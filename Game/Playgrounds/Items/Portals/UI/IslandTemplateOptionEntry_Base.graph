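# Base list-entry button for an island option in the portal island select UI, built on FortCreativeIslandTemplateButton.
# Setting the list item caches its IslandDataEntry, fills in the map name, collapses the code text if empty, and resets styling.
# Hover and selection events toggle name style and action prompt, though their row is collapsed by default.
# Entry click only calls the parent; hover animation, SelectPlot and MakeNewCreativePlot nodes have no incoming execution.
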WidgetBlueprint /Game/Playgrounds/Items/Portals/UI/IslandTemplateOptionEntry_Base
animations [Hovered {
  binding_count=1
  length_seconds=0.2
  object=&self:Hovered
  sequence {bindings [IslandOptionEntry_C_107 {
     id=#1
     possessed_class=IslandOptionEntry_C
     tracks [{
       type=MovieScene2DTransformTrack
       sections [{
         type=MovieScene2DTransformSection
         end_frame=4
         start_frame=0
         channels [
          Translation.X {type=MovieSceneScriptingFloatChannel}
          Translation.Y {type=MovieSceneScriptingFloatChannel}
          Angle {type=MovieSceneScriptingFloatChannel}
          Scale.X {type=MovieSceneScriptingFloatChannel, keys [[0,1],[4,1.1073]]}
          Scale.Y {type=MovieSceneScriptingFloatChannel, keys [[0,1],[4,1.1073]]}
          Shear.X {type=MovieSceneScriptingFloatChannel}
          Shear.Y {type=MovieSceneScriptingFloatChannel}
         ]
        }]
      }]
    }]}
 }]
graphs [
 EventGraph {nodes [
   n1 = Event OnListItemObjectSet()
   n2 = Cast To IslandDataEntry!(execute=@n24.then_0, Object=@n1.ListItemObject) -> (then, CastFailed, AsIsland Data Entry:Island Data Entry Object Reference) {is=title K2Node_DynamicCast}
   > n3 = Set IslandData(IslandData:Island Data Entry Object Reference=@n2) -> (Output_Get:Island Data Entry Object Reference)
   n4 = Get PublishedIslandCode!(self:Island Data Entry Object Reference=@n3) -> (PublishedIslandCode:String)
   n5 = Get PublishedCodeText() -> (PublishedCodeText:Common Text Object Reference)
   "SetText (Text)"(
    self=@n5
    InText=Format Text!(
     Format="INVTEXT(\"{Code} (ver.{Version})\")"
     Code:Text="To Text (String)"(InString=@n4)
     Version:int=Get PublishedVersion!(self:Island Data Entry Object Reference=@n3) -> (PublishedVersion:int=0)
    ) -> (Result)
   )
   > SetVisibility(self=@n5, InVisibility=SelfHitTestInvisible)
   n6 = Get Hovered() -> (Hovered:Widget Animation Object Reference)
   n7 = PlayAnimationReverse(InAnimation=@n6)
   Set bIsHovered(execute=[@n7.then, @n19.then], bIsHovered:bool=false) -> (Output_Get:bool)
   n8 = Get PlotName() -> (PlotName:Common Text Object Reference)
   > n9 = SetStyle!(
    execute
    self:Common Text Object Reference=@n8
    InStyle:Common Text Style Class Reference=/FortUILibrary/Styles/Typography/Foundation/TextStyle-BurbankSmall-M-Blue.TextStyle-BurbankSmall-M-Blue_C
   ) -> (then)
   n10 = Get InputActionWidget() -> (InputActionWidget:Common Action Widget Object Reference)
   SetVisibility(execute=[@n9.then, @n17.else], self=@n10, InVisibility=Hidden)
   n11 = PlayAnimationForward(InAnimation=@n6)
   Set bIsHovered(execute=[@n11.then, @n18.then], bIsHovered=true)
   > n12 = SetStyle(self=@n8, InStyle=/FortUILibrary/Styles/Typography/Foundation/TextStyle-BurbankSmall-M-Black.TextStyle-BurbankSmall-M-Black_C)
   SetVisibility(execute=[@n12.then, @n15.then], self=@n10, InVisibility=SelfHitTestInvisible)
   > PlaySound2D()
   n13 = Cast To FortPlayerControllerAthena(Object=GetOwningPlayer())
   > n14 = GetPlotTemplateDefinition!(execute, self:Fort Creative Island Template Button Object Reference) -> (then, ReturnValue:Fort Creative Real Estate Plot Item Definition Object Reference)
   > MakeNewCreativePlot!(
    execute
    self:Fort Player Controller Athena Object Reference=@n13
    PlotType:Fort Creative Real Estate Plot Item Definition Object Reference=@n14
    Locale:String
    Title:String
   ) -> (then)
   n15 = Event On Selected()
   n16 = Event OnEntryClick!() -> (OutputDelegate, then, Entry:Athena Creative Item Tile Button Object Reference) {is=title K2Node_Event}
   > Parent: OnEntryClick!(execute, Entry:Athena Creative Item Tile Button Object Reference=@n16.Entry) -> (then) {is=title K2Node_CallParentFunction}
   Event On Deselected()
   > n17 = Branch(Condition=Get bIsHovered() -> (bIsHovered:bool))
   n18 = Event On Hovered()
   n19 = Event On Unhovered()
   SetVisibility(execute=@n21.then, self=Get InputActionWidget(), InVisibility=Hidden)
   "SetText (Text)"(execute=@n22.then_0, self=Get PlotName(), InText=Get MapName!(self:Island Data Entry Object Reference=@n3) -> (MapName:Text))
   SelectPlot!(
    execute
    self:Island Select UI Object Reference=Get ParentWidget() -> (ParentWidget:Island Select UI Object Reference)
    SelectedPlot:String=Get McpId!(self:Island Data Entry Object Reference=Get IslandData() -> (IslandData:Island Data Entry Object Reference)) -> (McpId:String)
   ) -> (then)
   SetVisibility(execute=@n23.else, self=@n5, InVisibility=Collapsed)
   n20 = "integer > integer"(A=Len(S=@n4), B=0) {is=title K2Node_CallFunction}
   SetStyle(
    execute=@n24.then_1
    self=Get PlotName()
    InStyle=&TextStyle-BurbankSmall-M-Blue_C
   )
   > SetRenderTranslation()
   > n21 = SetRenderOpacity(InOpacity=1)
   n22 = Sequence(execute=@n3.then)
   n23 = Branch(execute=@n22.then_1, Condition=@n20)
   n24 = Sequence(execute=@n1.then)
  ]}
 InitializeOption {nodes [
   n1 = InitializeOption!() -> (then, ParentWidget:Island Select UI Object Reference) {is=title K2Node_FunctionEntry}
   > Set ParentWidget(ParentWidget:Island Select UI Object Reference=@n1) -> (Output_Get:Island Select UI Object Reference)
   > SetStyle(self=Get PlotName(), InStyle=&TextStyle-BurbankSmall-M-Blue_C)
  ]}
]
variables [
 bIsHovered {type=boolean}
 IslandData {type=IslandDataEntry_C, default=null}
 ParentWidget {type=IslandSelect_UI_C, default=null}
]
widgets {
 info {inheritedWidgetCount=2, parentClass=/Script/FortniteUI.FortCreativeIslandTemplateButton, rootWidgetClass=&Border, widgetCount=8}
 widgets [
  Border_1 {
   widgetClassPath=&Border
   widget=Border_1
   properties {
    horizontalAlignment=HAlign_Center
    verticalAlignment=VAlign_Center
    background {
     bIsDynamicallyLoaded=false
     drawAs=Image
     imageType=NoImage
     mirroring=NoMirror
     resourceName=None
     resourceObject=null
     tiling=NoTile
     imageSize=<32 32>
     margin=<ltrb 0 0 0 0>
     outlineSettings {
      bUseBrushTransparency=false
      roundingType=HalfHeightRadius
      width=0
      color {colorUseRule=UseColor_Specified, specifiedColor=<rgba 0 0 0 0>}
      cornerRadii=<0 0 0 0>
     }
     tintColor {colorUseRule=UseColor_Specified, specifiedColor=<rgba 1 1 1 0>}
     uVRegion {bIsValid=false, max=<0 0>, min=<0 0>}
    }
    padding=<ltrb 5 5 5 5>
   }
  }
  Border_0 {
   widgetClassPath=&Border
   parent=Border_1
   widget=Border_0
   properties {horizontalAlignment=HAlign_Center, verticalAlignment=VAlign_Center, padding=<ltrb 0 0 0 0>}
   slot {class=BorderSlot, properties {horizontalAlignment=HAlign_Center, verticalAlignment=VAlign_Center, padding=<ltrb 5 5 5 5>}}
  }
  VerticalBox_0 {
   widgetClassPath=&VerticalBox
   parent=Border_0
   widget=VerticalBox_0
   slot {class=BorderSlot, properties {horizontalAlignment=HAlign_Center, verticalAlignment=VAlign_Center, padding=<ltrb 0 0 0 0>}}
  }
  TemplateImage {
   widgetClassPath=&FortLazyImage
   bInherited=true
   bIsVariable=true
   parent=VerticalBox_0
   widget=TemplateImage
   properties {brush {
     bIsDynamicallyLoaded=false
     drawAs=Image
     imageType=NoImage
     mirroring=NoMirror
     resourceName=None
     resourceObject=/Game/Athena/Items/Consumables/PlaysetGrenade/T-Icon-PS-CP-Island_Temperate_60x60_a
     tiling=NoTile
     imageSize=<128 128>
     margin=<ltrb 0 0 0 0>
     outlineSettings {
      bUseBrushTransparency=false
      roundingType=HalfHeightRadius
      width=0
      color {colorUseRule=UseColor_Specified, specifiedColor=<rgba 0 0 0 0>}
      cornerRadii=<0 0 0 0>
     }
     tintColor {colorUseRule=UseColor_Specified, specifiedColor=<rgba 1 1 1 1>}
     uVRegion {bIsValid=false, max=<0 0>, min=<0 0>}
    }}
   slot {class=VerticalBoxSlot, properties {padding=<ltrb 5 5 5 5>}}
  }
  HorizontalBox_0 {
   widgetClassPath=&HorizontalBox
   parent=VerticalBox_0
   widget=HorizontalBox_0
   properties {visibility=Collapsed}
   slot {class=VerticalBoxSlot}
  }
  InputActionWidget {
   widgetClassPath=&CommonActionWidget
   bInherited=true
   bIsVariable=true
   parent=HorizontalBox_0
   widget=InputActionWidget
   properties {
    visibility=Hidden
    inputActions [{dataTable=&InputActionDataTable, rowName=Select}]
    renderTransform {angle=0, scale=<1 1>, shear=<0 0>, translation=<-8 0>}
   }
   slot {class=HorizontalBoxSlot, properties {horizontalAlignment=HAlign_Left, verticalAlignment=VAlign_Center}}
  }
  PlotName {
   widgetClassPath=&CommonTextBlock
   bIsVariable=true
   parent=HorizontalBox_0
   widget=PlotName
   properties {
    justification=Center
    minDesiredWidth=440
    scrollStyle=&TextScrollStyle-Base_C
    text="{Island Name}"
    colorAndOpacity {colorUseRule=UseColor_Specified, specifiedColor=<rgba 0 0.624 1 1>}
    font {
     bForceMonospaced=false
     bMaterialIsStencil=false
     fontMaterial=null
     fontObject=&BurbankSmall
     letterSpacing=0
     monospacedWidth=1
     size=24
     skewAmount=0
     typefaceFontName=Bold
     outlineSettings {
      bApplyOutlineToDropShadows=false
      bMiteredCorners=false
      bSeparateFillAlpha=false
      outlineMaterial=null
      outlineSize=0
      outlineColor=<rgba 0 0 0 1>
     }
    }
    shadowOffset=<0 0>
   }
   slot {class=HorizontalBoxSlot, properties {verticalAlignment=VAlign_Center}}
  }
  PublishedCodeText {
   widgetClassPath=&CommonTextBlock
   bIsVariable=true
   parent=HorizontalBox_0
   widget=PublishedCodeText
   properties {
    justification=Center
    minDesiredWidth=440
    scrollStyle=&TextScrollStyle-Base_C
    text="0000-0000-0000?v=8"
    visibility=Collapsed
    font {
     bForceMonospaced=false
     bMaterialIsStencil=false
     fontMaterial=null
     fontObject=&BurbankSmall
     letterSpacing=0
     monospacedWidth=1
     size=18
     skewAmount=0
     typefaceFontName=Bold
     outlineSettings {
      bApplyOutlineToDropShadows=false
      bMiteredCorners=false
      bSeparateFillAlpha=false
      outlineMaterial=null
      outlineSize=1
      outlineColor=<rgba 0 0 0 1>
     }
    }
    shadowOffset=<0 0>
   }
   slot {class=HorizontalBoxSlot, properties {verticalAlignment=VAlign_Center}}
  }
 ]
}
