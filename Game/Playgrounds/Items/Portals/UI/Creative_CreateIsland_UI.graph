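# Creative island creation panel with starter, template and optional collab tabs, each backed by a tile view.
# Accept opens an island naming popup, or the collab info dialog when collab approval is missing.
# Plot requests to the player controller disable Accept and show a spinner until the finished callback closes the panel or shows an error.
# The FortCreativeIslandGeneration parent supplies tab visibility, entitlement and template setup.
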WidgetBlueprint /Game/Playgrounds/Items/Portals/UI/Creative_CreateIsland_UI
animations [
 Intro {
  binding_count=2
  length_seconds=0.3715
  object=&self:Intro
  sequence {bindings [
    CommonTextBlock_2 {
     id=#1
     possessed_class=CommonTextBlock
     tracks [
      {
       type=MovieScene2DTransformTrack
       sections [{
         type=MovieScene2DTransformSection
         end_frame=7
         start_frame=0
         channels [
          Translation.X {type=MovieSceneScriptingFloatChannel, keys [[0,-80],[2.8583,-80],[5.0013,9.829],[6.3167,0],[7.4307,0]]}
          Angle {type=MovieSceneScriptingFloatChannel}
          Scale.X {type=MovieSceneScriptingFloatChannel}
          Scale.Y {type=MovieSceneScriptingFloatChannel}
          Shear.X {type=MovieSceneScriptingFloatChannel}
          Shear.Y {type=MovieSceneScriptingFloatChannel}
         ]
        }]
      }
      {
       type=MovieSceneFloatTrack
       display_name=Render Opacity
       sections [{
         type=MovieSceneFloatSection
         end_frame=7
         start_frame=0
         channels [None {type=MovieSceneScriptingFloatChannel, keys [[0,0],[2.6617,0],[7.4307,1]]}]
        }]
      }
     ]
    }
    SizeBox_0 {
     id=#2
     possessed_class=SizeBox
     tracks [
      {
       type=MovieScene2DTransformTrack
       sections [{
         type=MovieScene2DTransformSection
         end_frame=7
         start_frame=0
         channels [
          Translation.X {type=MovieSceneScriptingFloatChannel, keys [[0,-50],[2.738,0],[7.4307,0]]}
          Translation.Y {type=MovieSceneScriptingFloatChannel}
          Angle {type=MovieSceneScriptingFloatChannel}
          Shear.X {type=MovieSceneScriptingFloatChannel}
          Shear.Y {type=MovieSceneScriptingFloatChannel}
         ]
        }]
      }
      {
       type=MovieSceneFloatTrack
       display_name=Render Opacity
       sections [{type=MovieSceneFloatSection, end_frame=4, start_frame=0, channels [None {type=MovieSceneScriptingFloatChannel, keys [[0,0],[3.6303,1]]}]}]
      }
     ]
    }
   ]}
 }
 Outro {
  binding_count=2
  length_seconds=0.2
  object=&self:Outro
  sequence {bindings [
    CommonTextBlock_2 {
     id=#1
     possessed_class=CommonTextBlock
     tracks [{
       type=MovieScene2DTransformTrack
       sections [{
         type=MovieScene2DTransformSection
         end_frame=4
         start_frame=0
         channels [
          Translation.X {type=MovieSceneScriptingFloatChannel, const [0,4]}
          Angle {type=MovieSceneScriptingFloatChannel}
          Scale.X {type=MovieSceneScriptingFloatChannel}
          Scale.Y {type=MovieSceneScriptingFloatChannel}
          Shear.X {type=MovieSceneScriptingFloatChannel}
          Shear.Y {type=MovieSceneScriptingFloatChannel}
         ]
        }]
      }]
    }
    SizeBox_0 {
     id=#2
     possessed_class=SizeBox
     tracks [
      {
       type=MovieScene2DTransformTrack
       sections [{
         type=MovieScene2DTransformSection
         end_frame=3
         start_frame=0
         channels [
          Translation.X {type=MovieSceneScriptingFloatChannel}
          Translation.Y {type=MovieSceneScriptingFloatChannel}
          Angle {type=MovieSceneScriptingFloatChannel}
          Scale.X {type=MovieSceneScriptingFloatChannel, keys [[0,0.6],[2,1.05],[3.533,1]]}
          Scale.Y {type=MovieSceneScriptingFloatChannel, keys [[0,0.6],[2,1.15],[3.533,1]]}
          Shear.X {type=MovieSceneScriptingFloatChannel}
          Shear.Y {type=MovieSceneScriptingFloatChannel}
         ]
        }]
      }
      {
       type=MovieSceneFloatTrack
       display_name=Render Opacity
       sections [{type=MovieSceneFloatSection, end_frame=4, start_frame=0, channels [None {type=MovieSceneScriptingFloatChannel, keys [[0,0],[1.428,1]]}]}]
      }
     ]
    }
   ]}
 }
 LTDIntro {
  binding_count=1
  length_seconds=0.2
  object=&self:LTDIntro
  sequence {bindings [Overlay_1 {
     id=#3
     possessed_class=Overlay
     tracks [
      {
       type=MovieSceneFloatTrack
       display_name=Render Opacity
       sections [{type=MovieSceneFloatSection, end_frame=4, start_frame=0, channels [None {type=MovieSceneScriptingFloatChannel, keys [[0,0],[3,1],[4,1]]}]}]
      }
      {
       type=MovieScene2DTransformTrack
       sections [{
         type=MovieScene2DTransformSection
         end_frame=4
         start_frame=0
         channels [
          Translation.X {type=MovieSceneScriptingFloatChannel, const [0,3]}
          Translation.Y {type=MovieSceneScriptingFloatChannel, keys [[0,45],[3,-15.4186],[4,0]]}
          Angle {type=MovieSceneScriptingFloatChannel}
          Scale.X {type=MovieSceneScriptingFloatChannel}
          Scale.Y {type=MovieSceneScriptingFloatChannel}
          Shear.X {type=MovieSceneScriptingFloatChannel}
          Shear.Y {type=MovieSceneScriptingFloatChannel}
         ]
        }]
      }
     ]
    }]}
 }
 LTDOutro {
  binding_count=1
  length_seconds=0.2
  object=&self:LTDOutro
  sequence {bindings [Overlay_1 {
     id=#3
     possessed_class=Overlay
     tracks [
      {
       type=MovieSceneFloatTrack
       display_name=Render Opacity
       sections [{type=MovieSceneFloatSection, end_frame=4, start_frame=0, channels [None {type=MovieSceneScriptingFloatChannel, keys [[0,0],[3,1],[4,1]]}]}]
      }
      {
       type=MovieScene2DTransformTrack
       sections [{
         type=MovieScene2DTransformSection
         end_frame=4
         start_frame=0
         channels [
          Translation.X {type=MovieSceneScriptingFloatChannel, const [0,3]}
          Translation.Y {type=MovieSceneScriptingFloatChannel, keys [[0,45],[4,0]]}
          Angle {type=MovieSceneScriptingFloatChannel}
          Scale.X {type=MovieSceneScriptingFloatChannel}
          Scale.Y {type=MovieSceneScriptingFloatChannel}
          Shear.X {type=MovieSceneScriptingFloatChannel}
          Shear.Y {type=MovieSceneScriptingFloatChannel}
         ]
        }]
      }
     ]
    }]}
 }
]
graphs [
 ClosePanel {nodes [
   ClosePanel() {is=title K2Node_FunctionEntry}
   > n1 = Branch(Condition=Get bMCPRquestInFlight() -> (bMCPRquestInFlight:bool))
   PopContentWidget!(execute=@n1.else) -> (then)
   > Return Node!(execute, bCommitted:bool=false)
  ]}
 EventGraph {nodes [
   n1 = "On Clicked (CollabIslandsTab)"!() -> (OutputDelegate, then, Button:Common Button Base Object Reference) {is=title K2Node_ComponentBoundEvent}
   SelectButtonAtIndex!(
    execute=[@n1.then, @n16.1]
    self:Common Button Group Base Object Reference=Get TabButtonGroup() -> (TabButtonGroup:Common Button Group Legacy Object Reference)
    ButtonIndex:int=2
    bAllowSound:bool=true
   ) -> (then)
   n2 = Get CollabIslandsTileView() -> (CollabIslandsTileView:Common Tile View Object Reference)
   > SetActiveWidget(
    self:Widget Switcher Object Reference=Get IslandTemplatesWidgetSwitcher() -> (IslandTemplatesWidgetSwitcher:Widget Switcher Object Reference)
    Widget=@n2
   )
   > SetFocus(self=@n2)
   n3 = Get Selected Item!(self:List View Object Reference=@n2) -> (ReturnValue:Object Reference) {is=title K2Node_CallFunction}
   > Set Selected Item!(execute, self:List View Object Reference=@n2, Item:Object Reference=@n3) -> (then) {is=title K2Node_CallFunction}
   > CollabIslandsOnCurrentItemSelectionChanged!(execute, self:Fort Creative Island Generation Object Reference, Item:Object Reference=@n3) -> (then)
   > SetVisibility(self=Get IslandTemplateNameText() -> (IslandTemplateNameText:Common Text Object Reference), InVisibility=HitTestInvisible)
   > SetVisibility(self=Get IslandDescriptionText() -> (IslandDescriptionText:Common Text Object Reference), InVisibility=HitTestInvisible)
   > SetVisibility(self=Get PublishedIslandName() -> (PublishedIslandName:Common Text Object Reference), InVisibility=Collapsed)
   > SetVisibility(self=Get Overlay_Creator() -> (Overlay_Creator:Overlay Object Reference), InVisibility=Collapsed)
   > SetActiveWidgetIndex(self=Get Switcher_CreatorInfo() -> (Switcher_CreatorInfo:Common Widget Switcher Legacy Object Reference), Index=0)
   > SetScrollWidget!(execute, ToScroll:Widget Object Reference=Get ScrollBox_Details() -> (ScrollBox_Details:Scroll Box Object Reference)) -> (then)
   n4 = Get Num Items Selected!(self:List View Object Reference=Get CollabIslandsTileView()) -> (ReturnValue:int=0) {is=title K2Node_CallFunction}
   > SetVisibility(
    self=Get VBox_IslandDetails() -> (VBox_IslandDetails:Vertical Box Object Reference)
    InVisibility=Select(
     Option 0=SelfHitTestInvisible
     Option 1=Hidden
     Index="integer < integer"!(A:int=@n4, B:int=1) -> (ReturnValue:bool) {is=title K2Node_CallFunction}
    )
   )
   > UpdateAcceptButtonState!(execute, self, EntitlementState:bool=GetHasCollabApproval!(self:Fort Creative Island Generation Object Reference) -> (ReturnValue:bool)) -> (then)
   n5 = "On Button Clicked (StarterIslandsTab)"!() -> (OutputDelegate, then, Button:Common Button Legacy Object Reference) {is=title K2Node_ComponentBoundEvent}
   SelectButtonAtIndex(execute=[@n5.then, @n10.1], self=Get TabButtonGroup(), ButtonIndex=0, bAllowSound=true)
   n6 = Get StartingIslandsTileView() -> (StartingIslandsTileView:Common Tile View Object Reference)
   > SetActiveWidget(self=Get IslandTemplatesWidgetSwitcher(), Widget=@n6)
   > SetFocus(self=@n6)
   n7 = Get Selected Item(self=@n6)
   > Set Selected Item(self=@n6, Item=@n7)
   > StartingIslandsOnCurrentItemSelectionChanged!(execute, self:Fort Creative Island Generation Object Reference, Item:Object Reference=@n7) -> (then)
   > SetVisibility(self=Get IslandTemplateNameText(), InVisibility=HitTestInvisible)
   > SetVisibility(self=Get IslandDescriptionText(), InVisibility=HitTestInvisible)
   n8 = Get Overlay_Creator()
   > SetVisibility(self=@n8, InVisibility=Collapsed)
   > SetActiveWidgetIndex(self=Get Switcher_CreatorInfo(), Index=0)
   n9 = Get Num Items Selected(self=Get StartingIslandsTileView()) -> (ReturnValue=0)
   > SetVisibility(self=Get VBox_IslandDetails(), InVisibility=Select(Option 0=SelfHitTestInvisible, Option 1=Hidden, Index="integer < integer"(A=@n9, B=1)))
   > SetActiveWidget(
    self:Common Visibility Switcher Object Reference=Get CreateButtonSwitcher() -> (CreateButtonSwitcher:Common Visibility Switcher Object Reference)
    Widget=Get AcceptButton() -> (AcceptButton:UIKit Modular Button Object Reference)
   )
   "On Button Clicked (LeftTabButton)"!() -> (OutputDelegate, then, Button:Common Button Legacy Object Reference) {is=title K2Node_ComponentBoundEvent}
   > n10 = Switch on Int!(
    execute:Exec
    Selection:int=GetActiveWidgetIndex!(self:Widget Switcher Object Reference=Get IslandTemplatesWidgetSwitcher()) -> (ReturnValue:int=0)
   ) -> (Default:Exec, 0:Exec, 1:Exec, 2:Exec)
   SelectButtonAtIndex(execute=[@n16.0, @n14.then, @n15.then, @n10.2], self=Get TabButtonGroup(), ButtonIndex=1, bAllowSound=true)
   n11 = Get TemplateIslandsTileView() -> (TemplateIslandsTileView:Common Tile View Object Reference)
   > SetActiveWidget(self:Widget Switcher Object Reference=Get IslandTemplatesWidgetSwitcher(), Widget=@n11)
   > SetFocus(self=@n11)
   n12 = Get Selected Item(self=@n11)
   > Set Selected Item(self=@n11, Item=@n12)
   > TemplateIslandsOnCurrentItemSelectionChanged!(execute, self:Fort Creative Island Generation Object Reference, Item:Object Reference=@n12) -> (then)
   > SetVisibility(self=Get IslandTemplateNameText(), InVisibility=HitTestInvisible)
   > SetVisibility(self=Get IslandDescriptionText(), InVisibility=Collapsed)
   > SetVisibility(self=Get PublishedIslandName(), InVisibility=Collapsed)
   > SetVisibility(self=@n8, InVisibility=SelfHitTestInvisible)
   > SetActiveWidgetIndex(self=Get Switcher_CreatorInfo(), Index=1)
   > SetScrollWidget(ToScroll=Get ScrollBox_Details())
   n13 = Get Num Items Selected(self=Get TemplateIslandsTileView()) -> (ReturnValue=0)
   > SetVisibility(self=Get VBox_IslandDetails(), InVisibility=Select(Option 0=SelfHitTestInvisible, Option 1=Hidden, Index="integer < integer"(A=@n13, B=1)))
   > SetActiveWidget(self:Common Visibility Switcher Object Reference=Get CreateButtonSwitcher(), Widget=Get AcceptButton())
   n14 = "On Button Clicked (TemplateIslandsTab)"!() -> (OutputDelegate, then, Button:Common Button Legacy Object Reference) {is=title K2Node_ComponentBoundEvent}
   "On Button Clicked (RightTabButton)"!() -> (OutputDelegate, then, Button:Common Button Legacy Object Reference) {is=title K2Node_ComponentBoundEvent}
   > n15 = Branch(Condition=NOT Boolean(A=GetShowCollabIslandsTab!(self:Fort Creative Island Generation Object Reference) -> (ReturnValue:bool)))
   n16 = Switch on Int!(execute=@n15.else, Selection=GetActiveWidgetIndex(self=Get IslandTemplatesWidgetSwitcher()) -> (ReturnValue=0)) -> (Default, 0, 1)
   "On Clicked (AcceptButton)"!() -> (OutputDelegate, then, Button:Common Button Base Object Reference) {is=title K2Node_ComponentBoundEvent}
   > Cast To FortPlayerControllerAthena(Object=GetOwningPlayer())
   > n17 = Switch on Int!(execute, Selection=GetActiveWidgetIndex(self=Get IslandTemplatesWidgetSwitcher()) -> (ReturnValue=0)) -> (Default, 0, 1, 2:Exec)
   n18 = CreateWidgetAsync!(
    execute:Exec=@n17.0
    UserWidgetSoftClass:User Widget Soft Class Reference=&/PopupIslandNaming_UI_Modal.PopupIslandNaming_UI_Modal_C
    OwningPlayer:Player Controller Object Reference
    bSuspendInputUntilComplete:bool=true
   ) -> (then:Exec, OnComplete:Exec, OnFailure:Exec, UserWidget:User Widget Object Reference) {is=title K2Node_AsyncAction}
   SetUpNameIslandPopup!(execute=@n18.OnComplete, self, UserWidget:Object Reference=@n18, IslandTileView:List View Object Reference=Get StartingIslandsTileView()) -> (then)
   n19 = CreateWidgetAsync(
    execute=@n17.1
    UserWidgetSoftClass=&PopupIslandNaming_UI_Modal_C
    bSuspendInputUntilComplete=true
   )
   SetUpNameIslandPopup(execute=@n19.OnComplete, UserWidget=@n19, IslandTileView=Get TemplateIslandsTileView())
   n20 = Branch(execute=[@n27.then, @n17.2], Condition=GetHasCollabApproval())
   > n21 = CreateWidgetAsync(
    UserWidgetSoftClass=&PopupIslandNaming_UI_Modal_C
    bSuspendInputUntilComplete=true
   )
   SetUpNameIslandPopup(execute=@n21.OnComplete, UserWidget=@n21, IslandTileView=Get CollabIslandsTileView())
   CheckIfHasCollabEntitlement(execute=@n20.else, self:Fort Creative Island Generation Object Reference)
   > n22 = CreateWidgetAsync(
    UserWidgetSoftClass=Get CurrentCollabInfoDialog() -> (CurrentCollabInfoDialog:Common Activatable Widget Soft Class Reference)
    bSuspendInputUntilComplete=true
   )
   n23 = Cast To CommonActivatableWidget!(execute=@n22.OnComplete, Object=@n22) -> (then, CastFailed, AsCommon Activatable Widget:Common Activatable Widget Object Reference) {is=title K2Node_DynamicCast}
   > Bind Event to On Widget Deactivated(self:Common Activatable Widget Object Reference=@n23, Delegate=Create Event()) {is=title K2Node_AddDelegate}
   > TryPushWidgetAsModalPanel!(execute, PushWidget:User Widget Object Reference=@n22) -> (then)
   "On Button Clicked (Button_CloseTouch)"!() -> (OutputDelegate, then, Button:Common Button Legacy Object Reference) {is=title K2Node_ComponentBoundEvent}
   > ClosePanel!(execute, self) -> (then, bCommitted:bool) {is=title K2Node_CallFunction}
   n24 = "On Item Selection Changed (TemplateIslandsTileView)"!() -> (OutputDelegate, then, Item:Object Reference, bIsSelected:bool) {is=title K2Node_ComponentBoundEvent}
   > n25 = Cast To FortCreativeIslandLink!(execute, Object=@n24.Item) -> (then, CastFailed, AsFort Creative Island Link:Fort Creative Island Link Object Reference) {is=title K2Node_DynamicCast}
   n26 = Len(S=Get SupportCode!(self:Fort Creative Island Link Object Reference=@n25) -> (SupportCode:String))
   > SetActiveWidgetIndex(
    self=Get Switcher_CreatorType() -> (Switcher_CreatorType:Common Widget Switcher Legacy Object Reference)
    Index=Select(Option 0:int=0, Option 1:int=1, Index="Equal (Integer)"(A=@n26, B=0) {is=title K2Node_CallFunction}) -> (ReturnValue:int)
   )
   n27 = "On Clicked (JoinCollabButton)"!() -> (OutputDelegate, then, Button:Common Button Base Object Reference) {is=title K2Node_ComponentBoundEvent}
   UpdateInputActionHandler() {is=title K2Node_CustomEvent}
   > n28 = Branch(Condition=IsUsingTouch(PlayerController=GetOwningPlayer()))
   n29 = Get CloseActionBack() -> (CloseActionBack:Data Table Row Handle Structure)
   > Branch(Condition=HasInputActionHandler!(self:Common Activatable Panel Legacy Object Reference, InputActionRow:Data Table Row Handle Structure=@n29) -> (ReturnValue:bool))
   n30 = Self-Reference()
   > RemoveInputActionHandler!(execute, self:Common Activatable Panel Legacy Object Reference=@n30, InputActionRow:Data Table Row Handle Structure=@n29) -> (then)
   SetInputActionHandler!(
    execute=@n28.else
    self:Common Activatable Panel Legacy Object Reference=@n30
    InputActionRow:Data Table Row Handle Structure=@n29
    CommitedEvent:Delegate=Create Event()
   ) -> (then)
   n31 = Create Island Finished!() -> (OutputDelegate, then, bSuccess:bool=~, Reason:"Text (by ref)") {is=title K2Node_CustomEvent}
   > Set bMCPRquestInFlight(bMCPRquestInFlight:bool=false) -> (Output_Get:bool)
   > n32 = Branch(Condition=@n31.bSuccess)
   > SetIsEnabled(self=Get AcceptButton(), bInIsEnabled=true)
   > SetVisibility(self=Get CreateSpinner() -> (CreateSpinner:Image Object Reference), InVisibility=Collapsed)
   n33 = Get IslandSelectUI() -> (IslandSelectUI:Island Select UI V2 Modal Object Reference)
   > n34 = Is Valid(InputObject=@n33)
   > n35 = Set ShouldLoadFirstIslandOnInit!(execute, ShouldLoadFirstIslandOnInit:bool=true, self:Island Select UI V2 Modal Object Reference=@n33) -> (then, Output_Get:bool)
   ClosePanel(execute=[@n34.Is Not Valid, @n35.then])
   ShowErrorDialog!(
    execute=@n32.else
    DisplayMessage:Text=@n31.Reason
    ErrorCode:String=Island Creation Error
    ErrorSeverity:EFortErrorSeverity Enum=Unspecified
   ) -> (then)
   > SetIsEnabled(self=Get AcceptButton(), bInIsEnabled=true)
   > SetVisibility(self=Get CreateSpinner(), InVisibility=Collapsed)
   CollabInfoPopupComplete() {is=title K2Node_CustomEvent}
   > SetFocus(self=Get CollabIslandsTileView())
   SetupTabs() {is=title K2Node_CustomEvent}
   > n36 = Sequence!(execute) -> (then_0, then_1, then_2:Exec)
   > SetSelectionRequired!(execute, self:Common Button Group Base Object Reference=Get TabButtonGroup(), bRequireSelection:bool=false) -> (then)
   > DeselectAll(self:Common Button Group Base Object Reference=Get TabButtonGroup())
   > RemoveAll(self:Common Widget Group Base Object Reference=Get TabButtonGroup())
   > AddWidget!(
    execute
    self:Common Widget Group Base Object Reference=Get TabButtonGroup()
    InWidget:Widget Object Reference=Get StarterIslandsTab() -> (StarterIslandsTab:Icon Text Button Object Reference)
   ) -> (then)
   > AddWidget(self=Get TabButtonGroup(), InWidget=Get TemplateIslandsTab() -> (TemplateIslandsTab:Icon Text Button Object Reference))
   > Set Text!(
    execute
    self:Icon Text Button Object Reference=Get CollabIslandsTab() -> (CollabIslandsTab:Icon Text Button Object Reference)
    ButtonText:Text=GetCollabIslandsTabDisplayName!(self:Fort Creative Island Generation Object Reference) -> (ReturnValue:"Text (by ref)")
   ) -> (then) {is=title K2Node_CallFunction}
   > AddWidget(self=Get TabButtonGroup(), InWidget=Get CollabIslandsTab())
   > ChangeCollabIslandsTabVisibility!(execute, self:Fort Creative Island Generation Object Reference, bShowTab:bool=GetShowCollabIslandsTab()) -> (then)
   SetSelectionRequired(execute=@n36.then_1, self=Get TabButtonGroup(), bRequireSelection=true)
   > SetIsSelected!(execute, self:Common Button Base Object Reference=Get StarterIslandsTab(), InSelected:bool=true, bGiveClickFeedback:bool=false) -> (then)
   > SetActiveWidget(self:Widget Switcher Object Reference=Get IslandTemplatesWidgetSwitcher(), Widget=Get StartingIslandsTileView())
   SetVisibility(execute=@n36.then_2, self=Get Overlay_Creator(), InVisibility=Collapsed)
   n37 = Select(
    Option 0:ESlateVisibility Enum=Collapsed
    Option 1:ESlateVisibility Enum=Visible
    Index=OR Boolean(A=GetShowCollabIslandsTab(), B=GetShowTemplateIslands!(self:Fort Creative Island Generation Object Reference) -> (ReturnValue:bool))
   ) -> (ReturnValue:ESlateVisibility Enum)
   > SetVisibility(self=Get StarterIslandsTab(), InVisibility=@n37)
   > SetVisibility(self=Get LeftTabButton() -> (LeftTabButton:User Action Menu Input Button Object Reference), InVisibility=@n37)
   > SetVisibility(self=Get RightTabButton() -> (RightTabButton:User Action Menu Input Button Object Reference), InVisibility=@n37)
   > SetVisibility(self=Get TemplateIslandsTab(), InVisibility=Select(Option 0=Collapsed, Option 1=Visible, Index=GetShowTemplateIslands()))
   > SetVisibility(self=Get CollabIslandsTab(), InVisibility=Select(Option 0=Collapsed, Option 1=Visible, Index=GetShowCollabIslandsTab()))
   n38 = ConfirmNameForLinkCodeCreation!() -> (OutputDelegate, then, IslandName:String) {is=title K2Node_CustomEvent}
   > n39 = Cast To FortCreativeIslandLink(Object=Get Selected Item(self=Get TemplateIslandsTileView()))
   > MakeCreativePlotFromLinkCodeAndHandleButtonState!(
    execute
    self
    Mnemonic:String=Break Online Link Id!(OnlineLinkId:"Online Link Id Structure (by ref)"=Get LinkId!(self:Fort Creative Island Link Object Reference=@n39) -> (LinkId:Online Link Id Structure)) -> (Mnemonic:String, Version:int) {is=title K2Node_BreakStruct}
    Locale:String=GetCurrentLanguage!() -> (ReturnValue:String)
    Title:String=@n38.IslandName
   ) -> (then)
   n40 = HandleInitemplateSelection!() -> (OutputDelegate, then, Widget:User Widget Object Reference) {is=title K2Node_CustomEvent}
   n41 = Get TemplateIslandsTileView()
   > Branch(Condition="integer < integer"(A=Get Num Items Selected(self=@n41) -> (ReturnValue=0), B=1))
   > SetSelectedIndex!(execute, self:List View Object Reference=Get TemplateIslandsTileView(), Index:int=0) -> (then)
   > NavigateToIndex!(execute, self:List View Object Reference=Get TemplateIslandsTileView(), Index:int=0) -> (then)
   SupportCreator() {is=title K2Node_CustomEvent}
   > n42 = CreateWidgetAsync(UserWidgetSoftClass=/Game/Athena/UI/Frontend/AffiliateKeyModal.AffiliateKeyModal_C, bSuspendInputUntilComplete=true)
   n43 = Cast To AffiliateKeyModal!(execute=@n42.OnComplete, Object=@n42) -> (then, CastFailed, AsAffiliate Key Modal:Affiliate Key Modal Object Reference) {is=title K2Node_DynamicCast}
   > PushContentWidgetAdvanced!(execute, PushWidget:Widget Object Reference=@n43, bHideHeader:bool=true, bHideFooter:bool=true, bHideChatWidget:bool=true) -> (then)
   > n44 = Cast To FortCreativeIslandLink(Object=Get Selected Item(self=Get TemplateIslandsTileView()))
   > "SetText (Editable Text)"!(
    execute
    self:Editable Text Object Reference=Get EditableText!(self:Affiliate Key Modal Object Reference=@n43) -> (EditableText:UIKit Editable Text Object Reference)
    InText:Text="To Text (String)"(InString=Get SupportCode(self=@n44))
   ) -> (then)
   > Set SourceName!(execute, SourceName:String=Get Source Name() -> (Source Name:String), self:Fort Affilate Modal Object Reference=@n43) -> (then, Output_Get:String)
   > SetVisibility(
    self=Get LaunchWebsite!(self:Affiliate Key Modal Object Reference=@n43) -> (LaunchWebsite:WBP UIKit Button Regular Object Reference)
    InVisibility=Collapsed
   )
   n45 = HandleInitCollabSelection!() -> (OutputDelegate, then, Widget:User Widget Object Reference) {is=title K2Node_CustomEvent}
   > n46 = Sequence!(execute) -> (then_0, then_1)
   n47 = Get CollabIslandsTileView()
   > Branch(Condition="integer < integer"(A=Get Num Items Selected(self=@n47) -> (ReturnValue=0), B=1))
   n48 = Get CollabIslandsTileView()
   > SetSelectedIndex(self=@n48, Index=0)
   > NavigateToIndex(self=@n48, Index=0)
   n49 = Cast To UserObjectListEntry!(execute=@n46.then_1, Object=@n45.Widget) -> (then, CastFailed, AsUser Object List Entry:User Object List Entry Interface) {is=title K2Node_DynamicCast}
   n50 = Get Selected Item(self=Get CollabIslandsTileView())
   > Branch(Condition="Equal (Object)"(A=GetListItemObject!(UserObjectListEntry:User Object List Entry Interface=@n49) -> (ReturnValue:Object Reference), B=@n50))
   > CollabIslandsOnCurrentItemSelectionChanged(Item=@n50)
   n51 = HandleInitStarterSelection!() -> (OutputDelegate, then, Widget:User Widget Object Reference) {is=title K2Node_CustomEvent}
   n52 = Get StartingIslandsTileView()
   > Branch(Condition="integer < integer"(A=Get Num Items Selected(self=@n52) -> (ReturnValue=0), B=1))
   > SetSelectedIndex(self=Get StartingIslandsTileView(), Index=0)
   > NavigateToIndex(self=Get StartingIslandsTileView(), Index=0)
   Event On Deactivated()
   > n53 = Cast To FortPlayerControllerAthena(Object=GetOwningPlayer())
   > Unbind Event from On Make New Creative Plot Finished(self:Fort Player Controller Athena Object Reference=@n53, Delegate=Create Event()) {is=title K2Node_RemoveDelegate}
   Event Construct()
   > Parent: Construct()
   > n54 = Sequence!(execute) -> (then_0, then_1, then_2:Exec)
   > Bind Event to On Entry Generated(self:List View Base Object Reference=Get StartingIslandsTileView(), Delegate=@n51) {is=title K2Node_AddDelegate}
   Bind Event to On Entry Generated(execute=@n54.then_1, self=Get TemplateIslandsTileView(), Delegate=@n40)
   Bind Event to On Entry Generated(execute=@n54.then_2, self=Get CollabIslandsTileView(), Delegate=@n45)
   n55 = Event OnInputModeChanged!() -> (OutputDelegate, then, bUsingGamepad:bool) {is=title K2Node_Event}
   > n56 = Cast To CommonTileView!(
    execute
    Object=GetActiveWidget!(self:Widget Switcher Object Reference=Get IslandTemplatesWidgetSwitcher()) -> (ReturnValue:Widget Object Reference)
   ) -> (then, CastFailed, AsCommon Tile View:Common Tile View Object Reference) {is=title K2Node_DynamicCast}
   > UpdateInputActionHandler() {is=title K2Node_CallFunction}
   n57 = GetNumItems!(self:List View Object Reference=@n56) -> (ReturnValue:int=0)
   > Branch(Condition=AND Boolean(A=@n55.bUsingGamepad, B="integer > integer"(A=@n57, B=1) {is=title K2Node_CallFunction}))
   > NavigateToIndex(self=@n56, Index=0)
   Event OnIslandFilterComplete() {is=title K2Node_Event}
   > Initialize()
   Event On Activated()
   > PlayAnimationForward(InAnimation=Get Intro())
   > UpdateInputActionHandler()
   > PlaySound2D(Sound=Get PanelAppear_Sound() -> (PanelAppear_Sound:Sound Base Object Reference))
   > SetScrollWidget(ToScroll=Get ScrollBox_Details())
   > SetupTabs() {is=title K2Node_CallFunction}
   > UpdateAcceptButtonState(EntitlementState=GetHasCollabApproval())
   > n58 = Cast To FortPlayerControllerAthena(Object=GetOwningPlayer())
   > Bind Event to On Make New Creative Plot Finished(self:Fort Player Controller Athena Object Reference=@n58, Delegate=Create Event()) {is=title K2Node_AddDelegate}
   > CheckIfHasCollabEntitlement()
   n59 = Event InitFromObject!() -> (OutputDelegate, then, InitObject:Object Reference) {is=title K2Node_Event}
   > n60 = Cast To IslandSelect_UI_V2_Modal!(execute, Object=@n59.InitObject) -> (then, CastFailed, AsIsland Select UI V2 Modal:Island Select UI V2 Modal Object Reference) {is=title K2Node_DynamicCast}
   > n61 = Set IslandSelectUI(IslandSelectUI:Island Select UI V2 Modal Object Reference=@n60) -> (Output_Get:Island Select UI V2 Modal Object Reference)
   OpenPanel!(execute=[@n61.then, @n60.CastFailed], self, Portal:BGA Island Portal Object Reference) -> (then)
   n62 = Event ChangeTermsAndConditionsVisibility!() -> (OutputDelegate, then, NewVisibility:ESlateVisibility Enum=Visible) {is=title K2Node_Event}
   > SetVisibility(self=Get CollabTermsAndConditonsBox() -> (CollabTermsAndConditonsBox:Size Box Object Reference), InVisibility=@n62.NewVisibility)
   Event OnCollabEntitlementChanged!() -> (OutputDelegate, then, bEntitlementGranted:bool) {is=title K2Node_Event}
   > UpdateCollabWidgets()
   Event OnSelectionUpdated!() -> (OutputDelegate, then, CurrentSelection:Fort Creative Island Template Button Object Reference) {is=title K2Node_Event}
   > UpdateCollabWidgets()
   n63 = Event ChangeCollabIslandsTabVisibility!() -> (OutputDelegate, then, bShowTab:bool) {is=title K2Node_Event}
   > SetVisibility(self=Get CollabIslandsTab(), InVisibility=Select(Option 0=Collapsed, Option 1=Visible, Index=@n63.bShowTab))
  ]}
 Initialize {nodes [
   n1 = Get StartingIslandsTileView()
   SetSelectedIndex(self=@n1, Index=0)
   > CenterSelectedItemListEntryWidget!(execute:Exec, self:Object Reference=@n1) -> (then:Exec) {is=title K2Node_Message}
   Initialize() {is=title K2Node_FunctionEntry}
   > SetupIslandTemplateData(self:Fort Creative Island Generation Object Reference)
  ]}
 MakeCreativePlotAndHandleButtonState {nodes [
   n1 = MakeCreativePlotAndHandleButtonState!() -> (then, Locale:String, Title:String, PlotItemDef:Fort Creative Real Estate Plot Item Definition Object Reference) {is=title K2Node_FunctionEntry}
   > n2 = Is Valid(InputObject=@n1.PlotItemDef)
   PrintString!(execute=@n2.Is Not Valid, InString=PlotItemDef is not valid, bPrintToScreen, bPrintToLog, TextColor, Duration, Key) -> (then)
   n3 = Cast To FortPlayerControllerAthena(execute=@n2.Is Valid, Object=GetOwningPlayer())
   > MakeNewCreativePlot!(
    execute
    self:Fort Player Controller Athena Object Reference=@n3
    PlotType:Fort Creative Real Estate Plot Item Definition Object Reference=Get PlotItemDef() -> (PlotItemDef:Fort Creative Real Estate Plot Item Definition Object Reference)
    Locale:String=Get Locale() -> (Locale:String)
    Title:String=Get Title() -> (Title:String)
   ) -> (then)
   > Set bMCPRquestInFlight(bMCPRquestInFlight=true)
   > DisableButtonWithReason!(
    execute
    self:Common Button Base Object Reference=Get AcceptButton()
    DisabledReason:"Text (by ref)"=Get Working Text() -> (Working Text:Text)
   ) -> (then)
   > SetVisibility(self=Get CreateSpinner(), InVisibility=SelfHitTestInvisible)
  ]}
 MakeCreativePlotFromLinkCodeAndHandleButtonState {nodes [
   MakeCreativePlotFromLinkCodeAndHandleButtonState!() -> (then, Mnemonic:String, Locale:String, Title:String) {is=title K2Node_FunctionEntry}
   > n1 = Cast To FortPlayerControllerAthena(Object=GetOwningPlayer())
   > MakeNewCreativePlotFromLinkCode!(
    execute
    self:Fort Player Controller Athena Object Reference=@n1
    LinkCode:String=Get Mnemonic() -> (Mnemonic:String)
    Locale:String=Get Locale()
    Title:String=Get Title()
   ) -> (then)
   > Set bMCPRquestInFlight(bMCPRquestInFlight=true)
   > DisableButtonWithReason(self=Get AcceptButton(), DisabledReason=Get Working Text())
   > SetVisibility(self=Get CreateSpinner(), InVisibility=SelfHitTestInvisible)
  ]}
 OnTouchToClose {nodes [
   OnTouchToClose!() -> (then, MyGeometry:Geometry Structure, MouseEvent:"Pointer Event Structure (by ref)") {is=title K2Node_FunctionEntry}
   Return Node!(execute, ReturnValue:Event Reply Structure=Handled())
  ]}
 OpenPanel {nodes [
   OpenPanel!() -> (then, Portal:BGA Island Portal Object Reference) {is=title K2Node_FunctionEntry}
   > PushContentWidgetAdvanced(PushWidget=Self-Reference(), bHideHeader=true, bHideFooter=false, bHideChatWidget=true)
  ]}
 SetUpNameIslandPopup {nodes [
   n1 = SetUpNameIslandPopup!() -> (then, UserWidget:Object Reference, IslandTileView:List View Object Reference) {is=title K2Node_FunctionEntry}
   > n2 = Cast To PopupIslandNaming_UI_Modal!(execute, Object=@n1) -> (then, CastFailed, AsPopup Island Naming UI Modal:Popup Island Naming UI Modal Object Reference) {is=title K2Node_DynamicCast}
   > InitFromObject!(execute, self:Popup Island Naming UI Modal Object Reference=@n2, InitObject:Object Reference=Self-Reference()) -> (then)
   n3 = Get Selected Item(self=Get IslandTileView() -> (IslandTileView:List View Object Reference))
   > n4 = Cast To FortCreativeRealEstatePlotItemDefinition!(execute, Object=@n3) -> (then, CastFailed, AsFort Creative Real Estate Plot Item Definition:Fort Creative Real Estate Plot Item Definition Object Reference) {is=title K2Node_DynamicCast}
   > InitFromPlotItemDefinition!(
    execute
    self:Fort Creative Island Naming Modal Object Reference=@n2
    RealEstateItemDefinition:Fort Creative Real Estate Plot Item Definition Object Reference=@n4
   ) -> (then)
   > n5 = Set IsFromIslandLink(IsFromIslandLink:bool=false) -> (Output_Get:bool)
   TryPushWidgetAsModalPanel(execute=[@n7.then, @n5.then], PushWidget=@n2)
   > Set IsFromTemplate!(execute, IsFromTemplate:bool=Get IsFromIslandLink() -> (IsFromIslandLink:bool), self:Popup Island Naming UI Modal Object Reference=@n2) -> (then, Output_Get:bool)
   n6 = Cast To FortCreativeIslandLink(execute=@n4.CastFailed, Object=@n3)
   > InitFromIslandLink!(execute, self:Fort Creative Island Naming Modal Object Reference=@n2, IslandLink:Fort Creative Island Link Object Reference=@n6) -> (then)
   > n7 = Set IsFromIslandLink(IsFromIslandLink=true)
  ]}
 UpdateAcceptButtonState {nodes [
   n1 = UpdateAcceptButtonState!() -> (then, EntitlementState:bool=~) {is=title K2Node_FunctionEntry}
   n2 = "Not Equal (Object)"(A:Object Reference=GetActiveWidget(self=Get IslandTemplatesWidgetSwitcher()), B:Object Reference=Get CollabIslandsTileView()) -> (ReturnValue:bool=~) {is=title K2Node_PromotableOperator}
   > SetActiveWidget(
    self:Common Visibility Switcher Object Reference=Get CreateButtonSwitcher()
    Widget=Select(
     Option 0:Widget Object Reference=Get JoinCollabButton() -> (JoinCollabButton:WBP UIKit Button Generic Object Reference)
     Option 1:Widget Object Reference=Get AcceptButton()
     Index=OR Boolean(A=@n1, B=@n2)
    ) -> (ReturnValue:Widget Object Reference)
   )
  ]}
 UpdateCollabWidgets {nodes [
   UpdateCollabWidgets() {is=title K2Node_FunctionEntry}
   n1 = GetCurrentCollabProperties!(self:Fort Creative Island Generation Object Reference) -> (
    bOutIsValid:bool
    OutCurrentCollabProperties_CollabIcon:Texture 2D Soft Object Reference
    OutCurrentCollabProperties_CollabDisabledDisclaimerText:Text
    OutCurrentCollabProperties_CollabEnabledDisclaimerText:Text
    OutCurrentCollabProperties_CollabInfoDialog:Common Activatable Widget Soft Class Reference
   )
   > Branch(Condition=@n1)
   > Set CurrentlyEntitled(CurrentlyEntitled:bool=GetHasCollabApproval()) -> (Output_Get:bool)
   > n2 = Sequence!(execute) -> (then_0, then_1)
   > UpdateAcceptButtonState(EntitlementState=Get CurrentlyEntitled() -> (CurrentlyEntitled:bool)) {is=title K2Node_CallFunction}
   Set Text(
    execute=@n2.then_1
    Text=Select(
     Option 0:Text=@n1.OutCurrentCollabProperties_CollabDisabledDisclaimerText
     Option 1:Text=@n1.OutCurrentCollabProperties_CollabEnabledDisclaimerText
     Index=Get CurrentlyEntitled()
    ) -> (ReturnValue:Text)
    self:Rich Text Block Object Reference=Get CollabDisclaimerText() -> (CollabDisclaimerText:Common Rich Text Block Object Reference)
   ) {is=title K2Node_VariableSet}
  ]}
]
variables [
 bIsClosing {type=boolean}
 bMCPRquestInFlight {type=boolean}
 CloseActionBack {type=DataTableRowHandle, default {dataTable=&InputActionDataTable, rowName=Cancel}}
 CollabEntitled {type=string, default=CReady}
 CollabNotEntitled {type=string, default=CNotReady}
 IslandDefinition {type=FortCreativeRealEstatePlotItemDefinition, default=null}
 IslandSelectUI {type=IslandSelect_UI_V2_Modal_C, default=null}
 PanelAppear_Sound {type=SoundBase, default=/Game/Sounds/Fort_UI_Sounds/FrontEnd/New_FrontEnd/Fortnite_UI_MainMenu_Tab_03_Cue}
 PanelDismiss_Sound {type=SoundBase, default=/Game/Sounds/Fort_UI_Sounds/FrontEnd/New_FrontEnd/Fortnite_UI_MissionObj_Cue}
 Portal {type=BGA_IslandPortal_C, default=null}
 SelectPlot_Sound {type=SoundBase, default=/Game/Sounds/Fort_UI_Sounds/FrontEnd/New_FrontEnd/Fortnite_UI_Options_Menu_01_Cue}
 Source Name {type=string, default=Creative_CreateIsland_UI}
 Working Text {type=string, default=Working}
]
widgets {
 info {
  inheritedWidgetCount=13
  namedSlotCount=4
  parentClass=/Script/FortniteUI.FortCreativeIslandGeneration
  rootWidgetClass=&CommonBorder
  widgetCount=66
 }
 widgets [
  TouchToCloseZone {widgetClassPath=&CommonBorder, widget=TouchToCloseZone, properties {padding=<ltrb 0 0 0 0>}}
  BackgroundBlur_1 {
   widgetClassPath=&BackgroundBlur
   parent=TouchToCloseZone
   widget=BackgroundBlur_1
   properties {blurStrength=25}
   slot {class=BorderSlot, properties {padding=<ltrb 0 0 0 0>}}
  }
  SafeZone_0 {
   widgetClassPath=&SafeZone
   bIsVariable=true
   parent=BackgroundBlur_1
   widget=SafeZone_0
   slot {class=BackgroundBlurSlot, properties {padding=<ltrb 0 0 0 0>}}
  }
  Overlay_0 {widgetClassPath=&Overlay, parent=SafeZone_0, widget=Overlay_0, slot {class=SafeZoneSlot}}
  SizeBox_0 {
   widgetClassPath=&SizeBox
   parent=Overlay_0
   widget=SizeBox_0
   properties {bOverride_HeightOverride=true, bOverride_WidthOverride=true, heightOverride=812, minDesiredWidth=1600, widthOverride=1900}
   slot {class=OverlaySlot, properties {horizontalAlignment=HAlign_Center, verticalAlignment=VAlign_Center}}
  }
  BackgroundBlur_0 {
   widgetClassPath=&BackgroundBlur
   parent=SizeBox_0
   widget=BackgroundBlur_0
   properties {blurStrength=16}
   slot {class=SizeBoxSlot}
  }
  Overlay_1 {
   widgetClassPath=&Overlay
   parent=BackgroundBlur_0
   widget=Overlay_1
   slot {class=BackgroundBlurSlot, properties {padding=<ltrb 0 0 0 0>}}
  }
  Image_DropShadow {
   widgetClassPath=&Image
   parent=Overlay_1
   widget=Image_DropShadow
   properties {colorAndOpacity=<rgba 0 0 0 0.5>, renderTransform {angle=0, scale=<1 1>, shear=<0 0>, translation=<12 12>}}
   slot {class=OverlaySlot, properties {horizontalAlignment=HAlign_Fill, verticalAlignment=VAlign_Fill}}
  }
  Imge_Backplate {
   widgetClassPath=&Image
   parent=Overlay_1
   widget=Imge_Backplate
   properties {brush {
     bIsDynamicallyLoaded=false
     drawAs=Image
     imageType=NoImage
     mirroring=NoMirror
     resourceName=None
     resourceObject=null
     tiling=NoTile
     imageSize=<32 32>
     margin=<ltrb 0 0 0 0>
     outlineSettings {
      bUseBrushTransparency=false
      roundingType=HalfHeightRadius
      width=0
      color {colorUseRule=UseColor_Specified, specifiedColor=<rgba 0 0 0 0>}
      cornerRadii=<0 0 0 0>
     }
     tintColor {colorUseRule=UseColor_Specified, specifiedColor=<rgba 0.0137 0.1022 0.491 1>}
     uVRegion {bIsValid=false, max=<0 0>, min=<0 0>}
    }}
   slot {class=OverlaySlot, properties {horizontalAlignment=HAlign_Fill, verticalAlignment=VAlign_Fill}}
  }
  HorizontalBox_3 {
   widgetClassPath=&HorizontalBox
   parent=Overlay_1
   widget=HorizontalBox_3
   slot {class=OverlaySlot, properties {horizontalAlignment=HAlign_Fill, verticalAlignment=VAlign_Fill}}
  }
  VBox_LeftSide {
   widgetClassPath=&VerticalBox
   parent=HorizontalBox_3
   widget=VBox_LeftSide
   slot {class=HorizontalBoxSlot, properties {size {sizeRule=Fill, value=1}}}
  }
  CommonTextBlock_2 {
   widgetClassPath=&CommonTextBlock
   parent=VBox_LeftSide
   widget=CommonTextBlock_2
   properties {
    justification=Center
    text=Island Type
    textTransformPolicy=ToUpper
    colorAndOpacity {colorUseRule=UseColor_Specified, specifiedColor=<rgba 0.15 0.314 0.8714 1>}
    font {
     bForceMonospaced=false
     bMaterialIsStencil=false
     fontMaterial=null
     fontObject=&Burbank
     letterSpacing=0
     monospacedWidth=1
     size=64
     skewAmount=0
     typefaceFontName=Burbank-Black
     outlineSettings {
      bApplyOutlineToDropShadows=false
      bMiteredCorners=false
      bSeparateFillAlpha=false
      outlineMaterial=null
      outlineSize=0
      outlineColor=<rgba 0 0 0 1>
     }
    }
    shadowOffset=<4.9266 5.2555>
   }
   slot {class=VerticalBoxSlot, properties {padding=<ltrb 0 32 0 0>}}
  }
  HorizontalBox_144 {
   widgetClassPath=&HorizontalBox
   parent=VBox_LeftSide
   widget=HorizontalBox_144
   slot {class=VerticalBoxSlot, properties {padding=<ltrb 0 12 0 0>}}
  }
  LeftTabButton {
   widgetClassPath=/Game/UI/SocialPanel/UserActionMenuInputButton
   bIsVariable=true
   parent=HorizontalBox_144
   widget=LeftTabButton
   properties {
    onButtonBaseClicked=(&self:WidgetTree.LeftTabButton.HandleOnButtonClicked)
    onButtonBaseDoubleClicked=(&self:WidgetTree.LeftTabButton.HandleOnButtonDoubleClicked)
    onButtonBaseHovered=(&self:WidgetTree.LeftTabButton.HandleOnButtonHovered)
    onButtonBaseUnhovered=(&self:WidgetTree.LeftTabButton.HandleOnButtonUnhovered)
    onSelectedChangedBase=(&self:WidgetTree.LeftTabButton.HandleOnSelectedChanged)
    triggeringInputAction {dataTable=&InputActionDataTable, rowName=GenericLeftShoulder}
   }
   slot {class=HorizontalBoxSlot, properties {padding=<ltrb 40 0 5 0>}}
  }
  StarterIslandsTab {
   widgetClassPath=&IconTextButton
   bIsVariable=true
   parent=HorizontalBox_144
   widget=StarterIslandsTab
   properties {
    bHideInputAction=true
    bIsFocusable=false
    buttonText=Starter
    onButtonBaseClicked=(&self:WidgetTree.StarterIslandsTab.HandleOnButtonClicked)
    onButtonBaseDoubleClicked=(&self:WidgetTree.StarterIslandsTab.HandleOnButtonDoubleClicked)
    onButtonBaseHovered=(&self:WidgetTree.StarterIslandsTab.HandleOnButtonHovered)
    onButtonBaseUnhovered=(&self:WidgetTree.StarterIslandsTab.HandleOnButtonUnhovered)
    onSelectedChangedBase=(&self:WidgetTree.StarterIslandsTab.HandleOnSelectedChanged)
    style=/FortUILibrary/Foundation/Buttons/ButtonStyles/Tab/Athena_ButtonStyle_ChallengesTab.Athena_ButtonStyle_ChallengesTab_C
    visibility=Visible
    triggeringInputAction {dataTable=&InputActionDataTable, rowName=GenericLeftShoulder}
   }
   slot {class=HorizontalBoxSlot, properties {verticalAlignment=VAlign_Center, size {sizeRule=Fill, value=1}}}
  }
  RemoveSpinner_1 {
   widgetClassPath=&Image
   bIsVariable=true
   namedSlotHost=StarterIslandsTab
   widget=RemoveSpinner_1
   properties {
    visibility=Collapsed
    brush {
     bIsDynamicallyLoaded=false
     drawAs=Image
     imageType=NoImage
     mirroring=NoMirror
     resourceName=None
     resourceObject=null
     tiling=NoTile
     imageSize=<46 32>
     margin=<ltrb 0 0 0 0>
     outlineSettings {
      bUseBrushTransparency=false
      roundingType=HalfHeightRadius
      width=0
      color {colorUseRule=UseColor_Specified, specifiedColor=<rgba 0 0 0 0>}
      cornerRadii=<0 0 0 0>
     }
     tintColor {colorUseRule=UseColor_Specified, specifiedColor=<rgba 1 1 1 1>}
     uVRegion {bIsValid=false, max=<0 0>, min=<0 0>}
    }
   }
  }
  TemplateIslandsTab {
   widgetClassPath=&IconTextButton
   bIsVariable=true
   parent=HorizontalBox_144
   widget=TemplateIslandsTab
   properties {
    bHideInputAction=true
    bIsFocusable=false
    buttonText=Template
    onButtonBaseClicked=(&self:WidgetTree.TemplateIslandsTab.HandleOnButtonClicked)
    onButtonBaseDoubleClicked=(&self:WidgetTree.TemplateIslandsTab.HandleOnButtonDoubleClicked)
    onButtonBaseHovered=(&self:WidgetTree.TemplateIslandsTab.HandleOnButtonHovered)
    onButtonBaseUnhovered=(&self:WidgetTree.TemplateIslandsTab.HandleOnButtonUnhovered)
    onSelectedChangedBase=(&self:WidgetTree.TemplateIslandsTab.HandleOnSelectedChanged)
    style=&Athena_ButtonStyle_ChallengesTab_C
    visibility=Visible
   }
   slot {class=HorizontalBoxSlot, properties {verticalAlignment=VAlign_Center, padding=<ltrb 10 0 0 0>, size {sizeRule=Fill, value=1}}}
  }
  RemoveSpinner_2 {
   widgetClassPath=&Image
   bIsVariable=true
   namedSlotHost=TemplateIslandsTab
   widget=RemoveSpinner_2
   properties {
    visibility=Collapsed
    brush {
     bIsDynamicallyLoaded=false
     drawAs=Image
     imageType=NoImage
     mirroring=NoMirror
     resourceName=None
     resourceObject=null
     tiling=NoTile
     imageSize=<46 32>
     margin=<ltrb 0 0 0 0>
     outlineSettings {
      bUseBrushTransparency=false
      roundingType=HalfHeightRadius
      width=0
      color {colorUseRule=UseColor_Specified, specifiedColor=<rgba 0 0 0 0>}
      cornerRadii=<0 0 0 0>
     }
     tintColor {colorUseRule=UseColor_Specified, specifiedColor=<rgba 1 1 1 1>}
     uVRegion {bIsValid=false, max=<0 0>, min=<0 0>}
    }
   }
  }
  CollabIslandsTab {
   widgetClassPath=&IconTextButton
   bIsVariable=true
   parent=HorizontalBox_144
   widget=CollabIslandsTab
   properties {
    bHideInputAction=true
    bIsFocusable=false
    buttonText=Collab
    onButtonBaseClicked=()
    onButtonBaseDoubleClicked=()
    onButtonBaseHovered=()
    onButtonBaseUnhovered=()
    onSelectedChangedBase=()
    style=&Athena_ButtonStyle_ChallengesTab_C
    visibility=Visible
   }
   slot {class=HorizontalBoxSlot, properties {verticalAlignment=VAlign_Center, padding=<ltrb 10 0 0 0>, size {sizeRule=Fill, value=1}}}
  }
  RemoveSpinner {
   widgetClassPath=&Image
   bIsVariable=true
   namedSlotHost=CollabIslandsTab
   widget=RemoveSpinner
   properties {
    visibility=Collapsed
    brush {
     bIsDynamicallyLoaded=false
     drawAs=Image
     imageType=NoImage
     mirroring=NoMirror
     resourceName=None
     resourceObject=null
     tiling=NoTile
     imageSize=<46 32>
     margin=<ltrb 0 0 0 0>
     outlineSettings {
      bUseBrushTransparency=false
      roundingType=HalfHeightRadius
      width=0
      color {colorUseRule=UseColor_Specified, specifiedColor=<rgba 0 0 0 0>}
      cornerRadii=<0 0 0 0>
     }
     tintColor {colorUseRule=UseColor_Specified, specifiedColor=<rgba 1 1 1 1>}
     uVRegion {bIsValid=false, max=<0 0>, min=<0 0>}
    }
   }
  }
  RightTabButton {
   widgetClassPath=&UserActionMenuInputButton
   bIsVariable=true
   parent=HorizontalBox_144
   widget=RightTabButton
   properties {
    onButtonBaseClicked=(&self:WidgetTree.RightTabButton.HandleOnButtonClicked)
    onButtonBaseDoubleClicked=(&self:WidgetTree.RightTabButton.HandleOnButtonDoubleClicked)
    onButtonBaseHovered=(&self:WidgetTree.RightTabButton.HandleOnButtonHovered)
    onButtonBaseUnhovered=(&self:WidgetTree.RightTabButton.HandleOnButtonUnhovered)
    onSelectedChangedBase=(&self:WidgetTree.RightTabButton.HandleOnSelectedChanged)
    triggeringInputAction {dataTable=&InputActionDataTable, rowName=GenericRightShoulder}
   }
   slot {class=HorizontalBoxSlot, properties {padding=<ltrb 5 0 10 0>}}
  }
  SizeBox_1 {
   widgetClassPath=&SizeBox
   parent=VBox_LeftSide
   widget=SizeBox_1
   properties {bOverride_WidthOverride=true, widthOverride=1500}
   slot {class=VerticalBoxSlot, properties {horizontalAlignment=HAlign_Center, padding=<ltrb 0 12 0 0>, size {sizeRule=Fill, value=1}}}
  }
  IslandTemplatesWidgetSwitcher {
   widgetClassPath=&WidgetSwitcher
   bIsVariable=true
   parent=SizeBox_1
   widget=IslandTemplatesWidgetSwitcher
   slot {class=SizeBoxSlot}
  }
  StartingIslandsTileView {
   widgetClassPath=/Script/CommonUI.CommonTileView
   bInherited=true
   bIsVariable=true
   parent=IslandTemplatesWidgetSwitcher
   widget=StartingIslandsTileView
   properties {
    clipping=ClipToBoundsWithoutIntersecting
    entryHeight=150
    entryWidgetClass=&/IslandTemplateOptionEntry_Base.IslandTemplateOptionEntry_Base_C
    entryWidth=150
    horizontalEntrySpacing=12
    numDesignerPreviewEntries=17
    tileAlignment=EvenlySize
    verticalEntrySpacing=12
   }
   slot {class=WidgetSwitcherSlot, properties {padding=<ltrb 64 0 64 64>}}
  }
  TemplateIslandsTileView {
   widgetClassPath=&CommonTileView
   bInherited=true
   bIsVariable=true
   parent=IslandTemplatesWidgetSwitcher
   widget=TemplateIslandsTileView
   properties {
    clipping=ClipToBoundsWithoutIntersecting
    entryHeight=136
    entryWidgetClass=&IslandTemplateOptionEntry_Base_C
    entryWidth=244
    horizontalEntrySpacing=8
    numDesignerPreviewEntries=17
    tileAlignment=EvenlySize
    verticalEntrySpacing=8
   }
   slot {class=WidgetSwitcherSlot, properties {padding=<ltrb 56 0 56 64>}}
  }
  CollabIslandsTileView {
   widgetClassPath=&CommonTileView
   bInherited=true
   bIsVariable=true
   parent=IslandTemplatesWidgetSwitcher
   widget=CollabIslandsTileView
   properties {
    clipping=ClipToBoundsWithoutIntersecting
    entryHeight=136
    entryWidgetClass=&IslandTemplateOptionEntry_Base_C
    entryWidth=244
    horizontalEntrySpacing=8
    numDesignerPreviewEntries=17
    tileAlignment=EvenlySize
    verticalEntrySpacing=8
   }
   slot {class=WidgetSwitcherSlot, properties {padding=<ltrb 56 0 56 64>}}
  }
  CollabTermsAndConditonsBox {widgetClassPath=&SizeBox, bIsVariable=true, parent=VBox_LeftSide, widget=CollabTermsAndConditonsBox, slot {class=VerticalBoxSlot}}
  Overlay_134 {widgetClassPath=&Overlay, parent=CollabTermsAndConditonsBox, widget=Overlay_134, slot {class=SizeBoxSlot}}
  Image_DisclaimerBG {
   widgetClassPath=&Image
   bIsVariable=true
   parent=Overlay_134
   widget=Image_DisclaimerBG
   properties {
    brush {
     bIsDynamicallyLoaded=false
     drawAs=Image
     imageType=NoImage
     mirroring=NoMirror
     resourceName=None
     resourceObject=null
     tiling=NoTile
     imageSize=<32 150>
     margin=<ltrb 0 0 0 0>
     outlineSettings {
      bUseBrushTransparency=false
      roundingType=HalfHeightRadius
      width=0
      color {colorUseRule=UseColor_Specified, specifiedColor=<rgba 0 0 0 0>}
      cornerRadii=<0 0 0 0>
     }
     tintColor {colorUseRule=UseColor_Specified, specifiedColor=<rgba 1 1 1 1>}
     uVRegion {bIsValid=false, max=<0 0>, min=<0 0>}
    }
    colorAndOpacity=<rgba 0 0.0203 0.0953 1>
   }
   slot {class=OverlaySlot, properties {horizontalAlignment=HAlign_Fill, verticalAlignment=VAlign_Fill, padding=<ltrb 0 0 17 0>}}
  }
  HorizontalBox_240 {
   widgetClassPath=&HorizontalBox
   parent=Overlay_134
   widget=HorizontalBox_240
   slot {class=OverlaySlot, properties {horizontalAlignment=HAlign_Fill, verticalAlignment=VAlign_Fill, padding=<ltrb 0 0 30 0>}}
  }
  CollabPartnerIconOld {
   widgetClassPath=&Image
   parent=HorizontalBox_240
   widget=CollabPartnerIconOld
   properties {
    visibility=Collapsed
    brush {
     bIsDynamicallyLoaded=false
     drawAs=Image
     imageType=NoImage
     mirroring=NoMirror
     resourceName=None
     resourceObject=null
     tiling=NoTile
     imageSize=<64 64>
     margin=<ltrb 0 0 0 0>
     outlineSettings {
      bUseBrushTransparency=false
      roundingType=HalfHeightRadius
      width=0
      color {colorUseRule=UseColor_Specified, specifiedColor=<rgba 0 0 0 0>}
      cornerRadii=<0 0 0 0>
     }
     tintColor {colorUseRule=UseColor_Specified, specifiedColor=<rgba 1 1 1 1>}
     uVRegion {bIsValid=false, max=<0 0>, min=<0 0>}
    }
   }
   slot {class=HorizontalBoxSlot, properties {horizontalAlignment=HAlign_Left, verticalAlignment=VAlign_Center, padding=<ltrb 40 40 40 40>}}
  }
  CollabPartnerIcon {
   widgetClassPath=&FortLazyImage
   bInherited=true
   parent=HorizontalBox_240
   widget=CollabPartnerIcon
   properties {brush {
     bIsDynamicallyLoaded=false
     drawAs=Image
     imageType=NoImage
     mirroring=NoMirror
     resourceName=None
     resourceObject=null
     tiling=NoTile
     imageSize=<50 50>
     margin=<ltrb 0 0 0 0>
     outlineSettings {
      bUseBrushTransparency=false
      roundingType=HalfHeightRadius
      width=0
      color {colorUseRule=UseColor_Specified, specifiedColor=<rgba 0 0 0 0>}
      cornerRadii=<0 0 0 0>
     }
     tintColor {colorUseRule=UseColor_Specified, specifiedColor=<rgba 1 1 1 1>}
     uVRegion {bIsValid=false, max=<0 0>, min=<0 0>}
    }}
   slot {class=HorizontalBoxSlot, properties {horizontalAlignment=HAlign_Left, verticalAlignment=VAlign_Center, padding=<ltrb 40 40 40 40>}}
  }
  CollabDisclaimerText {
   widgetClassPath=&CommonRichTextBlock
   bIsVariable=true
   parent=HorizontalBox_240
   widget=CollabDisclaimerText
   properties {
    autoWrapText=true
    bOverrideDefaultStyle=true
    defaultTextStyleOverrideClass=/Game/Athena/Settings/SettingEntries/Styles/TextStyle_HeadingNow_75_19_LightGray.TextStyle_HeadingNow_75_19_LightGray_C
    text="In order to Publish a LEGO® island you are required to review the terms and condtions and understand the revenue rules."
    defaultTextStyleOverride {
     overflowPolicy=Clip
     transformPolicy=None
     colorAndOpacity {colorUseRule=UseColor_Specified, specifiedColor=<rgba 1 1 1 1>}
     font {
      bForceMonospaced=false
      bMaterialIsStencil=false
      fontMaterial=null
      fontObject=&NotoSans
      letterSpacing=0
      monospacedWidth=1
      size=24
      skewAmount=0
      typefaceFontName=NotoSans-Regular
      outlineSettings {
       bApplyOutlineToDropShadows=false
       bMiteredCorners=false
       bSeparateFillAlpha=false
       outlineMaterial=null
       outlineSize=0
       outlineColor=<rgba 0 0 0 1>
      }
     }
     highlightColor {colorUseRule=UseColor_Specified, specifiedColor=<rgba 0 0 0 1>}
     highlightShape {
      bIsDynamicallyLoaded=false
      drawAs=Image
      imageType=NoImage
      mirroring=NoMirror
      resourceName=None
      resourceObject=null
      tiling=NoTile
      imageSize=<32 32>
      margin=<ltrb 0 0 0 0>
      outlineSettings {
       bUseBrushTransparency=false
       roundingType=HalfHeightRadius
       width=0
       color {colorUseRule=UseColor_Specified, specifiedColor=<rgba 0 0 0 0>}
       cornerRadii=<0 0 0 0>
      }
      tintColor {colorUseRule=UseColor_Specified, specifiedColor=<rgba 1 1 1 1>}
      uVRegion {bIsValid=false, max=<0 0>, min=<0 0>}
     }
     selectedBackgroundColor {colorUseRule=UseColor_Foreground, specifiedColor=<rgba 1 0 1 1>}
     shadowColorAndOpacity=<rgba 0 0 0 1>
     shadowOffset=<0 0>
     strikeBrush {
      bIsDynamicallyLoaded=false
      drawAs=Image
      imageType=NoImage
      mirroring=NoMirror
      resourceName=None
      resourceObject=null
      tiling=NoTile
      imageSize=<32 32>
      margin=<ltrb 0 0 0 0>
      outlineSettings {
       bUseBrushTransparency=false
       roundingType=HalfHeightRadius
       width=0
       color {colorUseRule=UseColor_Specified, specifiedColor=<rgba 0 0 0 0>}
       cornerRadii=<0 0 0 0>
      }
      tintColor {colorUseRule=UseColor_Specified, specifiedColor=<rgba 1 1 1 1>}
      uVRegion {bIsValid=false, max=<0 0>, min=<0 0>}
     }
     underlineBrush {
      bIsDynamicallyLoaded=false
      drawAs=Image
      imageType=NoImage
      mirroring=NoMirror
      resourceName=None
      resourceObject=null
      tiling=NoTile
      imageSize=<32 32>
      margin=<ltrb 0 0 0 0>
      outlineSettings {
       bUseBrushTransparency=false
       roundingType=HalfHeightRadius
       width=0
       color {colorUseRule=UseColor_Specified, specifiedColor=<rgba 0 0 0 0>}
       cornerRadii=<0 0 0 0>
      }
      tintColor {colorUseRule=UseColor_Specified, specifiedColor=<rgba 1 1 1 1>}
      uVRegion {bIsValid=false, max=<0 0>, min=<0 0>}
     }
    }
   }
   slot {class=HorizontalBoxSlot, properties {verticalAlignment=VAlign_Center, size {sizeRule=Fill, value=1}}}
  }
  DetailsButton {
   widgetClassPath=&WBP_UIKit_Button_Regular
   bIsVariable=true
   parent=HorizontalBox_240
   widget=DetailsButton
   properties {text=Details, visibility=Collapsed, triggeringInputAction {dataTable=&InputActionDataTable, rowName=Details}}
   slot {class=HorizontalBoxSlot, properties {horizontalAlignment=HAlign_Right, verticalAlignment=VAlign_Center, padding=<ltrb 40 0 40 0>}}
  }
  SBox-RightSide {
   widgetClassPath=&SizeBox
   parent=HorizontalBox_3
   widget=SBox-RightSide
   properties {bOverride_WidthOverride=true, widthOverride=600}
   slot {class=HorizontalBoxSlot}
  }
  Overlay_113 {widgetClassPath=&Overlay, parent=SBox-RightSide, widget=Overlay_113, slot {class=SizeBoxSlot}}
  Image_868 {
   widgetClassPath=&Image
   parent=Overlay_113
   widget=Image_868
   properties {
    visibility=SelfHitTestInvisible
    brush {
     bIsDynamicallyLoaded=false
     drawAs=Image
     imageType=NoImage
     mirroring=NoMirror
     resourceName=None
     resourceObject=/Game/Athena/UI/InGame/Creative/Art/M_CreativeCreateIslandRightSideBacking
     tiling=NoTile
     imageSize=<32 32>
     margin=<ltrb 0 0 0 0>
     outlineSettings {
      bUseBrushTransparency=false
      roundingType=HalfHeightRadius
      width=0
      color {colorUseRule=UseColor_Specified, specifiedColor=<rgba 0 0 0 0>}
      cornerRadii=<0 0 0 0>
     }
     tintColor {colorUseRule=UseColor_Specified, specifiedColor=<rgba 1 1 1 1>}
     uVRegion {bIsValid=false, max=<0 0>, min=<0 0>}
    }
    colorAndOpacity=<rgba 0 0 0 0.2>
   }
   slot {class=OverlaySlot, properties {horizontalAlignment=HAlign_Fill, verticalAlignment=VAlign_Fill, padding=<ltrb -96 0 0 0>}}
  }
  VBox_IslandDetails {
   widgetClassPath=&VerticalBox
   bIsVariable=true
   parent=Overlay_113
   widget=VBox_IslandDetails
   slot {class=OverlaySlot, properties {horizontalAlignment=HAlign_Fill, verticalAlignment=VAlign_Fill, padding=<ltrb 0 145 0 0>}}
  }
  ScrollBox_Details {
   widgetClassPath=&ScrollBox
   bIsVariable=true
   parent=VBox_IslandDetails
   widget=ScrollBox_Details
   slot {class=VerticalBoxSlot, properties {padding=<ltrb 24 0 24 12>, size {sizeRule=Fill, value=1}}}
  }
  IslandTemplateNameText {
   widgetClassPath=&CommonTextBlock
   bInherited=true
   parent=ScrollBox_Details
   widget=IslandTemplateNameText
   properties {
    autoWrapText=true
    justification=Center
    text="Pick an Island!"
    textTransformPolicy=ToUpper
    font {
     bForceMonospaced=false
     bMaterialIsStencil=false
     fontMaterial=null
     fontObject=&Burbank
     letterSpacing=0
     monospacedWidth=1
     size=36
     skewAmount=0
     typefaceFontName=BurbankRegular-Black
     outlineSettings {
      bApplyOutlineToDropShadows=false
      bMiteredCorners=false
      bSeparateFillAlpha=false
      outlineMaterial=null
      outlineSize=0
      outlineColor=<rgba 0 0 0 1>
     }
    }
    shadowOffset=<0 0>
   }
   slot {class=ScrollBoxSlot}
  }
  Switcher_CreatorInfo {
   widgetClassPath=&CommonWidgetSwitcherLegacy
   bIsVariable=true
   parent=ScrollBox_Details
   widget=Switcher_CreatorInfo
   slot {class=ScrollBoxSlot, properties {padding=<ltrb 0 22 0 0>}}
  }
  IslandDescriptionText {
   widgetClassPath=&CommonTextBlock
   bInherited=true
   parent=Switcher_CreatorInfo
   widget=IslandDescriptionText
   properties {
    autoWrapText=true
    justification=Center
    font {
     bForceMonospaced=false
     bMaterialIsStencil=false
     fontMaterial=null
     fontObject=&Burbank
     letterSpacing=0
     monospacedWidth=1
     size=18
     skewAmount=0
     typefaceFontName=Burbank-PlayerPower
     outlineSettings {
      bApplyOutlineToDropShadows=false
      bMiteredCorners=false
      bSeparateFillAlpha=false
      outlineMaterial=null
      outlineSize=0
      outlineColor=<rgba 0 0 0 1>
     }
    }
    shadowOffset=<0 0>
   }
   slot {class=WidgetSwitcherSlot, properties {padding=<ltrb 0 0 0 32>}}
  }
  VBox_TemplateDescription {
   widgetClassPath=&VerticalBox
   bIsVariable=true
   parent=Switcher_CreatorInfo
   widget=VBox_TemplateDescription
   slot {class=WidgetSwitcherSlot}
  }
  PublishedIslandBulletOne {
   widgetClassPath=&CommonTextBlock
   bInherited=true
   parent=VBox_TemplateDescription
   widget=PublishedIslandBulletOne
   properties {
    autoWrapText=true
    font {
     bForceMonospaced=false
     bMaterialIsStencil=false
     fontMaterial=null
     fontObject=&Burbank
     letterSpacing=0
     monospacedWidth=1
     size=18
     skewAmount=0
     typefaceFontName=Burbank-PlayerPower
     outlineSettings {
      bApplyOutlineToDropShadows=false
      bMiteredCorners=false
      bSeparateFillAlpha=false
      outlineMaterial=null
      outlineSize=0
      outlineColor=<rgba 0 0 0 1>
     }
    }
    shadowOffset=<4.9266 5.2555>
   }
   slot {class=VerticalBoxSlot}
  }
  PublishedIslandBulletTwo {
   widgetClassPath=&CommonTextBlock
   bInherited=true
   parent=VBox_TemplateDescription
   widget=PublishedIslandBulletTwo
   properties {
    autoWrapText=true
    font {
     bForceMonospaced=false
     bMaterialIsStencil=false
     fontMaterial=null
     fontObject=&Burbank
     letterSpacing=0
     monospacedWidth=1
     size=18
     skewAmount=0
     typefaceFontName=Burbank-PlayerPower
     outlineSettings {
      bApplyOutlineToDropShadows=false
      bMiteredCorners=false
      bSeparateFillAlpha=false
      outlineMaterial=null
      outlineSize=0
      outlineColor=<rgba 0 0 0 1>
     }
    }
    shadowOffset=<4.9266 5.2555>
   }
   slot {class=VerticalBoxSlot}
  }
  PublishedIslandBulletThree {
   widgetClassPath=&CommonTextBlock
   bInherited=true
   parent=VBox_TemplateDescription
   widget=PublishedIslandBulletThree
   properties {
    autoWrapText=true
    font {
     bForceMonospaced=false
     bMaterialIsStencil=false
     fontMaterial=null
     fontObject=&Burbank
     letterSpacing=0
     monospacedWidth=1
     size=18
     skewAmount=0
     typefaceFontName=Burbank-PlayerPower
     outlineSettings {
      bApplyOutlineToDropShadows=false
      bMiteredCorners=false
      bSeparateFillAlpha=false
      outlineMaterial=null
      outlineSize=0
      outlineColor=<rgba 0 0 0 1>
     }
    }
    shadowOffset=<4.9266 5.2555>
   }
   slot {class=VerticalBoxSlot}
  }
  Overlay_Creator {
   widgetClassPath=&Overlay
   bIsVariable=true
   parent=VBox_IslandDetails
   widget=Overlay_Creator
   properties {visibility=Collapsed}
   slot {class=VerticalBoxSlot, properties {verticalAlignment=VAlign_Bottom, padding=<ltrb 0 0 0 10>}}
  }
  Image_1158 {
   widgetClassPath=&Image
   parent=Overlay_Creator
   widget=Image_1158
   properties {
    visibility=SelfHitTestInvisible
    brush {
     bIsDynamicallyLoaded=false
     drawAs=Image
     imageType=NoImage
     mirroring=NoMirror
     resourceName=None
     resourceObject=/Game/Athena/UI/InGame/Creative/Art/M_CreativeCreateIslandRightSideBottomBacking
     tiling=NoTile
     imageSize=<32 32>
     margin=<ltrb 0 0 0 0>
     outlineSettings {
      bUseBrushTransparency=false
      roundingType=HalfHeightRadius
      width=0
      color {colorUseRule=UseColor_Specified, specifiedColor=<rgba 0 0 0 0>}
      cornerRadii=<0 0 0 0>
     }
     tintColor {colorUseRule=UseColor_Specified, specifiedColor=<rgba 1 1 1 1>}
     uVRegion {bIsValid=false, max=<0 0>, min=<0 0>}
    }
    colorAndOpacity=<rgba 0 0 0 0.2>
   }
   slot {class=OverlaySlot, properties {horizontalAlignment=HAlign_Fill, verticalAlignment=VAlign_Fill, padding=<ltrb -102 0 0 0>}}
  }
  VBox_Creator {
   widgetClassPath=&VerticalBox
   parent=Overlay_Creator
   widget=VBox_Creator
   slot {class=OverlaySlot, properties {horizontalAlignment=HAlign_Fill, verticalAlignment=VAlign_Fill, padding=<ltrb 24 12 24 12>}}
  }
  PublishedIslandName {
   widgetClassPath=&CommonTextBlock
   bInherited=true
   parent=VBox_Creator
   widget=PublishedIslandName
   properties {
    autoWrapText=true
    text="[Published Island Name]"
    textTransformPolicy=ToUpper
    font {
     bForceMonospaced=false
     bMaterialIsStencil=false
     fontMaterial=null
     fontObject=&Burbank
     letterSpacing=0
     monospacedWidth=1
     size=18
     skewAmount=0
     typefaceFontName=Burbank-PlayerPower
     outlineSettings {
      bApplyOutlineToDropShadows=false
      bMiteredCorners=false
      bSeparateFillAlpha=false
      outlineMaterial=null
      outlineSize=0
      outlineColor=<rgba 0 0 0 1>
     }
    }
    shadowOffset=<4.9266 5.2555>
   }
   slot {class=VerticalBoxSlot}
  }
  HorizontalBox_5 {
   widgetClassPath=&HorizontalBox
   parent=VBox_Creator
   widget=HorizontalBox_5
   slot {class=VerticalBoxSlot, properties {padding=<ltrb 0 0 0 8>}}
  }
  Switcher_CreatorType {
   widgetClassPath=&CommonWidgetSwitcherLegacy
   bIsVariable=true
   parent=HorizontalBox_5
   widget=Switcher_CreatorType
   properties {transitionDuration=0}
   slot {class=HorizontalBoxSlot, properties {verticalAlignment=VAlign_Center}}
  }
  HBox_Creator {
   widgetClassPath=&HorizontalBox
   parent=Switcher_CreatorType
   widget=HBox_Creator
   slot {class=WidgetSwitcherSlot, properties {verticalAlignment=VAlign_Center}}
  }
  CommonTextBlock_102 {
   widgetClassPath=&CommonTextBlock
   parent=HBox_Creator
   widget=CommonTextBlock_102
   properties {
    text=CREATED BY
    colorAndOpacity {colorUseRule=UseColor_Specified, specifiedColor=<rgba 0.5271 1 0.8228 1>}
    font {
     bForceMonospaced=false
     bMaterialIsStencil=false
     fontMaterial=null
     fontObject=&Burbank
     letterSpacing=0
     monospacedWidth=1
     size=18
     skewAmount=0
     typefaceFontName=Burbank-Black
     outlineSettings {
      bApplyOutlineToDropShadows=false
      bMiteredCorners=false
      bSeparateFillAlpha=false
      outlineMaterial=null
      outlineSize=0
      outlineColor=<rgba 0 0 0 1>
     }
    }
    renderTransform {angle=0, scale=<1 1>, shear=<-8 0>, translation=<0 0>}
    shadowOffset=<0 0>
   }
   slot {class=HorizontalBoxSlot, properties {verticalAlignment=VAlign_Center, padding=<ltrb 0 0 6 0>}}
  }
  SizeBox_2 {
   widgetClassPath=&SizeBox
   parent=HBox_Creator
   widget=SizeBox_2
   properties {bOverride_WidthOverride=true, widthOverride=277.9369}
   slot {class=HorizontalBoxSlot, properties {horizontalAlignment=HAlign_Left, verticalAlignment=VAlign_Center, padding=<ltrb 0 0 0 -3>}}
  }
  PublishedIslandAuthor {
   widgetClassPath=&CommonTextBlock
   bInherited=true
   parent=SizeBox_2
   widget=PublishedIslandAuthor
   properties {
    scrollStyle=&TextScrollStyle-Base_C
    text="[creator name]"
    textTransformPolicy=ToUpper
    colorAndOpacity {colorUseRule=UseColor_Specified, specifiedColor=<rgba 0 1 0.6654 1>}
    font {
     bForceMonospaced=false
     bMaterialIsStencil=false
     fontMaterial=null
     fontObject=&Burbank
     letterSpacing=0
     monospacedWidth=1
     size=18
     skewAmount=0
     typefaceFontName=Burbank-Black
     outlineSettings {
      bApplyOutlineToDropShadows=false
      bMiteredCorners=false
      bSeparateFillAlpha=false
      outlineMaterial=null
      outlineSize=0
      outlineColor=<rgba 0 0 0 1>
     }
    }
    renderTransform {angle=0, scale=<1 1>, shear=<-8 0>, translation=<0 0>}
    shadowOffset=<4.9266 5.2555>
   }
   slot {class=SizeBoxSlot}
  }
  Overlay_Epic {
   widgetClassPath=&Overlay
   parent=Switcher_CreatorType
   widget=Overlay_Epic
   slot {class=WidgetSwitcherSlot, properties {horizontalAlignment=HAlign_Left}}
  }
  Image_164 {
   widgetClassPath=&Image
   parent=Overlay_Epic
   widget=Image_164
   properties {brush {
     bIsDynamicallyLoaded=false
     drawAs=Image
     imageType=NoImage
     mirroring=NoMirror
     resourceName=None
     resourceObject=/Game/UI/Login/Art/T_EpicLogo
     tiling=NoTile
     imageSize=<64 64>
     margin=<ltrb 0 0 0 0>
     outlineSettings {
      bUseBrushTransparency=false
      roundingType=HalfHeightRadius
      width=0
      color {colorUseRule=UseColor_Specified, specifiedColor=<rgba 0 0 0 0>}
      cornerRadii=<0 0 0 0>
     }
     tintColor {colorUseRule=UseColor_Specified, specifiedColor=<rgba 1 1 1 1>}
     uVRegion {bIsValid=false, max=<0 0>, min=<0 0>}
    }}
   slot {class=OverlaySlot, properties {verticalAlignment=VAlign_Fill}}
  }
  PublishedIslandCode {
   widgetClassPath=&CommonTextBlock
   bInherited=true
   parent=HorizontalBox_5
   widget=PublishedIslandCode
   properties {
    text=8888-8888-8888
    textTransformPolicy=ToUpper
    colorAndOpacity {colorUseRule=UseColor_Specified, specifiedColor=<rgba 0.502 0.8314 1 1>}
    font {
     bForceMonospaced=false
     bMaterialIsStencil=false
     fontMaterial=null
     fontObject=&Burbank
     letterSpacing=0
     monospacedWidth=1
     size=18
     skewAmount=0
     typefaceFontName=Burbank-Black
     outlineSettings {
      bApplyOutlineToDropShadows=false
      bMiteredCorners=false
      bSeparateFillAlpha=false
      outlineMaterial=null
      outlineSize=0
      outlineColor=<rgba 0 0 0 1>
     }
    }
    shadowOffset=<4.9266 5.2555>
   }
   slot {class=HorizontalBoxSlot, properties {horizontalAlignment=HAlign_Right, verticalAlignment=VAlign_Center, size {sizeRule=Fill, value=1}}}
  }
  CreateButtonSwitcher {
   widgetClassPath=&CommonVisibilitySwitcher
   bIsVariable=true
   parent=VBox_IslandDetails
   widget=CreateButtonSwitcher
   slot {class=VerticalBoxSlot, properties {horizontalAlignment=HAlign_Center, padding=<ltrb 0 0 0 24>}}
  }
  AcceptButton {
   widgetClassPath=&WBP_UIKit_Button_Generic
   bInherited=true
   bIsVariable=true
   parent=CreateButtonSwitcher
   widget=AcceptButton
   properties {
    backgroundMaterial=/FortUILibrary/UIKit/Buttons/MaterialInstances/CTAButtons/MI_UIKit_BgLoud
    minWidth=356
    text=Confirm
    buttonPadding=<ltrb 60 0 60 0>
    promptNudge=<10 0>
    triggeringInputAction {dataTable=&InputActionDataTable, rowName=Confirm}
   }
   slot {class=CommonVisibilitySwitcherSlot}
  }
  SizeBox_3 {widgetClassPath=&SizeBox, namedSlotHost=AcceptButton, widget=SizeBox_3}
  CreateSpinner {
   widgetClassPath=&Image
   bIsVariable=true
   parent=SizeBox_3
   widget=CreateSpinner
   properties {
    visibility=Collapsed
    brush {
     bIsDynamicallyLoaded=false
     drawAs=Image
     imageType=NoImage
     mirroring=NoMirror
     resourceName=None
     resourceObject=null
     tiling=NoTile
     imageSize=<42 42>
     margin=<ltrb 0 0 0 0>
     outlineSettings {
      bUseBrushTransparency=false
      roundingType=HalfHeightRadius
      width=0
      color {colorUseRule=UseColor_Specified, specifiedColor=<rgba 0 0 0 0>}
      cornerRadii=<0 0 0 0>
     }
     tintColor {colorUseRule=UseColor_Specified, specifiedColor=<rgba 0 0 0 1>}
     uVRegion {bIsValid=false, max=<0 0>, min=<0 0>}
    }
   }
   slot {class=SizeBoxSlot, properties {horizontalAlignment=HAlign_Right, verticalAlignment=VAlign_Center, padding=<ltrb 0 0 10 0>}}
  }
  JoinCollabButton {
   widgetClassPath=&WBP_UIKit_Button_Generic
   bIsVariable=true
   parent=CreateButtonSwitcher
   widget=JoinCollabButton
   properties {
    backgroundMaterial=&MI_UIKit_BgLoud
    minWidth=356
    text=Join Program
    promptNudge=<10 0>
    triggeringInputAction {dataTable=&InputActionDataTable, rowName=Join}
   }
   slot {class=CommonVisibilitySwitcherSlot}
  }
  VisibilityBorder_Touch {
   widgetClassPath=&CommonHardwareVisibilityBorder
   parent=Overlay_1
   widget=VisibilityBorder_Touch
   properties {visibilityQuery {
     autoDescription=" ANY( Input.Touch )"
     tokenStreamVersion=0
     userDescription=""
     queryTokenStream [0,1,1,1,0]
     tagDictionary [{tagName=Input.Touch}]
    }}
   slot {class=OverlaySlot, properties {horizontalAlignment=HAlign_Right, padding=<ltrb 0 20 20 0>}}
  }
  Button_CloseTouch {
   widgetClassPath=/FortUILibrary/Foundation/Buttons/CloseButton
   bIsVariable=true
   parent=VisibilityBorder_Touch
   widget=Button_CloseTouch
   properties {
    onButtonBaseClicked=(&self:WidgetTree.Button_CloseTouch.HandleOnButtonClicked)
    onButtonBaseDoubleClicked=(&self:WidgetTree.Button_CloseTouch.HandleOnButtonDoubleClicked)
    onButtonBaseHovered=(&self:WidgetTree.Button_CloseTouch.HandleOnButtonHovered)
    onButtonBaseUnhovered=(&self:WidgetTree.Button_CloseTouch.HandleOnButtonUnhovered)
    onSelectedChangedBase=(&self:WidgetTree.Button_CloseTouch.HandleOnSelectedChanged)
   }
   slot {class=BorderSlot, properties {padding=<ltrb 0 0 0 0>}}
  }
 ]
}
