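# Island select list entry widget, built on AthenaCreativeItemTileButton, that represents one island plot in the portal UI.
# When its list item is set, it caches the IslandDataEntry and shows the map name, plus published code and version if a code exists.
# Clicks pass the island's McpId to SelectPlot on the parent, either IslandSelect_UI or IslandSelect_UI_V2_Modal.
# Hover scales the entry and swaps name style; selection shows the input-action prompt.
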WidgetBlueprint /Game/Playgrounds/Items/Portals/UI/IslandOptionEntry
animations [Hovered {
  binding_count=1
  length_seconds=0.2
  object=&self:Hovered
  sequence {bindings [IslandOptionEntry_C_107 {
     id=#1
     possessed_class=IslandOptionEntry_C
     tracks [{
       type=MovieScene2DTransformTrack
       sections [{
         type=MovieScene2DTransformSection
         end_frame=4
         start_frame=0
         channels [
          Translation.X {type=MovieSceneScriptingFloatChannel}
          Translation.Y {type=MovieSceneScriptingFloatChannel}
          Angle {type=MovieSceneScriptingFloatChannel}
          Scale.X {type=MovieSceneScriptingFloatChannel, keys [[0,1],[4,1.1073]]}
          Scale.Y {type=MovieSceneScriptingFloatChannel, keys [[0,1],[4,1.1073]]}
          Shear.X {type=MovieSceneScriptingFloatChannel}
          Shear.Y {type=MovieSceneScriptingFloatChannel}
         ]
        }]
      }]
    }]}
 }]
graphs [
 EventGraph {nodes [
   n1 = PlayAnimationForward(InAnimation=Get Selected() -> (Selected:Widget Animation Object Reference))
   n2 = Get ParentWidget() -> (ParentWidget:Fort Activatable Panel Object Reference)
   n3 = Cast To IslandSelect_UI!(execute=[@n1.then, @n26.Is Valid], Object=@n2) -> (then, CastFailed, AsIsland Select UI:Island Select UI Object Reference) {is=title K2Node_DynamicCast}
   n4 = Get IslandData() -> (IslandData:Island Data Entry Object Reference)
   n5 = Get McpId!(self:Island Data Entry Object Reference=@n4) -> (McpId:String)
   > SelectPlot!(execute, self:Island Select UI Object Reference=@n3, SelectedPlot:String=@n5) -> (then)
   n6 = Cast To IslandSelect_UI_V2_Modal!(execute=@n3.CastFailed, Object=@n2) -> (then, CastFailed, AsIsland Select UI V2 Modal:Island Select UI V2 Modal Object Reference) {is=title K2Node_DynamicCast}
   > SelectPlot(self:Island Select UI V2 Modal Object Reference=@n6, SelectedPlot=@n5)
   n7 = Event On Selected()
   n8 = Get InputActionWidget() -> (InputActionWidget:Common Action Widget Object Reference)
   SetVisibility(execute=[@n24.then, @n7.then], self=@n8, InVisibility=SelfHitTestInvisible)
   > PlaySound2D()
   n9 = Event OnListItemObjectSet()
   > n10 = Sequence()
   > n11 = Cast To IslandDataEntry!(execute, Object=@n9.ListItemObject) -> (then, CastFailed, AsIsland Data Entry:Island Data Entry Object Reference) {is=title K2Node_DynamicCast}
   > n12 = Set IslandData(IslandData:Island Data Entry Object Reference=@n11) -> (Output_Get:Island Data Entry Object Reference)
   > n13 = Sequence()
   > "SetText (Text)"(
    self=Get PlotName() -> (PlotName:Common Text Object Reference)
    InText=Get MapName!(self:Island Data Entry Object Reference=@n12) -> (MapName:Text)
   )
   n14 = Get PublishedIslandCode!(self:Island Data Entry Object Reference=@n12) -> (PublishedIslandCode:String)
   n15 = Branch(execute=@n13.then_1, Condition="integer > integer"(A=Len(S=@n14), B=0) {is=title K2Node_CallFunction})
   n16 = Get PublishedCodeText() -> (PublishedCodeText:Common Text Object Reference)
   > "SetText (Text)"(
    self=@n16
    InText=Format Text!(
     Format="INVTEXT(\"{Code} (ver.{Version})\")"
     Code:Text="To Text (String)"(InString=@n14)
     Version:int=Get PublishedVersion!(self:Island Data Entry Object Reference=@n12) -> (PublishedVersion:int=0)
    ) -> (Result)
   )
   > SetVisibility(self=@n16, InVisibility=SelfHitTestInvisible)
   SetVisibility(execute=@n15.else, self=@n16, InVisibility=Collapsed)
   SetStyle!(
    execute=@n10.then_1
    self:Common Text Object Reference=Get PlotName()
    InStyle:Common Text Style Class Reference=/FortUILibrary/Styles/Typography/Foundation/TextStyle-BurbankSmall-M-Blue.TextStyle-BurbankSmall-M-Blue_C
   ) -> (then)
   > SetRenderTranslation()
   > SetRenderOpacity(InOpacity=1)
   > SetVisibility(self=Get InputActionWidget(), InVisibility=Hidden)
   Event On Clicked()
   n17 = Get ParentWidget()
   > n18 = Cast To IslandSelect_UI(Object=@n17)
   n19 = Get McpId(self=Get IslandData())
   > SelectPlot(self:Island Select UI Object Reference=@n18, SelectedPlot=@n19)
   n20 = Cast To IslandSelect_UI_V2_Modal(execute=@n18.CastFailed, Object=@n17)
   > SelectPlot(self:Island Select UI V2 Modal Object Reference=@n20, SelectedPlot=@n19)
   Event On Deselected()
   > n21 = Branch(Condition=Get bIsHovered() -> (bIsHovered:bool))
   SetVisibility(execute=[@n25.then, @n21.else], self=@n8, InVisibility=Hidden)
   Event On Hovered()
   n22 = Get Hovered() -> (Hovered:Widget Animation Object Reference)
   > PlayAnimationForward(InAnimation=@n22)
   > Set bIsHovered(bIsHovered:bool=true) -> (Output_Get:bool)
   n23 = Get PlotName()
   > n24 = SetStyle(self=@n23, InStyle=/FortUILibrary/Styles/Typography/Foundation/TextStyle-BurbankSmall-M-Black.TextStyle-BurbankSmall-M-Black_C)
   Event On Unhovered()
   > PlayAnimationReverse(InAnimation=@n22)
   > Set bIsHovered(bIsHovered=false)
   > n25 = SetStyle(self=@n23, InStyle=&TextStyle-BurbankSmall-M-Blue_C)
   Event OnEntryClick!() -> (OutputDelegate, then, Entry:Athena Creative Item Tile Button Object Reference) {is=title K2Node_Event}
   > n26 = Is Valid(InputObject=@n4)
   K2Node_Event_3 = CustomEvent!() -> (OutputDelegate, then, Entry:Athena Creative Item Tile Button Object Reference) {is=title K2Node_CustomEvent}
  ]}
 InitializeOption {nodes [
   n1 = InitializeOption!() -> (then, ParentWidget:Fort Activatable Panel Object Reference) {is=title K2Node_FunctionEntry}
   > n2 = Cast To IslandSelect_UI(Object=@n1)
   > n3 = Set ParentWidget(ParentWidget:Fort Activatable Panel Object Reference=@n2) -> (Output_Get:Fort Activatable Panel Object Reference)
   SetStyle(
    execute=[@n3.then, @n5.then]
    self=Get PlotName()
    InStyle=&TextStyle-BurbankSmall-M-Blue_C
   )
   n4 = Cast To IslandSelect_UI_V2_Modal(execute=@n2.CastFailed, Object=@n1)
   > n5 = Set ParentWidget(ParentWidget=@n4)
  ]}
]
variables [
 bIsHovered {type=boolean}
 IslandData {type=IslandDataEntry_C, default=null}
 ParentWidget {type=FortActivatablePanel, default=null}
]
widgets {
 info {inheritedWidgetCount=1, parentClass=/Script/FortniteUI.AthenaCreativeItemTileButton, rootWidgetClass=&VerticalBox, widgetCount=6}
 widgets [
  VerticalBox_0 {widgetClassPath=&VerticalBox, widget=VerticalBox_0}
  HorizontalBox_0 {widgetClassPath=&HorizontalBox, parent=VerticalBox_0, widget=HorizontalBox_0, slot {class=VerticalBoxSlot}}
  SizeBox_0 {
   widgetClassPath=&SizeBox
   parent=HorizontalBox_0
   widget=SizeBox_0
   properties {bOverride_HeightOverride=true, heightOverride=55, widthOverride=450}
   slot {class=HorizontalBoxSlot, properties {size {sizeRule=Fill, value=1}}}
  }
  PlotName {
   widgetClassPath=&CommonTextBlock
   bIsVariable=true
   parent=SizeBox_0
   widget=PlotName
   properties {
    justification=Center
    minDesiredWidth=440
    scrollStyle=&TextScrollStyle-Base_C
    text="{Island Name}"
    colorAndOpacity {colorUseRule=UseColor_Specified, specifiedColor=<rgba 0 0.624 1 1>}
    font {
     bForceMonospaced=false
     bMaterialIsStencil=false
     fontMaterial=null
     fontObject=&BurbankSmall
     letterSpacing=0
     monospacedWidth=1
     size=24
     skewAmount=0
     typefaceFontName=Bold
     outlineSettings {
      bApplyOutlineToDropShadows=false
      bMiteredCorners=false
      bSeparateFillAlpha=false
      outlineMaterial=null
      outlineSize=0
      outlineColor=<rgba 0 0 0 1>
     }
    }
    shadowOffset=<0 0>
   }
   slot {class=SizeBoxSlot, properties {verticalAlignment=VAlign_Center}}
  }
  InputActionWidget {
   widgetClassPath=&CommonActionWidget
   bInherited=true
   bIsVariable=true
   parent=HorizontalBox_0
   widget=InputActionWidget
   properties {
    visibility=Hidden
    inputActions [{dataTable=&InputActionDataTable, rowName=Select}]
    renderTransform {angle=0, scale=<1 1>, shear=<0 0>, translation=<-8 0>}
   }
   slot {class=HorizontalBoxSlot, properties {horizontalAlignment=HAlign_Left, verticalAlignment=VAlign_Center}}
  }
  PublishedCodeText {
   widgetClassPath=&CommonTextBlock
   bIsVariable=true
   parent=VerticalBox_0
   widget=PublishedCodeText
   properties {
    justification=Center
    minDesiredWidth=440
    scrollStyle=&TextScrollStyle-Base_C
    text="0000-0000-0000?v=8"
    font {
     bForceMonospaced=false
     bMaterialIsStencil=false
     fontMaterial=null
     fontObject=&BurbankSmall
     letterSpacing=0
     monospacedWidth=1
     size=18
     skewAmount=0
     typefaceFontName=Bold
     outlineSettings {
      bApplyOutlineToDropShadows=false
      bMiteredCorners=false
      bSeparateFillAlpha=false
      outlineMaterial=null
      outlineSize=1
      outlineColor=<rgba 0 0 0 1>
     }
    }
    shadowOffset=<0 0>
   }
   slot {class=VerticalBoxSlot, properties {verticalAlignment=VAlign_Center, padding=<ltrb 0 0 0 10>}}
  }
 ]
}
