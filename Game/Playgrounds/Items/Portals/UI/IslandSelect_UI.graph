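# Island Select panel for a Creative island portal (rift) that lets the player pick which of their islands it leads to.
# OpenPanel stores the portal, then Initialize either shows a limited-access date message or lists non-deleted CreativeIslands as IslandDataEntry objects, with an animated empty-list warning.
# SelectPlot sends the chosen plot and portal to ServerLoadPlotForPortal and closes the panel.
# Closing plays a reversed outro chosen by access state, guarded by bIsClosing.
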WidgetBlueprint /Game/Playgrounds/Items/Portals/UI/IslandSelect_UI
animations [
 Intro {
  binding_count=2
  length_seconds=0.3715
  object=&self:Intro
  sequence {bindings [
    CommonTextBlock_2 {
     id=#1
     possessed_class=CommonTextBlock
     tracks [
      {
       type=MovieScene2DTransformTrack
       sections [{
         type=MovieScene2DTransformSection
         end_frame=7
         start_frame=0
         channels [
          Translation.X {type=MovieSceneScriptingFloatChannel, keys [[0,-80],[2.8583,-80],[5.0013,9.829],[6.3167,0],[7.4307,0]]}
          Angle {type=MovieSceneScriptingFloatChannel}
          Scale.X {type=MovieSceneScriptingFloatChannel}
          Scale.Y {type=MovieSceneScriptingFloatChannel}
          Shear.X {type=MovieSceneScriptingFloatChannel}
          Shear.Y {type=MovieSceneScriptingFloatChannel}
         ]
        }]
      }
      {
       type=MovieSceneFloatTrack
       display_name=Render Opacity
       sections [{
         type=MovieSceneFloatSection
         end_frame=7
         start_frame=0
         channels [None {type=MovieSceneScriptingFloatChannel, keys [[0,0],[2.6617,0],[7.4307,1]]}]
        }]
      }
     ]
    }
    SizeBox_0 {
     id=#2
     possessed_class=SizeBox
     tracks [
      {
       type=MovieScene2DTransformTrack
       sections [{
         type=MovieScene2DTransformSection
         end_frame=7
         start_frame=0
         channels [
          Translation.X {type=MovieSceneScriptingFloatChannel, keys [[0,-50],[2.738,0],[7.4307,0]]}
          Translation.Y {type=MovieSceneScriptingFloatChannel}
          Angle {type=MovieSceneScriptingFloatChannel}
          Shear.X {type=MovieSceneScriptingFloatChannel}
          Shear.Y {type=MovieSceneScriptingFloatChannel}
         ]
        }]
      }
      {
       type=MovieSceneFloatTrack
       display_name=Render Opacity
       sections [{type=MovieSceneFloatSection, end_frame=4, start_frame=0, channels [None {type=MovieSceneScriptingFloatChannel, keys [[0,0],[3.6303,1]]}]}]
      }
     ]
    }
   ]}
 }
 NoIslands {
  binding_count=1
  length_seconds=0.149
  object=&self:NoIslands
  sequence {bindings [TextBlock_0 {
     id=#3
     possessed_class=TextBlock
     tracks [
      {
       type=MovieScene2DTransformTrack
       sections [{
         type=MovieScene2DTransformSection
         end_frame=2
         start_frame=0
         channels [
          Translation.X {type=MovieSceneScriptingFloatChannel, const [0,3]}
          Translation.Y {type=MovieSceneScriptingFloatChannel, keys [[0,-29],[2.0253,8],[2.98,0]]}
          Angle {type=MovieSceneScriptingFloatChannel}
          Scale.X {type=MovieSceneScriptingFloatChannel}
          Scale.Y {type=MovieSceneScriptingFloatChannel}
          Shear.X {type=MovieSceneScriptingFloatChannel}
          Shear.Y {type=MovieSceneScriptingFloatChannel}
         ]
        }]
      }
      {
       type=MovieSceneFloatTrack
       display_name=Render Opacity
       sections [{
         type=MovieSceneFloatSection
         end_frame=2
         start_frame=0
         channels [None {type=MovieSceneScriptingFloatChannel, keys [[0,0],[2.0253,1],[2.98,1]]}]
        }]
      }
     ]
    }]}
 }
 Outro {
  binding_count=2
  length_seconds=0.2
  object=&self:Outro
  sequence {bindings [
    CommonTextBlock_2 {
     id=#1
     possessed_class=CommonTextBlock
     tracks [{
       type=MovieScene2DTransformTrack
       sections [{
         type=MovieScene2DTransformSection
         end_frame=4
         start_frame=0
         channels [
          Translation.X {type=MovieSceneScriptingFloatChannel, const [0,4]}
          Angle {type=MovieSceneScriptingFloatChannel}
          Scale.X {type=MovieSceneScriptingFloatChannel}
          Scale.Y {type=MovieSceneScriptingFloatChannel}
          Shear.X {type=MovieSceneScriptingFloatChannel}
          Shear.Y {type=MovieSceneScriptingFloatChannel}
         ]
        }]
      }]
    }
    SizeBox_0 {
     id=#2
     possessed_class=SizeBox
     tracks [
      {
       type=MovieScene2DTransformTrack
       sections [{
         type=MovieScene2DTransformSection
         end_frame=3
         start_frame=0
         channels [
          Translation.X {type=MovieSceneScriptingFloatChannel}
          Translation.Y {type=MovieSceneScriptingFloatChannel}
          Angle {type=MovieSceneScriptingFloatChannel}
          Scale.X {type=MovieSceneScriptingFloatChannel, keys [[0,0.6],[2,1.05],[3.533,1]]}
          Scale.Y {type=MovieSceneScriptingFloatChannel, keys [[0,0.6],[2,1.15],[3.533,1]]}
          Shear.X {type=MovieSceneScriptingFloatChannel}
          Shear.Y {type=MovieSceneScriptingFloatChannel}
         ]
        }]
      }
      {
       type=MovieSceneFloatTrack
       display_name=Render Opacity
       sections [{type=MovieSceneFloatSection, end_frame=4, start_frame=0, channels [None {type=MovieSceneScriptingFloatChannel, keys [[0,0],[1.428,1]]}]}]
      }
     ]
    }
   ]}
 }
 LTDIntro {
  binding_count=2
  length_seconds=0.2
  object=&self:LTDIntro
  sequence {bindings [
    Overlay_1 {
     id=#4
     possessed_class=Overlay
     tracks [
      {
       type=MovieSceneFloatTrack
       display_name=Render Opacity
       sections [{type=MovieSceneFloatSection, end_frame=4, start_frame=0, channels [None {type=MovieSceneScriptingFloatChannel, keys [[0,0],[3,1],[4,1]]}]}]
      }
      {
       type=MovieScene2DTransformTrack
       sections [{
         type=MovieScene2DTransformSection
         end_frame=4
         start_frame=0
         channels [
          Translation.X {type=MovieSceneScriptingFloatChannel, const [0,3]}
          Translation.Y {type=MovieSceneScriptingFloatChannel, keys [[0,45],[3,-15.4186],[4,0]]}
          Angle {type=MovieSceneScriptingFloatChannel}
          Scale.X {type=MovieSceneScriptingFloatChannel}
          Scale.Y {type=MovieSceneScriptingFloatChannel}
          Shear.X {type=MovieSceneScriptingFloatChannel}
          Shear.Y {type=MovieSceneScriptingFloatChannel}
         ]
        }]
      }
     ]
    }
    Border_0 {
     id=#5
     possessed_class=Border
     tracks [{
       type=MovieSceneFloatTrack
       display_name=Render Opacity
       sections [{type=MovieSceneFloatSection, end_frame=4, start_frame=0, channels [None {type=MovieSceneScriptingFloatChannel, keys [[0,0],[3,1],[4,1]]}]}]
      }]
    }
   ]}
 }
 LTDOutro {
  binding_count=2
  length_seconds=0.2
  object=&self:LTDOutro
  sequence {bindings [
    Overlay_1 {
     id=#4
     possessed_class=Overlay
     tracks [
      {
       type=MovieSceneFloatTrack
       display_name=Render Opacity
       sections [{type=MovieSceneFloatSection, end_frame=4, start_frame=0, channels [None {type=MovieSceneScriptingFloatChannel, keys [[0,0],[3,1],[4,1]]}]}]
      }
      {
       type=MovieScene2DTransformTrack
       sections [{
         type=MovieScene2DTransformSection
         end_frame=4
         start_frame=0
         channels [
          Translation.X {type=MovieSceneScriptingFloatChannel, const [0,3]}
          Translation.Y {type=MovieSceneScriptingFloatChannel, keys [[0,45],[4,0]]}
          Angle {type=MovieSceneScriptingFloatChannel}
          Scale.X {type=MovieSceneScriptingFloatChannel}
          Scale.Y {type=MovieSceneScriptingFloatChannel}
          Shear.X {type=MovieSceneScriptingFloatChannel}
          Shear.Y {type=MovieSceneScriptingFloatChannel}
         ]
        }]
      }
     ]
    }
    Border_0 {
     id=#5
     possessed_class=Border
     tracks [{
       type=MovieSceneFloatTrack
       display_name=Render Opacity
       sections [{type=MovieSceneFloatSection, end_frame=4, start_frame=0, channels [None {type=MovieSceneScriptingFloatChannel, keys [[0,0],[3,1],[4,1]]}]}]
      }]
    }
   ]}
 }
]
graphs [
 ClosePanel {nodes [
   ClosePanel() {is=title K2Node_FunctionEntry}
   > PopContentWidget!(execute) -> (then)
   n1 = Get FortHUDContext!() -> (ReturnValue:Fort HUDContext Object Reference) {is=title K2Node_GetBlueprintContextEx}
   > SetCursorModeLocked!(execute, self:Fort HUDContext Object Reference=@n1, bLocked:bool=false) -> (then)
   > EnterCameraMode(self:Fort HUDContext Object Reference=@n1)
   > Return Node!(execute, bCommitted:bool=false)
  ]}
 EventGraph {nodes [
   SetInputActionHandler!(
    execute
    self:Common Activatable Panel Legacy Object Reference
    InputActionRow:Data Table Row Handle Structure=Get CloseButton() -> (CloseButton:Data Table Row Handle Structure)
    CommitedEvent:Delegate=Create Event()
   ) -> (then)
   > SetActionHandlerStateFromHandle!(
    execute
    self:Common Activatable Panel Legacy Object Reference
    InputActionRow:Data Table Row Handle Structure=Get CloseButton()
    State:EInputActionState Enum=Enabled
   ) -> (then)
   n1 = "On Button Clicked (CloseMainPanel)"!() -> (OutputDelegate, then, Button:Common Button Legacy Object Reference) {is=title K2Node_ComponentBoundEvent}
   ClosePanelWithAnim(execute=[@n1.then, @n2.then, @n5.then])
   n2 = "On Button Clicked (CloseLTDMsg)"!() -> (OutputDelegate, then, Button:Common Button Legacy Object Reference) {is=title K2Node_ComponentBoundEvent}
   n3 = "On Entry Generated (ListView_IslandList)"!() -> (OutputDelegate, then, Widget:User Widget Object Reference) {is=title K2Node_ComponentBoundEvent}
   > n4 = Cast To IslandOptionEntry!(execute, Object=@n3.Widget) -> (then, CastFailed, AsIsland Option Entry:Island Option Entry Object Reference) {is=title K2Node_DynamicCast}
   > InitializeOption!(execute, self:Island Option Entry Object Reference=@n4, ParentWidget:Fort Activatable Panel Object Reference=Self-Reference()) -> (then)
   n5 = "On Button Clicked (Button_CloseTouch)"!() -> (OutputDelegate, then, Button:Common Button Legacy Object Reference) {is=title K2Node_ComponentBoundEvent}
   PlayNoIslandAnim() {is=title K2Node_CustomEvent}
   > Delay(Duration=0.28)
   > PlayAnimationForward(InAnimation=Get NoIslands() -> (NoIslands:Widget Animation Object Reference))
   ClosePanelWithAnim() {is=title K2Node_CustomEvent}
   > Branch(Condition=NOT Boolean(A=Get bIsClosing() -> (bIsClosing:bool)))
   > Set bIsClosing(bIsClosing:bool=true) -> (Output_Get:bool)
   n6 = IsCreativeModeAccessLimited!(self:Fort Global UIContext Object Reference=Get FortGlobalUIContext() -> (ReturnValue:Fort Global UIContext Object Reference) {is=title K2Node_GetSubsystem}) -> (ReturnValue:bool)
   n7 = Get Outro() -> (Outro:Widget Animation Object Reference)
   > PlayAnimationReverse(
    InAnimation=Select(
     Option 0:Widget Animation Object Reference=@n7
     Option 1:Widget Animation Object Reference=Get LTDOutro() -> (LTDOutro:Widget Animation Object Reference)
     Index=@n6
    ) -> (ReturnValue:Widget Animation Object Reference)
    PlaybackSpeed=Select(Option 0:double=1, Option 1:double=1.25, Index=@n6) -> (ReturnValue:double)
   )
   > PlaySound2D(Sound=Get PanelDismiss_Sound() -> (PanelDismiss_Sound:Sound Base Object Reference))
   > Delay(Duration=GetEndTime(self=@n7))
   > Set bIsClosing(bIsClosing=false)
   > ClosePanel!(execute, self) -> (then, bCommitted:bool) {is=title K2Node_CallFunction}
   PlayLimitedAcessIntro() {is=title K2Node_CustomEvent}
   > PlayAnimationForward(InAnimation=Get LTDIntro() -> (LTDIntro:Widget Animation Object Reference))
   Event OnInitialized()
   > n8 = Construct Object from Class!(execute, Class=/Script/CommonUILegacy.CommonButtonGroupLegacy, self=Self-Reference(), bSelectionRequired:bool=true) -> (then, ReturnValue:Common Button Group Legacy Object Reference)
   > Set IslandButtonGroup(IslandButtonGroup:Common Button Group Legacy Object Reference=@n8) -> (Output_Get:Common Button Group Legacy Object Reference)
   Event Construct()
   n9 = GetRuntimeOptions()
   > SetVisibility(
    self=Get IslandCodesButton() -> (IslandCodesButton:Icon Text Button Object Reference)
    InVisibility=Select(
     Option 0:ESlateVisibility Enum=Collapsed
     Option 1:ESlateVisibility Enum=SelfHitTestInvisible
     Index=Get bEnableIslandCodeEntryOnPlayerPortal!(self:Fort Runtime Options Object Reference=@n9) -> (bEnableIslandCodeEntryOnPlayerPortal:bool)
    ) -> (ReturnValue:ESlateVisibility Enum)
   )
   n10 = Event OnInputModeChanged!() -> (OutputDelegate, then, bUsingGamepad:bool) {is=title K2Node_Event}
   n11 = Get ListView_IslandList() -> (ListView_IslandList:List View Object Reference)
   n12 = GetNumItems!(self:List View Object Reference=@n11) -> (ReturnValue:int=0)
   > Branch(Condition=AND Boolean(A=@n10.bUsingGamepad, B="integer > integer"(A=@n12, B=1) {is=title K2Node_CallFunction}))
   > NavigateToIndex!(execute, self:List View Object Reference=@n11, Index:int=0) -> (then)
   Event OnAddedToActivationStack() {is=title K2Node_Event}
   n13 = Get FortHUDContext()
   > EnterCursorMode!(execute, self:Fort HUDContext Object Reference=@n13, ActionName:Name=None, CursorModeWidget:User Widget Object Reference) -> (then)
   > SetCursorModeLocked(self=@n13, bLocked=true)
   Event On Activated()
   > PlayAnimationForward(InAnimation=Get Intro())
   > PlaySound2D(Sound=Get PanelAppear_Sound() -> (PanelAppear_Sound:Sound Base Object Reference))
   n14 = Event InitFromObject!() -> (OutputDelegate, then, InitObject:Object Reference) {is=title K2Node_Event}
   > n15 = Cast To BGA_IslandPortal!(execute, Object=@n14.InitObject) -> (then, CastFailed, AsBGA Island Portal:BGA Island Portal Object Reference) {is=title K2Node_DynamicCast}
   > OpenPanel!(execute, self, Portal:BGA Island Portal Object Reference=@n15) -> (then)
   "On Button Clicked (IslandCodesButton)"!() -> (OutputDelegate, then, Button:Common Button Legacy Object Reference) {is=title K2Node_ComponentBoundEvent}
   IslandCodeConfirmedEvent!() -> (OutputDelegate, then, IslandLink:Fort Creative Island Link Object Reference) {is=title K2Node_CustomEvent}
  ]}
 Initialize {nodes [
   Initialize() {is=title K2Node_FunctionEntry}
   > Set bIsClosing(bIsClosing=false)
   n1 = IsCreativeModeAccessLimited(self=Get FortGlobalUIContext())
   > SetActiveWidgetIndex(
    self=Get Switcher_Content() -> (Switcher_Content:Widget Switcher Object Reference)
    Index=Select(Option 0:int=0, Option 1:int=1, Index=@n1) -> (ReturnValue:int)
   )
   > n2 = Branch(Condition=@n1)
   > UpdateLimitedTimeText()
   > PlayLimitedAcessIntro() {is=title K2Node_CallFunction}
   n3 = Cast To FortPlayerControllerAthena(execute=@n2.else, Object=GetOwningPlayer())
   > n4 = For Each Loop(Array:Array of Creative Island Data Structures=Get CreativeIslands!(self:Fort Player Controller Athena Object Reference=@n3) -> (CreativeIslands:Array of Creative Island Data Structures)) -> (Array Element:Creative Island Data Structure)
   n5 = Break Creative Island Data!(CreativeIslandData:"Creative Island Data Structure (by ref)"=@n4) -> (
    McpId:String
    IslandName:Text
    IslandCode:Online Link Id Structure
    GameFeatureSetName:String
    LastLoadedDate:Date Time Structure
    DeletedAt:Date Time Structure
    bIsDeleted:bool=~
    bIsVKProject:bool=~
    OriginatingIslandId:String
   ) {is=title K2Node_BreakStruct}
   > n6 = Branch(Condition=@n5.bIsDeleted)
   n7 = Break Online Link Id!(OnlineLinkId:"Online Link Id Structure (by ref)"=@n5.IslandCode) -> (Mnemonic:String, Version:int) {is=title K2Node_BreakStruct}
   n8 = Construct Object from Class!(
    execute=@n6.else
    Class=&/IslandDataEntry.IslandDataEntry_C
    self=Self-Reference()
    McpId:String=@n5
    MapName:Text=@n5.IslandName
    PublishedIslandCode:String=@n7
    PublishedVersion:int=@n7.Version
    GameFeatureSetName:String
    LastUsedDate:Date Time Structure=@n5.LastLoadedDate
    DeletedAt:Date Time Structure
    IsDeleted:bool=False
    IsVKProject:bool=False
    IsLocked:bool=False
   ) -> (then, ReturnValue:Island Data Entry Object Reference)
   > Add(
    TargetArray:Array of Island Data Entry Object References=Get IslandEntries() -> (IslandEntries:Array of Island Data Entry Object References)
    NewItem:"Island Data Entry Object Reference (by ref)"=@n8
   ) {is=title K2Node_CallArrayFunction}
   Set List Items!(execute=@n4.Completed, self:List View Object Reference=Get ListView_IslandList(), InListItems:Array of Object References=Get IslandEntries()) -> (then) {is=title K2Node_CallFunction}
   n9 = "integer > integer"(A=GetNumItems(self=Get ListView_IslandList()) -> (ReturnValue=0), B=0)
   > SetActiveWidgetIndex(self=Get Switcher_EmptyList() -> (Switcher_EmptyList:Widget Switcher Object Reference), Index=Select(Option 0=1, Option 1=0, Index=@n9))
   > n10 = Branch(Condition=@n9)
   > Branch(Condition=IsUsingGamepad!(self:Fort Global UIContext Object Reference=Get FortGlobalUIContext()) -> (ReturnValue:bool))
   > NavigateToIndex(self=Get ListView_IslandList(), Index=0)
   PlayNoIslandAnim(execute=@n10.else) {is=title K2Node_CallFunction}
   GetCreativeProfile!(self:Fort Mcp Context Object Reference=Get FortMcpContext!() -> (ReturnValue:Fort Mcp Context Object Reference) {is=title K2Node_GetBlueprintContextEx}) -> (ReturnValue:Fort Mcp Profile Creative Object Reference)
  ]}
 OnTouchToClose {nodes [
   OnTouchToClose!() -> (then, MyGeometry:Geometry Structure, MouseEvent:"Pointer Event Structure (by ref)") {is=title K2Node_FunctionEntry}
   > ClosePanelWithAnim() {is=title K2Node_CallFunction}
   > Return Node!(execute, ReturnValue:Event Reply Structure=Handled())
  ]}
 OpenPanel {nodes [
   n1 = OpenPanel!() -> (then, Portal:BGA Island Portal Object Reference) {is=title K2Node_FunctionEntry}
   > Set Portal(Portal:BGA Island Portal Object Reference=@n1) -> (Output_Get:BGA Island Portal Object Reference)
   > PushContentWidgetAdvanced!(execute, PushWidget:Widget Object Reference=Self-Reference(), bHideHeader:bool=true, bHideFooter:bool=true, bHideChatWidget:bool=true) -> (then)
   > Initialize() {is=title K2Node_CallFunction}
  ]}
 SelectPlot {nodes [
   n1 = SelectPlot!() -> (then, SelectedPlot:String) {is=title K2Node_FunctionEntry}
   > n2 = Cast To FortPlayerControllerAthena(Object=GetOwningPlayer())
   > PlaySound2D(Sound=Get SelectPlot_Sound() -> (SelectPlot_Sound:Sound Base Object Reference))
   > ServerLoadPlotForPortal!(
    execute
    self:Fort Player Controller Athena Object Reference=@n2
    Portal:Fort Athena Creative Portal Object Reference=Get Portal() -> (Portal:BGA Island Portal Object Reference)
    PlotItemId:String=@n1
   ) -> (then)
   > ClosePanelWithAnim()
  ]}
 UpdateLimitedTimeText {nodes [
   UpdateLimitedTimeText() {is=title K2Node_FunctionEntry}
   n1 = GetCreativeModeLimitedAccessEndTime!(self:Fort Global UIContext Object Reference=Get FortGlobalUIContext()) -> (ReturnValue:Date Time Structure)
   > "SetText (Text)"(
    self=Get LimitAccessDateCalloutText() -> (LimitAccessDateCalloutText:Common Text Object Reference)
    InText=Format Text!(
     Format="LOCTEXT(\"Available on {Date}.\")"
     Date:Text=As Date!(InDateTime:"Date Time Structure (by ref)"=@n1, InDateStyle:EDateTimeStyle Enum=Default) -> (ReturnValue:Text)
    ) -> (Result)
   )
  ]}
]
variables [
 bIsClosing {type=boolean}
 CloseButton {type=DataTableRowHandle, default {dataTable=&InputActionDataTable, rowName=Close}}
 IslandButtonGroup {type=CommonButtonGroupLegacy, default=null}
 IslandEntries {type="array<IslandDataEntry_C>", default []}
 PanelAppear_Sound {type=SoundBase, default=/Game/Sounds/Fort_UI_Sounds/FrontEnd/New_FrontEnd/Fortnite_UI_MainMenu_Tab_03_Cue}
 PanelDismiss_Sound {type=SoundBase, default=/Game/Sounds/Fort_UI_Sounds/FrontEnd/New_FrontEnd/Fortnite_UI_MissionObj_Cue}
 Portal {type=BGA_IslandPortal_C, default=null}
 SelectPlot_Sound {type=SoundBase, default=/Game/Sounds/Fort_UI_Sounds/FrontEnd/New_FrontEnd/Fortnite_UI_Options_Menu_01_Cue}
]
widgets {
 info {inheritedWidgetCount=0, parentClass=/Script/FortniteUI.FortActivatablePanel, rootWidgetClass=&CommonBorder, widgetCount=57}
 widgets [
  TouchToCloseZone {widgetClassPath=&CommonBorder, widget=TouchToCloseZone, properties {padding=<ltrb 0 0 0 0>}}
  BackgroundBlur_1 {
   widgetClassPath=&BackgroundBlur
   parent=TouchToCloseZone
   widget=BackgroundBlur_1
   properties {blurStrength=25}
   slot {class=BorderSlot, properties {padding=<ltrb 0 0 0 0>}}
  }
  SafeZone_0 {
   widgetClassPath=&SafeZone
   bIsVariable=true
   parent=BackgroundBlur_1
   widget=SafeZone_0
   slot {class=BackgroundBlurSlot, properties {padding=<ltrb 0 0 0 0>}}
  }
  Overlay_0 {widgetClassPath=&Overlay, parent=SafeZone_0, widget=Overlay_0, slot {class=SafeZoneSlot}}
  Switcher_Content {
   widgetClassPath=&WidgetSwitcher
   bIsVariable=true
   parent=Overlay_0
   widget=Switcher_Content
   slot {class=OverlaySlot, properties {horizontalAlignment=HAlign_Center, verticalAlignment=VAlign_Center}}
  }
  SizeBox_0 {
   widgetClassPath=&SizeBox
   parent=Switcher_Content
   widget=SizeBox_0
   properties {
    bOverride_MaxDesiredHeight=true
    bOverride_MinDesiredHeight=true
    bOverride_MinDesiredWidth=true
    maxDesiredHeight=650
    maxDesiredWidth=800
    minDesiredHeight=400
    minDesiredWidth=800
    widthOverride=800
   }
   slot {class=WidgetSwitcherSlot}
  }
  BackgroundBlur_0 {
   widgetClassPath=&BackgroundBlur
   parent=SizeBox_0
   widget=BackgroundBlur_0
   properties {blurStrength=16}
   slot {class=SizeBoxSlot}
  }
  CommonBorder_0 {
   widgetClassPath=&CommonBorder
   parent=BackgroundBlur_0
   widget=CommonBorder_0
   properties {
    background {
     bIsDynamicallyLoaded=false
     drawAs=Image
     imageType=NoImage
     mirroring=NoMirror
     resourceName=None
     resourceObject=null
     tiling=NoTile
     imageSize=<32 32>
     margin=<ltrb 0 0 0 0>
     outlineSettings {
      bUseBrushTransparency=false
      roundingType=HalfHeightRadius
      width=0
      color {colorUseRule=UseColor_Specified, specifiedColor=<rgba 0 0 0 0>}
      cornerRadii=<0 0 0 0>
     }
     tintColor {colorUseRule=UseColor_Specified, specifiedColor=<rgba 0 0.0166 0.0677 0.7>}
     uVRegion {bIsValid=false, max=<0 0>, min=<0 0>}
    }
    padding=<ltrb 0 0 0 0>
    renderTransform {angle=0, scale=<1 1>, shear=<0 0>, translation=<1 1>}
   }
   slot {class=BackgroundBlurSlot, properties {padding=<ltrb 0 0 0 0>}}
  }
  Overlay_3 {widgetClassPath=&Overlay, parent=CommonBorder_0, widget=Overlay_3, slot {class=BorderSlot, properties {padding=<ltrb 0 0 0 0>}}}
  VerticalBox_0 {widgetClassPath=&VerticalBox, parent=Overlay_3, widget=VerticalBox_0, slot {class=OverlaySlot}}
  HorizontalBox_5 {
   widgetClassPath=&HorizontalBox
   parent=VerticalBox_0
   widget=HorizontalBox_5
   properties {renderTransform {angle=0, scale=<1 1>, shear=<0 0>, translation=<50 -35>}}
   slot {class=VerticalBoxSlot, properties {padding=<ltrb 0 0 0 -35>}}
  }
  Image_1 {
   widgetClassPath=&Image
   parent=HorizontalBox_5
   widget=Image_1
   properties {
    visibility=Collapsed
    brush {
     bIsDynamicallyLoaded=false
     drawAs=Image
     imageType=NoImage
     mirroring=NoMirror
     resourceName=None
     resourceObject=/Game/UI/Foundation/Textures/Icons/Cards/SquadTypes/T-Icon-Fortress-64
     tiling=NoTile
     imageSize=<64 64>
     margin=<ltrb 0 0 0 0>
     outlineSettings {
      bUseBrushTransparency=false
      roundingType=HalfHeightRadius
      width=0
      color {colorUseRule=UseColor_Specified, specifiedColor=<rgba 0 0 0 0>}
      cornerRadii=<0 0 0 0>
     }
     tintColor {colorUseRule=UseColor_Specified, specifiedColor=<rgba 1 1 1 1>}
     uVRegion {bIsValid=false, max=<0 0>, min=<0 0>}
    }
   }
   slot {class=HorizontalBoxSlot, properties {horizontalAlignment=HAlign_Left, verticalAlignment=VAlign_Center, padding=<ltrb 0 0 16 0>}}
  }
  CommonTextBlock_2 {
   widgetClassPath=&CommonTextBlock
   bIsVariable=true
   parent=HorizontalBox_5
   widget=CommonTextBlock_2
   properties {
    text=ISLAND SELECT
    font {
     bForceMonospaced=false
     bMaterialIsStencil=false
     fontMaterial=null
     fontObject=&Burbank
     letterSpacing=0
     monospacedWidth=1
     size=64
     skewAmount=0
     typefaceFontName=Burbank-Black
     outlineSettings {
      bApplyOutlineToDropShadows=false
      bMiteredCorners=false
      bSeparateFillAlpha=false
      outlineMaterial=null
      outlineSize=0
      outlineColor=<rgba 0 0 0 1>
     }
    }
    shadowColorAndOpacity=<rgba 0 0 0 0.3>
    shadowOffset=<4.9266 5.2555>
   }
   slot {class=HorizontalBoxSlot, properties {horizontalAlignment=HAlign_Right, verticalAlignment=VAlign_Center}}
  }
  SizeBox_1 {
   widgetClassPath=&SizeBox
   parent=VerticalBox_0
   widget=SizeBox_1
   properties {bOverride_HeightOverride=true, heightOverride=500, widthOverride=31549.3516}
   slot {class=VerticalBoxSlot}
  }
  HorizontalBox_0 {widgetClassPath=&HorizontalBox, parent=SizeBox_1, widget=HorizontalBox_0, slot {class=SizeBoxSlot}}
  VerticalBox_4 {
   widgetClassPath=&VerticalBox
   parent=HorizontalBox_0
   widget=VerticalBox_4
   slot {class=HorizontalBoxSlot, properties {verticalAlignment=VAlign_Center, padding=<ltrb 50 10 50 0>, size {sizeRule=Fill, value=1}}}
  }
  VerticalBox_1 {
   widgetClassPath=&VerticalBox
   parent=VerticalBox_4
   widget=VerticalBox_1
   properties {navigation=&self:WidgetTree.VerticalBox_1.WidgetNavigation_0}
   slot {class=VerticalBoxSlot, properties {padding=<ltrb 0 0 0 10>}}
  }
  CommonTextBlock_3 {
   widgetClassPath=&CommonTextBlock
   parent=VerticalBox_1
   widget=CommonTextBlock_3
   properties {
    text="Set destination Island for this rift: "
    font {
     bForceMonospaced=false
     bMaterialIsStencil=false
     fontMaterial=null
     fontObject=&BurbankSmall
     letterSpacing=0
     monospacedWidth=1
     size=24
     skewAmount=0
     typefaceFontName=Black
     outlineSettings {
      bApplyOutlineToDropShadows=false
      bMiteredCorners=false
      bSeparateFillAlpha=false
      outlineMaterial=null
      outlineSize=0
      outlineColor=<rgba 0 0 0 1>
     }
    }
    renderTransform {angle=0, scale=<1 1>, shear=<-6 0>, translation=<0 0>}
    shadowColorAndOpacity=<rgba 0 0 0 0.3>
    shadowOffset=<4.9266 5.2555>
   }
   slot {class=VerticalBoxSlot, properties {padding=<ltrb 0 0 0 10>}}
  }
  Switcher_EmptyList {widgetClassPath=&WidgetSwitcher, bIsVariable=true, parent=VerticalBox_1, widget=Switcher_EmptyList, slot {class=VerticalBoxSlot}}
  SizeBox_2 {
   widgetClassPath=&SizeBox
   parent=Switcher_EmptyList
   widget=SizeBox_2
   properties {bOverride_HeightOverride=true, bOverride_WidthOverride=true, heightOverride=400, widthOverride=530}
   slot {class=WidgetSwitcherSlot, properties {horizontalAlignment=HAlign_Center}}
  }
  ListView_IslandList {
   widgetClassPath=/Script/UMG.ListView
   bIsVariable=true
   parent=SizeBox_2
   widget=ListView_IslandList
   properties {
    entryWidgetClass=&/IslandOptionEntry.IslandOptionEntry_C
    horizontalEntrySpacing=10
    verticalEntrySpacing=10
   }
   slot {class=SizeBoxSlot}
  }
  SizeBox_EmptyListError {
   widgetClassPath=&SizeBox
   parent=Switcher_EmptyList
   widget=SizeBox_EmptyListError
   properties {bOverride_HeightOverride=true, bOverride_MaxDesiredHeight=true, heightOverride=300, maxDesiredHeight=300}
   slot {class=WidgetSwitcherSlot}
  }
  TextBlock_0 {
   widgetClassPath=&TextBlock
   bIsVariable=true
   parent=SizeBox_EmptyListError
   widget=TextBlock_0
   properties {
    renderOpacity=0
    text=You have no available Islands
    colorAndOpacity {colorUseRule=UseColor_Specified, specifiedColor=<rgba 0.9301 0.0742 0.107 1>}
    font {
     bForceMonospaced=false
     bMaterialIsStencil=false
     fontMaterial=null
     fontObject=&BurbankSmall
     letterSpacing=0
     monospacedWidth=1
     size=24
     skewAmount=0
     typefaceFontName=Black
     outlineSettings {
      bApplyOutlineToDropShadows=false
      bMiteredCorners=false
      bSeparateFillAlpha=false
      outlineMaterial=null
      outlineSize=0
      outlineColor=<rgba 0 0 0 1>
     }
    }
   }
   slot {class=SizeBoxSlot, properties {verticalAlignment=VAlign_Center}}
  }
  VB_EnterCodeSection {
   widgetClassPath=&VerticalBox
   parent=VerticalBox_4
   widget=VB_EnterCodeSection
   properties {navigation=&self:WidgetTree.VB_EnterCodeSection.WidgetNavigation_0, visibility=Collapsed}
   slot {class=VerticalBoxSlot}
  }
  CommonTextBlock_0 {
   widgetClassPath=&CommonTextBlock
   parent=VB_EnterCodeSection
   widget=CommonTextBlock_0
   properties {
    style=/FortUILibrary/Styles/Typography/Foundation/TextStyle-Base-S.TextStyle-Base-S_C
    text=Enter code
    font {
     bForceMonospaced=false
     bMaterialIsStencil=false
     fontMaterial=null
     fontObject=&NotoSans
     letterSpacing=0
     monospacedWidth=1
     size=18
     skewAmount=0
     typefaceFontName=NotoSans-Regular
     outlineSettings {
      bApplyOutlineToDropShadows=false
      bMiteredCorners=false
      bSeparateFillAlpha=false
      outlineMaterial=null
      outlineSize=0
      outlineColor=<rgba 0 0 0 1>
     }
    }
    shadowOffset=<0 0>
   }
   slot {class=VerticalBoxSlot, properties {verticalAlignment=VAlign_Center}}
  }
  CommonBorder_1 {
   widgetClassPath=&CommonBorder
   parent=VB_EnterCodeSection
   widget=CommonBorder_1
   properties {style=/FortUILibrary/Styles/Border/Border-SolidBG-DkBlue.Border-SolidBG-DkBlue_C, padding=<ltrb 0 0 0 0>}
   slot {class=VerticalBoxSlot, properties {verticalAlignment=VAlign_Center, padding=<ltrb 0 16 0 16>}}
  }
  FriendCodeEntry {
   widgetClassPath=/Script/UMG.EditableText
   bIsVariable=true
   parent=CommonBorder_1
   widget=FriendCodeEntry
   properties {
    hintText=Code
    widgetStyle {
     backgroundImageComposing {
      bIsDynamicallyLoaded=false
      drawAs=Border
      imageType=FullColor
      mirroring=NoMirror
      resourceName=../../../Engine/Content/Slate/Old/HyperlinkDotted.png
      resourceObject=null
      tiling=Both
      imageSize=<0 0>
      margin=<ltrb 0 0 0 0.1875>
      outlineSettings {
       bUseBrushTransparency=false
       roundingType=HalfHeightRadius
       width=0
       color {colorUseRule=UseColor_Specified, specifiedColor=<rgba 0 0 0 0>}
       cornerRadii=<0 0 0 0>
      }
      tintColor {colorUseRule=UseColor_Specified, specifiedColor=<rgba 1 1 1 1>}
      uVRegion {bIsValid=false, max=<0 0>, min=<0 0>}
     }
     backgroundImageSelected {
      bIsDynamicallyLoaded=false
      drawAs=Box
      imageType=FullColor
      mirroring=NoMirror
      resourceName=../../../Engine/Content/Slate/Common/EditableTextSelectionBackground.png
      resourceObject=null
      tiling=NoTile
      imageSize=<0 0>
      margin=<ltrb 0.25 0.25 0.25 0.25>
      outlineSettings {
       bUseBrushTransparency=false
       roundingType=HalfHeightRadius
       width=0
       color {colorUseRule=UseColor_Specified, specifiedColor=<rgba 0 0 0 0>}
       cornerRadii=<0 0 0 0>
      }
      tintColor {colorUseRule=UseColor_Specified, specifiedColor=<rgba 1 1 1 1>}
      uVRegion {bIsValid=false, max=<0 0>, min=<0 0>}
     }
     caretImage {
      bIsDynamicallyLoaded=false
      drawAs=Image
      imageType=FullColor
      mirroring=NoMirror
      resourceName=../../../Engine/Content/Slate/Old/White.png
      resourceObject=null
      tiling=NoTile
      imageSize=<16 16>
      margin=<ltrb 0 0 0 0>
      outlineSettings {
       bUseBrushTransparency=false
       roundingType=HalfHeightRadius
       width=0
       color {colorUseRule=UseColor_Specified, specifiedColor=<rgba 0 0 0 0>}
       cornerRadii=<0 0 0 0>
      }
      tintColor {colorUseRule=UseColor_Specified, specifiedColor=<rgba 1 1 1 1>}
      uVRegion {bIsValid=false, max=<0 0>, min=<0 0>}
     }
     colorAndOpacity {colorUseRule=UseColor_Specified, specifiedColor=<rgba 1 1 1 1>}
     font {
      bForceMonospaced=false
      bMaterialIsStencil=false
      fontMaterial=null
      fontObject=&NotoSans
      letterSpacing=0
      monospacedWidth=1
      size=16
      skewAmount=0
      typefaceFontName=NotoSans-Regular
      outlineSettings {
       bApplyOutlineToDropShadows=false
       bMiteredCorners=false
       bSeparateFillAlpha=false
       outlineMaterial=null
       outlineSize=0
       outlineColor=<rgba 0 0 0 1>
      }
     }
     hintColorAndOpacity {colorUseRule=UseColor_Foreground, specifiedColor=<rgba 1 0 1 1>}
     selectedBackgroundColor {colorUseRule=UseColor_Foreground, specifiedColor=<rgba 1 0 1 1>}
    }
   }
   slot {class=BorderSlot, properties {padding=<ltrb 0 0 0 0>}}
  }
  ErrorText {
   widgetClassPath=&CommonTextBlock
   bIsVariable=true
   parent=VB_EnterCodeSection
   widget=ErrorText
   properties {
    style=/FortUILibrary/Styles/Typography/Foundation/TextStyle-Base-XS-B-S-Red.TextStyle-Base-XS-B-S-Red_C
    text=Text Block
    visibility=Hidden
    colorAndOpacity {colorUseRule=UseColor_Specified, specifiedColor=<rgba 1 0.0561 0.0648 1>}
    font {
     bForceMonospaced=false
     bMaterialIsStencil=false
     fontMaterial=null
     fontObject=&NotoSans
     letterSpacing=0
     monospacedWidth=1
     size=16
     skewAmount=0
     typefaceFontName=NotoSans-Bold
     outlineSettings {
      bApplyOutlineToDropShadows=true
      bMiteredCorners=false
      bSeparateFillAlpha=true
      outlineMaterial=&M-FontShadow
      outlineSize=2
      outlineColor=<rgba 0.085 0.0146 0.0177 0.5>
     }
    }
    shadowOffset=<0 0>
   }
   slot {class=VerticalBoxSlot}
  }
  HorizontalBox_3 {
   widgetClassPath=&HorizontalBox
   parent=VB_EnterCodeSection
   widget=HorizontalBox_3
   slot {class=VerticalBoxSlot, properties {horizontalAlignment=HAlign_Left, verticalAlignment=VAlign_Center, padding=<ltrb 10 10 20 10>}}
  }
  IconTextButton {
   widgetClassPath=&IconTextButton
   bIsVariable=true
   parent=HorizontalBox_3
   widget=IconTextButton
   properties {
    buttonText=Use Code
    onButtonBaseClicked=(&self:WidgetTree.IconTextButton.HandleOnButtonClicked)
    onButtonBaseDoubleClicked=(&self:WidgetTree.IconTextButton.HandleOnButtonDoubleClicked)
    onButtonBaseHovered=(&self:WidgetTree.IconTextButton.HandleOnButtonHovered)
    onButtonBaseUnhovered=(&self:WidgetTree.IconTextButton.HandleOnButtonUnhovered)
    onSelectedChangedBase=(&self:WidgetTree.IconTextButton.HandleOnSelectedChanged)
   }
   slot {class=HorizontalBoxSlot, properties {horizontalAlignment=HAlign_Left}}
  }
  SelectedIslandInfo {
   widgetClassPath=&VerticalBox
   parent=HorizontalBox_0
   widget=SelectedIslandInfo
   properties {visibility=Collapsed}
   slot {class=HorizontalBoxSlot, properties {padding=<ltrb 0 0 40 0>}}
  }
  SizeBox_3 {
   widgetClassPath=&SizeBox
   parent=SelectedIslandInfo
   widget=SizeBox_3
   properties {bOverride_HeightOverride=true, bOverride_WidthOverride=true, heightOverride=304.6492, widthOverride=437.6042}
   slot {class=VerticalBoxSlot}
  }
  Border_4 {widgetClassPath=&Border, parent=SizeBox_3, widget=Border_4, properties {padding=<ltrb 10 10 10 10>}, slot {class=SizeBoxSlot}}
  Overlay_2 {widgetClassPath=&Overlay, parent=Border_4, widget=Overlay_2, slot {class=BorderSlot, properties {padding=<ltrb 10 10 10 10>}}}
  Image_2 {
   widgetClassPath=&Image
   bIsVariable=true
   parent=Overlay_2
   widget=Image_2
   properties {brush {
     bIsDynamicallyLoaded=false
     drawAs=Image
     imageType=NoImage
     mirroring=NoMirror
     resourceName=None
     resourceObject=/Game/UI/Foundation/Textures/Tiles/T_Athena_Tile_Matchmaking_Default
     tiling=NoTile
     imageSize=<1024 512>
     margin=<ltrb 0 0 0 0>
     outlineSettings {
      bUseBrushTransparency=false
      roundingType=HalfHeightRadius
      width=0
      color {colorUseRule=UseColor_Specified, specifiedColor=<rgba 0 0 0 0>}
      cornerRadii=<0 0 0 0>
     }
     tintColor {colorUseRule=UseColor_Specified, specifiedColor=<rgba 1 1 1 1>}
     uVRegion {bIsValid=false, max=<0 0>, min=<0 0>}
    }}
   slot {class=OverlaySlot}
  }
  Image_IslandImage {
   widgetClassPath=&Image
   bIsVariable=true
   parent=Overlay_2
   widget=Image_IslandImage
   properties {brush {
     bIsDynamicallyLoaded=false
     drawAs=Image
     imageType=NoImage
     mirroring=NoMirror
     resourceName=None
     resourceObject=/Game/Athena/UI/Frontend/Art/M_UI_SpeedLines_MOTD
     tiling=NoTile
     imageSize=<32 32>
     margin=<ltrb 0 0 0 0>
     outlineSettings {
      bUseBrushTransparency=false
      roundingType=HalfHeightRadius
      width=0
      color {colorUseRule=UseColor_Specified, specifiedColor=<rgba 0 0 0 0>}
      cornerRadii=<0 0 0 0>
     }
     tintColor {colorUseRule=UseColor_Specified, specifiedColor=<rgba 1 1 1 1>}
     uVRegion {bIsValid=false, max=<0 0>, min=<0 0>}
    }}
   slot {class=OverlaySlot, properties {horizontalAlignment=HAlign_Fill, verticalAlignment=VAlign_Fill}}
  }
  Text_Title {
   widgetClassPath=&CommonTextBlock
   parent=SelectedIslandInfo
   widget=Text_Title
   properties {
    text=Title of this Island
    font {
     bForceMonospaced=false
     bMaterialIsStencil=false
     fontMaterial=null
     fontObject=&BurbankSmall
     letterSpacing=0
     monospacedWidth=1
     size=24
     skewAmount=0
     typefaceFontName=Black
     outlineSettings {
      bApplyOutlineToDropShadows=false
      bMiteredCorners=false
      bSeparateFillAlpha=false
      outlineMaterial=null
      outlineSize=0
      outlineColor=<rgba 0 0 0 1>
     }
    }
    renderTransform {angle=0, scale=<1 1>, shear=<-6 0>, translation=<0 0>}
    shadowColorAndOpacity=<rgba 0 0 0 0.3>
    shadowOffset=<4.9266 5.2555>
   }
   slot {class=VerticalBoxSlot, properties {padding=<ltrb 0 10 0 0>}}
  }
  Text_Description {
   widgetClassPath=&CommonTextBlock
   parent=SelectedIslandInfo
   widget=Text_Description
   properties {
    text="More information about this island, this could be a longer description"
    wrapTextAt=412.7545
    font {
     bForceMonospaced=false
     bMaterialIsStencil=false
     fontMaterial=null
     fontObject=&BurbankSmall
     letterSpacing=0
     monospacedWidth=1
     size=24
     skewAmount=0
     typefaceFontName=Black
     outlineSettings {
      bApplyOutlineToDropShadows=false
      bMiteredCorners=false
      bSeparateFillAlpha=false
      outlineMaterial=null
      outlineSize=0
      outlineColor=<rgba 0 0 0 1>
     }
    }
    shadowColorAndOpacity=<rgba 0 0 0 0.3>
   }
   slot {class=VerticalBoxSlot, properties {padding=<ltrb 0 10 0 10>}}
  }
  Border_BotPanel {
   widgetClassPath=&CommonBorder
   parent=VerticalBox_0
   widget=Border_BotPanel
   properties {
    verticalAlignment=VAlign_Center
    background {
     bIsDynamicallyLoaded=false
     drawAs=Image
     imageType=NoImage
     mirroring=NoMirror
     resourceName=None
     resourceObject=null
     tiling=NoTile
     imageSize=<32 32>
     margin=<ltrb 0 0 0 0>
     outlineSettings {
      bUseBrushTransparency=false
      roundingType=HalfHeightRadius
      width=0
      color {colorUseRule=UseColor_Specified, specifiedColor=<rgba 0 0 0 0>}
      cornerRadii=<0 0 0 0>
     }
     tintColor {colorUseRule=UseColor_Specified, specifiedColor=<rgba 0.0024 0.0097 0.0262 1>}
     uVRegion {bIsValid=false, max=<0 0>, min=<0 0>}
    }
    padding=<ltrb 10 10 10 10>
   }
   slot {class=VerticalBoxSlot, properties {verticalAlignment=VAlign_Bottom, padding=<ltrb 0 10 0 0>}}
  }
  HorizontalBox_1 {
   widgetClassPath=&HorizontalBox
   parent=Border_BotPanel
   widget=HorizontalBox_1
   slot {class=BorderSlot, properties {verticalAlignment=VAlign_Center, padding=<ltrb 10 10 10 10>}}
  }
  IslandCodesButton {
   widgetClassPath=&IconTextButton
   bIsVariable=true
   parent=HorizontalBox_1
   widget=IslandCodesButton
   properties {
    buttonText=Island Codes
    onButtonBaseClicked=(&self:WidgetTree.IslandCodesButton.HandleOnButtonClicked)
    onButtonBaseDoubleClicked=(&self:WidgetTree.IslandCodesButton.HandleOnButtonDoubleClicked)
    onButtonBaseHovered=(&self:WidgetTree.IslandCodesButton.HandleOnButtonHovered)
    onButtonBaseUnhovered=(&self:WidgetTree.IslandCodesButton.HandleOnButtonUnhovered)
    onSelectedChangedBase=(&self:WidgetTree.IslandCodesButton.HandleOnSelectedChanged)
    style=/FortUILibrary/Foundation/Buttons/ButtonStyles/Skew/ButtonStyle-Skew_LessTransparent.ButtonStyle-Skew_LessTransparent_C
    triggeringInputAction {dataTable=&InputActionDataTable, rowName=GenericPreview}
   }
   slot {class=HorizontalBoxSlot, properties {size {sizeRule=Fill, value=1}}}
  }
  Spacer_0 {
   widgetClassPath=&Spacer
   parent=HorizontalBox_1
   widget=Spacer_0
   slot {class=HorizontalBoxSlot, properties {padding=<ltrb 10 10 10 10>}}
  }
  VisibilityBorder_NonTouch {
   widgetClassPath=&CommonHardwareVisibilityBorder
   parent=HorizontalBox_1
   widget=VisibilityBorder_NonTouch
   properties {visibilityQuery {
     autoDescription=" NONE( Input.Touch )"
     tokenStreamVersion=0
     userDescription=""
     queryTokenStream [0,1,3,1,0]
     tagDictionary [{tagName=Input.Touch}]
    }}
   slot {class=HorizontalBoxSlot, properties {horizontalAlignment=HAlign_Right, size {sizeRule=Fill, value=1}}}
  }
  CloseMainPanel {
   widgetClassPath=&IconTextButton
   bIsVariable=true
   parent=VisibilityBorder_NonTouch
   widget=CloseMainPanel
   properties {
    buttonText=CLOSE
    onButtonBaseClicked=(&self:WidgetTree.CloseMainPanel.HandleOnButtonClicked)
    onButtonBaseDoubleClicked=(&self:WidgetTree.CloseMainPanel.HandleOnButtonDoubleClicked)
    onButtonBaseHovered=(&self:WidgetTree.CloseMainPanel.HandleOnButtonHovered)
    onButtonBaseUnhovered=(&self:WidgetTree.CloseMainPanel.HandleOnButtonUnhovered)
    onSelectedChangedBase=(&self:WidgetTree.CloseMainPanel.HandleOnSelectedChanged)
    style=&ButtonStyle-Skew_LessTransparent_C
    triggeringInputAction {dataTable=/FortUILibrary/Input/CreativeInputActionDataTable, rowName=Close}
   }
   slot {class=BorderSlot, properties {padding=<ltrb 0 0 0 0>}}
  }
  VisibilityBorder_Touch {
   widgetClassPath=&CommonHardwareVisibilityBorder
   parent=Overlay_3
   widget=VisibilityBorder_Touch
   properties {visibilityQuery {
     autoDescription=" ANY( Input.Touch )"
     tokenStreamVersion=0
     userDescription=""
     queryTokenStream [0,1,1,1,0]
     tagDictionary [{tagName=Input.Touch}]
    }}
   slot {class=OverlaySlot, properties {horizontalAlignment=HAlign_Right}}
  }
  Button_CloseTouch {
   widgetClassPath=/FortUILibrary/Foundation/Buttons/CloseButton
   bIsVariable=true
   parent=VisibilityBorder_Touch
   widget=Button_CloseTouch
   properties {
    onButtonBaseClicked=(&self:WidgetTree.Button_CloseTouch.HandleOnButtonClicked)
    onButtonBaseDoubleClicked=(&self:WidgetTree.Button_CloseTouch.HandleOnButtonDoubleClicked)
    onButtonBaseHovered=(&self:WidgetTree.Button_CloseTouch.HandleOnButtonHovered)
    onButtonBaseUnhovered=(&self:WidgetTree.Button_CloseTouch.HandleOnButtonUnhovered)
    onSelectedChangedBase=(&self:WidgetTree.Button_CloseTouch.HandleOnSelectedChanged)
    renderTransform {angle=0, scale=<0.8 0.8>, shear=<0 0>, translation=<-10 -25>}
   }
   slot {class=BorderSlot, properties {padding=<ltrb 0 0 0 0>}}
  }
  Overlay_1 {
   widgetClassPath=&Overlay
   parent=Switcher_Content
   widget=Overlay_1
   properties {clipping=ClipToBounds}
   slot {class=WidgetSwitcherSlot, properties {verticalAlignment=VAlign_Bottom, padding=<ltrb 0 10 0 0>}}
  }
  Border_0 {
   widgetClassPath=&Border
   parent=Overlay_1
   widget=Border_0
   properties {
    background {
     bIsDynamicallyLoaded=false
     drawAs=Image
     imageType=NoImage
     mirroring=NoMirror
     resourceName=None
     resourceObject=/Game/Athena/UI/Frontend/Art/M_UI_RadialGradient_Translucent
     tiling=NoTile
     imageSize=<32 32>
     margin=<ltrb 0 0 0 0>
     outlineSettings {
      bUseBrushTransparency=false
      roundingType=HalfHeightRadius
      width=0
      color {colorUseRule=UseColor_Specified, specifiedColor=<rgba 0 0 0 0>}
      cornerRadii=<0 0 0 0>
     }
     tintColor {colorUseRule=UseColor_Specified, specifiedColor=<rgba 1 1 1 1>}
     uVRegion {bIsValid=false, max=<0 0>, min=<0 0>}
    }
    renderTransform {angle=0, scale=<1 1.6>, shear=<0 0>, translation=<0 0>}
   }
   slot {class=OverlaySlot, properties {horizontalAlignment=HAlign_Fill, verticalAlignment=VAlign_Fill}}
  }
  VerticalBox_MainContent {widgetClassPath=&VerticalBox, bIsVariable=true, parent=Overlay_1, widget=VerticalBox_MainContent, slot {class=OverlaySlot}}
  CommonTextBlock_5 {
   widgetClassPath=&CommonTextBlock
   parent=VerticalBox_MainContent
   widget=CommonTextBlock_5
   properties {
    autoWrapText=true
    justification=Center
    lineHeightPercentage=0.8
    text="CREATE AN \r\nISLAND"
    font {
     bForceMonospaced=false
     bMaterialIsStencil=false
     fontMaterial=null
     fontObject=&Burbank
     letterSpacing=0
     monospacedWidth=1
     size=48
     skewAmount=0
     typefaceFontName=Burbank-Black
     outlineSettings {
      bApplyOutlineToDropShadows=false
      bMiteredCorners=false
      bSeparateFillAlpha=false
      outlineMaterial=null
      outlineSize=2
      outlineColor=<rgba 0.0052 0.0052 0.0056 0.2>
     }
    }
    shadowOffset=<0 0>
   }
   slot {class=VerticalBoxSlot, properties {verticalAlignment=VAlign_Bottom, padding=<ltrb 0 15 0 0>}}
  }
  LimitAccessDateCalloutText {
   widgetClassPath=&CommonTextBlock
   bIsVariable=true
   parent=VerticalBox_MainContent
   widget=LimitAccessDateCalloutText
   properties {
    justification=Center
    text=Available on Dec 13.
    wrapTextAt=350
    font {
     bForceMonospaced=false
     bMaterialIsStencil=false
     fontMaterial=null
     fontObject=&BurbankSmall
     letterSpacing=0
     monospacedWidth=1
     size=20
     skewAmount=0
     typefaceFontName=Bold
     outlineSettings {
      bApplyOutlineToDropShadows=false
      bMiteredCorners=false
      bSeparateFillAlpha=false
      outlineMaterial=null
      outlineSize=2
      outlineColor=<rgba 0.0052 0.0052 0.0056 0.2>
     }
    }
    shadowOffset=<0 0>
   }
   slot {class=VerticalBoxSlot, properties {padding=<ltrb 0 15 0 10>}}
  }
  CommonBorder_5 {
   widgetClassPath=&CommonBorder
   parent=VerticalBox_MainContent
   widget=CommonBorder_5
   properties {
    background {
     bIsDynamicallyLoaded=false
     drawAs=Image
     imageType=NoImage
     mirroring=NoMirror
     resourceName=None
     resourceObject=/Game/UI/Frontend/Showdown/Materials/M_UI_SkewBox_TournamentPanelDetailsBG
     tiling=NoTile
     imageSize=<32 32>
     margin=<ltrb 0 0 0 0>
     outlineSettings {
      bUseBrushTransparency=false
      roundingType=HalfHeightRadius
      width=0
      color {colorUseRule=UseColor_Specified, specifiedColor=<rgba 0 0 0 0>}
      cornerRadii=<0 0 0 0>
     }
     tintColor {colorUseRule=UseColor_Specified, specifiedColor=<rgba 1 1 1 1>}
     uVRegion {bIsValid=false, max=<0 0>, min=<0 0>}
    }
    padding=<ltrb 10 10 10 10>
   }
   slot {class=VerticalBoxSlot}
  }
  SizeBox_6 {
   widgetClassPath=&SizeBox
   parent=CommonBorder_5
   widget=SizeBox_6
   properties {bOverride_MinDesiredHeight=true, bOverride_WidthOverride=true, minDesiredHeight=55, widthOverride=400}
   slot {class=BorderSlot, properties {padding=<ltrb 10 10 10 10>}}
  }
  VerticalBox_6 {
   widgetClassPath=&VerticalBox
   parent=SizeBox_6
   widget=VerticalBox_6
   slot {class=SizeBoxSlot, properties {padding=<ltrb 0 12 0 0>}}
  }
  CallToActionText {
   widgetClassPath=&CommonTextBlock
   parent=VerticalBox_6
   widget=CallToActionText
   properties {
    justification=Center
    text="Upgrade your Battle Pass to get early access now!"
    visibility=HitTestInvisible
    wrapTextAt=350
    colorAndOpacity {colorUseRule=UseColor_Specified, specifiedColor=<rgba 0.0116 0.016 0.0395 1>}
    font {
     bForceMonospaced=false
     bMaterialIsStencil=false
     fontMaterial=null
     fontObject=&BurbankSmall
     letterSpacing=0
     monospacedWidth=1
     size=20
     skewAmount=0
     typefaceFontName=Black
     outlineSettings {
      bApplyOutlineToDropShadows=false
      bMiteredCorners=false
      bSeparateFillAlpha=false
      outlineMaterial=null
      outlineSize=0
      outlineColor=<rgba 0 0 0 1>
     }
    }
    shadowOffset=<0 0>
   }
   slot {class=VerticalBoxSlot}
  }
  CloseLTDMsg {
   widgetClassPath=&IconTextButton
   bIsVariable=true
   parent=VerticalBox_6
   widget=CloseLTDMsg
   properties {
    buttonText=CLOSE
    contentAlignment=HAlign_Center
    onButtonBaseClicked=(&self:WidgetTree.CloseLTDMsg.HandleOnButtonClicked)
    onButtonBaseDoubleClicked=(&self:WidgetTree.CloseLTDMsg.HandleOnButtonDoubleClicked)
    onButtonBaseHovered=(&self:WidgetTree.CloseLTDMsg.HandleOnButtonHovered)
    onButtonBaseUnhovered=(&self:WidgetTree.CloseLTDMsg.HandleOnButtonUnhovered)
    onSelectedChangedBase=(&self:WidgetTree.CloseLTDMsg.HandleOnSelectedChanged)
    style=&ButtonStyle-Skew_LessTransparent_C
    triggeringInputAction {dataTable=&CreativeInputActionDataTable, rowName=Cancel}
   }
   slot {class=VerticalBoxSlot, properties {horizontalAlignment=HAlign_Right, verticalAlignment=VAlign_Bottom, padding=<ltrb 0 10 0 0>}}
  }
  Image_PanelBorder {
   widgetClassPath=&Image
   parent=Overlay_1
   widget=Image_PanelBorder
   properties {
    visibility=HitTestInvisible
    brush {
     bIsDynamicallyLoaded=false
     drawAs=Box
     imageType=NoImage
     mirroring=NoMirror
     resourceName=None
     resourceObject=/Game/UI/Creative/Materials/M_UI_ServerBorder_Selected
     tiling=NoTile
     imageSize=<32 128>
     margin=<ltrb 0 0.17 0 0.17>
     outlineSettings {
      bUseBrushTransparency=false
      roundingType=HalfHeightRadius
      width=0
      color {colorUseRule=UseColor_Specified, specifiedColor=<rgba 0 0 0 0>}
      cornerRadii=<0 0 0 0>
     }
     tintColor {colorUseRule=UseColor_Specified, specifiedColor=<rgba 1 1 1 1>}
     uVRegion {bIsValid=false, max=<0 0>, min=<0 0>}
    }
   }
   slot {class=OverlaySlot, properties {horizontalAlignment=HAlign_Fill, verticalAlignment=VAlign_Fill}}
  }
 ]
}
