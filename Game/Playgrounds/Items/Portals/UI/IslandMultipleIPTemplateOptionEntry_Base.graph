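# Base list entry button for picking an island template in the Creative portal UI, showing the template image in a rounded border.
# Setting its list item stores the IslandDataEntry, resets translation and opacity, and hides the code text and Select prompt.
# Hover or selection unhides the prompt and calls PlaySound2D; unhover hides it, as does deselection when not hovered.
# The prompt's parent row is collapsed, and the Hovered scale animation is never played in this graph.
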WidgetBlueprint /Game/Playgrounds/Items/Portals/UI/IslandMultipleIPTemplateOptionEntry_Base
animations [Hovered {
  binding_count=1
  length_seconds=0.2
  object=&self:Hovered
  sequence {bindings [IslandOptionEntry_C_107 {
     id=#1
     possessed_class=IslandOptionEntry_C
     tracks [{
       type=MovieScene2DTransformTrack
       sections [{
         type=MovieScene2DTransformSection
         end_frame=4
         start_frame=0
         channels [
          Translation.X {type=MovieSceneScriptingFloatChannel}
          Translation.Y {type=MovieSceneScriptingFloatChannel}
          Angle {type=MovieSceneScriptingFloatChannel}
          Scale.X {type=MovieSceneScriptingFloatChannel, keys [[0,1],[4,1.1073]]}
          Scale.Y {type=MovieSceneScriptingFloatChannel, keys [[0,1],[4,1.1073]]}
          Shear.X {type=MovieSceneScriptingFloatChannel}
          Shear.Y {type=MovieSceneScriptingFloatChannel}
         ]
        }]
      }]
    }]}
 }]
graphs [EventGraph {nodes [
   n1 = Event On Selected()
   n2 = Get InputActionWidget() -> (InputActionWidget:Common Action Widget Object Reference)
   SetVisibility(execute=[@n7.then, @n1.then], self=@n2, InVisibility=SelfHitTestInvisible)
   > PlaySound2D()
   n3 = Event OnListItemObjectSet()
   > n4 = Sequence()
   > n5 = Cast To IslandDataEntry!(execute, Object=@n3.ListItemObject) -> (then, CastFailed, AsIsland Data Entry:Island Data Entry Object Reference) {is=title K2Node_DynamicCast}
   > Set IslandData(IslandData:Island Data Entry Object Reference=@n5) -> (Output_Get:Island Data Entry Object Reference)
   > SetVisibility(self=Get PublishedCodeText() -> (PublishedCodeText:Common Text Object Reference), InVisibility=Collapsed)
   SetRenderTranslation(execute=@n4.then_1)
   > SetRenderOpacity(InOpacity=1)
   > SetVisibility(self=Get InputActionWidget(), InVisibility=Hidden)
   Event On Deselected()
   > n6 = Branch(Condition=Get bIsHovered() -> (bIsHovered:bool))
   SetVisibility(execute=[@n8.then, @n6.else], self=@n2, InVisibility=Hidden)
   Event On Hovered()
   > n7 = Set bIsHovered(bIsHovered:bool=true) -> (Output_Get:bool)
   Event On Unhovered()
   > n8 = Set bIsHovered(bIsHovered=false)
  ]}]
variables [
 bIsHovered {type=boolean}
 IslandData {type=IslandDataEntry_C, default=null}
]
widgets {
 info {
  inheritedWidgetCount=2
  parentClass=/Script/FortniteUI.FortCreativeIslandMultipleIPTemplateButton
  rootWidgetClass=&Border
  widgetCount=9
 }
 widgets [
  Border_1 {
   widgetClassPath=&Border
   widget=Border_1
   properties {
    horizontalAlignment=HAlign_Center
    verticalAlignment=VAlign_Center
    background {
     bIsDynamicallyLoaded=false
     drawAs=RoundedBox
     imageType=FullColor
     mirroring=NoMirror
     resourceName=None
     resourceObject=null
     tiling=NoTile
     imageSize=<32 32>
     margin=<ltrb 0 0 0 0>
     outlineSettings {
      bUseBrushTransparency=false
      roundingType=FixedRadius
      width=0
      color {colorUseRule=UseColor_Specified, specifiedColor=<rgba 1 1 1 0>}
      cornerRadii=<6 6 6 6>
     }
     tintColor {colorUseRule=UseColor_Specified, specifiedColor=<rgba 1 1 1 0>}
     uVRegion {bIsValid=false, max=<0 0>, min=<0 0>}
    }
    padding=<ltrb 5 5 5 5>
   }
  }
  Border_0 {
   widgetClassPath=&Border
   parent=Border_1
   widget=Border_0
   properties {
    horizontalAlignment=HAlign_Center
    verticalAlignment=VAlign_Center
    background {
     bIsDynamicallyLoaded=false
     drawAs=RoundedBox
     imageType=FullColor
     mirroring=NoMirror
     resourceName=None
     resourceObject=null
     tiling=NoTile
     imageSize=<32 32>
     margin=<ltrb 0 0 0 0>
     outlineSettings {
      bUseBrushTransparency=false
      roundingType=FixedRadius
      width=3
      color {colorUseRule=UseColor_Specified, specifiedColor=<rgba 1 1 1 0>}
      cornerRadii=<10 10 10 10>
     }
     tintColor {colorUseRule=UseColor_Specified, specifiedColor=<rgba 1 1 1 0.1>}
     uVRegion {bIsValid=false, max=<0 0>, min=<0 0>}
    }
    padding=<ltrb 0 0 0 0>
   }
   slot {class=BorderSlot, properties {horizontalAlignment=HAlign_Center, verticalAlignment=VAlign_Center, padding=<ltrb 5 5 5 5>}}
  }
  Overlay_0 {
   widgetClassPath=&Overlay
   parent=Border_0
   widget=Overlay_0
   slot {class=BorderSlot, properties {horizontalAlignment=HAlign_Center, verticalAlignment=VAlign_Center, padding=<ltrb 0 0 0 0>}}
  }
  ScaleBox_0 {
   widgetClassPath=&ScaleBox
   parent=Overlay_0
   widget=ScaleBox_0
   slot {class=OverlaySlot, properties {horizontalAlignment=HAlign_Fill, verticalAlignment=VAlign_Fill, padding=<ltrb 10 10 10 10>}}
  }
  TemplateImage {
   widgetClassPath=&FortLazyImage
   bInherited=true
   bIsVariable=true
   parent=ScaleBox_0
   widget=TemplateImage
   properties {brush {
     bIsDynamicallyLoaded=false
     drawAs=Image
     imageType=FullColor
     mirroring=NoMirror
     resourceName=None
     resourceObject=/Game/Athena/Items/Consumables/PlaysetGrenade/T-Icon-PS-CP-Island_Temperate_60x60_a
     tiling=NoTile
     imageSize=<256 256>
     margin=<ltrb 0 0 0 0>
     outlineSettings {
      bUseBrushTransparency=false
      roundingType=HalfHeightRadius
      width=0
      color {colorUseRule=UseColor_Specified, specifiedColor=<rgba 0 0 0 0>}
      cornerRadii=<0 0 0 0>
     }
     tintColor {colorUseRule=UseColor_Specified, specifiedColor=<rgba 1 1 1 1>}
     uVRegion {bIsValid=false, max=<0 0>, min=<0 0>}
    }}
   slot {class=ScaleBoxSlot}
  }
  HorizontalBox_0 {
   widgetClassPath=&HorizontalBox
   parent=Overlay_0
   widget=HorizontalBox_0
   properties {visibility=Collapsed}
   slot {class=OverlaySlot, properties {horizontalAlignment=HAlign_Fill, verticalAlignment=VAlign_Fill}}
  }
  InputActionWidget {
   widgetClassPath=&CommonActionWidget
   bInherited=true
   bIsVariable=true
   parent=HorizontalBox_0
   widget=InputActionWidget
   properties {
    visibility=Hidden
    inputActions [{dataTable=&InputActionDataTable, rowName=Select}]
    renderTransform {angle=0, scale=<1 1>, shear=<0 0>, translation=<-8 0>}
   }
   slot {class=HorizontalBoxSlot, properties {horizontalAlignment=HAlign_Left, verticalAlignment=VAlign_Center}}
  }
  PlotName {
   widgetClassPath=&CommonTextBlock
   bIsVariable=true
   parent=HorizontalBox_0
   widget=PlotName
   properties {
    justification=Center
    minDesiredWidth=440
    scrollStyle=&TextScrollStyle-Base_C
    text="{Island Name}"
    colorAndOpacity {colorUseRule=UseColor_Specified, specifiedColor=<rgba 0 0.624 1 1>}
    font {
     bForceMonospaced=false
     bMaterialIsStencil=false
     fontMaterial=null
     fontObject=&BurbankSmall
     letterSpacing=0
     monospacedWidth=1
     size=24
     skewAmount=0
     typefaceFontName=Bold
     outlineSettings {
      bApplyOutlineToDropShadows=false
      bMiteredCorners=false
      bSeparateFillAlpha=false
      outlineMaterial=null
      outlineSize=0
      outlineColor=<rgba 0 0 0 1>
     }
    }
    shadowOffset=<0 0>
   }
   slot {class=HorizontalBoxSlot, properties {verticalAlignment=VAlign_Center}}
  }
  PublishedCodeText {
   widgetClassPath=&CommonTextBlock
   bIsVariable=true
   parent=HorizontalBox_0
   widget=PublishedCodeText
   properties {
    justification=Center
    minDesiredWidth=440
    scrollStyle=&TextScrollStyle-Base_C
    text="0000-0000-0000?v=8"
    visibility=Collapsed
    font {
     bForceMonospaced=false
     bMaterialIsStencil=false
     fontMaterial=null
     fontObject=&BurbankSmall
     letterSpacing=0
     monospacedWidth=1
     size=18
     skewAmount=0
     typefaceFontName=Bold
     outlineSettings {
      bApplyOutlineToDropShadows=false
      bMiteredCorners=false
      bSeparateFillAlpha=false
      outlineMaterial=null
      outlineSize=1
      outlineColor=<rgba 0 0 0 1>
     }
    }
    shadowOffset=<0 0>
   }
   slot {class=HorizontalBoxSlot, properties {verticalAlignment=VAlign_Center}}
  }
 ]
}
