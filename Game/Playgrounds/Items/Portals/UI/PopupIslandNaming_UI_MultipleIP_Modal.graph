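# Modal popup for naming a Creative island when creating, renaming or duplicating one, with a filtered text field and optional island image.
# InitFromObject stores either the generation UI or island select modal as the owner that receives the name.
# Confirm waits for text sanitization, then ApplyName checks for duplicate island names and routes to link-code creation, plot creation, rename or duplicate.
# Closing broadcasts OnIslandNamingClosed with a success flag and pops the modal.
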WidgetBlueprint /Game/Playgrounds/Items/Portals/UI/PopupIslandNaming_UI_MultipleIP_Modal
animations [
 Intro {
  binding_count=2
  length_seconds=0.3715
  object=&self:Intro
  sequence {bindings [
    NameYourIsland {
     id=#1
     possessed_class=CommonTextBlock
     tracks [
      {
       type=MovieScene2DTransformTrack
       sections [{
         type=MovieScene2DTransformSection
         end_frame=7
         start_frame=0
         channels [
          Translation.X {type=MovieSceneScriptingFloatChannel, keys [[0,-80],[2.8583,-80],[5.0013,9.829],[6.3167,0],[7.4307,0]]}
          Angle {type=MovieSceneScriptingFloatChannel}
          Scale.X {type=MovieSceneScriptingFloatChannel}
          Scale.Y {type=MovieSceneScriptingFloatChannel}
          Shear.X {type=MovieSceneScriptingFloatChannel}
          Shear.Y {type=MovieSceneScriptingFloatChannel}
         ]
        }]
      }
      {
       type=MovieSceneFloatTrack
       display_name=Render Opacity
       sections [{
         type=MovieSceneFloatSection
         end_frame=7
         start_frame=0
         channels [None {type=MovieSceneScriptingFloatChannel, keys [[0,0],[2.6617,0],[7.4307,1]]}]
        }]
      }
     ]
    }
    SizeBox_0 {
     id=#2
     possessed_class=SizeBox
     tracks [
      {
       type=MovieScene2DTransformTrack
       sections [{
         type=MovieScene2DTransformSection
         end_frame=7
         start_frame=0
         channels [
          Translation.X {type=MovieSceneScriptingFloatChannel, keys [[0,-50],[2.738,0],[7.4307,0]]}
          Translation.Y {type=MovieSceneScriptingFloatChannel}
          Angle {type=MovieSceneScriptingFloatChannel}
          Shear.X {type=MovieSceneScriptingFloatChannel}
          Shear.Y {type=MovieSceneScriptingFloatChannel}
         ]
        }]
      }
      {
       type=MovieSceneFloatTrack
       display_name=Render Opacity
       sections [{type=MovieSceneFloatSection, end_frame=4, start_frame=0, channels [None {type=MovieSceneScriptingFloatChannel, keys [[0,0],[3.6303,1]]}]}]
      }
     ]
    }
   ]}
 }
 Outro {
  binding_count=2
  length_seconds=0.2
  object=&self:Outro
  sequence {bindings [
    NameYourIsland {
     id=#1
     possessed_class=CommonTextBlock
     tracks [{
       type=MovieScene2DTransformTrack
       sections [{
         type=MovieScene2DTransformSection
         end_frame=4
         start_frame=0
         channels [
          Translation.X {type=MovieSceneScriptingFloatChannel, const [0,4]}
          Angle {type=MovieSceneScriptingFloatChannel}
          Scale.X {type=MovieSceneScriptingFloatChannel}
          Scale.Y {type=MovieSceneScriptingFloatChannel}
          Shear.X {type=MovieSceneScriptingFloatChannel}
          Shear.Y {type=MovieSceneScriptingFloatChannel}
         ]
        }]
      }]
    }
    SizeBox_0 {
     id=#2
     possessed_class=SizeBox
     tracks [
      {
       type=MovieScene2DTransformTrack
       sections [{
         type=MovieScene2DTransformSection
         end_frame=3
         start_frame=0
         channels [
          Translation.X {type=MovieSceneScriptingFloatChannel}
          Translation.Y {type=MovieSceneScriptingFloatChannel}
          Angle {type=MovieSceneScriptingFloatChannel}
          Scale.X {type=MovieSceneScriptingFloatChannel, keys [[0,0.6],[2,1.05],[3.533,1]]}
          Scale.Y {type=MovieSceneScriptingFloatChannel, keys [[0,0.6],[2,1.15],[3.533,1]]}
          Shear.X {type=MovieSceneScriptingFloatChannel}
          Shear.Y {type=MovieSceneScriptingFloatChannel}
         ]
        }]
      }
      {
       type=MovieSceneFloatTrack
       display_name=Render Opacity
       sections [{type=MovieSceneFloatSection, end_frame=4, start_frame=0, channels [None {type=MovieSceneScriptingFloatChannel, keys [[0,0],[1.428,1]]}]}]
      }
     ]
    }
   ]}
 }
 LTDIntro {
  binding_count=1
  length_seconds=0.2
  object=&self:LTDIntro
  sequence {bindings [Overlay_1 {
     id=#3
     possessed_class=Overlay
     tracks [
      {
       type=MovieSceneFloatTrack
       display_name=Render Opacity
       sections [{
         type=MovieSceneFloatSection
         end_frame=0
         start_frame=-5
         channels [None {type=MovieSceneScriptingFloatChannel, keys [[-4.0003,0],[-1.0003,1],[-0.0003,1]]}]
        }]
      }
      {
       type=MovieScene2DTransformTrack
       sections [{
         type=MovieScene2DTransformSection
         end_frame=0
         start_frame=-5
         channels [
          Translation.X {type=MovieSceneScriptingFloatChannel, const [0,3]}
          Translation.Y {type=MovieSceneScriptingFloatChannel, keys [[-4.0003,45],[-1.0003,-15.4186],[-0.0003,0]]}
          Angle {type=MovieSceneScriptingFloatChannel}
          Scale.X {type=MovieSceneScriptingFloatChannel}
          Scale.Y {type=MovieSceneScriptingFloatChannel}
          Shear.X {type=MovieSceneScriptingFloatChannel}
          Shear.Y {type=MovieSceneScriptingFloatChannel}
         ]
        }]
      }
     ]
    }]}
 }
 LTDOutro {
  binding_count=1
  length_seconds=0.2
  object=&self:LTDOutro
  sequence {bindings [Overlay_1 {
     id=#3
     possessed_class=Overlay
     tracks [
      {
       type=MovieSceneFloatTrack
       display_name=Render Opacity
       sections [{type=MovieSceneFloatSection, end_frame=4, start_frame=0, channels [None {type=MovieSceneScriptingFloatChannel, keys [[0,0],[3,1],[4,1]]}]}]
      }
      {
       type=MovieScene2DTransformTrack
       sections [{
         type=MovieScene2DTransformSection
         end_frame=4
         start_frame=0
         channels [
          Translation.X {type=MovieSceneScriptingFloatChannel, const [0,3]}
          Translation.Y {type=MovieSceneScriptingFloatChannel, keys [[0,45],[4,0]]}
          Angle {type=MovieSceneScriptingFloatChannel}
          Scale.X {type=MovieSceneScriptingFloatChannel}
          Scale.Y {type=MovieSceneScriptingFloatChannel}
          Shear.X {type=MovieSceneScriptingFloatChannel}
          Shear.Y {type=MovieSceneScriptingFloatChannel}
         ]
        }]
      }
     ]
    }]}
 }
]
graphs [
 Append Copy Text {nodes [
   Append Copy Text() {is=title K2Node_FunctionEntry}
   > n1 = GetCurrentText!(
    execute
    self:WBP UIKit Filtered Input Field Object Reference=Get As Filtered Text Entry Widget() -> (As Filtered Text Entry Widget:WBP UIKit Filtered Input Field Object Reference)
   ) -> (then, Text:Text)
   > SetInputFieldText!(
    execute
    self:WBP UIKit Filtered Input Field Object Reference=Get As Filtered Text Entry Widget()
    Text:Text=Format Text!(Format="LOCTEXT(\"{IslandName} (copy)\")", IslandName:Text=@n1) -> (Result)
   ) -> (then)
  ]}
 BP_GetDesiredFocusTarget {nodes [
   Get Desired Focus Target()
   > Return Node!(
    execute
    ReturnValue:Widget Object Reference=Get FilteredTextEntryWidget_Name() -> (FilteredTextEntryWidget_Name:Fort Editable Filtered Counted Text Box Object Reference)
   )
  ]}
 ClosePanel {nodes [
   ClosePanel() {is=title K2Node_FunctionEntry}
   > PopCurrentModal!(execute) -> (then)
   > Return Node!(execute, bCommitted:bool=false)
  ]}
 ClosePanelWithResult {nodes [
   n1 = ClosePanelWithResult!() -> (then, Success:bool=~) {is=title K2Node_FunctionEntry}
   > Call On Island Naming Closed!(execute:Exec, self:Popup Island Naming UI Multiple IP Modal Object Reference, Success:bool=@n1) -> (then:Exec) {is=title K2Node_CallDelegate}
   > ClosePanel!(execute, self) -> (then, bCommitted:bool) {is=title K2Node_CallFunction}
  ]}
 ConfirmNameForLinkCodeCreation {nodes [
   n1 = ConfirmNameForLinkCodeCreation!() -> (then, GenerationUI:Fort Creative Island Multiple IPGeneration Object Reference, IslandName:String) {is=title K2Node_FunctionEntry}
   > Is Valid(InputObject=@n1)
   n2 = Get GenerationUI() -> (GenerationUI:Fort Creative Island Multiple IPGeneration Object Reference)
   > n3 = Cast To Creative_CreateIsland_UI_MultipleIP!(execute, Object=@n2) -> (then, CastFailed, AsCreative Create Island UI Multiple IP:Creative Create Island UI Multiple IP Object Reference) {is=title K2Node_DynamicCast}
   > ConfirmNameForLinkCodeCreation!(execute, self:Creative Create Island UI Multiple IP Object Reference=@n3, IslandName:String=Get IslandName() -> (IslandName:String)) -> (then) {is=title K2Node_CallFunction}
   n4 = Cast To Creative_CreateIsland_UI_MultipleIP_V2!(execute=@n3.CastFailed, Object=@n2) -> (then, CastFailed, AsCreative Create Island UI Multiple IP V2:Creative Create Island UI Multiple IP V2 Object Reference) {is=title K2Node_DynamicCast}
   > ConfirmNameForLinkCodeCreation(self:Creative Create Island UI Multiple IP V2 Object Reference=@n4, IslandName=Get IslandName())
  ]}
 Entry Has Same Name as Existing Island {nodes [
   Entry Has Same Name as Existing Island() {is=title K2Node_FunctionEntry}
   > n1 = Cast To FortPlayerControllerAthena(Object=GetOwningPlayer())
   > n2 = For Each Loop(Array:Array of Creative Island Data Structures=Get CreativeIslands!(self:Fort Player Controller Athena Object Reference=@n1) -> (CreativeIslands:Array of Creative Island Data Structures)) -> (Array Element:Creative Island Data Structure)
   Return Node!(execute=@n2.Completed, AlreadyExists:bool=false)
   > n3 = GetCurrentText(self=Get As Filtered Text Entry Widget())
   n4 = "To String (Text)"(InText=@n3)
   n5 = Break Creative Island Data!(CreativeIslandData:"Creative Island Data Structure (by ref)"=@n2) -> (IslandName:Text, bIsDeleted:bool=~, OriginatingIslandId:String) {is=title K2Node_BreakStruct}
   n6 = "To String (Text)"(InText=@n5)
   > Branch(Condition=AND Boolean(A="Equal Exactly (String)"(A=@n4, B=@n6), B=NOT Boolean(A=@n5.bIsDeleted)))
   > Return Node(AlreadyExists=true)
  ]}
 EventGraph {nodes [
   n1 = "On Button Clicked (Button_CloseTouch)"!() -> (OutputDelegate, then, Button:Common Button Legacy Object Reference) {is=title K2Node_ComponentBoundEvent}
   ClosePanelWithResult!(execute=[@n1.then, @n4.then], self, Success:bool=false) -> (then) {is=title K2Node_CallFunction}
   "On Text Committed (FilteredTextEntryWidget_Name)"!() -> (OutputDelegate, then, CommitMethod:ETextCommit Enum=Default) {is=title K2Node_ComponentBoundEvent}
   > Set IsTextSanitizationComplete(IsTextSanitizationComplete:bool=Get bCurrentlyValid!(self:Fort Editable Filtered Counted Text Box Object Reference=Get FilteredTextEntryWidget_Name()) -> (bCurrentlyValid:bool)) -> (Output_Get:bool)
   n2 = "On Sanitized Text Ready (FilteredTextEntryWidget_Name)"!() -> (OutputDelegate, then, Text:Text, BadText:bool) {is=title K2Node_ComponentBoundEvent}
   > Set IsTextSanitizationComplete(IsTextSanitizationComplete=true)
   > Branch(Condition=Get HasConfirmationWaitingForTextSanitization() -> (HasConfirmationWaitingForTextSanitization:bool))
   > Set HasConfirmationWaitingForTextSanitization(HasConfirmationWaitingForTextSanitization:bool=false) -> (Output_Get:bool)
   > SetIsEnabled(self=Get Button_Confirm() -> (Button_Confirm:Common Button Base Object Reference), bInIsEnabled=true)
   > n3 = Branch(Condition=@n2.BadText)
   ApplyName(execute=@n3.else)
   n4 = "On Clicked (BackButton)"!() -> (OutputDelegate, then, Button:Common Button Base Object Reference) {is=title K2Node_ComponentBoundEvent}
   "On Clicked (Button_Confirm)"!() -> (OutputDelegate, then, Button:Common Button Base Object Reference) {is=title K2Node_ComponentBoundEvent}
   > n5 = Branch(Condition=Get IsTextSanitizationComplete() -> (IsTextSanitizationComplete:bool))
   > ApplyName()
   Set HasConfirmationWaitingForTextSanitization(execute=@n5.else, HasConfirmationWaitingForTextSanitization=true)
   > SetIsEnabled(self=Get Button_Confirm(), bInIsEnabled=false)
   UpdateInputActionHandler() {is=title K2Node_CustomEvent}
   > n6 = Branch(Condition=IsUsingTouch(PlayerController=GetOwningPlayer()))
   n7 = Get CloseActionBack() -> (CloseActionBack:Data Table Row Handle Structure)
   > Branch(Condition=HasInputActionHandler!(self:Common Activatable Panel Legacy Object Reference, InputActionRow:Data Table Row Handle Structure=@n7) -> (ReturnValue:bool))
   n8 = Self-Reference()
   > RemoveInputActionHandler!(execute, self:Common Activatable Panel Legacy Object Reference=@n8, InputActionRow:Data Table Row Handle Structure=@n7) -> (then)
   SetInputActionHandler!(
    execute=@n6.else
    self:Common Activatable Panel Legacy Object Reference=@n8
    InputActionRow:Data Table Row Handle Structure=@n7
    CommitedEvent:Delegate=Create Event()
   ) -> (then)
   ApplyName() {is=title K2Node_CustomEvent}
   > n9 = Should Check for Duplicate Name!(execute:Exec) -> (then:Exec, ShouldCheck:bool=~) {is=title K2Node_MacroInstance}
   > n10 = Branch(Condition=@n9)
   > n11 = Entry Has Same Name as Existing Island!(execute, self) -> (then, AlreadyExists:bool) {is=title K2Node_CallFunction}
   > n12 = Branch(Condition=@n11)
   > ShowErrorDialog!(
    execute
    DisplayMessage:Text="LOCTEXT(\"You cannot name an island with the same name as one of your existing islands!\")"
    ErrorCode:String=Naming Island Error
    ErrorSeverity:EFortErrorSeverity Enum=Unspecified
   ) -> (then)
   n13 = GetCurrentText(execute=[@n12.else, @n10.else], self=Get As Filtered Text Entry Widget())
   > Set DesiredTitle(DesiredTitle:String="To String (Text)"(InText=@n13)) -> (Output_Get:String)
   > n14 = Is Valid(InputObject=Get IslandGenerationUI() -> (IslandGenerationUI:Fort Creative Island Multiple IPGeneration Object Reference))
   > n15 = Branch(Condition=Get IsFromTemplate() -> (IsFromTemplate:bool))
   > ConfirmNameForLinkCodeCreation!(
    execute
    self:Self Object Reference
    GenerationUI:Fort Creative Island Multiple IPGeneration Object Reference=Get IslandGenerationUI()
    IslandName=Get DesiredTitle() -> (DesiredTitle:String)
   ) -> (then)
   > ClosePanelWithResult(Success=true)
   n16 = Get InitialDisplayText() -> (InitialDisplayText:Text)
   n17 = Get DesiredTitle()
   n18 = Select(Option 0:String=@n17, Option 1:String, Index="Equal Exactly (String)"(A=@n17, B="To String (Text)"(InText=@n16))) -> (ReturnValue:String)
   MakeCreativePlotAndHandleButtonState!(
    execute=@n15.else
    self
    GenerationUI:Fort Creative Island Multiple IPGeneration Object Reference=Get IslandGenerationUI()
    Locale:String=GetCurrentLanguage!() -> (ReturnValue:String)
    Title:String=@n18
    PlotItemDef:Fort Creative Real Estate Plot Item Definition Object Reference=Get TemplateRealEstatePlotDefinition() -> (TemplateRealEstatePlotDefinition:Fort Creative Real Estate Plot Item Definition Object Reference)
   ) -> (then)
   > ClosePanelWithResult(Success=true)
   n19 = Is Valid(
    exec=@n14.Is Not Valid
    InputObject=Get IslandSelect_UI_Modal() -> (IslandSelect_UI_Modal:Island Select UI Multiple IP Modal Object Reference)
   )
   > n20 = Branch(Condition=Get IsForRenaming() -> (IsForRenaming:bool))
   > n21 = Branch(Condition="integer > integer"(A=Len(S=@n18), B=0) {is=title K2Node_CallFunction})
   n22 = Get IslandSelect_UI_Modal()
   > n23 = RenameCreativePlotAndHandleButtonState!(execute, self:Island Select UI Multiple IP Modal Object Reference=@n22, RenameTitle:String=@n18) -> (then)
   ClosePanelWithResult(execute=[@n23.then, @n25.then, @n24.then], Success=true)
   ShowErrorDialog(
    execute=@n21.else
    DisplayMessage="LOCTEXT(\"You cannot rename an island to a blank island name.\")"
    ErrorCode=Renaming Island Error
    ErrorSeverity=Unspecified
   )
   n24 = DuplicateCreativePlotAndHandleButtonState!(execute=@n20.else, self:Island Select UI Multiple IP Modal Object Reference=@n22, DuplicateTitle:String=@n18) -> (then)
   n25 = MakeCreativePlotAndHandleButtonState(
    execute=@n19.Is Not Valid
    GenerationUI=Get IslandGenerationUI()
    Locale=GetCurrentLanguage()
    Title=@n18
    PlotItemDef=Get TemplateRealEstatePlotDefinition()
   )
   Event On Deactivated()
   > Set IsTextSanitizationComplete(IsTextSanitizationComplete=true)
   > Set HasConfirmationWaitingForTextSanitization(HasConfirmationWaitingForTextSanitization=false)
   Event Construct()
   > Parent: Construct()
   > Set IsFromTemplate(IsFromTemplate:bool=false) -> (Output_Get:bool)
   > n26 = Cast To WBP_UIKit_FilteredInputField!(execute, Object=Get FilteredTextEntryWidget_Name()) -> (then, CastFailed, AsWBP UIKit Filtered Input Field:WBP UIKit Filtered Input Field Object Reference) {is=title K2Node_DynamicCast}
   > Set As Filtered Text Entry Widget(As Filtered Text Entry Widget:WBP UIKit Filtered Input Field Object Reference=@n26) -> (Output_Get:WBP UIKit Filtered Input Field Object Reference)
   > Branch(Condition=IsInCabinMode!() -> (ReturnValue:bool))
   > SetIsEnabled(self=Get FilteredTextEntryWidget_Name(), bInIsEnabled=false)
   Event OnInputModeChanged!() -> (OutputDelegate, then, bUsingGamepad:bool) {is=title K2Node_Event}
   > UpdateInputActionHandler() {is=title K2Node_CallFunction}
   Event On Activated()
   > UpdateInputActionHandler()
   > Manage Island Image()
   > PlayAnimationForward(InAnimation=Get Intro())
   > PlaySound2D(Sound=Get PanelAppear_Sound() -> (PanelAppear_Sound:Sound Base Object Reference))
   n27 = Event InitFromObject!() -> (OutputDelegate, then, InitObject:Object Reference) {is=title K2Node_Event}
   > n28 = Cast To FortCreativeIslandMultipleIPGeneration!(execute, Object=@n27.InitObject) -> (then, CastFailed, AsFort Creative Island Multiple IPGeneration:Fort Creative Island Multiple IPGeneration Object Reference) {is=title K2Node_DynamicCast}
   > n29 = Set IslandGenerationUI(IslandGenerationUI:Fort Creative Island Multiple IPGeneration Object Reference=@n28) -> (Output_Get:Fort Creative Island Multiple IPGeneration Object Reference)
   OpenPanel!(execute=[@n29.then, @n31.then], self, Portal:BGA Island Portal Object Reference) -> (then)
   n30 = Cast To IslandSelect_UI_MultipleIP_Modal!(execute=@n28.CastFailed, Object=@n27.InitObject) -> (then, CastFailed, AsIsland Select UI Multiple IP Modal:Island Select UI Multiple IP Modal Object Reference) {is=title K2Node_DynamicCast}
   > n31 = Set IslandSelect_UI_Modal(IslandSelect_UI_Modal:Island Select UI Multiple IP Modal Object Reference=@n30) -> (Output_Get:Island Select UI Multiple IP Modal Object Reference)
   PrintString!(
    execute=@n30.CastFailed
    InString=WARNING: This version only supports Multiple IP widgets
    bPrintToScreen
    bPrintToLog
    TextColor
    Duration
    Key
   ) -> (then)
  ]}
 InitFromIslandDataEntry {nodes [
   n1 = InitFromIslandDataEntry!() -> (then, IslandData:Island Data Entry Object Reference) {is=title K2Node_FunctionEntry}
   n2 = Get MapName!(self:Island Data Entry Object Reference=@n1) -> (MapName:Text)
   > SetInputFieldText(self=Get As Filtered Text Entry Widget(), Text=@n2)
   > Set InitialDisplayText(InitialDisplayText:Text) -> (Output_Get:Text)
   > Set DuplicatedIslandName(DuplicatedIslandName:String="To String (Text)"(InText=@n2)) -> (Output_Get:String)
  ]}
 MakeCreativePlotAndHandleButtonState {nodes [
   n1 = MakeCreativePlotAndHandleButtonState!() -> (
    then
    GenerationUI:Fort Creative Island Multiple IPGeneration Object Reference
    Locale:String
    Title:String
    PlotItemDef:Fort Creative Real Estate Plot Item Definition Object Reference
   ) {is=title K2Node_FunctionEntry}
   > Is Valid(InputObject=@n1)
   n2 = Get GenerationUI()
   > n3 = Cast To Creative_CreateIsland_UI_MultipleIP(Object=@n2)
   > MakeCreativePlotAndHandleButtonState!(
    execute
    self:Creative Create Island UI Multiple IP Object Reference=@n3
    Locale=Get Locale() -> (Locale:String)
    Title=Get Title() -> (Title:String)
    PlotItemDef=Get PlotItemDef() -> (PlotItemDef:Fort Creative Real Estate Plot Item Definition Object Reference)
   ) -> (then) {is=title K2Node_CallFunction}
   n4 = Cast To Creative_CreateIsland_UI_MultipleIP_V2(execute=@n3.CastFailed, Object=@n2)
   > MakeCreativePlotAndHandleButtonState(self:Creative Create Island UI Multiple IP V2 Object Reference=@n4, Locale=Get Locale(), Title=Get Title(), PlotItemDef=Get PlotItemDef())
  ]}
 Manage Island Image {nodes [
   Manage Island Image() {is=title K2Node_FunctionEntry}
   n1 = Break Slate Brush!(SlateBrush:"Slate Brush Structure (by ref)"=Get Brush!(self:Image Object Reference=Get IslandImage() -> (IslandImage:Fort Lazy Image Object Reference)) -> (Brush:Slate Brush Structure)) -> (ResourceObject:Object Reference) {is=title K2Node_BreakStruct}
   > SetVisibility(
    self=Get IslandImageScaleBox() -> (IslandImageScaleBox:Scale Box Object Reference)
    InVisibility=Select(Option 0:ESlateVisibility Enum=Collapsed, Option 1:ESlateVisibility Enum=SelfHitTestInvisible, Index=IsValid(Object=@n1)) -> (ReturnValue:ESlateVisibility Enum)
   )
  ]}
 OnIslandNamingClosed {nodes [OnIslandNamingClosed!() -> (then, Success:bool=~) {is=title K2Node_FunctionEntry}]}
 OnTouchToClose {nodes [
   OnTouchToClose!() -> (then, MyGeometry:Geometry Structure, MouseEvent:"Pointer Event Structure (by ref)") {is=title K2Node_FunctionEntry}
   Return Node!(execute, ReturnValue:Event Reply Structure=Handled())
  ]}
 OpenPanel {nodes [OpenPanel!() -> (then, Portal:BGA Island Portal Object Reference) {is=title K2Node_FunctionEntry}]}
 ShouldCheckForDuplicateName {nodes [
   Inputs()
   > n1 = GetCurrentText(self=Get As Filtered Text Entry Widget())
   n2 = "Equal Exactly (String)"(A="To String (Text)"(InText=@n1), B="To String (Text)"(InText=Get InitialDisplayText()))
   n3 = IsValid(Object=Get IslandGenerationUI())
   n4 = NOT Boolean(A=Get IsFromTemplate())
   > Outputs!(then:Exec, ShouldCheck:bool=NOT Boolean(A=AND Boolean!(A=@n2, B=@n4, C:bool=@n3) -> (ReturnValue)))
  ]}
]
variables [
 As Filtered Text Entry Widget {type=WBP_UIKit_FilteredInputField_C, default=null}
 bIsClosing {type=boolean}
 CloseActionBack {type=DataTableRowHandle, default {dataTable=&InputActionDataTable, rowName=Cancel}}
 DesiredTitle {type=string, default=""}
 DuplicatedIslandName {type=string, default="!@##!@##!@##)(*!@##"}
 DuplicationSameNameErrorText {type=string, default=When duplicating an Island you are required to change the name from the original.}
 HasConfirmationWaitingForTextSanitization {type=boolean}
 IsForRenaming {type=boolean}
 IsFromTemplate {type=boolean}
 IslandGenerationUI {type=FortCreativeIslandMultipleIPGeneration, default=null}
 IslandSelect_UI_Modal {type=IslandSelect_UI_MultipleIP_Modal_C, default=null}
 IsTextSanitizationComplete {type=boolean, default=true}
 NamingWidgetModal {type=FortActivatablePanel, default=null}
 OnIslandNamingClosed {}
 PanelAppear_Sound {type=SoundBase, default=/Game/Sounds/Fort_UI_Sounds/FrontEnd/New_FrontEnd/Fortnite_UI_MainMenu_Tab_03_Cue}
 PanelDismiss_Sound {type=SoundBase, default=/Game/Sounds/Fort_UI_Sounds/FrontEnd/New_FrontEnd/Fortnite_UI_MissionObj_Cue}
 Portal {type=BGA_IslandPortal_C, default=null}
 RenameSameNameErrorText {type=string, default=When renaming an Island you are required to change the name from the original.}
 SelectPlot_Sound {type=SoundBase, default=/Game/Sounds/Fort_UI_Sounds/FrontEnd/New_FrontEnd/Fortnite_UI_Options_Menu_01_Cue}
 Working Text {type=string, default=Working}
]
widgets {
 info {
  inheritedWidgetCount=3
  parentClass=/Script/FortniteUI.FortCreativeIslandMultipleIPNamingModal
  rootWidgetClass=&CommonBorder
  widgetCount=24
 }
 widgets [
  TouchToCloseZone {widgetClassPath=&CommonBorder, widget=TouchToCloseZone, properties {padding=<ltrb 0 0 0 0>}}
  SafeZone_1 {
   widgetClassPath=&SafeZone
   bIsVariable=true
   parent=TouchToCloseZone
   widget=SafeZone_1
   slot {class=BorderSlot, properties {padding=<ltrb 0 0 0 0>}}
  }
  Border_258 {
   widgetClassPath=&Border
   parent=SafeZone_1
   widget=Border_258
   properties {brushColor=<rgba 0 0 0 0.7>}
   slot {class=SafeZoneSlot}
  }
  Overlay_0 {widgetClassPath=&Overlay, parent=Border_258, widget=Overlay_0, slot {class=BorderSlot}}
  SizeBox_0 {
   widgetClassPath=&SizeBox
   parent=Overlay_0
   widget=SizeBox_0
   properties {bOverride_HeightOverride=true, bOverride_WidthOverride=true, heightOverride=512, minDesiredWidth=800, widthOverride=1424}
   slot {class=OverlaySlot, properties {horizontalAlignment=HAlign_Center, verticalAlignment=VAlign_Center}}
  }
  BackgroundBlur_0 {
   widgetClassPath=&BackgroundBlur
   parent=SizeBox_0
   widget=BackgroundBlur_0
   properties {blurStrength=16}
   slot {class=SizeBoxSlot}
  }
  Overlay_1 {
   widgetClassPath=&Overlay
   parent=BackgroundBlur_0
   widget=Overlay_1
   slot {class=BackgroundBlurSlot, properties {padding=<ltrb 0 0 0 0>}}
  }
  Image_149 {
   widgetClassPath=&Image
   parent=Overlay_1
   widget=Image_149
   properties {colorAndOpacity=<rgba 0 0.002 0.026 1>}
   slot {class=OverlaySlot, properties {horizontalAlignment=HAlign_Fill, verticalAlignment=VAlign_Fill}}
  }
  WBP_CaptureForPostBufferUpdate {
   widgetClassPath=&WBP_CaptureForPostBufferUpdate
   bIsVariable=true
   parent=Overlay_1
   slot {class=OverlaySlot, properties {horizontalAlignment=HAlign_Fill, verticalAlignment=VAlign_Fill}}
  }
  CommonBorder_0 {
   widgetClassPath=&CommonBorder
   parent=Overlay_1
   widget=CommonBorder_0
   properties {padding=<ltrb 0 0 0 0>}
   slot {class=OverlaySlot, properties {horizontalAlignment=HAlign_Fill, verticalAlignment=VAlign_Fill}}
  }
  HorizontalBox_1 {
   widgetClassPath=&HorizontalBox
   parent=CommonBorder_0
   widget=HorizontalBox_1
   slot {class=BorderSlot, properties {padding=<ltrb 0 0 0 0>}}
  }
  VerticalBox_0 {
   widgetClassPath=&VerticalBox
   parent=HorizontalBox_1
   widget=VerticalBox_0
   slot {class=HorizontalBoxSlot, properties {padding=<ltrb 80 50 80 50>, size {sizeRule=Fill, value=2}}}
  }
  Overlay_2 {widgetClassPath=&Overlay, parent=VerticalBox_0, widget=Overlay_2, slot {class=VerticalBoxSlot}}
  HorizontalBox_85 {
   widgetClassPath=&HorizontalBox
   parent=Overlay_2
   widget=HorizontalBox_85
   slot {class=OverlaySlot, properties {horizontalAlignment=HAlign_Fill}}
  }
  NameYourIsland {
   widgetClassPath=&CommonTextBlock
   parent=HorizontalBox_85
   widget=NameYourIsland
   properties {
    text=Name Your Island
    textTransformPolicy=ToUpper
    wrapTextAt=960
    font {
     bForceMonospaced=false
     bMaterialIsStencil=false
     fontMaterial=null
     fontObject=&HeadingNow
     letterSpacing=0
     monospacedWidth=1
     size=25.5
     skewAmount=0
     typefaceFontName=86 Bold Italic
     outlineSettings {
      bApplyOutlineToDropShadows=true
      bMiteredCorners=false
      bSeparateFillAlpha=false
      outlineMaterial=null
      outlineSize=2
      outlineColor=<rgba 1 1 1 0>
     }
    }
    shadowOffset=<4.9266 5.2555>
   }
   slot {class=HorizontalBoxSlot, properties {verticalAlignment=VAlign_Center, padding=<ltrb 0 0 24 0>}}
  }
  VisibilityBorder_Touch {
   widgetClassPath=&CommonHardwareVisibilityBorder
   parent=HorizontalBox_85
   widget=VisibilityBorder_Touch
   properties {visibilityQuery {
     autoDescription=" ANY( Input.Touch )"
     tokenStreamVersion=0
     userDescription=""
     queryTokenStream [0,1,1,1,0]
     tagDictionary [{tagName=Input.Touch}]
    }}
   slot {class=HorizontalBoxSlot, properties {horizontalAlignment=HAlign_Right, verticalAlignment=VAlign_Top, size {sizeRule=Fill, value=1}}}
  }
  Button_CloseTouch {
   widgetClassPath=/FortUILibrary/Foundation/Buttons/CloseButton
   bIsVariable=true
   parent=VisibilityBorder_Touch
   widget=Button_CloseTouch
   properties {
    onButtonBaseClicked=(&self:WidgetTree.Button_CloseTouch.HandleOnButtonClicked)
    onButtonBaseDoubleClicked=(&self:WidgetTree.Button_CloseTouch.HandleOnButtonDoubleClicked)
    onButtonBaseHovered=(&self:WidgetTree.Button_CloseTouch.HandleOnButtonHovered)
    onButtonBaseUnhovered=(&self:WidgetTree.Button_CloseTouch.HandleOnButtonUnhovered)
    onSelectedChangedBase=(&self:WidgetTree.Button_CloseTouch.HandleOnSelectedChanged)
    renderTransform {angle=0, scale=<0.8 0.8>, shear=<0 0>, translation=<0 0>}
   }
   slot {class=BorderSlot, properties {padding=<ltrb 0 0 0 0>}}
  }
  FilteredTextEntryWidget_Name {
   widgetClassPath=/FortUILibrary/UIKit/InputField/WBP_UIKit_FilteredInputField
   bInherited=true
   bIsVariable=true
   parent=VerticalBox_0
   widget=FilteredTextEntryWidget_Name
   properties {
    bAllowSanitizedStrings=true
    bIsFocusable=true
    entryCharLimit=40
    navigation=&self:WidgetTree.FilteredTextEntryWidget_Name.WidgetNavigation_0
    width Override {enable=true, value=1000}
   }
   slot {class=VerticalBoxSlot, properties {padding=<ltrb 0 24 0 24>}}
  }
  HorizontalBox_0 {
   widgetClassPath=&HorizontalBox
   parent=VerticalBox_0
   widget=HorizontalBox_0
   slot {class=VerticalBoxSlot, properties {verticalAlignment=VAlign_Bottom, padding=<ltrb 0 48 0 0>, size {sizeRule=Fill, value=1}}}
  }
  Button_Confirm {
   widgetClassPath=&WBP_UIKit_Button_Generic
   bInherited=true
   bIsVariable=true
   parent=HorizontalBox_0
   widget=Button_Confirm
   properties {
    backgroundMaterial=/FortUILibrary/UIKit/Buttons/MaterialInstances/CTAButtons/MI_UIKit_BgLoud
    minWidth=250
    navigation=&self:WidgetTree.Button_Confirm.WidgetNavigation_0
    text=Confirm
    buttonPadding=<ltrb 0 0 0 0>
    font {
     bForceMonospaced=false
     bMaterialIsStencil=false
     fontMaterial=/FortUILibrary/UIKit/Buttons/MaterialInstances/CTAButtons/MI_UIKit_TextQuiet
     fontObject=&HeadingNow
     letterSpacing=0
     monospacedWidth=1
     size=18
     skewAmount=0
     typefaceFontName=76 Bold
     outlineSettings {
      bApplyOutlineToDropShadows=false
      bMiteredCorners=false
      bSeparateFillAlpha=false
      outlineMaterial=null
      outlineSize=0
      outlineColor=<rgba 0 0 0 1>
     }
    }
    promptNudge=<10 0>
    triggeringInputAction {dataTable=&InputActionDataTable, rowName=GenericFaceButtonLeft}
   }
   slot {class=HorizontalBoxSlot, properties {horizontalAlignment=HAlign_Center, verticalAlignment=VAlign_Center}}
  }
  Spacer_110 {widgetClassPath=&Spacer, parent=HorizontalBox_0, widget=Spacer_110, properties {size=<24 1>}, slot {class=HorizontalBoxSlot}}
  BackButton {
   widgetClassPath=&WBP_UIKit_Button_Generic
   bIsVariable=true
   parent=HorizontalBox_0
   widget=BackButton
   properties {
    backgroundMaterial=/FortUILibrary/UIKit/Buttons/MaterialInstances/CTAButtons/MI_UIKit_BgRegular
    minWidth=250
    navigation=&self:WidgetTree.BackButton.WidgetNavigation_0
    text=Back
    buttonPadding=<ltrb 0 0 0 0>
    font {
     bForceMonospaced=false
     bMaterialIsStencil=false
     fontMaterial=&MI_UIKit_TextQuiet
     fontObject=&HeadingNow
     letterSpacing=48
     monospacedWidth=1
     size=18
     skewAmount=0
     typefaceFontName=76 Bold
     outlineSettings {
      bApplyOutlineToDropShadows=false
      bMiteredCorners=false
      bSeparateFillAlpha=false
      outlineMaterial=null
      outlineSize=0
      outlineColor=<rgba 0 0 0 1>
     }
    }
    promptNudge=<10 0>
    triggeringInputAction {dataTable=&InputActionDataTable, rowName=GenericBack}
   }
   slot {class=HorizontalBoxSlot, properties {horizontalAlignment=HAlign_Center, verticalAlignment=VAlign_Center}}
  }
  IslandImageScaleBox {
   widgetClassPath=&ScaleBox
   bIsVariable=true
   parent=HorizontalBox_1
   widget=IslandImageScaleBox
   properties {stretch=ScaleToFill}
   slot {class=HorizontalBoxSlot, properties {size {sizeRule=Fill, value=1.3}}}
  }
  IslandImage {
   widgetClassPath=&FortLazyImage
   bInherited=true
   bIsVariable=true
   parent=IslandImageScaleBox
   widget=IslandImage
   properties {brush {
     bIsDynamicallyLoaded=false
     drawAs=Image
     imageType=FullColor
     mirroring=NoMirror
     resourceName=None
     resourceObject=null
     tiling=NoTile
     imageSize=<1920 1080>
     margin=<ltrb 0 0 0 0>
     outlineSettings {
      bUseBrushTransparency=false
      roundingType=HalfHeightRadius
      width=0
      color {colorUseRule=UseColor_Specified, specifiedColor=<rgba 0 0 0 0>}
      cornerRadii=<0 0 0 0>
     }
     tintColor {colorUseRule=UseColor_Specified, specifiedColor=<rgba 1 1 1 1>}
     uVRegion {bIsValid=false, max=<0 0>, min=<0 0>}
    }}
   slot {class=ScaleBoxSlot}
  }
 ]
}
