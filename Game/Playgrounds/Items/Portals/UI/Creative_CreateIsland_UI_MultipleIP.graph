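# Island creation panel in Creative, opened from the island select modal, where players pick a template from per-IP collaboration tabs.
# Tab and selection changes update the tile view, details and collab widgets, with approval checks per gating context choosing buttons and disclaimers.
# Accepting pushes async-loaded collab info or island naming modals, then creates the plot through the player controller.
# Closing is blocked while the request runs; success closes the panel, failure shows an error.
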WidgetBlueprint /Game/Playgrounds/Items/Portals/UI/Creative_CreateIsland_UI_MultipleIP
animations [
 Intro {
  binding_count=2
  length_seconds=0.3715
  object=&self:Intro
  sequence {bindings [
    Header_Text {
     id=#1
     display_name=CommonTextBlock
     possessed_class=CommonTextBlock
     tracks [
      {
       type=MovieScene2DTransformTrack
       sections [{
         type=MovieScene2DTransformSection
         end_frame=7
         start_frame=0
         channels [
          Translation.X {type=MovieSceneScriptingFloatChannel, keys [[0,-80],[2.8583,-80],[5.0013,9.829],[6.3167,0],[7.4307,0]]}
          Angle {type=MovieSceneScriptingFloatChannel}
          Scale.X {type=MovieSceneScriptingFloatChannel}
          Scale.Y {type=MovieSceneScriptingFloatChannel}
          Shear.X {type=MovieSceneScriptingFloatChannel}
          Shear.Y {type=MovieSceneScriptingFloatChannel}
         ]
        }]
      }
      {
       type=MovieSceneFloatTrack
       display_name=Render Opacity
       sections [{
         type=MovieSceneFloatSection
         end_frame=7
         start_frame=0
         channels [None {type=MovieSceneScriptingFloatChannel, keys [[0,0],[2.6617,0],[7.4307,1]]}]
        }]
      }
     ]
    }
    Overlay_661 {
     id=#2
     possessed_class=Overlay
     tracks [
      {
       type=MovieScene2DTransformTrack
       sections [{
         type=MovieScene2DTransformSection
         end_frame=7
         start_frame=0
         channels [
          Translation.X {type=MovieSceneScriptingFloatChannel, keys [[0,-50],[2.738,0],[7.4307,0]]}
          Translation.Y {type=MovieSceneScriptingFloatChannel}
          Angle {type=MovieSceneScriptingFloatChannel}
          Shear.X {type=MovieSceneScriptingFloatChannel}
          Shear.Y {type=MovieSceneScriptingFloatChannel}
         ]
        }]
      }
      {
       type=MovieSceneFloatTrack
       display_name=Render Opacity
       sections [{type=MovieSceneFloatSection, end_frame=4, start_frame=0, channels [None {type=MovieSceneScriptingFloatChannel, keys [[0,0],[3.6303,1]]}]}]
      }
     ]
    }
   ]}
 }
 Outro {
  binding_count=2
  length_seconds=0.2
  object=&self:Outro
  sequence {bindings [
    Overlay_661 {
     id=#3
     possessed_class=Overlay
     tracks [
      {
       type=MovieScene2DTransformTrack
       sections [{
         type=MovieScene2DTransformSection
         end_frame=3
         start_frame=0
         channels [
          Translation.X {type=MovieSceneScriptingFloatChannel}
          Translation.Y {type=MovieSceneScriptingFloatChannel}
          Angle {type=MovieSceneScriptingFloatChannel}
          Scale.X {type=MovieSceneScriptingFloatChannel, keys [[0,0.6],[2,1.05],[3.533,1]]}
          Scale.Y {type=MovieSceneScriptingFloatChannel, keys [[0,0.6],[2,1.15],[3.533,1]]}
          Shear.X {type=MovieSceneScriptingFloatChannel}
          Shear.Y {type=MovieSceneScriptingFloatChannel}
         ]
        }]
      }
      {
       type=MovieSceneFloatTrack
       display_name=Render Opacity
       sections [{type=MovieSceneFloatSection, end_frame=4, start_frame=0, channels [None {type=MovieSceneScriptingFloatChannel, keys [[0,0],[1.428,1]]}]}]
      }
     ]
    }
    Header_Text {
     id=#4
     display_name=CommonTextBlock
     possessed_class=CommonTextBlock
     tracks [{
       type=MovieScene2DTransformTrack
       sections [{
         type=MovieScene2DTransformSection
         end_frame=4
         start_frame=0
         channels [
          Translation.X {type=MovieSceneScriptingFloatChannel, const [0,4]}
          Angle {type=MovieSceneScriptingFloatChannel}
          Scale.X {type=MovieSceneScriptingFloatChannel}
          Scale.Y {type=MovieSceneScriptingFloatChannel}
          Shear.X {type=MovieSceneScriptingFloatChannel}
          Shear.Y {type=MovieSceneScriptingFloatChannel}
         ]
        }]
      }]
    }
   ]}
 }
]
graphs [
 CheckFor_IP_Dialog {nodes [
   CheckFor_IP_Dialog() {is=title K2Node_FunctionEntry}
   n1 = "To String (Name)"(InName=GetSelectedTabId!(self:Common Tab List Widget Base Object Reference=Get NavTabs() -> (NavTabs:WBP UIKit Tabs Object Reference)) -> (ReturnValue:Name=None))
   > Return Node!(
    execute
    ShowDialog:bool=OR Boolean(
     A=GetHasCollabApproval!(
      self:Fort Creative Island Multiple IPUIBase Object Reference
      GameFeatureSetName:String=@n1
      Context:ECollabIPGatingContext Enum=BrandAndICPDialog
     ) -> (ReturnValue:bool)
     B=GetHasCollabApproval(GameFeatureSetName=@n1, Context=BrandOnlyDialog)
    )
   )
  ]}
 ClosePanel {nodes [
   ClosePanel() {is=title K2Node_FunctionEntry}
   > n1 = Branch(Condition=Get bMCPRquestInFlight() -> (bMCPRquestInFlight:bool))
   Return Node!(execute=[@n2.then, @n1.then], bCommitted:bool=false)
   n2 = PopContentWidget!(execute=@n1.else) -> (then)
  ]}
 DialogResultSet {nodes [
   n1 = DialogResultSet!() -> (then, Action:ECollabDialogExitResult Enum=None) {is=title K2Node_FunctionEntry}
   > n2 = Switch on ECollabDialogExitResult!(execute, Selection:ECollabDialogExitResult Enum=@n1) -> (None:Exec, Close:Exec, Continue) {is=title K2Node_SwitchEnum}
   Set ContinueToNaming(execute=@n2.Close, ContinueToNaming:bool=false) -> (Output_Get:bool)
   Set ContinueToNaming(execute=@n2.Continue, ContinueToNaming=true)
  ]}
 EventGraph {nodes [
   n1 = "On Clicked (ProgramDetailsButton)"!() -> (OutputDelegate, then, Button:Common Button Base Object Reference) {is=title K2Node_ComponentBoundEvent}
   n2 = CreateWidgetAsync!(
    execute:Exec=[@n1.then, @n17.else]
    UserWidgetSoftClass:User Widget Soft Class Reference=Get CurrentCollabInfoDialog() -> (CurrentCollabInfoDialog:Common Activatable Widget Soft Class Reference)
    OwningPlayer:Player Controller Object Reference
    bSuspendInputUntilComplete:bool=true
   ) -> (then:Exec, OnComplete:Exec, OnFailure:Exec, UserWidget:User Widget Object Reference) {is=title K2Node_AsyncAction}
   n3 = Cast To FortCreativeCollabDialogWidget!(execute=@n2.OnComplete, Object=@n2) -> (then, CastFailed, AsFort Creative Collab Dialog Widget:Fort Creative Collab Dialog Widget Object Reference) {is=title K2Node_DynamicCast}
   > SetCloseOnly!(execute, self:Fort Creative Collab Dialog Widget Object Reference=@n3, bCloseOnly:bool=true) -> (then)
   > Bind Event to On Widget Deactivated(self:Common Activatable Widget Object Reference=@n3, Delegate=Create Event()) {is=title K2Node_AddDelegate}
   > TryPushWidgetAsModalPanel!(execute, PushWidget:User Widget Object Reference=@n2) -> (then)
   "On Clicked (ReviewTermsButton)"!() -> (OutputDelegate, then, Button:Common Button Base Object Reference) {is=title K2Node_ComponentBoundEvent}
   > n4 = Set CanCreateIslands(CanCreateIslands:bool=false) -> (Output_Get:bool)
   n5 = CheckFor_IP_Dialog!(execute=[@n12.then, @n4.then], self) -> (then, ShowDialog:bool) {is=title K2Node_CallFunction}
   > n6 = Branch(Condition=@n5)
   > n7 = CreateWidgetAsync(UserWidgetSoftClass=Get CurrentCollabInfoDialog(), bSuspendInputUntilComplete=true)
   n8 = Set CurrentDialogWidget(execute=@n7.OnComplete, CurrentDialogWidget:User Widget Object Reference=@n7) -> (Output_Get:User Widget Object Reference)
   > n9 = Cast To FortCreativeCollabDialogWidget(Object=@n8)
   > SetCanCreateIslands!(execute, self:Fort Creative Collab Dialog Widget Object Reference=@n9, bCanCreate:bool=Get CanCreateIslands() -> (CanCreateIslands:bool)) -> (then)
   > Bind Event to On Dialog Result Set(self:Fort Creative Collab Dialog Widget Object Reference=@n9, Delegate=Create Event()) {is=title K2Node_AddDelegate}
   > Bind Event to On Widget Deactivated(self=@n9, Delegate=Create Event())
   > TryPushWidgetAsModalPanel(PushWidget=@n7)
   OpenNameIslandDialog(execute=@n6.else)
   n10 = "On UIKit Tab Clicked (NavTabs)"!() -> (OutputDelegate, then, TabID:Name=None, TabVM:MVVM Base Viewmodel Object Reference) {is=title K2Node_ComponentBoundEvent}
   > SwitchTileView!(execute, self, Selection:Name=@n10.TabID) -> (then)
   n11 = "On UIKit Tab Selected (NavTabs)"!() -> (OutputDelegate, then, TabID:Name=None, TabVM:MVVM Base Viewmodel Object Reference) {is=title K2Node_ComponentBoundEvent}
   > SwitchTileView(Selection=@n11.TabID)
   "On Clicked (AcceptButton)"!() -> (OutputDelegate, then, Button:Common Button Base Object Reference) {is=title K2Node_ComponentBoundEvent}
   > n12 = Set CanCreateIslands(CanCreateIslands=true)
   "On Button Clicked (Button_CloseTouch)"!() -> (OutputDelegate, then, Button:Common Button Legacy Object Reference) {is=title K2Node_ComponentBoundEvent}
   > ClosePanel!(execute, self) -> (then, bCommitted:bool) {is=title K2Node_CallFunction}
   n13 = "On Item Selection Changed (TemplateIslandsTileView)"!() -> (OutputDelegate, then, Item:Object Reference, bIsSelected:bool) {is=title K2Node_ComponentBoundEvent}
   > n14 = Cast To FortCreativeIslandLink!(execute, Object=@n13.Item) -> (then, CastFailed, AsFort Creative Island Link:Fort Creative Island Link Object Reference) {is=title K2Node_DynamicCast}
   n15 = Len(S=Get SupportCode!(self:Fort Creative Island Link Object Reference=@n14) -> (SupportCode:String))
   > SetActiveWidgetIndex(
    self=Get Switcher_CreatorType() -> (Switcher_CreatorType:Common Widget Switcher Legacy Object Reference)
    Index=Select(Option 0:int=0, Option 1:int=1, Index="Equal (Integer)"(A=@n15, B=0) {is=title K2Node_CallFunction}) -> (ReturnValue:int)
   )
   "On Clicked (JoinCollabButton)"!() -> (OutputDelegate, then, Button:Common Button Base Object Reference) {is=title K2Node_ComponentBoundEvent}
   n16 = GetSelectedTabId(self=Get NavTabs()) -> (ReturnValue=None)
   > n17 = Branch(Condition=GetHasCollabApproval(GameFeatureSetName="To String (Name)"(InName=@n16), Context=NameIslandDialog))
   n18 = Cast To PopupIslandNaming_UI_MultipleIP_Modal!(execute=[@n17.then, @n37.then], Object=Get CurrentDialogWidget() -> (CurrentDialogWidget:User Widget Object Reference)) -> (then, CastFailed, AsPopup Island Naming UI Multiple IP Modal:Popup Island Naming UI Multiple IP Modal Object Reference) {is=title K2Node_DynamicCast}
   n19 = CreateWidgetAsync(
    execute=@n18.CastFailed
    UserWidgetSoftClass=Get IslandNamingPopup() -> (IslandNamingPopup:User Widget Soft Class Reference)
    bSuspendInputUntilComplete=true
   )
   n20 = Set CurrentDialogWidget(execute=@n19.OnComplete, CurrentDialogWidget=@n19)
   > n21 = Cast To PopupIslandNaming_UI_MultipleIP_Modal(Object=@n20)
   > Bind Event to On Widget Deactivated(self=@n21, Delegate=Create Event())
   > Assign On Island Naming Closed!(execute:Exec, self:Popup Island Naming UI Multiple IP Modal Object Reference=@n21, Delegate:"Delegate (by ref)"=Create Event()) -> (then:Exec) {is=title K2Node_AssignDelegate}
   > SetUpNameIslandPopup!(
    execute
    self
    UserWidget:Object Reference=@n21
    IslandTileView:List View Object Reference=GetTileViewForTab!(self:Fort Creative Island Multiple IPGeneration Object Reference, InTabID:Name=@n16) -> (ReturnValue:Common Tile View Object Reference)
   ) -> (then)
   > SetIsEnabled(self=Get AcceptButton() -> (AcceptButton:UIKit Modular Button Object Reference), bInIsEnabled=false)
   UpdateInputActionHandler() {is=title K2Node_CustomEvent}
   n22 = GetOwningPlayer()
   > Is Valid(InputObject=@n22)
   > n23 = Branch(Condition=IsUsingTouch(PlayerController=@n22))
   n24 = Get CloseActionBack() -> (CloseActionBack:Data Table Row Handle Structure)
   > Branch(Condition=HasInputActionHandler!(self:Common Activatable Panel Legacy Object Reference, InputActionRow:Data Table Row Handle Structure=@n24) -> (ReturnValue:bool))
   n25 = Self-Reference()
   > RemoveInputActionHandler!(execute, self:Common Activatable Panel Legacy Object Reference=@n25, InputActionRow:Data Table Row Handle Structure=@n24) -> (then)
   SetInputActionHandler!(
    execute=@n23.else
    self:Common Activatable Panel Legacy Object Reference=@n25
    InputActionRow:Data Table Row Handle Structure=@n24
    CommitedEvent:Delegate=Create Event()
   ) -> (then)
   n26 = Island Created!() -> (OutputDelegate, then, bSuccess:bool=~, Reason:"Text (by ref)") {is=title K2Node_CustomEvent}
   > Set bMCPRquestInFlight(bMCPRquestInFlight:bool=false) -> (Output_Get:bool)
   > n27 = Branch(Condition=@n26.bSuccess)
   > SetIsEnabled(self=Get AcceptButton(), bInIsEnabled=true)
   > SetVisibility(self=Get CreateSpinner() -> (CreateSpinner:Image Object Reference), InVisibility=Collapsed)
   n28 = Get IslandSelectUI() -> (IslandSelectUI:Island Select UI Multiple IP Modal Object Reference)
   > n29 = Is Valid(InputObject=@n28)
   > n30 = Set ShouldLoadFirstIslandOnInit!(execute, ShouldLoadFirstIslandOnInit:bool=true, self:Island Select UI Multiple IP Modal Object Reference=@n28) -> (then, Output_Get:bool)
   ClosePanel(execute=[@n29.Is Not Valid, @n30.then])
   ShowErrorDialog!(
    execute=@n27.else
    DisplayMessage:Text=@n26.Reason
    ErrorCode:String=Island Creation Error
    ErrorSeverity:EFortErrorSeverity Enum=Unspecified
   ) -> (then)
   > SetIsEnabled(self=Get AcceptButton(), bInIsEnabled=true)
   > SetVisibility(self=Get CreateSpinner(), InVisibility=Collapsed)
   CollabInfoPopupComplete() {is=title K2Node_CustomEvent}
   > n31 = Cast To FortCreativeCollabDialogWidget(Object=Get CurrentDialogWidget())
   > Unbind Event from On Widget Deactivated(self:Common Activatable Widget Object Reference=@n31, Delegate=Create Event()) {is=title K2Node_RemoveDelegate}
   > n32 = Unbind Event from On Dialog Result Set(self:Fort Creative Collab Dialog Widget Object Reference=@n31, Delegate=Create Event()) {is=title K2Node_RemoveDelegate}
   Set CurrentDialogWidget(execute=[@n32.then, @n31.CastFailed])
   n33 = GetSelectedTabId(self=Get NavTabs()) -> (ReturnValue=None)
   > SetFocus(self=GetTileViewForTab(InTabID=@n33))
   n34 = "To String (Name)"(InName=@n33)
   n35 = GetHasCollabApproval(GameFeatureSetName=@n34, Context=NameIslandDialog)
   n36 = NOT Boolean(A=GetHasCollabApproval(GameFeatureSetName=@n34, Context=ProgramDetailsButton))
   > Branch(Condition=AND Boolean(A=Get ContinueToNaming() -> (ContinueToNaming:bool), B=AND Boolean(A=@n35, B=@n36)))
   > OpenNameIslandDialog()
   n37 = OpenNameIslandDialog() {is=title K2Node_CustomEvent}
   n38 = ConfirmNameForLinkCodeCreation!() -> (OutputDelegate, then, IslandName:String) {is=title K2Node_CustomEvent}
   > n39 = Cast To FortCreativeIslandLink(Object=Get Selected Item!(self:List View Object Reference=Get TemplateIslandsTileView() -> (TemplateIslandsTileView:Common Tile View Object Reference)) -> (ReturnValue:Object Reference) {is=title K2Node_CallFunction})
   > MakeCreativePlotFromLinkCodeAndHandleButtonState!(
    execute
    self
    Mnemonic:String=Break Online Link Id!(OnlineLinkId:"Online Link Id Structure (by ref)"=Get LinkId!(self:Fort Creative Island Link Object Reference=@n39) -> (LinkId:Online Link Id Structure)) -> (Mnemonic:String, Version:int) {is=title K2Node_BreakStruct}
    Locale:String=GetCurrentLanguage!() -> (ReturnValue:String)
    Title:String=@n38.IslandName
   ) -> (then)
   NamingPopupDeactivated() {is=title K2Node_CustomEvent}
   > n40 = Cast To PopupIslandNaming_UI_MultipleIP_Modal(Object=Get CurrentDialogWidget())
   > n41 = Unbind Event from On Widget Deactivated(self=@n40, Delegate=Create Event())
   Set CurrentDialogWidget(execute=[@n41.then, @n40.CastFailed])
   SupportCreator() {is=title K2Node_CustomEvent}
   > n42 = CreateWidgetAsync(UserWidgetSoftClass=/Game/Athena/UI/Frontend/AffiliateKeyModal.AffiliateKeyModal_C, bSuspendInputUntilComplete=true)
   n43 = Cast To AffiliateKeyModal!(execute=@n42.OnComplete, Object=@n42) -> (then, CastFailed, AsAffiliate Key Modal:Affiliate Key Modal Object Reference) {is=title K2Node_DynamicCast}
   > PushContentWidgetAdvanced!(execute, PushWidget:Widget Object Reference=@n43, bHideHeader:bool=true, bHideFooter:bool=true, bHideChatWidget:bool=true) -> (then)
   > n44 = Cast To FortCreativeIslandLink(Object=Get Selected Item(self=Get TemplateIslandsTileView()))
   > "SetText (Editable Text)"!(
    execute
    self:Editable Text Object Reference=Get EditableText!(self:Affiliate Key Modal Object Reference=@n43) -> (EditableText:UIKit Editable Text Object Reference)
    InText:Text="To Text (String)"(InString=Get SupportCode(self=@n44))
   ) -> (then)
   > Set SourceName!(execute, SourceName:String=Get Source Name() -> (Source Name:String), self:Fort Affilate Modal Object Reference=@n43) -> (then, Output_Get:String)
   > SetVisibility(
    self=Get LaunchWebsite!(self:Affiliate Key Modal Object Reference=@n43) -> (LaunchWebsite:WBP UIKit Button Regular Object Reference)
    InVisibility=Collapsed
   )
   n45 = OnIslandNamingClosed!() -> (OutputDelegate, then, Success:bool=~) {is=title K2Node_CustomEvent}
   > SetIsEnabled(self=Get AcceptButton(), bInIsEnabled=NOT Boolean(A=@n45.Success))
   Event On Deactivated()
   > n46 = Cast To FortPlayerControllerAthena(Object=GetOwningPlayer())
   > Unbind Event from On Make New Creative Plot Finished(self:Fort Player Controller Athena Object Reference=@n46, Delegate=Create Event()) {is=title K2Node_RemoveDelegate}
   Event Construct()
   > Parent: Construct()
   n47 = Event InitFromObject!() -> (OutputDelegate, then, InitObject:Object Reference) {is=title K2Node_Event}
   > n48 = Cast To IslandSelect_UI_MultipleIP_Modal!(execute, Object=@n47.InitObject) -> (then, CastFailed, AsIsland Select UI Multiple IP Modal:Island Select UI Multiple IP Modal Object Reference) {is=title K2Node_DynamicCast}
   > n49 = Set IslandSelectUI(IslandSelectUI:Island Select UI Multiple IP Modal Object Reference=@n48) -> (Output_Get:Island Select UI Multiple IP Modal Object Reference)
   OpenPanel!(execute=[@n49.then, @n48.CastFailed], self, Portal:BGA Island Portal Object Reference) -> (then)
   n50 = Event OnInputModeChanged!() -> (OutputDelegate, then, bUsingGamepad:bool) {is=title K2Node_Event}
   > n51 = Get!(execute:Exec) -> (then:Exec, else:Exec, CurrentDialogWidget:User Widget Object Reference)
   n52 = Cast To CommonTileView!(
    execute=@n51.else
    Object=GetActiveWidget!(self:Widget Switcher Object Reference=Get IslandTemplatesWidgetSwitcher() -> (IslandTemplatesWidgetSwitcher:Widget Switcher Object Reference)) -> (ReturnValue:Widget Object Reference)
   ) -> (then, CastFailed, AsCommon Tile View:Common Tile View Object Reference) {is=title K2Node_DynamicCast}
   > UpdateInputActionHandler() {is=title K2Node_CallFunction}
   n53 = GetNumItems!(self:List View Object Reference=@n52) -> (ReturnValue:int=0)
   > Branch(Condition=AND Boolean(A=@n50.bUsingGamepad, B="integer > integer"(A=@n53, B=1) {is=title K2Node_CallFunction}))
   > NavigateToIndex!(execute, self:List View Object Reference=@n52, Index:int=0) -> (then)
   Event OnIslandFilterComplete() {is=title K2Node_Event}
   > Initialize()
   Event On Activated()
   > PlayAnimationForward(InAnimation=Get Intro())
   > UpdateInputActionHandler()
   > PlaySound2D(Sound=Get PanelAppear_Sound() -> (PanelAppear_Sound:Sound Base Object Reference))
   > SetScrollWidget!(execute, ToScroll:Widget Object Reference=Get ScrollBox_Details() -> (ScrollBox_Details:Scroll Box Object Reference)) -> (then)
   > n54 = Cast To FortPlayerControllerAthena(Object=GetOwningPlayer())
   > Bind Event to On Make New Creative Plot Finished(self:Fort Player Controller Athena Object Reference=@n54, Delegate=Create Event()) {is=title K2Node_AddDelegate}
   > CheckIfHasCollabEntitlement(self:Fort Creative Island Multiple IPUIBase Object Reference)
   n55 = Event ChangeTermsAndConditionsVisibility!() -> (OutputDelegate, then, NewVisibility:ESlateVisibility Enum=Visible) {is=title K2Node_Event}
   > SetVisibility(self=Get CollabTermsAndConditonsBox() -> (CollabTermsAndConditonsBox:Size Box Object Reference), InVisibility=@n55.NewVisibility)
   Event OnCollabEntitlementChanged() {is=title K2Node_Event}
   > UpdateCollabWidgets()
   Event OnSelectionUpdated!() -> (OutputDelegate, then, CurrentSelection:Fort Creative Island Multiple IPTemplate Button Object Reference) {is=title K2Node_Event}
   > UpdateCollabWidgets()
  ]}
 GetModalFocusTarget {nodes [
   GetModalFocusTarget() {is=title K2Node_FunctionEntry}
   n1 = Get CurrentDialogWidget()
   > n2 = Is Valid(InputObject=@n1)
   > n3 = Cast To CommonActivatableWidget!(execute, Object=@n1) -> (then, CastFailed, AsCommon Activatable Widget:Common Activatable Widget Object Reference) {is=title K2Node_DynamicCast}
   > n4 = Branch(Condition=IsActivated!(self:Common Activatable Widget Object Reference=@n3) -> (ReturnValue:bool))
   > Return Node!(
    execute
    ReturnValue:Widget Object Reference=GetDesiredFocusTarget!(self:Common Activatable Widget Object Reference=@n3) -> (ReturnValue:Widget Object Reference)
   )
   Return Node(execute=[@n4.else, @n3.CastFailed, @n2.Is Not Valid])
  ]}
 GetNavTabButtonByIndex {nodes [
   n1 = GetNavTabButtonByIndex!() -> (then, Index:int) {is=title K2Node_FunctionEntry}
   n2 = Get NavTabs()
   > Return Node!(
    execute
    NewParam:Common Button Base Object Reference=GetTabButtonBaseByID!(
     self:Common Tab List Widget Base Object Reference=@n2
     TabNameID:Name=GetTabIDbyIndex!(self:UIKit Tab Group Object Reference=@n2, Index:int=@n1) -> (ReturnValue:Name=None)
    ) -> (ReturnValue:Common Button Base Object Reference)
   )
  ]}
 Initialize {nodes [
   Initialize() {is=title K2Node_FunctionEntry}
   > SetupIslandTemplateData(self:Fort Creative Island Multiple IPGeneration Object Reference)
   n1 = Get NavTabs()
   > Set Tabs Data!(
    execute
    self:WBP UIKit Tabs Base Object Reference=@n1
    InTabs Data:Array of UI Kit Tab Group Button View Model Object References=Get TabViewModels() -> (TabViewModels:Array of UI Kit Tab Group Button View Model Object References)
   ) -> (then) {is=title K2Node_CallFunction}
   > SetFocus(self=@n1)
   > SelectTabByIndex!(execute, self:UIKit Tab Group Object Reference=@n1, Index:int=0, bSuppressClickFeedback:bool=true) -> (then, ReturnValue:bool)
  ]}
 MakeCreativePlotAndHandleButtonState {nodes [
   n1 = MakeCreativePlotAndHandleButtonState!() -> (then, Locale:String, Title:String, PlotItemDef:Fort Creative Real Estate Plot Item Definition Object Reference) {is=title K2Node_FunctionEntry}
   > n2 = Is Valid(InputObject=@n1.PlotItemDef)
   PrintString!(execute=@n2.Is Not Valid, InString=PlotItemDef is not valid, bPrintToScreen, bPrintToLog, TextColor, Duration, Key) -> (then)
   n3 = Cast To FortPlayerControllerAthena(execute=@n2.Is Valid, Object=GetOwningPlayer())
   > MakeNewCreativePlot!(
    execute
    self:Fort Player Controller Athena Object Reference=@n3
    PlotType:Fort Creative Real Estate Plot Item Definition Object Reference=Get PlotItemDef() -> (PlotItemDef:Fort Creative Real Estate Plot Item Definition Object Reference)
    Locale:String=Get Locale() -> (Locale:String)
    Title:String=Get Title() -> (Title:String)
   ) -> (then)
   > Set bMCPRquestInFlight(bMCPRquestInFlight=true)
   > DisableButtonWithReason!(
    execute
    self:Common Button Base Object Reference=Get AcceptButton()
    DisabledReason:"Text (by ref)"=Get Working Text() -> (Working Text:Text)
   ) -> (then)
   > SetVisibility(self=Get CreateSpinner(), InVisibility=SelfHitTestInvisible)
  ]}
 MakeCreativePlotFromLinkCodeAndHandleButtonState {nodes [
   MakeCreativePlotFromLinkCodeAndHandleButtonState!() -> (then, Mnemonic:String, Locale:String, Title:String) {is=title K2Node_FunctionEntry}
   > n1 = Cast To FortPlayerControllerAthena(Object=GetOwningPlayer())
   > MakeNewCreativePlotFromLinkCode!(
    execute
    self:Fort Player Controller Athena Object Reference=@n1
    LinkCode:String=Get Mnemonic() -> (Mnemonic:String)
    Locale:String=Get Locale()
    Title:String=Get Title()
   ) -> (then)
   > Set bMCPRquestInFlight(bMCPRquestInFlight=true)
   > DisableButtonWithReason(self=Get AcceptButton(), DisabledReason=Get Working Text())
   > SetVisibility(self=Get CreateSpinner(), InVisibility=SelfHitTestInvisible)
  ]}
 Navigate to Tabs {nodes [
   Navigate to Tabs!() -> (then, Navigation:EUINavigation Enum) {is=title K2Node_FunctionEntry}
   n1 = GetSelectedTabIndex!(self:UIKit Tab Group Object Reference=Get NavTabs()) -> (ReturnValue:int=0)
   n2 = GetNavTabButtonByIndex!(Index:int=@n1) -> (NewParam:Common Button Base Object Reference) {is=title K2Node_CallFunction}
   > n3 = Is Valid(InputObject=@n2)
   > n4 = SelectTabByIndex(self=Get NavTabs(), Index=@n1, bSuppressClickFeedback=true)
   > n5 = Branch(Condition=@n4)
   > Return Node!(execute, ReturnValue:Widget Object Reference=@n2)
   Return Node(execute=[@n5.else, @n3.Is Not Valid], ReturnValue=Get NavTabs())
  ]}
 NavigateToGrid {nodes [
   NavigateToGrid!() -> (then, Navigation:EUINavigation Enum) {is=title K2Node_FunctionEntry}
   n1 = Get NavTabs()
   n2 = GetTileViewForTab(InTabID=GetTabIDbyIndex(self=@n1, Index=GetSelectedTabIndex(self=@n1) -> (ReturnValue=0)) -> (ReturnValue=None))
   > n3 = Is Valid(InputObject=@n2)
   n4 = Get Selected Item(self=@n2)
   > n5 = Is Valid(InputObject=@n4)
   n6 = GetTileViewItemWidget!(
    self:Fort Creative Island Multiple IPGeneration Object Reference
    TileView:Common Tile View Object Reference=@n2
    InItem:Object Reference=@n4
   ) -> (ReturnValue:Widget Object Reference)
   > n7 = Is Valid(InputObject=@n6)
   > Return Node(ReturnValue=@n6)
   Return Node(execute=[@n7.Is Not Valid, @n5.Is Not Valid, @n3.Is Not Valid], ReturnValue=GetActiveWidget(self=Get IslandTemplatesWidgetSwitcher()))
  ]}
 OnTouchToClose {nodes [
   OnTouchToClose!() -> (then, MyGeometry:Geometry Structure, MouseEvent:"Pointer Event Structure (by ref)") {is=title K2Node_FunctionEntry}
   Return Node(ReturnValue:Event Reply Structure=Handled())
  ]}
 OpenPanel {nodes [
   OpenPanel!() -> (then, Portal:BGA Island Portal Object Reference) {is=title K2Node_FunctionEntry}
   > PushContentWidgetAdvanced(PushWidget=Self-Reference(), bHideHeader=true, bHideFooter=false, bHideChatWidget=true)
  ]}
 SetUpNameIslandPopup {nodes [
   n1 = SetUpNameIslandPopup!() -> (then, UserWidget:Object Reference, IslandTileView:List View Object Reference) {is=title K2Node_FunctionEntry}
   > n2 = Cast To PopupIslandNaming_UI_MultipleIP_Modal(Object=@n1)
   > InitFromObject!(execute, self:Popup Island Naming UI Multiple IP Modal Object Reference=@n2, InitObject:Object Reference=Self-Reference()) -> (then)
   n3 = Get Selected Item(self=Get IslandTileView() -> (IslandTileView:List View Object Reference))
   > n4 = Cast To FortCreativeRealEstatePlotItemDefinition!(execute, Object=@n3) -> (then, CastFailed, AsFort Creative Real Estate Plot Item Definition:Fort Creative Real Estate Plot Item Definition Object Reference) {is=title K2Node_DynamicCast}
   > InitFromPlotItemDefinition!(
    execute
    self:Fort Creative Island Multiple IPNaming Modal Object Reference=@n2
    RealEstateItemDefinition:Fort Creative Real Estate Plot Item Definition Object Reference=@n4
   ) -> (then)
   > n5 = Set IsFromIslandLink(IsFromIslandLink:bool=false) -> (Output_Get:bool)
   TryPushWidgetAsModalPanel(execute=[@n7.then, @n5.then], PushWidget=@n2)
   > Set IsFromTemplate!(
    execute
    IsFromTemplate:bool=Get IsFromIslandLink() -> (IsFromIslandLink:bool)
    self:Popup Island Naming UI Multiple IP Modal Object Reference=@n2
   ) -> (then, Output_Get:bool)
   n6 = Cast To FortCreativeIslandLink(execute=@n4.CastFailed, Object=@n3)
   > InitFromIslandLink!(execute, self:Fort Creative Island Multiple IPNaming Modal Object Reference=@n2, IslandLink:Fort Creative Island Link Object Reference=@n6) -> (then)
   > n7 = Set IsFromIslandLink(IsFromIslandLink=true)
  ]}
 SwitchTileView {nodes [
   n1 = SwitchTileView!() -> (then, Selection:Name) {is=title K2Node_FunctionEntry}
   > n2 = Set CurrentSelectedTabID(CurrentSelectedTabID:Name=@n1) -> (Output_Get:Name=None)
   > n3 = GetPropertiesForTab!(execute, self:Fort Creative Island Multiple IPGeneration Object Reference, InTabID:Name=@n2) -> (then, OutTabProperties:Fort Creative Island Multiple IPSelect Collab Properties Structure)
   > Set TabProperties(TabProperties:Fort Creative Island Multiple IPSelect Collab Properties Structure=@n3) -> (Output_Get:Fort Creative Island Multiple IPSelect Collab Properties Structure)
   > n4 = Sequence!(execute) -> (then_0, then_1, then_2:Exec, then_3:Exec, then_4:Exec)
   n5 = Break Fort Creative Island Multiple IPSelect Collab Properties!(FortCreativeIslandMultipleIPSelectCollabProperties:"Fort Creative Island Multiple IPSelect Collab Properties Structure (by ref)"=Get TabProperties() -> (TabProperties:Fort Creative Island Multiple IPSelect Collab Properties Structure)) -> (CollabDisabledDisclaimerText:Text, CollabEnabledDisclaimerText:Text) {is=title K2Node_BreakStruct}
   n6 = TextIsEmpty(InText=@n5)
   n7 = TextIsEmpty(InText=@n5.CollabEnabledDisclaimerText)
   SetVisibility(
    execute=@n4.then_1
    self=Get CollabTermsAndConditonsBox()
    InVisibility=Select(Option 0:ESlateVisibility Enum=Visible, Option 1:ESlateVisibility Enum=Collapsed, Index=AND Boolean(A=@n6, B=@n7)) -> (ReturnValue:ESlateVisibility Enum)
   )
   n8 = GetTileViewForTab(InTabID=Get CurrentSelectedTabID() -> (CurrentSelectedTabID:Name=None))
   SetActiveWidget(execute=@n4.then_0, self:Widget Switcher Object Reference=Get IslandTemplatesWidgetSwitcher(), Widget=@n8)
   > SetFocus(self=@n8)
   n9 = Get Selected Item(self=@n8)
   > Set Selected Item!(execute, self:List View Object Reference=@n8, Item:Object Reference=@n9) -> (then) {is=title K2Node_CallFunction}
   > OnCurrentItemSelectionChanged!(execute, self:Fort Creative Island Multiple IPGeneration Object Reference, Item:Object Reference=@n9) -> (then)
   n10 = Branch(
    execute=@n4.then_2
    Condition=Break Fort Creative Island Multiple IPSelect Collab Properties!(FortCreativeIslandMultipleIPSelectCollabProperties=Get TabProperties()) -> (bShowCreatorInfo:bool=~)
   )
   > SetVisibility(self=Get IslandTemplateNameText() -> (IslandTemplateNameText:Common Text Object Reference), InVisibility=HitTestInvisible)
   > SetVisibility(self=Get IslandDescriptionText() -> (IslandDescriptionText:Common Text Object Reference), InVisibility=Collapsed)
   > SetVisibility(self=Get PublishedIslandName() -> (PublishedIslandName:Common Text Object Reference), InVisibility=Collapsed)
   n11 = Get Overlay_Creator() -> (Overlay_Creator:Overlay Object Reference)
   > SetVisibility(self=@n11, InVisibility=SelfHitTestInvisible)
   > SetActiveWidget(
    self=Get Switcher_CreatorInfo() -> (Switcher_CreatorInfo:Common Widget Switcher Legacy Object Reference)
    Widget=Get TemplateDescriptionSection() -> (TemplateDescriptionSection:Vertical Box Object Reference)
   )
   > SetScrollWidget(ToScroll=Get ScrollBox_Details())
   SetVisibility(execute=@n10.else, self=Get IslandTemplateNameText(), InVisibility=HitTestInvisible)
   > SetVisibility(self=Get IslandDescriptionText(), InVisibility=HitTestInvisible)
   > SetVisibility(self=@n11, InVisibility=Collapsed)
   > SetActiveWidget(self=Get Switcher_CreatorInfo(), Widget=Get IslandDescriptionText())
   UpdateCollabWidgets(execute=@n4.then_3)
   UpdateHeader!(
    execute=@n4.then_4
    self
    InHeaderText:Text=Break Fort Creative Island Multiple IPSelect Collab Properties!(FortCreativeIslandMultipleIPSelectCollabProperties=Get TabProperties()) -> (DisplayName:Text)
   ) -> (then)
  ]}
 TabNavigation {nodes [
   n1 = TabNavigation!() -> (then, Navigation:EUINavigation Enum) {is=title K2Node_FunctionEntry}
   > n2 = Switch on EUINavigation!(execute, Selection:EUINavigation Enum=@n1) -> (Left:Exec, Right:Exec, Up:Exec, Down:Exec, Next:Exec, Previous:Exec, Invalid:Exec) {is=title K2Node_SwitchEnum}
   n3 = GetSelectedTabIndex(self=Get NavTabs()) -> (ReturnValue=0)
   n4 = Increment Int(" "=[@n2.Down, @n2.Next], Value=@n3)
   > n5 = Set NewIndex(NewIndex:int=@n4) -> (Output_Get:int=0)
   n6 = Get NewIndex() -> (NewIndex:int=0)
   n7 = GetNavTabButtonByIndex(Index=@n6)
   n8 = Is Valid(exec=[@n12.then, @n5.then], InputObject=@n7)
   > n9 = SelectTabByIndex(self=Get NavTabs(), Index=@n6, bSuppressClickFeedback=true)
   > n10 = Branch(Condition=@n9)
   > Return Node(ReturnValue:Widget Object Reference=@n7)
   Return Node(execute=[@n10.else, @n8.Is Not Valid], ReturnValue=Get NavTabs())
   n11 = Decrement Int(" "=[@n2.Previous, @n2.Up], Value=@n3)
   > n12 = Set NewIndex(NewIndex=@n11) -> (Output_Get=0)
  ]}
 Update Accept Button State {nodes [
   n1 = Update Accept Button State!() -> (then, GameFeatureSetName:String) {is=title K2Node_FunctionEntry}
   n2 = Get ReviewTermsBox() -> (ReviewTermsBox:Vertical Box Object Reference)
   n3 = GetHasCollabApproval(GameFeatureSetName=@n1, Context=ReviewTermsButton)
   n4 = Select(
    Option 0:Widget Object Reference=Select(
     Option 0=Select(Option 1=@n2, Index=@n3)
     Option 1=Get JoinCollabButton() -> (JoinCollabButton:WBP UIKit Button Generic Object Reference)
     Index=GetHasCollabApproval(GameFeatureSetName=@n1, Context=JoinProgramButton)
    )
    Option 1:Widget Object Reference=Get AcceptButton()
    Index=GetHasCollabApproval(GameFeatureSetName=@n1, Context=CreateButton)
   ) -> (ReturnValue:Widget Object Reference)
   > Is Valid(InputObject=@n4)
   > SetActiveWidget(
    self:Common Visibility Switcher Object Reference=Get CreateButtonSwitcher() -> (CreateButtonSwitcher:Common Visibility Switcher Object Reference)
    Widget=@n4
   )
   n5 = GetHasCollabApproval(GameFeatureSetName=@n1, Context=ProgramDetailsButton)
   n6 = Get ProgramDetailsButton() -> (ProgramDetailsButton:WBP UIKit Button Generic Object Reference)
   > SetVisibility(
    self=@n6
    InVisibility=Select(Option 0:ESlateVisibility Enum=Collapsed, Option 1:ESlateVisibility Enum=Visible, Index=@n5) -> (ReturnValue:ESlateVisibility Enum)
   )
   n7 = Select(
    Option 0:Text="LOCTEXT(\"Templates\")"
    Option 1:Text
    Index=GetCurrentCollabProperties!(self:Fort Creative Island Multiple IPGeneration Object Reference) -> (bOutIsValid:bool, OutCurrentCollabProperties:Fort Creative Island Multiple IPSelect Collab Properties Structure)
   ) -> (ReturnValue:Text)
   n8 = Get ProgramDetailsButtonTextFormat() -> (ProgramDetailsButtonTextFormat:Text)
   > SetText!(
    execute
    self:WBP UIKit Button Generic Base Object Reference=@n6
    Text:Text=Select(Option 0="LOCTEXT(\"Program Details\")", Option 1=Format Text!(Format=@n8, 0:Text=@n7) -> (Result), Index=@n5)
   ) -> (then)
   n9 = GetCurrentCollabProperties()
   > SetText(
    self:WBP UIKit Block Text Object Reference=Get Block_SecondText!(self:WBP UIKit Button Generic Base Object Reference=Get ProgramDetailsButton()) -> (Block_SecondText:WBP UIKit Block Text Object Reference)
    Text=Select(
     Option 1=Break Fort Creative Island Multiple IPSelect Collab Properties(FortCreativeIslandMultipleIPSelectCollabProperties=@n9.OutCurrentCollabProperties)
     Index=@n9
    )
   )
  ]}
 Update Terms and Conditions {nodes [
   n1 = Update Terms and Conditions!() -> (then, DisclaimerText:Text) {is=title K2Node_FunctionEntry}
   > n2 = Set Text(Text=@n1, self:Rich Text Block Object Reference=Get CollabDisclaimerText() -> (CollabDisclaimerText:Common Rich Text Block Object Reference))
   > SetVisibility(
    self=Get CollabTermsAndConditonsBox()
    InVisibility=Select(Option 0:ESlateVisibility Enum=SelfHitTestInvisible, Option 1:ESlateVisibility Enum=Collapsed, Index=TextIsEmpty(InText=@n2)) -> (ReturnValue:ESlateVisibility Enum)
   )
  ]}
 UpdateCollabWidgets {nodes [
   UpdateCollabWidgets() {is=title K2Node_FunctionEntry}
   n1 = GetCurrentCollabProperties!(self) -> (
    bOutIsValid
    OutCurrentCollabProperties_GFSName:String
    OutCurrentCollabProperties_DisplayName:Text
    OutCurrentCollabProperties_bEntitlementIsRequired:bool=~
    OutCurrentCollabProperties_bPlayerIsEntitledToGFS:bool=~
    OutCurrentCollabProperties_bShowCreatorInfo:bool=~
    OutCurrentCollabProperties_IP_GatingType:ECollabIPGatingType Enum
    OutCurrentCollabProperties_CollabIcon:Texture 2D Soft Object Reference
    OutCurrentCollabProperties_CollabDisabledDisclaimerText:Text
    OutCurrentCollabProperties_CollabEnabledDisclaimerText:Text
    OutCurrentCollabProperties_CollabInfoDialog:Common Activatable Widget Soft Class Reference
    OutCurrentCollabProperties_TileView:Common Tile View Object Reference
    OutCurrentCollabProperties_CompatibleStarterListItems:Array of Object References
   )
   > Branch(Condition=@n1)
   > Set CurrentGFSName(CurrentGFSName:String=@n1.OutCurrentCollabProperties_GFSName) -> (Output_Get:String)
   > n2 = Sequence!(execute) -> (then_0, then_1, then_2)
   n3 = Get IP_PartnerIcon() -> (IP_PartnerIcon:Fort Lazy Image Object Reference)
   SetBrushFromLazyTexture!(
    execute=@n2.then_2
    self:Common Lazy Image Object Reference=@n3
    LazyTexture:"Texture 2D Soft Object Reference (by ref)"=@n1.OutCurrentCollabProperties_CollabIcon
    bMatchSize:bool=false
   ) -> (then)
   n4 = AND Boolean(
    A=NOT Boolean(A=IsAssetNull!(Asset:Soft Object Reference=@n1.OutCurrentCollabProperties_CollabIcon) -> (ReturnValue:bool))
    B=GetHasCollabApproval(GameFeatureSetName=Get CurrentGFSName() -> (CurrentGFSName:String), Context=Icon)
   )
   > SetVisibility(self=@n3, InVisibility=Select(Option 0=Collapsed, Option 1=SelfHitTestInvisible, Index=@n4))
   > SetVisibility(
    self=Get CollabPartnerIcon() -> (CollabPartnerIcon:Fort Lazy Image Object Reference)
    InVisibility=Select(Option 0=HitTestInvisible, Option 1=Hidden, Index=@n4)
   )
   Update Accept Button State!(execute=@n2.then_0, self, GameFeatureSetName:String=Get CurrentGFSName()) -> (then) {is=title K2Node_CallFunction}
   n5 = Get CurrentGFSName()
   Update Terms and Conditions!(
    execute=@n2.then_1
    self
    DisclaimerText:Text=Select(
     Option 0:Text=@n1.OutCurrentCollabProperties_CollabDisabledDisclaimerText
     Option 1:Text=@n1.OutCurrentCollabProperties_CollabEnabledDisclaimerText
     Index=GetHasCollabApproval(GameFeatureSetName=@n5, Context=Disclaimer)
    ) -> (ReturnValue:Text)
   ) -> (then) {is=title K2Node_CallFunction}
  ]}
 UpdateHeader {nodes [
   n1 = UpdateHeader!() -> (then, InHeaderText:Text) {is=title K2Node_FunctionEntry}
   > "SetText (Text)"(self=Get Subheader_text() -> (Subheader_text:Common Text Object Reference), InText=@n1)
  ]}
]
variables [
 bIsClosing {type=boolean}
 bMCPRquestInFlight {type=boolean}
 CanCreateIslands {type=boolean}
 CloseActionBack {type=DataTableRowHandle, default {dataTable=&InputActionDataTable, rowName=Cancel}}
 CollabEntitled {type=string, default=CReady}
 CollabNotEntitled {type=string, default=CNotReady}
 ContinueToNaming {type=boolean}
 CurrentDialogWidget {type=UserWidget, default=null}
 CurrentSelectedTabID {type=string, default=None}
 IslandDefinition {type=FortCreativeRealEstatePlotItemDefinition, default=null}
 IslandNamingPopup {type=Class, default=&/PopupIslandNaming_UI_MultipleIP_Modal.PopupIslandNaming_UI_MultipleIP_Modal_C}
 IslandSelectUI {type=IslandSelect_UI_MultipleIP_Modal_C, default=null}
 PanelAppear_Sound {type=SoundBase, default=/Game/Sounds/Fort_UI_Sounds/FrontEnd/New_FrontEnd/Fortnite_UI_MainMenu_Tab_03_Cue}
 PanelDismiss_Sound {type=SoundBase, default=/Game/Sounds/Fort_UI_Sounds/FrontEnd/New_FrontEnd/Fortnite_UI_MissionObj_Cue}
 Portal {type=BGA_IslandPortal_C, default=null}
 ProgramDetailsButtonTextFormat {type=string, default="Creating with {0}"}
 SelectPlot_Sound {type=SoundBase, default=/Game/Sounds/Fort_UI_Sounds/FrontEnd/New_FrontEnd/Fortnite_UI_Options_Menu_01_Cue}
 Source Name {type=string, default=Creative_CreateIsland_UI}
 Working Text {type=string, default=Working}
]
widgets {
 info {
  inheritedWidgetCount=20
  namedSlotCount=1
  parentClass=/Script/FortniteUI.FortCreativeIslandMultipleIPGeneration
  rootWidgetClass=&CommonBorder
  widgetCount=68
 }
 widgets [
  TouchToCloseZone {widgetClassPath=&CommonBorder, widget=TouchToCloseZone, properties {padding=<ltrb 0 0 0 0>}}
  BackgroundBlur_1 {
   widgetClassPath=&BackgroundBlur
   parent=TouchToCloseZone
   widget=BackgroundBlur_1
   properties {blurStrength=25}
   slot {class=BorderSlot, properties {padding=<ltrb 0 0 0 0>}}
  }
  SafeZone_0 {
   widgetClassPath=&SafeZone
   bIsVariable=true
   parent=BackgroundBlur_1
   widget=SafeZone_0
   slot {class=BackgroundBlurSlot, properties {padding=<ltrb 0 0 0 0>}}
  }
  Overlay_661 {widgetClassPath=&Overlay, parent=SafeZone_0, widget=Overlay_661, slot {class=SafeZoneSlot}}
  Imge_Backplate {
   widgetClassPath=&Image
   parent=Overlay_661
   widget=Imge_Backplate
   properties {brush {
     bIsDynamicallyLoaded=false
     drawAs=Image
     imageType=FullColor
     mirroring=NoMirror
     resourceName=None
     resourceObject=/Game/Playgrounds/Items/Portals/UI/Art/MI_UI_Creative_NebulousBackground
     tiling=NoTile
     imageSize=<32 32>
     margin=<ltrb 0 0 0 0>
     outlineSettings {
      bUseBrushTransparency=false
      roundingType=HalfHeightRadius
      width=0
      color {colorUseRule=UseColor_Specified, specifiedColor=<rgba 0 0 0 0>}
      cornerRadii=<0 0 0 0>
     }
     tintColor {colorUseRule=UseColor_Specified, specifiedColor=<rgba 1 1 1 1>}
     uVRegion {bIsValid=false, max=<0 0>, min=<0 0>}
    }}
   slot {class=OverlaySlot, properties {horizontalAlignment=HAlign_Fill, verticalAlignment=VAlign_Fill}}
  }
  UIKitGridPanel_671 {
   widgetClassPath=/Script/FortniteUI.UIKitGridPanel
   parent=Overlay_661
   widget=UIKitGridPanel_671
   properties {columnFill [0.225,0.55,0.225], rowFill [0.18,0.82]}
   slot {class=OverlaySlot, properties {horizontalAlignment=HAlign_Fill, verticalAlignment=VAlign_Fill, padding=<ltrb 0 24 0 0>}}
  }
  Nav_Overlay {
   widgetClassPath=&Overlay
   parent=UIKitGridPanel_671
   widget=Nav_Overlay
   slot {class=GridSlot, properties {row=1, padding=<ltrb 48 12 0 0>}}
  }
  Overlay_0 {
   widgetClassPath=&Overlay
   parent=Nav_Overlay
   widget=Overlay_0
   slot {class=OverlaySlot, properties {horizontalAlignment=HAlign_Fill}}
  }
  WBP_CaptureForPostBufferUpdate {
   widgetClassPath=&WBP_CaptureForPostBufferUpdate
   bIsVariable=true
   parent=Overlay_0
   slot {class=OverlaySlot, properties {horizontalAlignment=HAlign_Fill, verticalAlignment=VAlign_Fill}}
  }
  NavTabs {
   widgetClassPath=/FortUILibrary/UIKit/Tabs/WBP_UIKit_Tabs
   bIsVariable=true
   parent=Overlay_0
   widget=NavTabs
   properties {
    bIsFocusable=true
    backplateCornerRadius=Tab
    buttonClass=&/Creative_CreateIsland_Button_Tab.Creative_CreateIsland_Button_Tab_C
    inputActions=None
    isNaturalNavigation=true
    isVertical=true
    minButtonHeight=48
    navigation=&self:WidgetTree.NavTabs.WidgetNavigation_3
    nextPromptHorizontalAlignment=HAlign_Left
    tabGroupHorizontalAlignment=HAlign_Fill
    useBackplate=true
    usePromptBackplate=true
    localTabsViewModels [&self:WidgetTree.NavTabs.FortTabViewModel_0]
   }
   slot {class=OverlaySlot, properties {horizontalAlignment=HAlign_Fill}}
  }
  Tiles_Overlay {
   widgetClassPath=&Overlay
   parent=UIKitGridPanel_671
   widget=Tiles_Overlay
   slot {class=GridSlot, properties {column=1, row=1, padding=<ltrb 0 12 0 142>}}
  }
  VerticalBox_0 {
   widgetClassPath=&VerticalBox
   parent=Tiles_Overlay
   widget=VerticalBox_0
   slot {class=OverlaySlot, properties {horizontalAlignment=HAlign_Fill, verticalAlignment=VAlign_Fill, padding=<ltrb 24 0 24 0>}}
  }
  IslandTemplatesWidgetSwitcher {
   widgetClassPath=&WidgetSwitcher
   bIsVariable=true
   parent=VerticalBox_0
   widget=IslandTemplatesWidgetSwitcher
   slot {class=VerticalBoxSlot, properties {size {sizeRule=Fill, value=1}}}
  }
  StartingIslandsTileView {
   widgetClassPath=/Script/CommonUI.CommonTileView
   bInherited=true
   bIsVariable=true
   parent=IslandTemplatesWidgetSwitcher
   widget=StartingIslandsTileView
   properties {
    clipping=ClipToBoundsWithoutIntersecting
    entryHeight=150
    entryWidgetClass=&/IslandMultipleIPTemplateOptionEntry_Base.IslandMultipleIPTemplateOptionEntry_Base_C
    entryWidth=150
    horizontalEntrySpacing=12
    navigation=&self:WidgetTree.StartingIslandsTileView.WidgetNavigation_0
    numDesignerPreviewEntries=17
    tileAlignment=EvenlySize
    verticalEntrySpacing=12
   }
   slot {class=WidgetSwitcherSlot}
  }
  TemplateIslandsTileView {
   widgetClassPath=&CommonTileView
   bInherited=true
   bIsVariable=true
   parent=IslandTemplatesWidgetSwitcher
   widget=TemplateIslandsTileView
   properties {
    clipping=ClipToBoundsWithoutIntersecting
    entryHeight=136
    entryWidgetClass=&IslandMultipleIPTemplateOptionEntry_Base_C
    entryWidth=244
    horizontalEntrySpacing=8
    navigation=&self:WidgetTree.TemplateIslandsTileView.WidgetNavigation_0
    numDesignerPreviewEntries=17
    tileAlignment=EvenlySize
    verticalEntrySpacing=8
   }
   slot {class=WidgetSwitcherSlot}
  }
  CollabIslandsTileView {
   widgetClassPath=&CommonTileView
   bInherited=true
   bIsVariable=true
   parent=IslandTemplatesWidgetSwitcher
   widget=CollabIslandsTileView
   properties {
    clipping=ClipToBoundsWithoutIntersecting
    entryHeight=136
    entryWidgetClass=&IslandMultipleIPTemplateOptionEntry_Base_C
    entryWidth=244
    horizontalEntrySpacing=8
    navigation=&self:WidgetTree.CollabIslandsTileView.WidgetNavigation_0
    numDesignerPreviewEntries=17
    tileAlignment=EvenlySize
    verticalEntrySpacing=8
   }
   slot {class=WidgetSwitcherSlot}
  }
  CollabTileView_1 {
   widgetClassPath=&CommonTileView
   bInherited=true
   bIsVariable=true
   parent=IslandTemplatesWidgetSwitcher
   widget=CollabTileView_1
   properties {
    clipping=ClipToBoundsWithoutIntersecting
    entryHeight=136
    entryWidgetClass=&IslandMultipleIPTemplateOptionEntry_Base_C
    entryWidth=244
    horizontalEntrySpacing=8
    navigation=&self:WidgetTree.CollabTileView_1.WidgetNavigation_0
    numDesignerPreviewEntries=17
    tileAlignment=EvenlySize
    verticalEntrySpacing=8
   }
   slot {class=WidgetSwitcherSlot}
  }
  CollabTileView_2 {
   widgetClassPath=&CommonTileView
   bInherited=true
   bIsVariable=true
   parent=IslandTemplatesWidgetSwitcher
   widget=CollabTileView_2
   properties {
    clipping=ClipToBoundsWithoutIntersecting
    entryHeight=136
    entryWidgetClass=&IslandMultipleIPTemplateOptionEntry_Base_C
    entryWidth=244
    horizontalEntrySpacing=8
    navigation=&self:WidgetTree.CollabTileView_2.WidgetNavigation_0
    numDesignerPreviewEntries=17
    tileAlignment=EvenlySize
    verticalEntrySpacing=8
   }
   slot {class=WidgetSwitcherSlot}
  }
  CollabTileView_3 {
   widgetClassPath=&CommonTileView
   bInherited=true
   bIsVariable=true
   parent=IslandTemplatesWidgetSwitcher
   widget=CollabTileView_3
   properties {
    clipping=ClipToBoundsWithoutIntersecting
    entryHeight=136
    entryWidgetClass=&IslandMultipleIPTemplateOptionEntry_Base_C
    entryWidth=244
    horizontalEntrySpacing=8
    navigation=&self:WidgetTree.CollabTileView_3.WidgetNavigation_0
    numDesignerPreviewEntries=17
    tileAlignment=EvenlySize
    verticalEntrySpacing=8
   }
   slot {class=WidgetSwitcherSlot}
  }
  CollabTileView_4 {
   widgetClassPath=&CommonTileView
   bInherited=true
   bIsVariable=true
   parent=IslandTemplatesWidgetSwitcher
   widget=CollabTileView_4
   properties {
    clipping=ClipToBoundsWithoutIntersecting
    entryHeight=136
    entryWidgetClass=&IslandMultipleIPTemplateOptionEntry_Base_C
    entryWidth=244
    horizontalEntrySpacing=8
    navigation=&self:WidgetTree.CollabTileView_4.WidgetNavigation_0
    numDesignerPreviewEntries=17
    tileAlignment=EvenlySize
    verticalEntrySpacing=8
   }
   slot {class=WidgetSwitcherSlot}
  }
  CollabTileView_5 {
   widgetClassPath=&CommonTileView
   bInherited=true
   bIsVariable=true
   parent=IslandTemplatesWidgetSwitcher
   widget=CollabTileView_5
   properties {
    clipping=ClipToBoundsWithoutIntersecting
    entryHeight=136
    entryWidgetClass=&IslandMultipleIPTemplateOptionEntry_Base_C
    entryWidth=244
    horizontalEntrySpacing=8
    navigation=&self:WidgetTree.CollabTileView_5.WidgetNavigation_0
    numDesignerPreviewEntries=17
    tileAlignment=EvenlySize
    verticalEntrySpacing=8
   }
   slot {class=WidgetSwitcherSlot}
  }
  CollabTileView_6 {
   widgetClassPath=&CommonTileView
   bInherited=true
   bIsVariable=true
   parent=IslandTemplatesWidgetSwitcher
   widget=CollabTileView_6
   properties {
    clipping=ClipToBoundsWithoutIntersecting
    entryHeight=136
    entryWidgetClass=&IslandMultipleIPTemplateOptionEntry_Base_C
    entryWidth=244
    horizontalEntrySpacing=8
    navigation=&self:WidgetTree.CollabTileView_6.WidgetNavigation_0
    numDesignerPreviewEntries=17
    tileAlignment=EvenlySize
    verticalEntrySpacing=8
   }
   slot {class=WidgetSwitcherSlot}
  }
  Details_Overlay {
   widgetClassPath=&Overlay
   parent=UIKitGridPanel_671
   widget=Details_Overlay
   slot {class=GridSlot, properties {column=2, row=1, padding=<ltrb 12 12 48 148>}}
  }
  VBox_IslandDetails {
   widgetClassPath=&VerticalBox
   bIsVariable=true
   parent=Details_Overlay
   widget=VBox_IslandDetails
   slot {class=OverlaySlot, properties {horizontalAlignment=HAlign_Fill, verticalAlignment=VAlign_Fill}}
  }
  SizeBox_248 {
   widgetClassPath=&SizeBox
   parent=VBox_IslandDetails
   widget=SizeBox_248
   properties {bOverride_WidthOverride=true, widthOverride=416}
   slot {class=VerticalBoxSlot, properties {horizontalAlignment=HAlign_Left}}
  }
  IP_PartnerIcon {
   widgetClassPath=&FortLazyImage
   bIsVariable=true
   parent=VBox_IslandDetails
   widget=IP_PartnerIcon
   properties {
    visibility=Collapsed
    brush {
     bIsDynamicallyLoaded=false
     drawAs=Image
     imageType=NoImage
     mirroring=NoMirror
     resourceName=None
     resourceObject=null
     tiling=NoTile
     imageSize=<68 68>
     margin=<ltrb 0 0 0 0>
     outlineSettings {
      bUseBrushTransparency=false
      roundingType=HalfHeightRadius
      width=0
      color {colorUseRule=UseColor_Specified, specifiedColor=<rgba 0 0 0 0>}
      cornerRadii=<0 0 0 0>
     }
     tintColor {colorUseRule=UseColor_Specified, specifiedColor=<rgba 1 1 1 1>}
     uVRegion {bIsValid=false, max=<0 0>, min=<0 0>}
    }
   }
   slot {class=VerticalBoxSlot, properties {horizontalAlignment=HAlign_Left, verticalAlignment=VAlign_Top, padding=<ltrb 0 0 0 20>}}
  }
  ScrollBox_Details {
   widgetClassPath=&ScrollBox
   bIsVariable=true
   parent=VBox_IslandDetails
   widget=ScrollBox_Details
   slot {class=VerticalBoxSlot, properties {padding=<ltrb 0 0 0 12>, size {sizeRule=Fill, value=1}}}
  }
  IslandTemplateNameText {
   widgetClassPath=&CommonTextBlock
   bInherited=true
   parent=ScrollBox_Details
   widget=IslandTemplateNameText
   properties {
    autoWrapText=true
    text=Pick an Island
    textTransformPolicy=ToUpper
    font {
     bForceMonospaced=false
     bMaterialIsStencil=false
     fontMaterial=null
     fontObject=&HeadingNow
     letterSpacing=0
     monospacedWidth=1
     size=22.5
     skewAmount=0
     typefaceFontName=75 Medium Italic
     outlineSettings {
      bApplyOutlineToDropShadows=false
      bMiteredCorners=false
      bSeparateFillAlpha=false
      outlineMaterial=null
      outlineSize=0
      outlineColor=<rgba 0 0 0 1>
     }
    }
    shadowOffset=<0 0>
   }
   slot {class=ScrollBoxSlot}
  }
  Switcher_CreatorInfo {
   widgetClassPath=&CommonWidgetSwitcherLegacy
   bIsVariable=true
   parent=ScrollBox_Details
   widget=Switcher_CreatorInfo
   slot {class=ScrollBoxSlot, properties {padding=<ltrb 0 22 0 0>}}
  }
  IslandDescriptionText {
   widgetClassPath=&CommonTextBlock
   bInherited=true
   parent=Switcher_CreatorInfo
   widget=IslandDescriptionText
   properties {
    autoWrapText=true
    text="[Island Description]"
    font {
     bForceMonospaced=false
     bMaterialIsStencil=false
     fontMaterial=null
     fontObject=&HeadingNow
     letterSpacing=0
     monospacedWidth=1
     size=18
     skewAmount=0
     typefaceFontName=74 Regular
     outlineSettings {
      bApplyOutlineToDropShadows=false
      bMiteredCorners=false
      bSeparateFillAlpha=false
      outlineMaterial=null
      outlineSize=0
      outlineColor=<rgba 0 0 0 1>
     }
    }
    shadowOffset=<0 0>
   }
   slot {class=WidgetSwitcherSlot, properties {padding=<ltrb 0 0 0 32>}}
  }
  TemplateDescriptionSection {
   widgetClassPath=&VerticalBox
   bIsVariable=true
   parent=Switcher_CreatorInfo
   widget=TemplateDescriptionSection
   slot {class=WidgetSwitcherSlot}
  }
  PublishedIslandBulletOne {
   widgetClassPath=&CommonTextBlock
   bInherited=true
   parent=TemplateDescriptionSection
   widget=PublishedIslandBulletOne
   properties {
    autoWrapText=true
    font {
     bForceMonospaced=false
     bMaterialIsStencil=false
     fontMaterial=null
     fontObject=&HeadingNow
     letterSpacing=0
     monospacedWidth=1
     size=18
     skewAmount=0
     typefaceFontName=74 Regular
     outlineSettings {
      bApplyOutlineToDropShadows=false
      bMiteredCorners=false
      bSeparateFillAlpha=false
      outlineMaterial=null
      outlineSize=0
      outlineColor=<rgba 0 0 0 1>
     }
    }
    shadowOffset=<4.9266 5.2555>
   }
   slot {class=VerticalBoxSlot}
  }
  PublishedIslandBulletTwo {
   widgetClassPath=&CommonTextBlock
   bInherited=true
   parent=TemplateDescriptionSection
   widget=PublishedIslandBulletTwo
   properties {
    autoWrapText=true
    font {
     bForceMonospaced=false
     bMaterialIsStencil=false
     fontMaterial=null
     fontObject=&HeadingNow
     letterSpacing=0
     monospacedWidth=1
     size=18
     skewAmount=0
     typefaceFontName=74 Regular
     outlineSettings {
      bApplyOutlineToDropShadows=false
      bMiteredCorners=false
      bSeparateFillAlpha=false
      outlineMaterial=null
      outlineSize=0
      outlineColor=<rgba 0 0 0 1>
     }
    }
    shadowOffset=<4.9266 5.2555>
   }
   slot {class=VerticalBoxSlot}
  }
  PublishedIslandBulletThree {
   widgetClassPath=&CommonTextBlock
   bInherited=true
   parent=TemplateDescriptionSection
   widget=PublishedIslandBulletThree
   properties {
    autoWrapText=true
    font {
     bForceMonospaced=false
     bMaterialIsStencil=false
     fontMaterial=null
     fontObject=&HeadingNow
     letterSpacing=0
     monospacedWidth=1
     size=18
     skewAmount=0
     typefaceFontName=74 Regular
     outlineSettings {
      bApplyOutlineToDropShadows=false
      bMiteredCorners=false
      bSeparateFillAlpha=false
      outlineMaterial=null
      outlineSize=0
      outlineColor=<rgba 0 0 0 1>
     }
    }
    shadowOffset=<4.9266 5.2555>
   }
   slot {class=VerticalBoxSlot}
  }
  Overlay_Creator {
   widgetClassPath=&Overlay
   bIsVariable=true
   parent=VBox_IslandDetails
   widget=Overlay_Creator
   properties {visibility=Collapsed}
   slot {class=VerticalBoxSlot, properties {verticalAlignment=VAlign_Bottom, padding=<ltrb 0 0 0 10>}}
  }
  VBox_Creator {
   widgetClassPath=&VerticalBox
   parent=Overlay_Creator
   widget=VBox_Creator
   slot {class=OverlaySlot, properties {horizontalAlignment=HAlign_Fill, verticalAlignment=VAlign_Fill, padding=<ltrb 0 12 0 12>}}
  }
  PublishedIslandName {
   widgetClassPath=&CommonTextBlock
   bInherited=true
   parent=VBox_Creator
   widget=PublishedIslandName
   properties {
    autoWrapText=true
    text="[Published Island Name]"
    textTransformPolicy=ToUpper
    font {
     bForceMonospaced=false
     bMaterialIsStencil=false
     fontMaterial=null
     fontObject=&HeadingNow
     letterSpacing=0
     monospacedWidth=1
     size=18
     skewAmount=0
     typefaceFontName=75 Medium
     outlineSettings {
      bApplyOutlineToDropShadows=false
      bMiteredCorners=false
      bSeparateFillAlpha=false
      outlineMaterial=null
      outlineSize=0
      outlineColor=<rgba 0 0 0 1>
     }
    }
    shadowOffset=<4.9266 5.2555>
   }
   slot {class=VerticalBoxSlot}
  }
  VerticalBox_2 {
   widgetClassPath=&VerticalBox
   parent=VBox_Creator
   widget=VerticalBox_2
   slot {class=VerticalBoxSlot, properties {verticalAlignment=VAlign_Center}}
  }
  Switcher_CreatorType {
   widgetClassPath=&CommonWidgetSwitcherLegacy
   bIsVariable=true
   parent=VerticalBox_2
   widget=Switcher_CreatorType
   properties {transitionDuration=0}
   slot {class=VerticalBoxSlot}
  }
  VBox_CreatorDetails {
   widgetClassPath=&VerticalBox
   parent=Switcher_CreatorType
   widget=VBox_CreatorDetails
   slot {class=WidgetSwitcherSlot, properties {verticalAlignment=VAlign_Center}}
  }
  CommonTextBlock_102 {
   widgetClassPath=&CommonTextBlock
   parent=VBox_CreatorDetails
   widget=CommonTextBlock_102
   properties {
    text=Created by:
    font {
     bForceMonospaced=false
     bMaterialIsStencil=false
     fontMaterial=null
     fontObject=&HeadingNow
     letterSpacing=0
     monospacedWidth=1
     size=18
     skewAmount=0
     typefaceFontName=75 Medium
     outlineSettings {
      bApplyOutlineToDropShadows=false
      bMiteredCorners=false
      bSeparateFillAlpha=false
      outlineMaterial=null
      outlineSize=0
      outlineColor=<rgba 0 0 0 1>
     }
    }
    renderTransform {angle=0, scale=<1 1>, shear=<-8 0>, translation=<0 0>}
    shadowOffset=<0 0>
   }
   slot {class=VerticalBoxSlot}
  }
  PublishedIslandAuthor {
   widgetClassPath=&CommonTextBlock
   bInherited=true
   parent=VBox_CreatorDetails
   widget=PublishedIslandAuthor
   properties {
    scrollStyle=&TextScrollStyle-Base_C
    text="[creator name]"
    textTransformPolicy=ToUpper
    font {
     bForceMonospaced=false
     bMaterialIsStencil=false
     fontMaterial=null
     fontObject=&HeadingNow
     letterSpacing=0
     monospacedWidth=1
     size=18
     skewAmount=0
     typefaceFontName=75 Medium
     outlineSettings {
      bApplyOutlineToDropShadows=false
      bMiteredCorners=false
      bSeparateFillAlpha=false
      outlineMaterial=null
      outlineSize=0
      outlineColor=<rgba 0 0 0 1>
     }
    }
    renderTransform {angle=0, scale=<1 1>, shear=<-8 0>, translation=<0 0>}
    shadowOffset=<4.9266 5.2555>
   }
   slot {class=VerticalBoxSlot}
  }
  PublishedIslandCode {
   widgetClassPath=&CommonTextBlock
   bInherited=true
   parent=VBox_CreatorDetails
   widget=PublishedIslandCode
   properties {
    text=8888-8888-8888
    textTransformPolicy=ToUpper
    font {
     bForceMonospaced=false
     bMaterialIsStencil=false
     fontMaterial=null
     fontObject=&HeadingNow
     letterSpacing=0
     monospacedWidth=1
     size=18
     skewAmount=0
     typefaceFontName=74 Regular
     outlineSettings {
      bApplyOutlineToDropShadows=false
      bMiteredCorners=false
      bSeparateFillAlpha=false
      outlineMaterial=null
      outlineSize=0
      outlineColor=<rgba 0 0 0 1>
     }
    }
    shadowOffset=<4.9266 5.2555>
   }
   slot {class=VerticalBoxSlot, properties {verticalAlignment=VAlign_Center}}
  }
  Overlay_Epic {
   widgetClassPath=&Overlay
   parent=Switcher_CreatorType
   widget=Overlay_Epic
   slot {class=WidgetSwitcherSlot, properties {horizontalAlignment=HAlign_Left}}
  }
  Image_164 {
   widgetClassPath=&Image
   parent=Overlay_Epic
   widget=Image_164
   properties {brush {
     bIsDynamicallyLoaded=false
     drawAs=Image
     imageType=NoImage
     mirroring=NoMirror
     resourceName=None
     resourceObject=/Game/UI/Login/Art/T_EpicLogo
     tiling=NoTile
     imageSize=<64 64>
     margin=<ltrb 0 0 0 0>
     outlineSettings {
      bUseBrushTransparency=false
      roundingType=HalfHeightRadius
      width=0
      color {colorUseRule=UseColor_Specified, specifiedColor=<rgba 0 0 0 0>}
      cornerRadii=<0 0 0 0>
     }
     tintColor {colorUseRule=UseColor_Specified, specifiedColor=<rgba 1 1 1 1>}
     uVRegion {bIsValid=false, max=<0 0>, min=<0 0>}
    }}
   slot {class=OverlaySlot, properties {verticalAlignment=VAlign_Fill}}
  }
  ProgramDetailsButton {
   widgetClassPath=&WBP_UIKit_Button_Generic
   bIsVariable=true
   parent=VBox_IslandDetails
   widget=ProgramDetailsButton
   properties {
    minWidth=356
    secondText=""
    text=Program Details
    useSecondTextBlock=true
    visibility=Collapsed
    font {
     bForceMonospaced=false
     bMaterialIsStencil=false
     fontMaterial=/FortUILibrary/UIKit/Buttons/MaterialInstances/CTAButtons/MI_UIKit_TextQuiet
     fontObject=&HeadingNow
     letterSpacing=48
     monospacedWidth=1
     size=15.75
     skewAmount=0
     typefaceFontName=76 Bold
     outlineSettings {
      bApplyOutlineToDropShadows=false
      bMiteredCorners=false
      bSeparateFillAlpha=false
      outlineMaterial=null
      outlineSize=0
      outlineColor=<rgba 0 0 0 1>
     }
    }
    promptNudge=<10 0>
    secondFont {
     bForceMonospaced=false
     bMaterialIsStencil=false
     fontMaterial=&MI_UIKit_TextQuiet
     fontObject=&HeadingNow
     letterSpacing=48
     monospacedWidth=1
     size=15.75
     skewAmount=0
     typefaceFontName=76 Bold
     outlineSettings {
      bApplyOutlineToDropShadows=false
      bMiteredCorners=false
      bSeparateFillAlpha=false
      outlineMaterial=null
      outlineSize=0
      outlineColor=<rgba 0 0 0 1>
     }
    }
    secondTextPadding=<ltrb 0 -5 0 0>
    textPadding=<ltrb 20 0 20 0>
    triggeringInputAction {dataTable=&InputActionDataTable, rowName=Join}
   }
   slot {class=VerticalBoxSlot}
  }
  CreateButtonSwitcher {
   widgetClassPath=&CommonVisibilitySwitcher
   bIsVariable=true
   parent=VBox_IslandDetails
   widget=CreateButtonSwitcher
   slot {class=VerticalBoxSlot, properties {padding=<ltrb 0 20 0 0>}}
  }
  AcceptButton {
   widgetClassPath=&WBP_UIKit_Button_Generic
   bInherited=true
   bIsVariable=true
   parent=CreateButtonSwitcher
   widget=AcceptButton
   properties {
    backgroundMaterial=/FortUILibrary/UIKit/Buttons/MaterialInstances/CTAButtons/MI_UIKit_BgLoud
    minHeight=80
    minWidth=356
    text=Confirm
    buttonPadding=<ltrb 60 0 60 0>
    font {
     bForceMonospaced=false
     bMaterialIsStencil=false
     fontMaterial=&MI_UIKit_TextQuiet
     fontObject=&HeadingNow
     letterSpacing=48
     monospacedWidth=1
     size=21
     skewAmount=0
     typefaceFontName=76 Bold
     outlineSettings {
      bApplyOutlineToDropShadows=false
      bMiteredCorners=false
      bSeparateFillAlpha=false
      outlineMaterial=null
      outlineSize=0
      outlineColor=<rgba 0 0 0 1>
     }
    }
    promptNudge=<10 0>
    triggeringInputAction {dataTable=&InputActionDataTable, rowName=Confirm}
   }
   slot {class=CommonVisibilitySwitcherSlot}
  }
  SizeBox_3 {widgetClassPath=&SizeBox, namedSlotHost=AcceptButton, widget=SizeBox_3}
  CreateSpinner {
   widgetClassPath=&Image
   bIsVariable=true
   parent=SizeBox_3
   widget=CreateSpinner
   properties {
    visibility=Collapsed
    brush {
     bIsDynamicallyLoaded=false
     drawAs=Image
     imageType=NoImage
     mirroring=NoMirror
     resourceName=None
     resourceObject=null
     tiling=NoTile
     imageSize=<42 42>
     margin=<ltrb 0 0 0 0>
     outlineSettings {
      bUseBrushTransparency=false
      roundingType=HalfHeightRadius
      width=0
      color {colorUseRule=UseColor_Specified, specifiedColor=<rgba 0 0 0 0>}
      cornerRadii=<0 0 0 0>
     }
     tintColor {colorUseRule=UseColor_Specified, specifiedColor=<rgba 0 0 0 1>}
     uVRegion {bIsValid=false, max=<0 0>, min=<0 0>}
    }
   }
   slot {class=SizeBoxSlot, properties {horizontalAlignment=HAlign_Right, verticalAlignment=VAlign_Center, padding=<ltrb 0 0 10 0>}}
  }
  JoinCollabButton {
   widgetClassPath=&WBP_UIKit_Button_Generic
   bIsVariable=true
   parent=CreateButtonSwitcher
   widget=JoinCollabButton
   properties {
    backgroundMaterial=&MI_UIKit_BgLoud
    minWidth=356
    text=Join Program
    font {
     bForceMonospaced=false
     bMaterialIsStencil=false
     fontMaterial=&MI_UIKit_TextQuiet
     fontObject=&HeadingNow
     letterSpacing=48
     monospacedWidth=1
     size=21
     skewAmount=0
     typefaceFontName=76 Bold
     outlineSettings {
      bApplyOutlineToDropShadows=false
      bMiteredCorners=false
      bSeparateFillAlpha=false
      outlineMaterial=null
      outlineSize=0
      outlineColor=<rgba 0 0 0 1>
     }
    }
    promptNudge=<10 0>
    triggeringInputAction {dataTable=&InputActionDataTable, rowName=Join}
   }
   slot {class=CommonVisibilitySwitcherSlot}
  }
  ReviewTermsBox {
   widgetClassPath=&VerticalBox
   bIsVariable=true
   parent=CreateButtonSwitcher
   widget=ReviewTermsBox
   slot {class=CommonVisibilitySwitcherSlot}
  }
  ReviewTermsButton {
   widgetClassPath=&WBP_UIKit_Button_Generic
   bIsVariable=true
   parent=ReviewTermsBox
   widget=ReviewTermsButton
   properties {
    backgroundMaterial=&MI_UIKit_BgLoud
    minWidth=356
    text=Review Terms
    font {
     bForceMonospaced=false
     bMaterialIsStencil=false
     fontMaterial=&MI_UIKit_TextQuiet
     fontObject=&HeadingNow
     letterSpacing=48
     monospacedWidth=1
     size=21
     skewAmount=0
     typefaceFontName=76 Bold
     outlineSettings {
      bApplyOutlineToDropShadows=false
      bMiteredCorners=false
      bSeparateFillAlpha=false
      outlineMaterial=null
      outlineSize=0
      outlineColor=<rgba 0 0 0 1>
     }
    }
    promptNudge=<10 0>
    triggeringInputAction {dataTable=&InputActionDataTable, rowName=Join}
   }
   slot {class=VerticalBoxSlot}
  }
  CommonTextBlock_439 {
   widgetClassPath=&CommonTextBlock
   parent=ReviewTermsBox
   widget=CommonTextBlock_439
   properties {
    autoWrapText=true
    text=Already signed up? Restart Creative to access these templates.
    font {
     bForceMonospaced=false
     bMaterialIsStencil=false
     fontMaterial=null
     fontObject=&HeadingNow
     letterSpacing=0
     monospacedWidth=1
     size=11.25
     skewAmount=0
     typefaceFontName=74 Regular
     outlineSettings {
      bApplyOutlineToDropShadows=false
      bMiteredCorners=false
      bSeparateFillAlpha=false
      outlineMaterial=null
      outlineSize=0
      outlineColor=<rgba 0 0 0 1>
     }
    }
    shadowOffset=<0 0>
   }
   slot {class=VerticalBoxSlot, properties {horizontalAlignment=HAlign_Center, verticalAlignment=VAlign_Center, padding=<ltrb 5 10 5 0>}}
  }
  Header_Overlay {
   widgetClassPath=&Overlay
   parent=UIKitGridPanel_671
   widget=Header_Overlay
   slot {class=GridSlot, properties {columnSpan=3, padding=<ltrb 48 0 0 0>}}
  }
  VerticalBox_1 {
   widgetClassPath=&VerticalBox
   parent=Header_Overlay
   widget=VerticalBox_1
   slot {class=OverlaySlot, properties {horizontalAlignment=HAlign_Fill, verticalAlignment=VAlign_Bottom, padding=<ltrb 0 0 0 12>}}
  }
  Header_Text {
   widgetClassPath=&CommonTextBlock
   parent=VerticalBox_1
   widget=Header_Text
   properties {
    text=New Island
    textTransformPolicy=ToUpper
    font {
     bForceMonospaced=false
     bMaterialIsStencil=false
     fontMaterial=null
     fontObject=&HeadingNow
     letterSpacing=0
     monospacedWidth=1
     size=51.75
     skewAmount=0
     typefaceFontName=86 Bold Italic
     outlineSettings {
      bApplyOutlineToDropShadows=false
      bMiteredCorners=false
      bSeparateFillAlpha=false
      outlineMaterial=null
      outlineSize=0
      outlineColor=<rgba 0 0 0 1>
     }
    }
   }
   slot {class=VerticalBoxSlot}
  }
  Subheader_text {
   widgetClassPath=&CommonTextBlock
   bIsVariable=true
   parent=VerticalBox_1
   widget=Subheader_text
   properties {
    bAutoCollapseWithEmptyText=true
    textTransformPolicy=ToUpper
    colorAndOpacity {colorUseRule=UseColor_Specified, specifiedColor=<rgba 0.2502 0.5271 1 1>}
    font {
     bForceMonospaced=false
     bMaterialIsStencil=false
     fontMaterial=null
     fontObject=&HeadingNow
     letterSpacing=0
     monospacedWidth=1
     size=25.5
     skewAmount=0
     typefaceFontName=86 Bold Italic
     outlineSettings {
      bApplyOutlineToDropShadows=false
      bMiteredCorners=false
      bSeparateFillAlpha=false
      outlineMaterial=null
      outlineSize=0
      outlineColor=<rgba 0 0 0 1>
     }
    }
   }
   slot {class=VerticalBoxSlot}
  }
  CollabTermsAndConditonsBox {
   widgetClassPath=&SizeBox
   bIsVariable=true
   parent=UIKitGridPanel_671
   widget=CollabTermsAndConditonsBox
   properties {visibility=Collapsed}
   slot {class=GridSlot, properties {columnSpan=3, row=1, verticalAlignment=VAlign_Bottom}}
  }
  Overlay_134 {widgetClassPath=&Overlay, parent=CollabTermsAndConditonsBox, widget=Overlay_134, slot {class=SizeBoxSlot}}
  Image_DisclaimerBG {
   widgetClassPath=&Image
   bIsVariable=true
   parent=Overlay_134
   widget=Image_DisclaimerBG
   properties {visibility=Collapsed, colorAndOpacity=<rgba 1 1 1 0.05>}
   slot {class=OverlaySlot, properties {horizontalAlignment=HAlign_Fill, verticalAlignment=VAlign_Fill}}
  }
  HorizontalBox {
   widgetClassPath=&HorizontalBox
   parent=Overlay_134
   widget=HorizontalBox
   slot {class=OverlaySlot, properties {horizontalAlignment=HAlign_Fill, verticalAlignment=VAlign_Fill, padding=<ltrb 0 0 456 0>}}
  }
  CollabPartnerIcon {
   widgetClassPath=&FortLazyImage
   bInherited=true
   parent=HorizontalBox
   widget=CollabPartnerIcon
   properties {brush {
     bIsDynamicallyLoaded=false
     drawAs=Image
     imageType=NoImage
     mirroring=NoMirror
     resourceName=None
     resourceObject=null
     tiling=NoTile
     imageSize=<50 50>
     margin=<ltrb 0 0 0 0>
     outlineSettings {
      bUseBrushTransparency=false
      roundingType=HalfHeightRadius
      width=0
      color {colorUseRule=UseColor_Specified, specifiedColor=<rgba 0 0 0 0>}
      cornerRadii=<0 0 0 0>
     }
     tintColor {colorUseRule=UseColor_Specified, specifiedColor=<rgba 1 1 1 1>}
     uVRegion {bIsValid=false, max=<0 0>, min=<0 0>}
    }}
   slot {class=HorizontalBoxSlot, properties {horizontalAlignment=HAlign_Left, verticalAlignment=VAlign_Center, padding=<ltrb 40 40 40 40>}}
  }
  CollabDisclaimerText {
   widgetClassPath=&CommonRichTextBlock
   bIsVariable=true
   parent=HorizontalBox
   widget=CollabDisclaimerText
   properties {
    autoWrapText=true
    bOverrideDefaultStyle=true
    defaultTextStyleOverrideClass=/Game/Athena/Settings/SettingEntries/Styles/TextStyle_HeadingNow_75_19_LightGray.TextStyle_HeadingNow_75_19_LightGray_C
    text="In order to Publish a LEGO® island you are required to review the terms and conditions and understand the revenue rules."
    defaultTextStyleOverride {
     overflowPolicy=Clip
     transformPolicy=None
     colorAndOpacity {colorUseRule=UseColor_Specified, specifiedColor=<rgba 1 1 1 1>}
     font {
      bForceMonospaced=false
      bMaterialIsStencil=false
      fontMaterial=null
      fontObject=&NotoSans
      letterSpacing=0
      monospacedWidth=1
      size=24
      skewAmount=0
      typefaceFontName=NotoSans-Regular
      outlineSettings {
       bApplyOutlineToDropShadows=false
       bMiteredCorners=false
       bSeparateFillAlpha=false
       outlineMaterial=null
       outlineSize=0
       outlineColor=<rgba 0 0 0 1>
      }
     }
     highlightColor {colorUseRule=UseColor_Specified, specifiedColor=<rgba 0 0 0 1>}
     highlightShape {
      bIsDynamicallyLoaded=false
      drawAs=Image
      imageType=NoImage
      mirroring=NoMirror
      resourceName=None
      resourceObject=null
      tiling=NoTile
      imageSize=<32 32>
      margin=<ltrb 0 0 0 0>
      outlineSettings {
       bUseBrushTransparency=false
       roundingType=HalfHeightRadius
       width=0
       color {colorUseRule=UseColor_Specified, specifiedColor=<rgba 0 0 0 0>}
       cornerRadii=<0 0 0 0>
      }
      tintColor {colorUseRule=UseColor_Specified, specifiedColor=<rgba 1 1 1 1>}
      uVRegion {bIsValid=false, max=<0 0>, min=<0 0>}
     }
     selectedBackgroundColor {colorUseRule=UseColor_Foreground, specifiedColor=<rgba 1 0 1 1>}
     shadowColorAndOpacity=<rgba 0 0 0 1>
     shadowOffset=<0 0>
     strikeBrush {
      bIsDynamicallyLoaded=false
      drawAs=Image
      imageType=NoImage
      mirroring=NoMirror
      resourceName=None
      resourceObject=null
      tiling=NoTile
      imageSize=<32 32>
      margin=<ltrb 0 0 0 0>
      outlineSettings {
       bUseBrushTransparency=false
       roundingType=HalfHeightRadius
       width=0
       color {colorUseRule=UseColor_Specified, specifiedColor=<rgba 0 0 0 0>}
       cornerRadii=<0 0 0 0>
      }
      tintColor {colorUseRule=UseColor_Specified, specifiedColor=<rgba 1 1 1 1>}
      uVRegion {bIsValid=false, max=<0 0>, min=<0 0>}
     }
     underlineBrush {
      bIsDynamicallyLoaded=false
      drawAs=Image
      imageType=NoImage
      mirroring=NoMirror
      resourceName=None
      resourceObject=null
      tiling=NoTile
      imageSize=<32 32>
      margin=<ltrb 0 0 0 0>
      outlineSettings {
       bUseBrushTransparency=false
       roundingType=HalfHeightRadius
       width=0
       color {colorUseRule=UseColor_Specified, specifiedColor=<rgba 0 0 0 0>}
       cornerRadii=<0 0 0 0>
      }
      tintColor {colorUseRule=UseColor_Specified, specifiedColor=<rgba 1 1 1 1>}
      uVRegion {bIsValid=false, max=<0 0>, min=<0 0>}
     }
    }
   }
   slot {class=HorizontalBoxSlot, properties {verticalAlignment=VAlign_Center, size {sizeRule=Fill, value=1}}}
  }
  DetailsButton {
   widgetClassPath=&WBP_UIKit_Button_Regular
   bIsVariable=true
   parent=HorizontalBox
   widget=DetailsButton
   properties {text=Details, visibility=Collapsed, triggeringInputAction {dataTable=&InputActionDataTable, rowName=Details}}
   slot {class=HorizontalBoxSlot, properties {horizontalAlignment=HAlign_Right, verticalAlignment=VAlign_Center, padding=<ltrb 40 0 40 0>}}
  }
  VisibilityBorder_Touch {
   widgetClassPath=&CommonHardwareVisibilityBorder
   parent=Overlay_661
   widget=VisibilityBorder_Touch
   properties {visibilityQuery {
     autoDescription=" ANY( Input.Touch )"
     tokenStreamVersion=0
     userDescription=""
     queryTokenStream [0,1,1,1,0]
     tagDictionary [{tagName=Input.Touch}]
    }}
   slot {class=OverlaySlot, properties {horizontalAlignment=HAlign_Right, padding=<ltrb 0 20 20 0>}}
  }
  Button_CloseTouch {
   widgetClassPath=/FortUILibrary/Foundation/Buttons/CloseButton
   bIsVariable=true
   parent=VisibilityBorder_Touch
   widget=Button_CloseTouch
   properties {
    onButtonBaseClicked=(&self:WidgetTree.Button_CloseTouch.HandleOnButtonClicked)
    onButtonBaseDoubleClicked=(&self:WidgetTree.Button_CloseTouch.HandleOnButtonDoubleClicked)
    onButtonBaseHovered=(&self:WidgetTree.Button_CloseTouch.HandleOnButtonHovered)
    onButtonBaseUnhovered=(&self:WidgetTree.Button_CloseTouch.HandleOnButtonUnhovered)
    onSelectedChangedBase=(&self:WidgetTree.Button_CloseTouch.HandleOnSelectedChanged)
   }
   slot {class=BorderSlot, properties {padding=<ltrb 0 0 0 0>}}
  }
  CompatibleStarterIslandsTileView {widgetClassPath=&CommonTileView, bInherited=true, widget=null}
 ]
}
