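# List-row button widget for one island in the Creative portal UI, showing name, published code, date and badge/lock icons.
# OnListItemObjectSet reads an IslandDataEntry to fill the row, prefixing deleted islands' names and showing their deletion date instead of last-used.
# Hover and selection events toggle backplate and outline images, with UpdateTextStyle tinting text and lock by selection.
# Construct binds to input-method changes, swapping in a touch-specific button style.
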WidgetBlueprint /Game/Playgrounds/Items/Portals/UI/IslandOptionEntry_V2
animations [Hovered {
  binding_count=1
  length_seconds=0.2
  object=&self:Hovered
  sequence {bindings [IslandOptionEntry_C_107 {
     id=#1
     possessed_class=IslandOptionEntry_C
     tracks [{
       type=MovieScene2DTransformTrack
       sections [{
         type=MovieScene2DTransformSection
         end_frame=4
         start_frame=0
         channels [
          Translation.X {type=MovieSceneScriptingFloatChannel}
          Translation.Y {type=MovieSceneScriptingFloatChannel}
          Angle {type=MovieSceneScriptingFloatChannel}
          Scale.X {type=MovieSceneScriptingFloatChannel, keys [[0,1],[4,1.1073]]}
          Scale.Y {type=MovieSceneScriptingFloatChannel, keys [[0,1],[4,1.1073]]}
          Shear.X {type=MovieSceneScriptingFloatChannel}
          Shear.Y {type=MovieSceneScriptingFloatChannel}
         ]
        }]
      }]
    }]}
 }]
graphs [
 EventGraph {nodes [
   n1 = InputModeChanged!() -> (OutputDelegate, then, InputType:ECommonInputType Enum) {is=title K2Node_CustomEvent}
   n2 = Get DefaultButtonStyle() -> (DefaultButtonStyle:Common Button Style Class Reference)
   > SetStyle!(
    execute
    self:Common Button Base Object Reference
    InStyle:Common Button Style Class Reference=Select!(
     MouseAndKeyboard:Common Button Style Class Reference=@n2
     Gamepad:Common Button Style Class Reference=@n2
     Touch:Common Button Style Class Reference=Get MobileButtonStyle() -> (MobileButtonStyle:Common Button Style Class Reference)
     Count:Common Button Style Class Reference
     Index:ECommonInputType Enum=@n1.InputType
    ) -> (ReturnValue:Common Button Style Class Reference)
   ) -> (then)
   n3 = InputUpdated!() -> (OutputDelegate, then, bNewInputType:ECommonInputType Enum) {is=title K2Node_CustomEvent}
   > InputModeChanged!(execute, self, InputType:ECommonInputType Enum=@n3.bNewInputType) -> (then) {is=title K2Node_CallFunction}
   Event On Selected()
   > Set isSelected(isSelected:bool=true) -> (Output_Get:bool)
   n4 = Get Image_Selection() -> (Image_Selection:Image Object Reference)
   > SetRenderOpacity(self=@n4, InOpacity=1)
   n5 = Get Image_Backplate() -> (Image_Backplate:Image Object Reference)
   n6 = Get Image_Selection()
   > SetVisibility(self=[@n5, @n6], InVisibility=SelfHitTestInvisible)
   n7 = Get Image_Backplate_Back() -> (Image_Backplate_Back:Image Object Reference)
   > SetVisibility(self=@n7, InVisibility=Hidden)
   n8 = Get Image_Backplate()
   > n9 = SetOpacity!(execute, self:Image Object Reference=@n8, InOpacity:float=0.35) -> (then)
   UpdateTextStyle!(execute=[@n34.then, @n32.then, @n9.then, @n10.then], self, isSelected:bool=Get isSelected() -> (isSelected:bool)) -> (then)
   Event On Deselected()
   > Set isSelected(isSelected=false)
   > SetRenderOpacity(self=@n4, InOpacity=1)
   > SetVisibility(self=[@n5, @n6], InVisibility=Collapsed)
   > SetVisibility(self=@n7, InVisibility=SelfHitTestInvisible)
   > n10 = SetOpacity(self=@n8, InOpacity=0.1)
   n11 = Event OnListItemObjectSet()
   > Sequence()
   > n12 = Cast To IslandDataEntry!(execute, Object=@n11.ListItemObject) -> (then, CastFailed, AsIsland Data Entry:Island Data Entry Object Reference) {is=title K2Node_DynamicCast}
   > n13 = Set IslandData(IslandData:Island Data Entry Object Reference=@n12) -> (Output_Get:Island Data Entry Object Reference)
   > n14 = Sequence!(execute) -> (then_0, then_1, then_2:Exec, then_3:Exec, then_4:Exec)
   n15 = Get PublishedIslandCode!(self:Island Data Entry Object Reference=@n13) -> (PublishedIslandCode:String)
   n16 = Branch(execute=@n14.then_1, Condition="integer > integer"(A=Len(S=@n15), B=0) {is=title K2Node_CallFunction})
   n17 = Get PublishedCodeText() -> (PublishedCodeText:Common Text Object Reference)
   > "SetText (Text)"(self=@n17, InText="To Text (String)"(InString=@n15))
   > SetVisibility(self=@n17, InVisibility=SelfHitTestInvisible)
   SetVisibility(execute=@n16.else, self=@n17, InVisibility=Hidden)
   n18 = Branch(execute=@n14.then_0, Condition=Get IsDeleted!(self:Island Data Entry Object Reference=@n13) -> (IsDeleted:bool))
   n19 = Get MapName!(self:Island Data Entry Object Reference=@n13) -> (MapName:Text)
   > "SetText (Text)"(
    self=Get PlotName() -> (PlotName:Common Text Object Reference)
    InText=Format Text!(Format="LOCTEXT(\"{0} {1}\")", 0:Text=Get DeletedText() -> (DeletedText:Text), 1:Text=@n19) -> (Result)
   )
   "SetText (Text)"(execute=@n18.else, self=Get PlotName(), InText=@n19)
   SetVisibility(
    execute=@n14.then_4
    self=Get SizeBox_Badge() -> (SizeBox_Badge:Size Box Object Reference)
    InVisibility=Select!(
     Option 0:ESlateVisibility Enum=Collapsed
     Option 1:ESlateVisibility Enum=SelfHitTestInvisible
     Index:bool=Get IsVKProject!(self:Island Data Entry Object Reference=@n13) -> (IsVKProject:bool)
    ) -> (ReturnValue:ESlateVisibility Enum)
   )
   > SetVisibility(
    self=Get SizeBox_Lock() -> (SizeBox_Lock:Size Box Object Reference)
    InVisibility=Select(Option 0=Hidden, Option 1=SelfHitTestInvisible, Index=Get IsLocked!(self:Island Data Entry Object Reference=@n13) -> (IsLocked:bool))
   )
   n20 = Branch(execute=@n14.then_2, Condition=Get IsDeleted(self=@n13))
   > n21 = Branch(Condition="Equal (DateTime)"!(
     A:Date Time Structure="MinValue (DateTime)"!() -> (ReturnValue:Date Time Structure)
     B:Date Time Structure=Get LastUsedDate!(self:Island Data Entry Object Reference=@n13) -> (LastUsedDate:Date Time Structure)
    ) -> (ReturnValue:bool))
   n22 = Get LastVisitedTextBlock() -> (LastVisitedTextBlock:Common Text Object Reference)
   > "SetText (Text)"(self=@n22)
   "SetText (Text)"(
    execute=@n21.else
    self=@n22
    InText="As DateTime (from UTC)"!(
     InDateTime:"Date Time Structure (by ref)"=Get DeletedAt!(self:Island Data Entry Object Reference=@n13) -> (DeletedAt:Date Time Structure)
     InTimeZone:String
     InDateStyle:EDateTimeStyle Enum=Default
     InTimeStyle:EDateTimeStyle Enum=Default
    ) -> (ReturnValue:Text)
   )
   n23 = Get LastUsedDate(self=@n13)
   n24 = Branch(execute=@n20.else, Condition="Equal (DateTime)"(A="MinValue (DateTime)"(), B=@n23))
   > "SetText (Text)"(self=@n22)
   "SetText (Text)"(execute=@n24.else, self=@n22, InText="As DateTime (from UTC)"(InDateTime=@n23, InDateStyle=Default, InTimeStyle=Default))
   n25 = Select(
    Option 0:Slate Color Structure=Get NormalColor() -> (NormalColor:Slate Color Structure)
    Option 1:Slate Color Structure=Get DeletedColor() -> (DeletedColor:Slate Color Structure)
    Index=Get IsDeleted(self=@n13)
   ) -> (ReturnValue:Slate Color Structure)
   SetColorAndOpacity(execute=@n14.then_3, self=Get LastVisitedTextBlock(), InColorAndOpacity=@n25)
   > SetColorAndOpacity(self=Get PlotName(), InColorAndOpacity=@n25)
   > SetColorAndOpacity(self=Get PublishedCodeText(), InColorAndOpacity=@n25)
   Event On Hovered()
   > n26 = Branch(Condition=Get isSelected())
   n27 = Get Image_Selection()
   SetRenderOpacity(execute=@n26.else, self=@n27, InOpacity=0.3)
   n28 = Get Image_Selection()
   n29 = Get Image_Backplate()
   > SetVisibility(self=[@n28, @n29], InVisibility=SelfHitTestInvisible)
   n30 = Get Image_Backplate_Back()
   > SetVisibility(self=@n30, InVisibility=Hidden)
   n31 = Get Image_Backplate()
   > n32 = SetOpacity(self=@n31, InOpacity=0.1)
   Event On Unhovered()
   > n33 = Branch(Condition=Get isSelected())
   SetRenderOpacity(execute=@n33.else, self=@n27, InOpacity=1)
   > SetVisibility(self=[@n28, @n29], InVisibility=Collapsed)
   > SetVisibility(self=@n30, InVisibility=SelfHitTestInvisible)
   > n34 = SetOpacity(self=@n31, InOpacity=0.1)
   Event Construct()
   n35 = Get CommonInputSubsystem!() -> (ReturnValue:Common Input Object Reference) {is=title K2Node_GetBlueprintContextEx}
   > InputModeChanged(InputType=GetCurrentInputType(self=@n35))
   > Assign On Input Method Changed!(execute:Exec, self:Common Input Object Reference=@n35, Delegate:"Delegate (by ref)"=@n3) -> (then:Exec) {is=title K2Node_AssignDelegate}
  ]}
 UpdateTextStyle {nodes [
   UpdateTextStyle!() -> (then, isSelected:bool=~) {is=title K2Node_FunctionEntry}
   n1 = Select(
    Option 0=Make Slate Color(SpecifiedColor=(R=0.849,G=0.849,B=0.849,A=1))
    Option 1=Make Slate Color(SpecifiedColor=(R=0.9911,G=0.9911,B=0.9911,A=1))
    Index=Get isSelected()
   )
   > SetColorAndOpacity(self=[Get PlotName(), Get PublishedCodeText(), Get LastVisitedTextBlock()], InColorAndOpacity=@n1)
   > SetColorAndOpacity(
    self:Image Object Reference=Get LockedIcon() -> (LockedIcon:Image Object Reference)
    InColorAndOpacity:Linear Color Structure="To Color (SlateColor)"!(Value:"Slate Color Structure (by ref)"=@n1) -> (ReturnValue:Linear Color Structure)
   )
  ]}
]
variables [
 DefaultButtonStyle {type=Class, default=/FortUILibrary/Foundation/Buttons/ButtonStyles/ButtonStyle-TextOnlyBase_Empty.ButtonStyle-TextOnlyBase_Empty_C}
 DeletedColor {type=SlateColor, default {colorUseRule=UseColor_Specified, specifiedColor=<rgba 1 1 1 1>}}
 DeletedText {type=string, default=(Deleted)}
 IslandData {type=IslandDataEntry_C, default=null}
 isSelected {type=boolean}
 MobileButtonStyle {
  type=Class
  default=/FortUILibrary/Foundation/Buttons/ButtonStyles/ButtonStyle-TextOnlyBase_Empty_BlueHover_Touch.ButtonStyle-TextOnlyBase_Empty_BlueHover_Touch_C
 }
 NormalColor {type=SlateColor, default {colorUseRule=UseColor_Specified, specifiedColor=<rgba 1 1 1 1>}}
 ParentWidget {type=CommonActivatableWidget, default=null}
]
widgets {
 info {inheritedWidgetCount=1, parentClass=/Script/FortniteUI.AthenaCreativeItemTileButton, rootWidgetClass=&Overlay, widgetCount=18}
 widgets [
  Overlay_0 {widgetClassPath=&Overlay, widget=Overlay_0}
  Overlay_1 {
   widgetClassPath=&Overlay
   parent=Overlay_0
   widget=Overlay_1
   slot {class=OverlaySlot, properties {horizontalAlignment=HAlign_Fill, verticalAlignment=VAlign_Center}}
  }
  Image_Backplate_Back {
   widgetClassPath=&Image
   bIsVariable=true
   parent=Overlay_1
   widget=Image_Backplate_Back
   properties {
    visibility=SelfHitTestInvisible
    brush {
     bIsDynamicallyLoaded=false
     drawAs=Image
     imageType=NoImage
     mirroring=NoMirror
     resourceName=None
     resourceObject=null
     tiling=NoTile
     imageSize=<32 32>
     margin=<ltrb 0 0 0 0>
     outlineSettings {
      bUseBrushTransparency=false
      roundingType=HalfHeightRadius
      width=0
      color {colorUseRule=UseColor_Specified, specifiedColor=<rgba 0 0 0 0>}
      cornerRadii=<10 10 10 10>
     }
     tintColor {colorUseRule=UseColor_Specified, specifiedColor=<rgba 1 1 1 1>}
     uVRegion {bIsValid=false, max=<0 0>, min=<0 0>}
    }
    colorAndOpacity=<rgba 1 1 1 0.05>
   }
   slot {class=OverlaySlot, properties {horizontalAlignment=HAlign_Fill, verticalAlignment=VAlign_Fill, padding=<ltrb 24 2 24 2>}}
  }
  Image_Backplate {
   widgetClassPath=&Image
   bIsVariable=true
   parent=Overlay_1
   widget=Image_Backplate
   properties {
    visibility=Hidden
    brush {
     bIsDynamicallyLoaded=false
     drawAs=RoundedBox
     imageType=NoImage
     mirroring=NoMirror
     resourceName=None
     resourceObject=null
     tiling=NoTile
     imageSize=<32 32>
     margin=<ltrb 0 0 0 0>
     outlineSettings {
      bUseBrushTransparency=false
      roundingType=HalfHeightRadius
      width=0
      color {colorUseRule=UseColor_Specified, specifiedColor=<rgba 0 0 0 0>}
      cornerRadii=<10 10 10 10>
     }
     tintColor {colorUseRule=UseColor_Specified, specifiedColor=<rgba 1 1 1 1>}
     uVRegion {bIsValid=false, max=<0 0>, min=<0 0>}
    }
    colorAndOpacity=<rgba 1 1 1 0.1>
   }
   slot {class=OverlaySlot, properties {horizontalAlignment=HAlign_Fill, verticalAlignment=VAlign_Fill, padding=<ltrb 6 6 6 6>}}
  }
  Image_Selection {
   widgetClassPath=&Image
   bIsVariable=true
   parent=Overlay_1
   widget=Image_Selection
   properties {
    visibility=Hidden
    brush {
     bIsDynamicallyLoaded=false
     drawAs=Image
     imageType=FullColor
     mirroring=NoMirror
     resourceName=None
     resourceObject=/FortUILibrary/UIKit/Buttons/MaterialInstances/TileOutline/MI_UIKit_Outline
     tiling=NoTile
     imageSize=<32 32>
     margin=<ltrb 0 0 0 0>
     outlineSettings {
      bUseBrushTransparency=false
      roundingType=HalfHeightRadius
      width=0
      color {colorUseRule=UseColor_Specified, specifiedColor=<rgba 0 0 0 0>}
      cornerRadii=<0 0 0 0>
     }
     tintColor {colorUseRule=UseColor_Specified, specifiedColor=<rgba 1 1 1 1>}
     uVRegion {bIsValid=false, max=<0 0>, min=<0 0>}
    }
    colorAndOpacity=<rgba 1 1 1 0.5>
   }
   slot {class=OverlaySlot, properties {horizontalAlignment=HAlign_Fill, verticalAlignment=VAlign_Fill, padding=<ltrb 0 -4 0 -4>}}
  }
  SizeBox_1 {
   widgetClassPath=&SizeBox
   parent=Overlay_1
   widget=SizeBox_1
   properties {bOverride_HeightOverride=true, heightOverride=64}
   slot {class=OverlaySlot, properties {horizontalAlignment=HAlign_Fill, verticalAlignment=VAlign_Center}}
  }
  HorizontalBox_0 {
   widgetClassPath=&HorizontalBox
   parent=SizeBox_1
   widget=HorizontalBox_0
   slot {class=SizeBoxSlot, properties {verticalAlignment=VAlign_Center}}
  }
  Spacer_Compensateforcurvedbackplate {
   widgetClassPath=&Spacer
   parent=HorizontalBox_0
   widget=Spacer_Compensateforcurvedbackplate
   slot {class=HorizontalBoxSlot, properties {padding=<ltrb 25 0 0 0>}}
  }
  SizeBox_Badge {
   widgetClassPath=&SizeBox
   bIsVariable=true
   parent=HorizontalBox_0
   widget=SizeBox_Badge
   properties {heightOverride=48, visibility=Collapsed, widthOverride=48}
   slot {class=HorizontalBoxSlot, properties {padding=<ltrb 2 5 18 5>}}
  }
  EditBadge {
   widgetClassPath=&Image
   bIsVariable=true
   parent=SizeBox_Badge
   widget=EditBadge
   properties {
    visibility=SelfHitTestInvisible
    brush {
     bIsDynamicallyLoaded=false
     drawAs=Image
     imageType=NoImage
     mirroring=NoMirror
     resourceName=None
     resourceObject=/Game/Playgrounds/Items/Portals/UI/Art/48x48
     tiling=NoTile
     imageSize=<48 48>
     margin=<ltrb 0 0 0 0>
     outlineSettings {
      bUseBrushTransparency=false
      roundingType=HalfHeightRadius
      width=0
      color {colorUseRule=UseColor_Specified, specifiedColor=<rgba 0 0 0 0>}
      cornerRadii=<0 0 0 0>
     }
     tintColor {colorUseRule=UseColor_Specified, specifiedColor=<rgba 1 1 1 1>}
     uVRegion {bIsValid=false, max=<0 0>, min=<0 0>}
    }
   }
   slot {class=SizeBoxSlot, properties {horizontalAlignment=HAlign_Center, verticalAlignment=VAlign_Center}}
  }
  SizeBox_Lock {
   widgetClassPath=&SizeBox
   bIsVariable=true
   parent=HorizontalBox_0
   widget=SizeBox_Lock
   properties {heightOverride=48, visibility=Collapsed, widthOverride=48}
   slot {class=HorizontalBoxSlot, properties {padding=<ltrb 2 5 18 5>}}
  }
  LockedIcon {
   widgetClassPath=&Image
   bIsVariable=true
   parent=SizeBox_Lock
   widget=LockedIcon
   properties {
    visibility=SelfHitTestInvisible
    brush {
     bIsDynamicallyLoaded=false
     drawAs=Image
     imageType=FullColor
     mirroring=NoMirror
     resourceName=None
     resourceObject=/Game/UI/Foundation/Textures/Icons/Locks/T-Icon-Lock-32
     tiling=NoTile
     imageSize=<32 32>
     margin=<ltrb 0 0 0 0>
     outlineSettings {
      bUseBrushTransparency=false
      roundingType=HalfHeightRadius
      width=0
      color {colorUseRule=UseColor_Specified, specifiedColor=<rgba 0 0 0 0>}
      cornerRadii=<0 0 0 0>
     }
     tintColor {colorUseRule=UseColor_Specified, specifiedColor=<rgba 1 1 1 1>}
     uVRegion {bIsValid=false, max=<0 0>, min=<0 0>}
    }
   }
   slot {class=SizeBoxSlot, properties {horizontalAlignment=HAlign_Center, verticalAlignment=VAlign_Center}}
  }
  SizeBox_0 {
   widgetClassPath=&SizeBox
   parent=HorizontalBox_0
   widget=SizeBox_0
   properties {heightOverride=48, widthOverride=450}
   slot {class=HorizontalBoxSlot, properties {size {sizeRule=Fill, value=0.5}}}
  }
  PlotName {
   widgetClassPath=&CommonTextBlock
   bIsVariable=true
   parent=SizeBox_0
   widget=PlotName
   properties {
    minDesiredWidth=440
    scrollStyle=&TextScrollStyle-Base_C
    text="{Island Name}"
    font {
     bForceMonospaced=false
     bMaterialIsStencil=false
     fontMaterial=null
     fontObject=&HeadingNow
     letterSpacing=0
     monospacedWidth=1
     size=16.5
     skewAmount=0
     typefaceFontName=75 Medium
     outlineSettings {
      bApplyOutlineToDropShadows=false
      bMiteredCorners=false
      bSeparateFillAlpha=false
      outlineMaterial=null
      outlineSize=0
      outlineColor=<rgba 0 0 0 1>
     }
    }
    shadowOffset=<0 0>
   }
   slot {class=SizeBoxSlot, properties {verticalAlignment=VAlign_Center}}
  }
  PublishedCodeText {
   widgetClassPath=&CommonTextBlock
   bIsVariable=true
   parent=HorizontalBox_0
   widget=PublishedCodeText
   properties {
    justification=Center
    scrollStyle=&TextScrollStyle-Base_C
    text=0000-0000-0000
    font {
     bForceMonospaced=false
     bMaterialIsStencil=false
     fontMaterial=null
     fontObject=&HeadingNow
     letterSpacing=0
     monospacedWidth=1
     size=16.5
     skewAmount=0
     typefaceFontName=75 Medium
     outlineSettings {
      bApplyOutlineToDropShadows=false
      bMiteredCorners=false
      bSeparateFillAlpha=false
      outlineMaterial=null
      outlineSize=0
      outlineColor=<rgba 0 0 0 1>
     }
    }
    shadowOffset=<0 0>
   }
   slot {class=HorizontalBoxSlot, properties {verticalAlignment=VAlign_Center, size {sizeRule=Fill, value=0.3}}}
  }
  Spacer_424 {
   widgetClassPath=&Spacer
   parent=HorizontalBox_0
   widget=Spacer_424
   slot {class=HorizontalBoxSlot, properties {size {sizeRule=Fill, value=0.01}}}
  }
  LastVisitedTextBlock {
   widgetClassPath=&CommonTextBlock
   bIsVariable=true
   parent=HorizontalBox_0
   widget=LastVisitedTextBlock
   properties {
    scrollStyle=&TextScrollStyle-Base_C
    text=11/29/1994
    font {
     bForceMonospaced=false
     bMaterialIsStencil=false
     fontMaterial=null
     fontObject=&HeadingNow
     letterSpacing=0
     monospacedWidth=1
     size=16.5
     skewAmount=0
     typefaceFontName=75 Medium
     outlineSettings {
      bApplyOutlineToDropShadows=false
      bMiteredCorners=false
      bSeparateFillAlpha=false
      outlineMaterial=null
      outlineSize=0
      outlineColor=<rgba 0 0 0 1>
     }
    }
    shadowOffset=<0 0>
   }
   slot {class=HorizontalBoxSlot, properties {verticalAlignment=VAlign_Center, size {sizeRule=Fill, value=0.3}}}
  }
  InputActionWidget {widgetClassPath=&CommonActionWidget, bInherited=true, widget=null}
 ]
}
