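# Modal island-select panel for Creative island portals, listing the player's active and deleted islands to select, create, rename, duplicate, remove or restore.
# List entries are built from the player controller's island data, with lock status from playability and collab approval.
# Actions call the controller; its finish delegates clear spinners, re-enable buttons and show errors.
# Intro and reversed Outro animations frame opening and closing, with separate touch and gamepad input handling.
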WidgetBlueprint /Game/Playgrounds/Items/Portals/UI/IslandSelect_UI_MultipleIP_Modal
animations [
 Intro {
  binding_count=2
  length_seconds=0.3715
  object=&self:Intro
  sequence {bindings [
    CommonTextBlock_2 {
     id=#1
     possessed_class=CommonTextBlock
     tracks [
      {
       type=MovieScene2DTransformTrack
       sections [{
         type=MovieScene2DTransformSection
         end_frame=7
         start_frame=0
         channels [
          Translation.X {type=MovieSceneScriptingFloatChannel, keys [[0,-80],[2.8583,-80],[5.0013,9.829],[6.3167,0],[7.4307,0]]}
          Angle {type=MovieSceneScriptingFloatChannel}
          Scale.X {type=MovieSceneScriptingFloatChannel}
          Scale.Y {type=MovieSceneScriptingFloatChannel}
          Shear.X {type=MovieSceneScriptingFloatChannel}
          Shear.Y {type=MovieSceneScriptingFloatChannel}
         ]
        }]
      }
      {
       type=MovieSceneFloatTrack
       display_name=Render Opacity
       sections [{
         type=MovieSceneFloatSection
         end_frame=7
         start_frame=0
         channels [None {type=MovieSceneScriptingFloatChannel, keys [[0,0],[2.6617,0],[7.4307,1]]}]
        }]
      }
     ]
    }
    SizeBox_0 {
     id=#2
     possessed_class=SizeBox
     tracks [
      {
       type=MovieScene2DTransformTrack
       sections [{
         type=MovieScene2DTransformSection
         end_frame=7
         start_frame=0
         channels [
          Translation.X {type=MovieSceneScriptingFloatChannel, keys [[0,-50],[2.738,0],[7.4307,0]]}
          Translation.Y {type=MovieSceneScriptingFloatChannel}
          Angle {type=MovieSceneScriptingFloatChannel}
          Shear.X {type=MovieSceneScriptingFloatChannel}
          Shear.Y {type=MovieSceneScriptingFloatChannel}
         ]
        }]
      }
      {
       type=MovieSceneFloatTrack
       display_name=Render Opacity
       sections [{type=MovieSceneFloatSection, end_frame=4, start_frame=0, channels [None {type=MovieSceneScriptingFloatChannel, keys [[0,0],[3.6303,1]]}]}]
      }
     ]
    }
   ]}
 }
 NoIslands {
  binding_count=1
  length_seconds=0.149
  object=&self:NoIslands
  sequence {bindings [TextBlock_0 {
     id=#3
     possessed_class=TextBlock
     tracks [
      {
       type=MovieScene2DTransformTrack
       sections [{
         type=MovieScene2DTransformSection
         end_frame=2
         start_frame=0
         channels [
          Translation.X {type=MovieSceneScriptingFloatChannel, const [0,3]}
          Translation.Y {type=MovieSceneScriptingFloatChannel, keys [[0,-29],[2.0253,8],[2.98,0]]}
          Angle {type=MovieSceneScriptingFloatChannel}
          Scale.X {type=MovieSceneScriptingFloatChannel}
          Scale.Y {type=MovieSceneScriptingFloatChannel}
          Shear.X {type=MovieSceneScriptingFloatChannel}
          Shear.Y {type=MovieSceneScriptingFloatChannel}
         ]
        }]
      }
      {
       type=MovieSceneFloatTrack
       display_name=Render Opacity
       sections [{
         type=MovieSceneFloatSection
         end_frame=2
         start_frame=0
         channels [None {type=MovieSceneScriptingFloatChannel, keys [[0,0],[2.0253,1],[2.98,1]]}]
        }]
      }
     ]
    }]}
 }
 Outro {
  binding_count=2
  length_seconds=0.2
  object=&self:Outro
  sequence {bindings [
    CommonTextBlock_2 {
     id=#1
     possessed_class=CommonTextBlock
     tracks [{
       type=MovieScene2DTransformTrack
       sections [{
         type=MovieScene2DTransformSection
         end_frame=4
         start_frame=0
         channels [
          Translation.X {type=MovieSceneScriptingFloatChannel, const [0,4]}
          Angle {type=MovieSceneScriptingFloatChannel}
          Scale.X {type=MovieSceneScriptingFloatChannel}
          Scale.Y {type=MovieSceneScriptingFloatChannel}
          Shear.X {type=MovieSceneScriptingFloatChannel}
          Shear.Y {type=MovieSceneScriptingFloatChannel}
         ]
        }]
      }]
    }
    SizeBox_0 {
     id=#2
     possessed_class=SizeBox
     tracks [
      {
       type=MovieScene2DTransformTrack
       sections [{
         type=MovieScene2DTransformSection
         end_frame=3
         start_frame=0
         channels [
          Translation.X {type=MovieSceneScriptingFloatChannel}
          Translation.Y {type=MovieSceneScriptingFloatChannel}
          Angle {type=MovieSceneScriptingFloatChannel}
          Scale.X {type=MovieSceneScriptingFloatChannel, keys [[0,0.6],[2,1.05],[3.533,1]]}
          Scale.Y {type=MovieSceneScriptingFloatChannel, keys [[0,0.6],[2,1.15],[3.533,1]]}
          Shear.X {type=MovieSceneScriptingFloatChannel}
          Shear.Y {type=MovieSceneScriptingFloatChannel}
         ]
        }]
      }
      {
       type=MovieSceneFloatTrack
       display_name=Render Opacity
       sections [{type=MovieSceneFloatSection, end_frame=4, start_frame=0, channels [None {type=MovieSceneScriptingFloatChannel, keys [[0,0],[1.428,1]]}]}]
      }
     ]
    }
   ]}
 }
]
graphs [
 ClosePanel {nodes [
   ClosePanel() {is=title K2Node_FunctionEntry}
   > DeactivateWidget()
   > n1 = Get!(execute:Exec) -> (then:Exec, else:Exec, Portal:BGA Island Portal Object Reference)
   > n2 = ClearDisplayedWidget(self:BGA Island Portal Object Reference=@n1)
   Return Node!(execute=[@n2.then, @n1.else], bCommitted:bool=false)
  ]}
 DeleteIsland {nodes [
   DeleteIsland!() -> (then, MCPIslandId:String) {is=title K2Node_FunctionEntry}
   > n1 = Cast To FortPlayerControllerAthena(Object=GetOwningPlayer())
   > DestroyCreativePlot!(execute, self:Fort Player Controller Athena Object Reference=@n1, IslandId:String=Get MCPIslandId() -> (MCPIslandId:String)) -> (then)
  ]}
 DuplicateCreativePlotAndHandleButtonState {nodes [
   DuplicateCreativePlotAndHandleButtonState!() -> (then, DuplicateTitle:String) {is=title K2Node_FunctionEntry}
   > n1 = Cast To IslandDataEntry!(
    execute
    Object=Get Selected Item!(self:List View Object Reference=Get ListView_IslandList() -> (ListView_IslandList:Common List View Object Reference)) -> (ReturnValue:Object Reference) {is=title K2Node_CallFunction}
   ) -> (then, CastFailed, AsIsland Data Entry:Island Data Entry Object Reference) {is=title K2Node_DynamicCast}
   > DuplicateIsland!(
    execute
    self
    PlotToDuplicateID:String=Get McpId!(self:Island Data Entry Object Reference=@n1) -> (McpId:String)
    Title:String=Get DuplicateTitle() -> (DuplicateTitle:String)
   ) -> (then)
   > DisableButtonWithReason!(
    execute
    self:Common Button Base Object Reference=[
     Get DuplicateIslandButton() -> (DuplicateIslandButton:Icon Text Button Object Reference)
     Get Duplicate_Button() -> (Duplicate_Button:WBP UIKit Button Regular Object Reference)
    ]
    DisabledReason:"Text (by ref)"=Get Working Text() -> (Working Text:Text)
   ) -> (then)
   > SetVisibility(self=Get DuplicateIslandButtonSpinner() -> (DuplicateIslandButtonSpinner:Image Object Reference), InVisibility=SelfHitTestInvisible)
  ]}
 DuplicateIsland {nodes [
   DuplicateIsland!() -> (then, PlotToDuplicateID:String, Title:String) {is=title K2Node_FunctionEntry}
   > n1 = Cast To FortPlayerControllerAthena(Object=GetOwningPlayer())
   > DuplicateCreativePlot!(
    execute
    self:Fort Player Controller Athena Object Reference=@n1
    IslandId:String=Get PlotToDuplicateID() -> (PlotToDuplicateID:String)
    Locale:String=GetCurrentLanguage!() -> (ReturnValue:String)
    Title:String=Get Title() -> (Title:String)
   ) -> (then)
  ]}
 EventGraph {nodes [
   n1 = "On Button Clicked (RemoveIslandButton)"!() -> (OutputDelegate, then, Button:Common Button Legacy Object Reference) {is=title K2Node_ComponentBoundEvent}
   n2 = Cast To IslandDataEntry(execute=[@n1.then, @n42.then], Object=Get Selected Item(self=Get ListView_IslandList()))
   > n3 = Branch(Condition=IsSelectingCurrentlyLoadedIsland!() -> (IsCurrentlyLoaded:bool))
   > ShowErrorDialog!(
    execute
    DisplayMessage:Text="LOCTEXT(\"You cannot remove your currently loaded island!\")"
    ErrorCode:String=Island Removal Error
    ErrorSeverity:EFortErrorSeverity Enum=Unspecified
   ) -> (then)
   Set DialogFinishHoldoverIndex(execute=@n3.else, DialogFinishHoldoverIndex:int=Get LastSelection() -> (LastSelection:int=0)) -> (Output_Get:int=0)
   n4 = Get PublishedIslandCode!(self:Island Data Entry Object Reference=@n2) -> (PublishedIslandCode:String)
   n5 = MakeLiteralText(Value="To Text (String)"(InString=@n4))
   n6 = Get BasicWarningText() -> (BasicWarningText:Text)
   n7 = Len(S=@n4)
   > n8 = ShowSimpleConfirmationDialog_NUI!(
    execute:Exec
    Icon:Slate Brush Structure
    Title:Text="LOCTEXT(\"Remove Island?\")"
    Message:Text=Select(
     Option 0:Text=Format Text!(
      Format="LOCTEXT(\"{0}\\r\\n\\r\\nThis does not affect published island codes. Content will continue to be accessible at {islandCode} but you will be unable to update it.\")"
      0:Text=@n6
      islandCode:Text=@n5
     ) -> (Result)
     Option 1:Text=Get BasicWarningText()
     Index="Equal (Integer)"(A=@n7, B=0) {is=title K2Node_CallFunction}
    ) -> (ReturnValue:Text)
    bShowConfirm:bool=true
    bShowDecline:bool=true
    AdditionalContent:Widget Object Reference
    LeftAdditionalContent:Widget Object Reference
    TargetPlayer:Local Player Object Reference
   ) -> (then:Exec, DialogResult:Exec, Result:EFortDialogResult Enum, ResultName:Name) {is=title K2Node_AsyncAction}
   n9 = Get DialogFinishHoldoverIndex() -> (DialogFinishHoldoverIndex:int=0)
   Set LastSelection(execute=@n8.DialogResult, LastSelection:int=@n9) -> (Output_Get:int=0)
   > Switch on EFortDialogResult!(execute, Selection:EFortDialogResult Enum=@n8) -> (Confirmed:Exec, Declined:Exec, Ignored:Exec, Killed:Exec, TimedOut:Exec) {is=title K2Node_SwitchEnum}
   > n10 = Cast To IslandDataEntry(Object=GetItemAt!(self:List View Object Reference=Get ListView_IslandList(), Index:int=@n9) -> (ReturnValue:Object Reference))
   > DeleteIsland!(execute, self, MCPIslandId:String=Get McpId(self=@n10)) -> (then) {is=title K2Node_CallFunction}
   > DisableButtonWithReason(
    self=[
     Get RemoveIslandButton() -> (RemoveIslandButton:Icon Text Button Object Reference)
     Get Remove_Button() -> (Remove_Button:WBP UIKit Button Regular Object Reference)
    ]
    DisabledReason=Get Working Text()
   )
   > SetVisibility(self=Get RemoveSpinner() -> (RemoveSpinner:Image Object Reference), InVisibility=SelfHitTestInvisible)
   n11 = "On Button Clicked (CloseMainPanel)"!() -> (OutputDelegate, then, Button:Common Button Legacy Object Reference) {is=title K2Node_ComponentBoundEvent}
   ClosePanelWithAnim(execute=[@n11.then, @n17.then])
   "On Button Clicked (UserActionMenuInputButton)"!() -> (OutputDelegate, then, Button:Common Button Legacy Object Reference) {is=title K2Node_ComponentBoundEvent}
   > MyIslandButtonClicked()
   "On Button Clicked (UserActionMenuInputButton_1)"!() -> (OutputDelegate, then, Button:Common Button Legacy Object Reference) {is=title K2Node_ComponentBoundEvent}
   > DeletedIslandsButtonClicked()
   n12 = "On Button Clicked (DuplicateIslandButton)"!() -> (OutputDelegate, then, Button:Common Button Legacy Object Reference) {is=title K2Node_ComponentBoundEvent}
   n13 = Cast To IslandDataEntry(execute=[@n12.then, @n43.then], Object=Get Selected Item(self=Get ListView_IslandList()))
   > n14 = Branch(Condition=Get IsDeleted!(self:Island Data Entry Object Reference=@n13) -> (IsDeleted:bool))
   n15 = CreateWidgetAsync!(
    execute:Exec=@n14.else
    UserWidgetSoftClass:User Widget Soft Class Reference=Get IslandNamingPopup() -> (IslandNamingPopup:User Widget Soft Class Reference)
    OwningPlayer:Player Controller Object Reference
    bSuspendInputUntilComplete:bool=true
   ) -> (then:Exec, OnComplete:Exec, OnFailure:Exec, UserWidget:User Widget Object Reference) {is=title K2Node_AsyncAction}
   n16 = Cast To PopupIslandNaming_UI_MultipleIP_Modal!(execute=@n15.OnComplete, Object=@n15) -> (then, CastFailed, AsPopup Island Naming UI Multiple IP Modal:Popup Island Naming UI Multiple IP Modal Object Reference) {is=title K2Node_DynamicCast}
   > Set CurrentModalWidget(CurrentModalWidget:User Widget Object Reference=@n16) -> (Output_Get:User Widget Object Reference)
   > Bind Event to On Widget Deactivated(self:Common Activatable Widget Object Reference=@n16, Delegate=Create Event()) {is=title K2Node_AddDelegate}
   > InitFromObject!(execute, self:Popup Island Naming UI Multiple IP Modal Object Reference=@n16, InitObject:Object Reference=Self-Reference()) -> (then)
   > TryPushWidgetAsModalPanel!(execute, PushWidget:User Widget Object Reference=@n16) -> (then)
   > InitFromIslandDataEntry!(execute, self:Popup Island Naming UI Multiple IP Modal Object Reference=@n16, IslandData:Island Data Entry Object Reference=@n13) -> (then)
   > Append Copy Text(self:Popup Island Naming UI Multiple IP Modal Object Reference=@n16)
   > Set IsFromTemplate!(execute, IsFromTemplate:bool=true, self:Popup Island Naming UI Multiple IP Modal Object Reference=@n16) -> (then, Output_Get:bool)
   > SetIsEnabled(self=Get SetPortal_Button() -> (SetPortal_Button:WBP UIKit Button Loud Object Reference), bInIsEnabled=false)
   "On Button Hovered (CreateIsland)"!() -> (OutputDelegate, then, Button:Common Button Legacy Object Reference) {is=title K2Node_ComponentBoundEvent}
   > Set Icon!(
    execute
    self:Icon Text Button Object Reference=Get CreateIsland() -> (CreateIsland:Icon Text Button Object Reference)
    IconBrush:Slate Brush Structure=Make Slate Brush!(
     TintColor:Slate Color Structure=Make Slate Color(SpecifiedColor=(R=0,G=0,B=0,A=1))
     DrawAs:ESlateBrushDrawType Enum=Image
     Tiling:ESlateBrushTileType Enum=NoTile
     Mirroring:ESlateBrushMirrorType Enum=NoMirror
     ImageSize:Vector 2D Structure=(X=42,Y=42)
     Margin:Margin Structure=(Left=0,Top=0,Right=0,Bottom=0)
     ResourceObject:Object Reference=/Game/Athena/UI/Frontend/Parties/Art/T_UI_PlusSign_64x
     OutlineSettings:Slate Brush Outline Settings Structure=(CornerRadii=(X=0,Y=0,Z=0,W=0),Color=(SpecifiedColor=(R=0,G=0,B=0,A=0),ColorUseRule=UseColor_Specified),Width=0,RoundingType=HalfHeightRadius,bUseBrushTransparency=False)
    ) -> (SlateBrush:Slate Brush Structure) {is=title K2Node_MakeStruct}
   ) -> (then) {is=title K2Node_CallFunction}
   "On Button Unhovered (CreateIsland)"!() -> (OutputDelegate, then, Button:Common Button Legacy Object Reference) {is=title K2Node_ComponentBoundEvent}
   > Set Icon(
    self=Get CreateIsland()
    IconBrush=Make Slate Brush(
     TintColor=Make Slate Color(SpecifiedColor=(R=1,G=1,B=1,A=1))
     DrawAs=Image
     Tiling=NoTile
     Mirroring=NoMirror
     ImageSize=(X=42,Y=42)
     Margin=(Left=0,Top=0,Right=0,Bottom=0)
     ResourceObject=&T_UI_PlusSign_64x
     OutlineSettings=(CornerRadii=(X=0,Y=0,Z=0,W=0),Color=(SpecifiedColor=(R=0,G=0,B=0,A=0),ColorUseRule=UseColor_Specified),Width=0,RoundingType=HalfHeightRadius,bUseBrushTransparency=False)
    )
   )
   n17 = "On Button Clicked (Button_CloseTouch)"!() -> (OutputDelegate, then, Button:Common Button Legacy Object Reference) {is=title K2Node_ComponentBoundEvent}
   n18 = "On Tab Selected (WBP_UIKit_Tabs)"!() -> (OutputDelegate, then, TabId:Name=None) {is=title K2Node_ComponentBoundEvent}
   n19 = Get WBP_UIKit_Tabs() -> (WBP_UIKit_Tabs:WBP UIKit Tabs Object Reference)
   n20 = Branch(
    execute=[@n18.then, @n67.then, @n68.then]
    Condition="Equal (Integer)"(A=GetSelectedTabIndex!(self:UIKit Tab Group Object Reference=@n19) -> (ReturnValue:int=0)) {is=title K2Node_PromotableOperator}
   )
   > SetActiveWidget(
    self:Widget Switcher Object Reference=Get ListWidgetSwitcher() -> (ListWidgetSwitcher:Widget Switcher Object Reference)
    Widget=Get ActiveIslandsSection() -> (ActiveIslandsSection:Horizontal Box Object Reference)
   )
   n21 = Get ListView_IslandList()
   > n22 = Branch(Condition="integer > integer"(A=GetNumItems!(self:List View Object Reference=@n21) -> (ReturnValue:int=0), B=0) {is=title K2Node_CallFunction})
   > CenterWidget!(execute, ToCenter:Widget Object Reference=@n21) -> (then)
   n23 = FixupSelectionIndex!(ListView:Common List View Object Reference=@n21) -> (FixedValue:int=0)
   > NavigateToIndex!(execute, self:List View Object Reference=@n21, Index:int=@n23) -> (then)
   > SetSelectedIndex!(execute, self:List View Object Reference=@n21, Index:int=@n23) -> (then)
   > OnSelectionChanged!(
    execute
    self
    Item:Object Reference=GetItemAt(self=@n21, Index=@n23)
    Is Selected:bool=true
    ListView:Common List View Object Reference=@n21
   ) -> (then)
   SetVisibility(execute=@n22.else, self=[Get RemoveIslandButton(), Get Remove_Button()], InVisibility=Hidden)
   > SetVisibility(self=[Get SetPortal_Button(), Get SelectIsland() -> (SelectIsland:Icon Text Button Object Reference)], InVisibility=Hidden)
   SelectButtonAtIndex!(
    execute=@n20.else
    self:Common Button Group Base Object Reference=Get TabButtonGroup() -> (TabButtonGroup:Common Button Group Legacy Object Reference)
    ButtonIndex:int=1
    bAllowSound:bool=true
   ) -> (then)
   > SetActiveWidget(self=Get ListWidgetSwitcher(), Widget=Get DeletedIslandsSection() -> (DeletedIslandsSection:Horizontal Box Object Reference))
   n24 = Get ListView_DeletedIslandList() -> (ListView_DeletedIslandList:Common List View Object Reference)
   > n25 = Branch(Condition="integer > integer"(A=GetNumItems(self=@n24) -> (ReturnValue=0), B=0))
   > CenterWidget(ToCenter=@n24)
   n26 = FixupSelectionIndex(ListView=@n24) -> (FixedValue=0)
   > NavigateToIndex(self=@n24, Index=@n26)
   > SetSelectedIndex(self=@n24, Index=@n26)
   > OnSelectionChanged(Item=GetItemAt(self=@n24, Index=@n26), Is Selected=true, ListView=@n24)
   SetVisibility(execute=@n25.else, self=[Get RemoveIslandButton(), Get Remove_Button()], InVisibility=Hidden)
   > SetVisibility(self=[Get SetPortal_Button(), Get SelectIsland()], InVisibility=Hidden)
   > SetVisibility(self=[Get DuplicateIslandButton(), Get Duplicate_Button()], InVisibility=Hidden)
   > SetVisibility(
    self=[
     Get RenameIslandButton() -> (RenameIslandButton:Icon Text Button Object Reference)
     Get Rename_Button() -> (Rename_Button:WBP UIKit Button Regular Object Reference)
    ]
    InVisibility=Hidden
   )
   n27 = "On Clicked (SetPortal_Button)"!() -> (OutputDelegate, then, Button:Common Button Base Object Reference) {is=title K2Node_ComponentBoundEvent}
   n28 = Cast To IslandDataEntry(execute=[@n27.then, @n44.then], Object=Get Selected Item(self=GetActiveListView!(self) -> (ReturnValue:Common List View Object Reference)))
   > n29 = Branch(Condition=Get IsDeleted(self=@n28))
   n30 = Get McpId(self=@n28)
   > RecoverIsland!(execute, self, MCPIslandID:String=@n30) -> (then)
   > DisableButtonWithReason(self=[Get SetPortal_Button(), Get SelectIsland()], DisabledReason=Get Working Text())
   > SetVisibility(self=Get SelectIslandSpinner() -> (SelectIslandSpinner:Image Object Reference), InVisibility=SelfHitTestInvisible)
   SelectPlot!(execute=@n29.else, self, SelectedPlot:String=@n30) -> (then)
   "On Clicked (Rename_Button)"!() -> (OutputDelegate, then, Button:Common Button Base Object Reference) {is=title K2Node_ComponentBoundEvent}
   > n31 = Cast To IslandDataEntry(Object=Get Selected Item(self=Get ListView_IslandList()))
   > n32 = Branch(Condition=Get IsDeleted(self=@n31))
   n33 = CreateWidgetAsync(execute=@n32.else, UserWidgetSoftClass=Get IslandNamingPopup(), bSuspendInputUntilComplete=true)
   n34 = Cast To PopupIslandNaming_UI_MultipleIP_Modal(execute=@n33.OnComplete, Object=@n33)
   > Set CurrentModalWidget(CurrentModalWidget=@n34)
   > Bind Event to On Widget Deactivated(self=@n34, Delegate=Create Event())
   > InitFromObject(self=@n34, InitObject=Self-Reference())
   > TryPushWidgetAsModalPanel(PushWidget=@n34)
   > InitFromIslandDataEntry(self=@n34, IslandData=@n31)
   > Set IsForRenaming!(execute, IsForRenaming:bool=true, self:Popup Island Naming UI Multiple IP Modal Object Reference=@n34) -> (then, Output_Get:bool)
   > SetIsEnabled(self=Get SetPortal_Button(), bInIsEnabled=false)
   n35 = "On Clicked (CreateNew_Button)"!() -> (OutputDelegate, then, Button:Common Button Base Object Reference) {is=title K2Node_ComponentBoundEvent}
   n36 = CreateWidgetAsync(
    execute=[@n35.then, @n40.then]
    UserWidgetSoftClass=Select(
     Option 0:User Widget Soft Class Reference=Get CreateIslandUI() -> (CreateIslandUI:User Widget Soft Class Reference)
     Option 1:User Widget Soft Class Reference=Get CreateIslandUI_V2() -> (CreateIslandUI_V2:User Widget Soft Class Reference)
     Index=ShouldUseGFSCompatibleStarterIslandsUI!(self:Fort Creative Island Multiple IPSelect Object Reference) -> (ReturnValue:bool)
    ) -> (ReturnValue:User Widget Soft Class Reference)
    OwningPlayer=GetOwningPlayer()
    bSuspendInputUntilComplete=true
   )
   InitFromObject!(execute:Exec=@n36.OnComplete, self:Object Reference=@n36, InitObject:Object Reference=Self-Reference()) -> (then:Exec) {is=title K2Node_Message}
   n37 = "On Entry Generated (ListView_IslandList)"!() -> (OutputDelegate, then, Widget:User Widget Object Reference) {is=title K2Node_ComponentBoundEvent}
   > n38 = Cast To IslandOptionEntry_V2!(execute, Object=@n37.Widget) -> (then, CastFailed, AsIsland Option Entry V2:Island Option Entry V2 Object Reference) {is=title K2Node_DynamicCast}
   n39 = Get PublishedCodeText!(self:Island Option Entry V2 Object Reference=@n38) -> (PublishedCodeText:Common Text Object Reference)
   > Branch(Condition="Equal (Enum)"(
     A:ESlateVisibility Enum=GetVisibility!(self:Widget Object Reference=@n39) -> (ReturnValue:ESlateVisibility Enum=Visible)
     B:ESlateVisibility Enum=SelfHitTestInvisible
    ))
   > SetVisibility(self=Get IslandCodeHeaderColumn() -> (IslandCodeHeaderColumn:Common Text Object Reference), InVisibility=SelfHitTestInvisible)
   n40 = "On Button Clicked (CreateIsland)"!() -> (OutputDelegate, then, Button:Common Button Legacy Object Reference) {is=title K2Node_ComponentBoundEvent}
   n41 = "On Item Selection Changed (ListView_IslandList)"!() -> (OutputDelegate, then, Item:Object Reference, bIsSelected:bool) {is=title K2Node_ComponentBoundEvent}
   > OnSelectionChanged(Item=@n41.Item, Is Selected=@n41.bIsSelected, ListView=Get ListView_IslandList())
   n42 = "On Clicked (Remove_Button)"!() -> (OutputDelegate, then, Button:Common Button Base Object Reference) {is=title K2Node_ComponentBoundEvent}
   n43 = "On Clicked (Duplicate_Button)"!() -> (OutputDelegate, then, Button:Common Button Base Object Reference) {is=title K2Node_ComponentBoundEvent}
   n44 = "On Button Clicked (SelectIsland)"!() -> (OutputDelegate, then, Button:Common Button Legacy Object Reference) {is=title K2Node_ComponentBoundEvent}
   n45 = "On Item Selection Changed (ListView_DeletedIslandList)"!() -> (OutputDelegate, then, Item:Object Reference, bIsSelected:bool) {is=title K2Node_ComponentBoundEvent}
   > OnSelectionChanged(Item=@n45.Item, Is Selected=@n45.bIsSelected, ListView=Get ListView_DeletedIslandList())
   UpdateInputActionHandler() {is=title K2Node_CustomEvent}
   > n46 = Branch(Condition=IsUsingTouch(PlayerController=GetOwningPlayer()))
   n47 = Get CloseButton() -> (CloseButton:Data Table Row Handle Structure)
   > Branch(Condition=HasInputActionHandler!(self:Common Activatable Panel Legacy Object Reference, InputActionRow:Data Table Row Handle Structure=@n47) -> (ReturnValue:bool))
   n48 = Self-Reference()
   > RemoveInputActionHandler!(execute, self:Common Activatable Panel Legacy Object Reference=@n48, InputActionRow:Data Table Row Handle Structure=@n47) -> (then)
   SetInputActionHandler!(
    execute=@n46.else
    self:Common Activatable Panel Legacy Object Reference=@n48
    InputActionRow:Data Table Row Handle Structure=@n47
    CommitedEvent:Delegate=Create Event()
   ) -> (then)
   n49 = RestoreIslandFinished!() -> (OutputDelegate, then, bSuccess:bool=~, Reason:"Text (by ref)") {is=title K2Node_CustomEvent}
   > n50 = Branch(Condition=@n49.bSuccess)
   SetIsEnabled(execute=[@n51.then, @n50.then], self=[Get SetPortal_Button(), Get SelectIsland()], bInIsEnabled=true)
   > SetVisibility(self=Get SelectIslandSpinner(), InVisibility=Collapsed)
   n51 = ShowErrorDialog(execute=@n50.else, DisplayMessage=@n49.Reason, ErrorCode=Restore Plot Error, ErrorSeverity=Unspecified)
   n52 = OnUpdateCreativePlotNameFinished!() -> (OutputDelegate, then, bSuccess:bool=~, Reason:"Text (by ref)") {is=title K2Node_CustomEvent}
   > n53 = Branch(Condition=@n52.bSuccess)
   SetIsEnabled(execute=[@n54.then, @n53.then], self=Get Rename_Button(), bInIsEnabled=true)
   > SetVisibility(self=Get RenameIslandButtonSpinner() -> (RenameIslandButtonSpinner:Image Object Reference), InVisibility=Collapsed)
   n54 = ShowErrorDialog(execute=@n53.else, DisplayMessage=@n52.Reason, ErrorCode=Rename Island Error, ErrorSeverity=Unspecified)
   n55 = DuplicateIslandFinished!() -> (OutputDelegate, then, bSuccess:bool=~, Reason:"Text (by ref)") {is=title K2Node_CustomEvent}
   > n56 = Branch(Condition=@n55.bSuccess)
   SetIsEnabled(execute=[@n57.then, @n56.then], self=[Get DuplicateIslandButton(), Get Duplicate_Button()], bInIsEnabled=true)
   > SetVisibility(self=Get DuplicateIslandButtonSpinner(), InVisibility=Collapsed)
   n57 = ShowErrorDialog(execute=@n56.else, DisplayMessage=@n55.Reason, ErrorCode=Duplicate Island Error, ErrorSeverity=Unspecified)
   PlayNoIslandAnim() {is=title K2Node_CustomEvent}
   > PlayAnimationForward(InAnimation=Get NoIslands() -> (NoIslands:Widget Animation Object Reference))
   n58 = CloseShown() {is=title K2Node_CustomEvent}
   > ClosePanel!(execute, self) -> (then, bCommitted:bool) {is=title K2Node_CallFunction}
   OnModalWidgetDeactivated() {is=title K2Node_CustomEvent}
   > n59 = Sequence()
   > n60 = Cast To PopupIslandNaming_UI_MultipleIP_Modal(Object=Get CurrentModalWidget() -> (CurrentModalWidget:User Widget Object Reference))
   > Unbind Event from On Widget Deactivated(self:Common Activatable Widget Object Reference=@n60, Delegate=Create Event()) {is=title K2Node_RemoveDelegate}
   SetIsEnabled(execute=@n59.then_1, self=Get SetPortal_Button(), bInIsEnabled=true)
   > Set CurrentModalWidget()
   ClosePanelWithAnim() {is=title K2Node_CustomEvent}
   > Branch(Condition=NOT Boolean(A=Get bIsClosing() -> (bIsClosing:bool)))
   > Set bIsClosing(bIsClosing:bool=true) -> (Output_Get:bool)
   > PlaySound2D(Sound=Get PanelDismiss_Sound() -> (PanelDismiss_Sound:Sound Base Object Reference))
   n61 = Get Outro() -> (Outro:Widget Animation Object Reference)
   > BindToAnimationFinished(Animation=@n61, Delegate=Create Event())
   > PlayAnimationReverse(InAnimation=@n61, PlaybackSpeed=1.25)
   n62 = DestroyPlotFinished!() -> (OutputDelegate, then, bSuccess:bool=~, Reason:"Text (by ref)") {is=title K2Node_CustomEvent}
   > n63 = Branch(Condition=@n62.bSuccess)
   SetIsEnabled(execute=[@n64.then, @n63.then], self=[Get RemoveIslandButton(), Get Remove_Button()], bInIsEnabled=true)
   > SetVisibility(self=Get RemoveSpinner(), InVisibility=Collapsed)
   n64 = ShowErrorDialog(execute=@n63.else, DisplayMessage=@n62.Reason, ErrorCode=Destroy Plot Error, ErrorSeverity=Unspecified)
   Reinitialize() {is=title K2Node_CustomEvent}
   > Initialize()
   > FocusAfterInitialize!(execute, self, ListView:List View Object Reference=GetActiveListView()) -> (then)
   SetupTabs() {is=title K2Node_CustomEvent}
   n65 = Get TabButtonGroup()
   > SetSelectionRequired!(execute, self:Common Button Group Base Object Reference=@n65, bRequireSelection:bool=false) -> (then)
   > DeselectAll(self:Common Button Group Base Object Reference=@n65)
   > RemoveAll(self:Common Widget Group Base Object Reference=@n65)
   > AddWidget!(
    execute
    self:Common Widget Group Base Object Reference=@n65
    InWidget:Widget Object Reference=Get MyIslandsTab() -> (MyIslandsTab:Icon Text Button Object Reference)
   ) -> (then)
   > AddWidget(self=@n65, InWidget=Get DeletedIslandsTab() -> (DeletedIslandsTab:Icon Text Button Object Reference))
   > SetSelectionRequired(self=@n65, bRequireSelection=true)
   > SetIsSelected!(execute, self:Common Button Base Object Reference=Get MyIslandsTab(), InSelected:bool=true, bGiveClickFeedback:bool=false) -> (then)
   n66 = Get ListView_IslandList()
   > Clear Selection(self:List View Object Reference=@n66)
   > OnSelectionChanged(Item=Get Selected Item(self=@n66), Is Selected=false, ListView=@n66)
   OnStandardOutroFinished() {is=title K2Node_CustomEvent}
   > Set bIsClosing(bIsClosing=false)
   > ClosePanel()
   > UnbindFromAnimationFinished!(execute, self:User Widget Object Reference, Animation:Widget Animation Object Reference=Get Outro(), Delegate:Delegate=Create Event()) -> (then)
   n67 = MyIslandButtonClicked() {is=title K2Node_CustomEvent}
   n68 = DeletedIslandsButtonClicked() {is=title K2Node_CustomEvent}
   Event OnInitialized()
   > n69 = Construct Object from Class!(execute, Class=/Script/CommonUILegacy.CommonButtonGroupLegacy, self=Self-Reference(), bSelectionRequired:bool=true) -> (then, ReturnValue:Common Button Group Legacy Object Reference)
   > Set IslandButtonGroup(IslandButtonGroup:Common Button Group Legacy Object Reference=@n69) -> (Output_Get:Common Button Group Legacy Object Reference)
   Event Construct()
   > SetupTabs() {is=title K2Node_CallFunction}
   > Branch(Condition=IsInCabinMode!() -> (ReturnValue:bool))
   > SetVisibility(self=Get Rename_Button(), InVisibility=Collapsed)
   n70 = Event OnInputModeChanged!() -> (OutputDelegate, then, bUsingGamepad:bool) {is=title K2Node_Event}
   > UpdateInputActionHandler() {is=title K2Node_CallFunction}
   n71 = GetNumItems(self=Get ListView_IslandList()) -> (ReturnValue=0)
   > Branch(Condition=AND Boolean(A=@n70.bUsingGamepad, B="integer > integer"(A=@n71, B=1)))
   n72 = GetActiveListView()
   n73 = Get LastSelection() -> (LastSelection=0)
   > ScrollIndexIntoView!(execute, self:List View Object Reference=@n72, Index:int=@n73) -> (then)
   > SetSelectedIndex(self=@n72, Index=@n73)
   > n74 = Cast To IslandDataEntry(Object=Get Selected Item(self=@n72))
   n75 = Get ListWidgetSwitcher()
   > UpdateSelectButtonText!(
    execute
    self
    IslandIsDeleted:bool="Equal (Object)"(A=GetActiveWidget!(self:Widget Switcher Object Reference=@n75) -> (ReturnValue:Widget Object Reference), B=Get ListView_DeletedIslandList())
    CanEdit:bool=NOT Boolean(A=Get IsLocked!(self:Island Data Entry Object Reference=@n74) -> (IsLocked:bool))
   ) -> (then)
   Event OnCollabEntitlementChanged() {is=title K2Node_Event}
   > n76 = Branch(Condition=Get ShouldLoadFirstIslandOnInit() -> (ShouldLoadFirstIslandOnInit:bool))
   n77 = Cast To IslandDataEntry(execute=[@n76.then, @n78.then], Object=Get Selected Item(self=GetActiveListView()))
   > UpdateSelectButtonText(IslandIsDeleted=Get IsDeleted(self=@n77), CanEdit=NOT Boolean(A=Get IsLocked(self=@n77)))
   n78 = Reinitialize(execute=@n76.else) {is=title K2Node_CallFunction}
   Event On Activated()
   > n79 = Sequence!(execute) -> (then_0, then_1, then_2:Exec)
   > PlayAnimationForward(InAnimation=Get Intro())
   > PlaySound2D(Sound=Get PanelAppear_Sound() -> (PanelAppear_Sound:Sound Base Object Reference))
   > SetActiveWidget(self=Get ListWidgetSwitcher(), Widget=Get ActiveIslandsSection())
   > Set LastSelection(LastSelection=0) -> (Output_Get=0)
   > SetupTabs()
   > CheckIfHasCollabEntitlement(self:Fort Creative Island Multiple IPUIBase Object Reference)
   UpdateInputActionHandler(execute=@n79.then_2)
   n80 = Cast To FortPlayerControllerAthena(execute=@n79.then_1, Object=GetOwningPlayer())
   > Bind Event to On Restore Creative Plot Finished(self:Fort Player Controller Athena Object Reference=@n80, Delegate=Create Event()) {is=title K2Node_AddDelegate}
   > Bind Event to On Destroy Creative Plot Finished(self:Fort Player Controller Athena Object Reference=@n80, Delegate=Create Event()) {is=title K2Node_AddDelegate}
   > Bind Event to On Island Entries Changed(self:Fort Player Controller Athena Object Reference=@n80, Delegate=Create Event()) {is=title K2Node_AddDelegate}
   > Bind Event to On Duplicate Creative Plot Finished(self:Fort Player Controller Athena Object Reference=@n80, Delegate=Create Event()) {is=title K2Node_AddDelegate}
   > Bind Event to On Update Creative Plot Name(self:Fort Player Controller Athena Object Reference=@n80, Delegate=Create Event()) {is=title K2Node_AddDelegate}
   n81 = Event InitFromObject!() -> (OutputDelegate, then, InitObject:Object Reference) {is=title K2Node_Event}
   > n82 = Cast To BGA_IslandPortal!(execute, Object=@n81.InitObject) -> (then, CastFailed, AsBGA Island Portal:BGA Island Portal Object Reference) {is=title K2Node_DynamicCast}
   > OpenPanel!(execute, self, Portal:BGA Island Portal Object Reference=@n82) -> (then)
   > Bind Event to Close Shown UI(self:BGA Island Portal Object Reference=@n82, Delegate=@n58) {is=title K2Node_AddDelegate}
   Event On Deactivated()
   > ClosePanel()
   > n83 = Cast To FortPlayerControllerAthena(Object=GetOwningPlayer())
   > Unbind Event from On Restore Creative Plot Finished(self:Fort Player Controller Athena Object Reference=@n83, Delegate=Create Event()) {is=title K2Node_RemoveDelegate}
   > Unbind Event from On Destroy Creative Plot Finished(self:Fort Player Controller Athena Object Reference=@n83, Delegate=Create Event()) {is=title K2Node_RemoveDelegate}
   > Unbind Event from On Island Entries Changed(self:Fort Player Controller Athena Object Reference=@n83, Delegate=Create Event()) {is=title K2Node_RemoveDelegate}
   > Unbind Event from On Duplicate Creative Plot Finished(self:Fort Player Controller Athena Object Reference=@n83, Delegate=Create Event()) {is=title K2Node_RemoveDelegate}
   > Unbind Event from On Update Creative Plot Name(self:Fort Player Controller Athena Object Reference=@n83, Delegate=Create Event()) {is=title K2Node_RemoveDelegate}
  ]}
 FixupSelectionIndex {nodes [
   n1 = FixupSelectionIndex!() -> (then, ListView:Common List View Object Reference) {is=title K2Node_FunctionEntry}
   n2 = Length(TargetArray:Array of Object References=GetListItems!(self:List View Object Reference=@n1) -> (ReturnValue:Array of Object References))
   > Return Node!(
    execute
    FixedValue:int="Clamp (Integer)"(Value=Get LastSelection() -> (LastSelection=0), Min=0, Max=int - int(A=@n2, B=1) {is=title K2Node_CallFunction}) -> (ReturnValue=0)
   )
  ]}
 FocusAfterInitialize {nodes [
   n1 = FocusAfterInitialize!() -> (then, ListView:List View Object Reference) {is=title K2Node_FunctionEntry}
   n2 = GetNumItems(self=@n1) -> (ReturnValue=0)
   > Branch(Condition="integer > integer"(A=@n2, B=0))
   n3 = Get LastSelection() -> (LastSelection=0)
   n4 = Select(Option 0:int=@n2, Option 1:int=@n3, Index="integer < integer"!(A:int=@n3, B:int=@n2) -> (ReturnValue:bool) {is=title K2Node_CallFunction}) -> (ReturnValue:int)
   > SetSelectedIndex(self=@n1, Index=@n4)
   > NavigateToIndex(self=@n1, Index=@n4)
  ]}
 GetActiveListView {nodes [
   GetActiveListView() {is=title K2Node_FunctionEntry}
   n1 = Get ListWidgetSwitcher()
   > Return Node!(
    execute
    ReturnValue:Common List View Object Reference=Select(
     Option 0:Common List View Object Reference=Get ListView_IslandList()
     Option 1:Common List View Object Reference=Get ListView_DeletedIslandList()
     Index:int=GetActiveWidgetIndex!(self:Widget Switcher Object Reference=@n1) -> (ReturnValue:int=0)
    ) -> (ReturnValue:Common List View Object Reference)
   )
  ]}
 GetLockStatus {nodes [
   n1 = GetLockStatus!() -> (then, IslandDataEntry:Island Data Entry Object Reference, CreativeIslandData:Creative Island Data Structure) {is=title K2Node_FunctionEntry}
   > n2 = Get!(execute) -> (then, else, IslandDataEntry:Island Data Entry Object Reference)
   > Set IslandDataMCP_ID(IslandDataMCP_ID:String=Get McpId(self=@n2)) -> (Output_Get:String)
   > n3 = Set IslandDataGFS_Name(IslandDataGFS_Name:String=Get GameFeatureSetName!(self:Island Data Entry Object Reference=@n2) -> (GameFeatureSetName:String)) -> (Output_Get:String)
   n4 = IsIslandPlayable!(self:Fort Creative Island Multiple IPUIBase Object Reference, McpId:String=Get IslandDataMCP_ID() -> (IslandDataMCP_ID:String)) -> (ReturnValue:bool)
   n5 = GetHasCollabApproval!(
    self:Fort Creative Island Multiple IPUIBase Object Reference
    GameFeatureSetName:String=Get IslandDataGFS_Name() -> (IslandDataGFS_Name:String)
    Context:ECollabIPGatingContext Enum=Edit
   ) -> (ReturnValue:bool)
   n6 = Break Creative Island Data!(CreativeIslandData:"Creative Island Data Structure (by ref)"=@n1.CreativeIslandData) -> (
    McpId:String
    IslandName:Text
    IslandCode:Online Link Id Structure
    GameFeatureSetName:String
    LastLoadedDate:Date Time Structure
    DeletedAt:Date Time Structure
    bIsDeleted:bool=~
    bIsVKProject:bool=~
    OriginatingIslandId:String
   ) {is=title K2Node_BreakStruct}
   Return Node!(
    execute=[@n3.then, @n7.then]
    IsLocked:bool=OR Boolean(A=NAND Boolean(A=@n4, B=@n5) {is=title K2Node_CommutativeAssociativeBinaryOperator}, B=@n6.bIsVKProject)
   )
   Set IslandDataMCP_ID(execute=@n2.else, IslandDataMCP_ID=@n6)
   > n7 = Set IslandDataGFS_Name(IslandDataGFS_Name=@n6.GameFeatureSetName)
  ]}
 Initialize {nodes [
   Initialize() {is=title K2Node_FunctionEntry}
   > n1 = Sequence()
   > Set bIsClosing(bIsClosing=false)
   > n2 = Cast To FortPlayerControllerAthena(Object=GetOwningPlayer())
   > Clear(TargetArray:Array of Island Data Entry Object References=Get IslandEntries() -> (IslandEntries:Array of Island Data Entry Object References))
   > Clear(TargetArray=Get DeletedIslandEntries() -> (DeletedIslandEntries:Array of Island Data Entry Object References))
   > n3 = For Each Loop(Array:Array of Creative Island Data Structures=Get CreativeIslands!(self:Fort Player Controller Athena Object Reference=@n2) -> (CreativeIslands:Array of Creative Island Data Structures)) -> (Array Element:Creative Island Data Structure)
   n4 = Break Creative Island Data(CreativeIslandData=@n3) -> (bIsDeleted=~, bIsVKProject=~)
   n5 = Break Online Link Id!(OnlineLinkId:"Online Link Id Structure (by ref)"=@n4.IslandCode) -> (Mnemonic:String, Version:int) {is=title K2Node_BreakStruct}
   > n6 = Construct Object from Class!(
    execute
    Class=&/IslandDataEntry.IslandDataEntry_C
    self=Self-Reference()
    McpId:String=@n4
    MapName:Text=@n4.IslandName
    PublishedIslandCode:String=@n5
    PublishedVersion:int=@n5.Version
    GameFeatureSetName:String=@n4.GameFeatureSetName
    LastUsedDate:Date Time Structure=@n4.LastLoadedDate
    DeletedAt:Date Time Structure=@n4.DeletedAt
    IsDeleted:bool=@n4.bIsDeleted
    IsVKProject:bool=@n4.bIsVKProject
    IsLocked:bool=GetLockStatus!(self, IslandDataEntry:Island Data Entry Object Reference, CreativeIslandData:Creative Island Data Structure=@n3) -> (IsLocked:bool) {is=title K2Node_CallFunction}
   ) -> (then, ReturnValue:Island Data Entry Object Reference)
   > n7 = Branch(Condition=Get IsDeleted(self=@n6))
   > Add(
    TargetArray:Array of Island Data Entry Object References=Get DeletedIslandEntries()
    NewItem:"Island Data Entry Object Reference (by ref)"=@n6
   ) {is=title K2Node_CallArrayFunction}
   Add(execute=@n7.else, TargetArray=Get IslandEntries(), NewItem=@n6)
   UpdateIslandNumbersText(execute=@n1.then_1)
   > Set List Items!(execute, self:List View Object Reference=Get ListView_IslandList(), InListItems:Array of Object References=Get IslandEntries()) -> (then) {is=title K2Node_CallFunction}
   > Set List Items(self=Get ListView_DeletedIslandList(), InListItems=Get DeletedIslandEntries())
   n8 = "integer > integer"(A=GetNumItems(self=Get ListView_IslandList()) -> (ReturnValue=0), B=0)
   > SetActiveWidget(
    self=Get Switcher_ActiveList() -> (Switcher_ActiveList:Widget Switcher Object Reference)
    Widget=Select(
     Option 0:Widget Object Reference=Get ActiveListViewErrorMessage() -> (ActiveListViewErrorMessage:Size Box Object Reference)
     Option 1:Widget Object Reference=Get ActiveListView() -> (ActiveListView:Vertical Box Object Reference)
     Index:bool=@n8
    ) -> (ReturnValue:Widget Object Reference)
   )
   > n9 = Branch(Condition=@n8)
   PlayNoIslandAnim(execute=@n9.else) {is=title K2Node_CallFunction}
   n10 = Get IslandEntries()
   Branch(
    execute=@n1.then_2
    Condition=AND Boolean(A=Get ShouldLoadFirstIslandOnInit(), B=Is Not Empty!(TargetArray:Array of Island Data Entry Object References=@n10) -> (ReturnValue:bool))
   )
   > Set ShouldLoadFirstIslandOnInit(ShouldLoadFirstIslandOnInit:bool=false) -> (Output_Get:bool)
   n11 = "Get (a copy)"(Array:Array of Island Data Entry Object References=@n10) -> (Output:Island Data Entry Object Reference)
   > Is Valid(InputObject=@n11)
   > n12 = Cast To FortPlayerControllerAthena(Object=GetOwningPlayer())
   > PlaySound2D(Sound=Get SelectPlot_Sound() -> (SelectPlot_Sound:Sound Base Object Reference))
   > ServerLoadPlotForPortal!(
    execute
    self:Fort Player Controller Athena Object Reference=@n12
    Portal:Fort Athena Creative Portal Object Reference=Get Portal() -> (Portal:BGA Island Portal Object Reference)
    PlotItemId:String=Get McpId(self=@n11)
   ) -> (then)
   > Set InFlightLoadingPlot(InFlightLoadingPlot:Island Data Entry Object Reference=@n11) -> (Output_Get:Island Data Entry Object Reference)
   > ClosePanel()
  ]}
 IsSelectingCachedEditIsland {nodes [
   IsSelectingCachedEditIsland() {is=title K2Node_FunctionEntry}
   > n1 = Cast To IslandDataEntry(Object=Get Selected Item(self=Get ListView_IslandList()))
   > n2 = Get!(execute) -> (then, else, Portal:BGA Island Portal Object Reference)
   > Return Node!(
    execute
    IsSelectingCachedEditIsland:bool="Equal Exactly (Text)"!(
     A:"Text (by ref)"=GetCachedEditIslandDataIslandName!(self:Fort Athena Creative Portal Object Reference=@n2) -> (ReturnValue:"Text (by ref)")
     B:"Text (by ref)"=Get MapName!(self:Island Data Entry Object Reference=@n1) -> (MapName:Text)
    ) -> (ReturnValue:bool)
   )
   Return Node(execute=[@n2.else, @n1.CastFailed], IsSelectingCachedEditIsland=false)
  ]}
 IsSelectingCurrentlyLoadedIsland {nodes [
   IsSelectingCurrentlyLoadedIsland() {is=title K2Node_FunctionEntry}
   > n1 = Cast To IslandDataEntry(Object=Get Selected Item(self=Get ListView_IslandList()))
   n2 = Get Portal()
   > n3 = Branch(Condition="Equal Exactly (String)"(A=Get McpId(self=@n1), B=GetIslandID!(self:Fort Athena Creative Portal Object Reference=@n2) -> (ReturnValue:String)))
   > Return Node!(execute, IsCurrentlyLoaded:bool=true)
   n4 = Get InFlightLoadingPlot() -> (InFlightLoadingPlot:Island Data Entry Object Reference)
   n5 = Is Valid(exec=@n3.else, InputObject=@n4)
   Return Node(execute=@n5.Is Not Valid, IsCurrentlyLoaded=false)
   > Return Node(IsCurrentlyLoaded="Equal Exactly (Text)"(A=Get MapName(self=@n1), B=Get MapName(self=@n4)))
   Return Node(execute=@n1.CastFailed, IsCurrentlyLoaded=false)
  ]}
 OnSelectionChanged {nodes [
   n1 = OnSelectionChanged!() -> (then, Item:Object Reference, Is Selected:bool=~, ListView:Common List View Object Reference) {is=title K2Node_FunctionEntry}
   > n2 = Is Valid(InputObject=@n1)
   > Set LastSelection(LastSelection=GetIndexForItem!(self:List View Object Reference=@n1.ListView, Item:Object Reference=@n1) -> (ReturnValue:int=0)) -> (Output_Get=0)
   > n3 = Cast To IslandDataEntry(Object=@n1)
   > n4 = Set SelectedIslandDataEntry(SelectedIslandDataEntry:Island Data Entry Object Reference=@n3) -> (Output_Get:Island Data Entry Object Reference)
   > Set IslandIsDeleted(IslandIsDeleted:bool=Get IsDeleted(self=@n4)) -> (Output_Get:bool)
   > Set SelectedGameFeatureSet(SelectedGameFeatureSet:String=Get GameFeatureSetName(self=@n4)) -> (Output_Get:String)
   > n5 = Sequence!(execute) -> (then_0, then_1)
   > n6 = Branch(Condition=Get Is Selected() -> (Is Selected:bool))
   n7 = Get SelectedIslandDataEntry() -> (SelectedIslandDataEntry:Island Data Entry Object Reference)
   > UpdateSelectButtonText(IslandIsDeleted=Get IslandIsDeleted() -> (IslandIsDeleted:bool), CanEdit=NOT Boolean(A=Get IsLocked(self=@n7)))
   n8 = Select(Option 0:ESlateVisibility Enum=Visible, Option 1:ESlateVisibility Enum=Hidden, Index=Get IslandIsDeleted()) -> (ReturnValue:ESlateVisibility Enum)
   > SetVisibility(self=[Get RemoveIslandButton(), Get Remove_Button()], InVisibility=@n8)
   > SetVisibility(self=[Get Duplicate_Button(), Get DuplicateIslandButton()], InVisibility=@n8)
   > SetVisibility(self=Get Rename_Button(), InVisibility=Select(Option 0=@n8, Option 1=Collapsed, Index=IsInCabinMode()))
   > SetVisibility(self=[Get SetPortal_Button(), Get SelectIsland()], InVisibility=Visible)
   SetVisibility(execute=[@n6.else, @n3.CastFailed, @n2.Is Not Valid], self=[Get RemoveIslandButton(), Get Remove_Button()], InVisibility=Collapsed)
   > SetVisibility(self=[Get Duplicate_Button(), Get DuplicateIslandButton()], InVisibility=Collapsed)
   > SetVisibility(self=[Get SetPortal_Button(), Get SelectIsland()], InVisibility=Collapsed)
   > SetVisibility(self=Get Rename_Button(), InVisibility=Collapsed)
   Branch(
    execute=@n5.then_1
    Condition="Equal (Object)"(A=Get ListView() -> (ListView:Common List View Object Reference), B=Get ListView_IslandList()) {is=title K2Node_CallFunction}
   )
   n9 = Get SelectedIslandDataEntry()
   n10 = NOT Boolean(A=OR Boolean!(
     A=Get IsVKProject!(self:Island Data Entry Object Reference=@n9) -> (IsVKProject:bool)
     B=IsSelectingCurrentlyLoadedIsland() {is=title K2Node_CallFunction}
     C:bool=IsSelectingCachedEditIsland!() -> (IsSelectingCachedEditIsland:bool) {is=title K2Node_CallFunction}
    ) -> (ReturnValue))
   n11 = Get Remove_Button()
   n12 = Get Rename_Button()
   n13 = Get RemoveIslandButton()
   > SetIsEnabled(self=[@n11, @n12, @n13], bInIsEnabled=@n10)
   n14 = Get IsLocked(self=@n9)
   n15 = Get DuplicateIslandButton()
   n16 = Get Duplicate_Button()
   > SetIsEnabled(self=[@n15, @n16], bInIsEnabled=AND Boolean(A=@n10, B=NOT Boolean(A=@n14)))
   > InvalidateLayoutAndVolatility(self:Widget Object Reference=[@n11, @n12, @n16, @n15, @n13])
  ]}
 OnTouchToClose {nodes [
   OnTouchToClose!() -> (then, MyGeometry:Geometry Structure, MouseEvent:"Pointer Event Structure (by ref)") {is=title K2Node_FunctionEntry}
   > ClosePanelWithAnim() {is=title K2Node_CallFunction}
   > Return Node!(execute, ReturnValue:Event Reply Structure=Handled())
  ]}
 OpenPanel {nodes [
   n1 = OpenPanel!() -> (then, Portal:BGA Island Portal Object Reference) {is=title K2Node_FunctionEntry}
   > Set Portal(Portal:BGA Island Portal Object Reference=@n1) -> (Output_Get:BGA Island Portal Object Reference)
   > PushContentWidgetAdvanced!(execute, PushWidget:Widget Object Reference=Self-Reference(), bHideHeader:bool=true, bHideFooter:bool=false, bHideChatWidget:bool=true) -> (then)
   > Initialize() {is=title K2Node_CallFunction}
  ]}
 RecoverIsland {nodes [
   RecoverIsland!() -> (then, MCPIslandID:String) {is=title K2Node_FunctionEntry}
   > n1 = Cast To FortPlayerControllerAthena(Object=GetOwningPlayer())
   > RestoreCreativePlot!(execute, self:Fort Player Controller Athena Object Reference=@n1, IslandId:String=Get MCPIslandID() -> (MCPIslandID:String)) -> (then)
  ]}
 RenameCreativePlotAndHandleButtonState {nodes [
   RenameCreativePlotAndHandleButtonState!() -> (then, RenameTitle:String) {is=title K2Node_FunctionEntry}
   > n1 = Cast To IslandDataEntry(Object=Get Selected Item(self=Get ListView_IslandList()))
   > RenameIsland!(execute, self, PlotToRenameID:String=Get McpId(self=@n1), Title:String=Get RenameTitle() -> (RenameTitle:String)) -> (then)
   > DisableButtonWithReason(self=Get Rename_Button(), DisabledReason=Get Working Text())
   > SetVisibility(self=Get RenameIslandButtonSpinner(), InVisibility=SelfHitTestInvisible)
  ]}
 RenameIsland {nodes [
   RenameIsland!() -> (then, PlotToRenameID:String, Title:String) {is=title K2Node_FunctionEntry}
   > n1 = Cast To FortPlayerControllerAthena(Object=GetOwningPlayer())
   > UpdateCreativePlotName!(
    execute
    self:Fort Player Controller Athena Object Reference=@n1
    IslandId:String=Get PlotToRenameID() -> (PlotToRenameID:String)
    Locale:String=GetCurrentLanguage()
    Title:String=Get Title()
   ) -> (then)
  ]}
 SelectPlot {nodes [
   SelectPlot!() -> (then, SelectedPlot:String) {is=title K2Node_FunctionEntry}
   > n1 = Cast To FortPlayerControllerAthena(Object=GetOwningPlayer())
   > PlaySound2D(Sound=Get SelectPlot_Sound())
   > ServerLoadPlotForPortal(self=@n1, Portal=Get Portal(), PlotItemId=Get SelectedPlot() -> (SelectedPlot:String))
   > ClosePanelWithAnim()
  ]}
 UpdateIslandNumbersText {nodes [
   UpdateIslandNumbersText() {is=title K2Node_FunctionEntry}
   > n1 = Cast To FortPlayerControllerAthena(Object=GetOwningPlayer())
   n2 = GetProfilePlotEntries!(self:Fort Player Controller Athena Object Reference=@n1, bIncludeDeleted:bool=false) -> (ReturnValue:int=0)
   n3 = GetMaxNumberOfPlots!(self:Fort Player Controller Athena Object Reference=@n1) -> (ReturnValue:int=0)
   > "SetText (Text)"(
    self=Get IslandLimitText() -> (IslandLimitText:Common Text Object Reference)
    InText=Format Text!(Format="LOCTEXT(\"{Current} / {Maximum}\")", Current:int=@n2, Maximum:int=@n3) -> (Result)
   )
   n4 = Branch(Condition="integer >= integer"!(A:int=@n2, B:int=@n3) -> (ReturnValue:bool) {is=title K2Node_CallFunction})
   > SetIsEnabled(self=Get CreateIsland(), bInIsEnabled=false)
   SetIsEnabled(execute=@n4.else, bInIsEnabled=true)
  ]}
 UpdateSelectButtonText {nodes [
   n1 = UpdateSelectButtonText!() -> (then, IslandIsDeleted:bool=~, CanEdit:bool=~) {is=title K2Node_FunctionEntry}
   > n2 = Branch(Condition=OR Boolean!(A=@n1, B=@n1.CanEdit) -> (ReturnValue))
   n3 = Get SelectIsland()
   > Set Text!(
    execute
    self:Icon Text Button Object Reference=@n3
    ButtonText:Text=Select(Option 0:Text=Get SelectedText() -> (SelectedText:Text), Option 1:Text=Get RecoverText() -> (RecoverText:Text), Index=Get IslandIsDeleted()) -> (ReturnValue:Text)
   ) -> (then) {is=title K2Node_CallFunction}
   n4 = Get SetPortal_Button()
   > SetIsEnabled(self=[@n4, @n3], bInIsEnabled=true)
   n5 = Get SetPortal_Button()
   > SetText!(
    execute
    self:Fort CTAButton Object Reference=@n5
    ButtonText:"Text (by ref)"=Select(Option 0=Get SelectedText(), Option 1=Get RecoverText(), Index=Get IslandIsDeleted())
   ) -> (then)
   Set Text(execute=@n2.else, self=@n3, ButtonText=Get NotEntitledText() -> (NotEntitledText:Text))
   > DisableButtonWithReason(self=[@n4, @n3], DisabledReason=Get NotEntitledText())
   > SetText(self=@n5, ButtonText=Get NotEntitledText())
  ]}
]
variables [
 BasicWarningText {type=string, default=Are you sure you want to delete your island? A deleted island may not be accessible in the future.}
 bIsClosing {type=boolean}
 CloseButton {type=DataTableRowHandle, default {dataTable=&InputActionDataTable, rowName=Close}}
 CreateIslandUI {type=Class, default=&/Creative_CreateIsland_UI_MultipleIP.Creative_CreateIsland_UI_MultipleIP_C}
 CreateIslandUI_V2 {type=Class, default=&/Creative_CreateIsland_UI_MultipleIP_V2.Creative_CreateIsland_UI_MultipleIP_V2_C}
 CurrentModalWidget {type=UserWidget, default=null}
 DeletedIslandEntries {type="array<IslandDataEntry_C>", default []}
 DialogFinishHoldoverIndex {type=integer}
 Entitlement Granted {type=boolean}
 InFlightLoadingPlot {type=IslandDataEntry_C, default=null}
 IslandButtonGroup {type=CommonButtonGroupLegacy, default=null}
 IslandEntries {type="array<IslandDataEntry_C>", default []}
 IslandNamingPopup {type=Class, default=&/PopupIslandNaming_UI_MultipleIP_Modal.PopupIslandNaming_UI_MultipleIP_Modal_C}
 LastSelection {type=integer}
 NotEntitledText {type=string, default=Unavailable}
 PanelAppear_Sound {type=SoundBase, default=/Game/Sounds/Fort_UI_Sounds/FrontEnd/New_FrontEnd/Fortnite_UI_MainMenu_Tab_03_Cue}
 PanelDismiss_Sound {type=SoundBase, default=/Game/Sounds/Fort_UI_Sounds/FrontEnd/New_FrontEnd/Fortnite_UI_MissionObj_Cue}
 Portal {type=BGA_IslandPortal_C, default=null}
 RecoverText {type=string, default=Recover}
 SelectedText {type=string, default=SET PORTAL}
 SelectPlot_Sound {type=SoundBase, default=/Game/Sounds/Fort_UI_Sounds/FrontEnd/New_FrontEnd/Fortnite_UI_Options_Menu_01_Cue}
 ShouldLoadFirstIslandOnInit {type=boolean}
 TabButtonGroup {type=CommonButtonGroupLegacy, default=null}
 Working Text {type=string, default=Working}
]
widgets {
 info {
  inheritedWidgetCount=1
  parentClass=/Script/FortniteUI.FortCreativeIslandMultipleIPSelect
  rootWidgetClass=&CommonBorder
  widgetCount=104
 }
 widgets [
  TouchToCloseZone {widgetClassPath=&CommonBorder, widget=TouchToCloseZone, properties {padding=<ltrb 0 0 0 0>}}
  BackgroundBlur_1 {
   widgetClassPath=&BackgroundBlur
   parent=TouchToCloseZone
   widget=BackgroundBlur_1
   properties {blurStrength=25, visibility=Visible}
   slot {class=BorderSlot, properties {padding=<ltrb 0 0 0 0>}}
  }
  SafeZone_0 {
   widgetClassPath=&SafeZone
   bIsVariable=true
   parent=BackgroundBlur_1
   widget=SafeZone_0
   slot {class=BackgroundBlurSlot, properties {padding=<ltrb 0 0 0 0>}}
  }
  Overlay_0 {widgetClassPath=&Overlay, parent=SafeZone_0, widget=Overlay_0, slot {class=SafeZoneSlot}}
  DeleteMe {widgetClassPath=&HorizontalBox, parent=Overlay_0, widget=DeleteMe, properties {visibility=Collapsed}, slot {class=OverlaySlot}}
  MyIslandsTab {
   widgetClassPath=&IconTextButton
   bIsVariable=true
   parent=DeleteMe
   widget=MyIslandsTab
   properties {
    bHideInputAction=true
    bIsFocusable=false
    buttonText=My Islands
    minWidth=300
    onButtonBaseClicked=(&self:WidgetTree.MyIslandsTab.HandleOnButtonClicked)
    onButtonBaseDoubleClicked=(&self:WidgetTree.MyIslandsTab.HandleOnButtonDoubleClicked)
    onButtonBaseHovered=(&self:WidgetTree.MyIslandsTab.HandleOnButtonHovered)
    onButtonBaseUnhovered=(&self:WidgetTree.MyIslandsTab.HandleOnButtonUnhovered)
    onSelectedChangedBase=(&self:WidgetTree.MyIslandsTab.HandleOnSelectedChanged)
    style=/FortUILibrary/Foundation/Buttons/ButtonStyles/Tab/Athena_ButtonStyle_ChallengesTab.Athena_ButtonStyle_ChallengesTab_C
    visibility=Visible
    font {
     bForceMonospaced=false
     bMaterialIsStencil=false
     fontMaterial=null
     fontObject=&HeadingNow
     letterSpacing=0
     monospacedWidth=1
     size=18
     skewAmount=0
     typefaceFontName=86 Bold
     outlineSettings {
      bApplyOutlineToDropShadows=false
      bMiteredCorners=false
      bSeparateFillAlpha=false
      outlineMaterial=null
      outlineSize=0
      outlineColor=<rgba 0 0 0 1>
     }
    }
    triggeringInputAction {dataTable=&InputActionDataTable, rowName=GenericLeftShoulder}
   }
   slot {class=HorizontalBoxSlot}
  }
  DeletedIslandsTab {
   widgetClassPath=&IconTextButton
   bIsVariable=true
   parent=DeleteMe
   widget=DeletedIslandsTab
   properties {
    bHideInputAction=true
    bIsFocusable=false
    buttonText=Deleted Islands
    minWidth=300
    onButtonBaseClicked=(&self:WidgetTree.DeletedIslandsTab.HandleOnButtonClicked)
    onButtonBaseDoubleClicked=(&self:WidgetTree.DeletedIslandsTab.HandleOnButtonDoubleClicked)
    onButtonBaseHovered=(&self:WidgetTree.DeletedIslandsTab.HandleOnButtonHovered)
    onButtonBaseUnhovered=(&self:WidgetTree.DeletedIslandsTab.HandleOnButtonUnhovered)
    onSelectedChangedBase=(&self:WidgetTree.DeletedIslandsTab.HandleOnSelectedChanged)
    style=&Athena_ButtonStyle_ChallengesTab_C
    visibility=Visible
   }
   slot {class=HorizontalBoxSlot}
  }
  UserActionMenuInputButton {
   widgetClassPath=/Game/UI/SocialPanel/UserActionMenuInputButton
   bIsVariable=true
   parent=DeleteMe
   widget=UserActionMenuInputButton
   properties {
    onButtonBaseClicked=(&self:WidgetTree.UserActionMenuInputButton.HandleOnButtonClicked)
    onButtonBaseDoubleClicked=(&self:WidgetTree.UserActionMenuInputButton.HandleOnButtonDoubleClicked)
    onButtonBaseHovered=(&self:WidgetTree.UserActionMenuInputButton.HandleOnButtonHovered)
    onButtonBaseUnhovered=(&self:WidgetTree.UserActionMenuInputButton.HandleOnButtonUnhovered)
    onSelectedChangedBase=(&self:WidgetTree.UserActionMenuInputButton.HandleOnSelectedChanged)
    triggeringInputAction {dataTable=&InputActionDataTable, rowName=GenericLeftShoulder}
   }
   slot {class=HorizontalBoxSlot, properties {padding=<ltrb 40 0 5 0>}}
  }
  UserActionMenuInputButton_1 {
   widgetClassPath=&UserActionMenuInputButton
   bIsVariable=true
   parent=DeleteMe
   widget=UserActionMenuInputButton_1
   properties {
    onButtonBaseClicked=(&self:WidgetTree.UserActionMenuInputButton_1.HandleOnButtonClicked)
    onButtonBaseDoubleClicked=(&self:WidgetTree.UserActionMenuInputButton_1.HandleOnButtonDoubleClicked)
    onButtonBaseHovered=(&self:WidgetTree.UserActionMenuInputButton_1.HandleOnButtonHovered)
    onButtonBaseUnhovered=(&self:WidgetTree.UserActionMenuInputButton_1.HandleOnButtonUnhovered)
    onSelectedChangedBase=(&self:WidgetTree.UserActionMenuInputButton_1.HandleOnSelectedChanged)
    triggeringInputAction {dataTable=&InputActionDataTable, rowName=GenericRightShoulder}
   }
   slot {class=HorizontalBoxSlot, properties {padding=<ltrb 5 0 0 0>}}
  }
  CreateIsland {
   widgetClassPath=&IconTextButton
   bIsVariable=true
   parent=DeleteMe
   widget=CreateIsland
   properties {
    bIsFocusable=false
    buttonText=Create New
    minHeight=96
    minWidth=312
    onButtonBaseClicked=(&self:WidgetTree.CreateIsland.HandleOnButtonClicked)
    onButtonBaseDoubleClicked=(&self:WidgetTree.CreateIsland.HandleOnButtonDoubleClicked)
    onButtonBaseHovered=(&self:WidgetTree.CreateIsland.HandleOnButtonHovered)
    onButtonBaseUnhovered=(&self:WidgetTree.CreateIsland.HandleOnButtonUnhovered)
    onSelectedChangedBase=(&self:WidgetTree.CreateIsland.HandleOnSelectedChanged)
    style=/FortUILibrary/Foundation/Buttons/ButtonStyles/Skew/ButtonStyle-Skew_LBlue.ButtonStyle-Skew_LBlue_C
    visibility=Collapsed
    font {
     bForceMonospaced=false
     bMaterialIsStencil=false
     fontMaterial=null
     fontObject=&HeadingNow
     letterSpacing=0
     monospacedWidth=1
     size=21
     skewAmount=0
     typefaceFontName=76 Bold
     outlineSettings {
      bApplyOutlineToDropShadows=false
      bMiteredCorners=false
      bSeparateFillAlpha=false
      outlineMaterial=null
      outlineSize=0
      outlineColor=<rgba 0 0 0 1>
     }
    }
    iconBrush {
     bIsDynamicallyLoaded=false
     drawAs=Image
     imageType=NoImage
     mirroring=NoMirror
     resourceName=None
     resourceObject=&T_UI_PlusSign_64x
     tiling=NoTile
     imageSize=<42 42>
     margin=<ltrb 0 0 0 0>
     outlineSettings {
      bUseBrushTransparency=false
      roundingType=HalfHeightRadius
      width=0
      color {colorUseRule=UseColor_Specified, specifiedColor=<rgba 0 0 0 0>}
      cornerRadii=<0 0 0 1>
     }
     tintColor {colorUseRule=UseColor_Specified, specifiedColor=<rgba 1 1 1 1>}
     uVRegion {bIsValid=false, max=<0 0>, min=<0 0>}
    }
    triggeringInputAction {dataTable=&InputActionDataTable, rowName=SecondaryTabRight}
   }
   slot {class=HorizontalBoxSlot, properties {horizontalAlignment=HAlign_Right, size {sizeRule=Automatic, value=0.4}}}
  }
  SizeBox_0 {
   widgetClassPath=&SizeBox
   parent=Overlay_0
   widget=SizeBox_0
   properties {
    bOverride_HeightOverride=true
    bOverride_MinDesiredWidth=true
    heightOverride=812
    maxDesiredHeight=1000
    maxDesiredWidth=800
    minDesiredHeight=1000
    minDesiredWidth=1600
    widthOverride=800
   }
   slot {class=OverlaySlot, properties {horizontalAlignment=HAlign_Center, verticalAlignment=VAlign_Center}}
  }
  Overlay_4 {widgetClassPath=&Overlay, parent=SizeBox_0, widget=Overlay_4, slot {class=SizeBoxSlot}}
  Image_118 {
   widgetClassPath=&Image
   parent=Overlay_4
   widget=Image_118
   properties {colorAndOpacity=<rgba 0.0246 0.0507 0.2031 0.3>, renderTransform {angle=0, scale=<1 1>, shear=<0 0>, translation=<12 12>}}
   slot {class=OverlaySlot, properties {horizontalAlignment=HAlign_Fill, verticalAlignment=VAlign_Fill}}
  }
  BackgroundBlur_0 {
   widgetClassPath=&BackgroundBlur
   parent=Overlay_4
   widget=BackgroundBlur_0
   properties {blurStrength=16}
   slot {class=OverlaySlot, properties {horizontalAlignment=HAlign_Fill, verticalAlignment=VAlign_Fill}}
  }
  CommonBorder_0 {
   widgetClassPath=&CommonBorder
   parent=BackgroundBlur_0
   widget=CommonBorder_0
   properties {
    background {
     bIsDynamicallyLoaded=false
     drawAs=Image
     imageType=NoImage
     mirroring=NoMirror
     resourceName=None
     resourceObject=null
     tiling=NoTile
     imageSize=<32 32>
     margin=<ltrb 0 0 0 0>
     outlineSettings {
      bUseBrushTransparency=false
      roundingType=HalfHeightRadius
      width=0
      color {colorUseRule=UseColor_Specified, specifiedColor=<rgba 0 0 0 0>}
      cornerRadii=<0 0 0 0>
     }
     tintColor {colorUseRule=UseColor_Specified, specifiedColor=<rgba 1 1 1 1>}
     uVRegion {bIsValid=false, max=<0 0>, min=<0 0>}
    }
    padding=<ltrb 0 0 0 0>
    renderTransform {angle=0, scale=<1 1>, shear=<0 0>, translation=<1 1>}
   }
   slot {class=BackgroundBlurSlot, properties {padding=<ltrb 0 0 0 0>}}
  }
  Overlay_1 {widgetClassPath=&Overlay, parent=CommonBorder_0, widget=Overlay_1, slot {class=BorderSlot, properties {padding=<ltrb 0 0 0 0>}}}
  Imge_Backplate {
   widgetClassPath=&Image
   parent=Overlay_1
   widget=Imge_Backplate
   properties {brush {
     bIsDynamicallyLoaded=false
     drawAs=Image
     imageType=FullColor
     mirroring=NoMirror
     resourceName=None
     resourceObject=/Game/Playgrounds/Items/Portals/UI/Art/MI_UI_Creative_NebulousBackground
     tiling=NoTile
     imageSize=<32 32>
     margin=<ltrb 0 0 0 0>
     outlineSettings {
      bUseBrushTransparency=false
      roundingType=HalfHeightRadius
      width=0
      color {colorUseRule=UseColor_Specified, specifiedColor=<rgba 0 0 0 0>}
      cornerRadii=<0 0 0 0>
     }
     tintColor {colorUseRule=UseColor_Specified, specifiedColor=<rgba 1 1 1 1>}
     uVRegion {bIsValid=false, max=<0 0>, min=<0 0>}
    }}
   slot {class=OverlaySlot, properties {horizontalAlignment=HAlign_Fill, verticalAlignment=VAlign_Fill}}
  }
  VerticalBox_0 {
   widgetClassPath=&VerticalBox
   parent=Overlay_1
   widget=VerticalBox_0
   slot {class=OverlaySlot, properties {horizontalAlignment=HAlign_Fill, verticalAlignment=VAlign_Fill}}
  }
  SizeBox_5 {
   widgetClassPath=&SizeBox
   parent=VerticalBox_0
   widget=SizeBox_5
   properties {bOverride_MinDesiredHeight=true, heightOverride=130, minDesiredHeight=124}
   slot {class=VerticalBoxSlot}
  }
  Overlay_2 {widgetClassPath=&Overlay, parent=SizeBox_5, widget=Overlay_2, slot {class=SizeBoxSlot}}
  Image_180 {
   widgetClassPath=&Image
   parent=Overlay_2
   widget=Image_180
   properties {
    visibility=Collapsed
    brush {
     bIsDynamicallyLoaded=false
     drawAs=Image
     imageType=FullColor
     mirroring=NoMirror
     resourceName=None
     resourceObject=null
     tiling=NoTile
     imageSize=<32 32>
     margin=<ltrb 0 0 0 0>
     outlineSettings {
      bUseBrushTransparency=false
      roundingType=HalfHeightRadius
      width=0
      color {colorUseRule=UseColor_Specified, specifiedColor=<rgba 0 0 0 0>}
      cornerRadii=<0 0 0 0>
     }
     tintColor {colorUseRule=UseColor_Specified, specifiedColor=<rgba 0.0212 0.1651 0.6376 0.5>}
     uVRegion {bIsValid=false, max=<0 0>, min=<0 0>}
    }
    renderTransform {angle=0, scale=<1 -1>, shear=<0 0>, translation=<0 -10>}
   }
   slot {class=OverlaySlot, properties {horizontalAlignment=HAlign_Fill, verticalAlignment=VAlign_Fill, padding=<ltrb 0 10 0 0>}}
  }
  HorizontalBox_5 {
   widgetClassPath=&HorizontalBox
   parent=Overlay_2
   widget=HorizontalBox_5
   slot {class=OverlaySlot, properties {horizontalAlignment=HAlign_Fill, verticalAlignment=VAlign_Fill, padding=<ltrb 0 0 32 0>}}
  }
  CommonTextBlock_2 {
   widgetClassPath=&CommonTextBlock
   bIsVariable=true
   parent=HorizontalBox_5
   widget=CommonTextBlock_2
   properties {
    text=SELECT ISLAND
    font {
     bForceMonospaced=false
     bMaterialIsStencil=false
     fontMaterial=null
     fontObject=&HeadingNow
     letterSpacing=0
     monospacedWidth=1
     size=51.75
     skewAmount=0
     typefaceFontName=86 Bold Italic
     outlineSettings {
      bApplyOutlineToDropShadows=false
      bMiteredCorners=false
      bSeparateFillAlpha=false
      outlineMaterial=null
      outlineSize=0
      outlineColor=<rgba 0 0 0 1>
     }
    }
    shadowOffset=<4.9266 5.2555>
   }
   slot {class=HorizontalBoxSlot, properties {horizontalAlignment=HAlign_Left, verticalAlignment=VAlign_Bottom, padding=<ltrb 52 0 0 0>}}
  }
  IslandLimitText {
   widgetClassPath=&CommonTextBlock
   bIsVariable=true
   parent=HorizontalBox_5
   widget=IslandLimitText
   properties {
    justification=Right
    text="{Current} / {Maximum}"
    colorAndOpacity {colorUseRule=UseColor_Specified, specifiedColor=<rgba 0.2195 0.6584 1 1>}
    font {
     bForceMonospaced=false
     bMaterialIsStencil=false
     fontMaterial=null
     fontObject=&HeadingNow
     letterSpacing=0
     monospacedWidth=1
     size=25.5
     skewAmount=0
     typefaceFontName=86 Bold Italic
     outlineSettings {
      bApplyOutlineToDropShadows=false
      bMiteredCorners=false
      bSeparateFillAlpha=false
      outlineMaterial=null
      outlineSize=0
      outlineColor=<rgba 0 0 0 1>
     }
    }
    shadowOffset=<0 0>
   }
   slot {class=HorizontalBoxSlot, properties {verticalAlignment=VAlign_Bottom, padding=<ltrb 0 10 12 10>, size {sizeRule=Fill, value=1}}}
  }
  VisibilityBorder_Touch_Top {
   widgetClassPath=&CommonHardwareVisibilityBorder
   parent=HorizontalBox_5
   widget=VisibilityBorder_Touch_Top
   properties {visibilityQuery {
     autoDescription=" ANY( Input.Touch )"
     tokenStreamVersion=0
     userDescription=""
     queryTokenStream [0,1,1,1,0]
     tagDictionary [{tagName=Input.Touch}]
    }}
   slot {class=HorizontalBoxSlot, properties {horizontalAlignment=HAlign_Right, verticalAlignment=VAlign_Top}}
  }
  Spacer_81 {
   widgetClassPath=&Spacer
   parent=VisibilityBorder_Touch_Top
   widget=Spacer_81
   properties {size=<100 1>}
   slot {class=BorderSlot, properties {padding=<ltrb 0 0 0 0>}}
  }
  HorizontalBox_144 {
   widgetClassPath=&HorizontalBox
   parent=VerticalBox_0
   widget=HorizontalBox_144
   slot {class=VerticalBoxSlot, properties {padding=<ltrb 20 18 42 0>}}
  }
  Overlay_6 {
   widgetClassPath=&Overlay
   parent=HorizontalBox_144
   widget=Overlay_6
   slot {class=HorizontalBoxSlot, properties {padding=<ltrb 24 0 0 0>}}
  }
  WBP_CaptureForPostBufferUpdate {
   widgetClassPath=&WBP_CaptureForPostBufferUpdate
   bIsVariable=true
   parent=Overlay_6
   slot {class=OverlaySlot, properties {horizontalAlignment=HAlign_Fill, verticalAlignment=VAlign_Fill}}
  }
  WBP_UIKit_Tabs {
   widgetClassPath=/FortUILibrary/UIKit/Tabs/WBP_UIKit_Tabs
   bIsVariable=true
   parent=Overlay_6
   widget=WBP_UIKit_Tabs
   properties {
    initialSelectedTabIndex=0
    inputActions=Bumpers
    useBackplate=true
    localTabsViewModels [
     &self:WidgetTree.WBP_UIKit_Tabs.FortTabViewModel_1
     &self:WidgetTree.WBP_UIKit_Tabs.FortTabViewModel_0
    ]
    nextTabInputActionData {dataTable=&InputActionDataTable, rowName=GenericRightShoulder}
    previousTabInputActionData {dataTable=&InputActionDataTable, rowName=GenericLeftShoulder}
   }
   slot {class=OverlaySlot, properties {horizontalAlignment=HAlign_Fill, verticalAlignment=VAlign_Center}}
  }
  RemoveSpinner_1 {
   widgetClassPath=&Image
   bIsVariable=true
   parent=HorizontalBox_144
   widget=RemoveSpinner_1
   properties {
    visibility=Collapsed
    brush {
     bIsDynamicallyLoaded=false
     drawAs=Image
     imageType=NoImage
     mirroring=NoMirror
     resourceName=None
     resourceObject=/FortUILibrary/UIKit/Throbber/MI_UIKit_Monochrome_Throbber
     tiling=NoTile
     imageSize=<32 32>
     margin=<ltrb 0 0 0 0>
     outlineSettings {
      bUseBrushTransparency=false
      roundingType=HalfHeightRadius
      width=0
      color {colorUseRule=UseColor_Specified, specifiedColor=<rgba 0 0 0 0>}
      cornerRadii=<0 0 0 0>
     }
     tintColor {colorUseRule=UseColor_Specified, specifiedColor=<rgba 1 1 1 1>}
     uVRegion {bIsValid=false, max=<0 0>, min=<0 0>}
    }
   }
   slot {class=HorizontalBoxSlot, properties {horizontalAlignment=HAlign_Center, verticalAlignment=VAlign_Center}}
  }
  RemoveSpinner_2 {
   widgetClassPath=&Image
   bIsVariable=true
   parent=HorizontalBox_144
   widget=RemoveSpinner_2
   properties {
    visibility=Collapsed
    brush {
     bIsDynamicallyLoaded=false
     drawAs=Image
     imageType=NoImage
     mirroring=NoMirror
     resourceName=None
     resourceObject=&MI_UIKit_Monochrome_Throbber
     tiling=NoTile
     imageSize=<32 32>
     margin=<ltrb 0 0 0 0>
     outlineSettings {
      bUseBrushTransparency=false
      roundingType=HalfHeightRadius
      width=0
      color {colorUseRule=UseColor_Specified, specifiedColor=<rgba 0 0 0 0>}
      cornerRadii=<0 0 0 0>
     }
     tintColor {colorUseRule=UseColor_Specified, specifiedColor=<rgba 1 1 1 1>}
     uVRegion {bIsValid=false, max=<0 0>, min=<0 0>}
    }
   }
   slot {class=HorizontalBoxSlot, properties {horizontalAlignment=HAlign_Center, verticalAlignment=VAlign_Center}}
  }
  CreateNew_Button {
   widgetClassPath=&WBP_UIKit_Button_Regular
   bIsVariable=true
   parent=HorizontalBox_144
   widget=CreateNew_Button
   properties {text=Create New Island, triggeringInputAction {dataTable=&InputActionDataTable, rowName=SecondaryTabRight}}
   slot {class=HorizontalBoxSlot, properties {horizontalAlignment=HAlign_Right, verticalAlignment=VAlign_Center, size {sizeRule=Fill, value=1}}}
  }
  ListWidgetSwitcher {
   widgetClassPath=&WidgetSwitcher
   bIsVariable=true
   parent=VerticalBox_0
   widget=ListWidgetSwitcher
   slot {class=VerticalBoxSlot, properties {padding=<ltrb 32 16 32 0>, size {sizeRule=Fill, value=1}}}
  }
  ActiveIslandsSection {
   widgetClassPath=&HorizontalBox
   bIsVariable=true
   parent=ListWidgetSwitcher
   widget=ActiveIslandsSection
   slot {class=WidgetSwitcherSlot}
  }
  VerticalBox_4 {
   widgetClassPath=&VerticalBox
   parent=ActiveIslandsSection
   widget=VerticalBox_4
   slot {class=HorizontalBoxSlot, properties {padding=<ltrb 20 10 20 0>, size {sizeRule=Fill, value=1}}}
  }
  Text_Islands {
   widgetClassPath=&CommonTextBlock
   parent=VerticalBox_4
   widget=Text_Islands
   properties {
    text=Select the destination for this rift
    visibility=Collapsed
    colorAndOpacity {colorUseRule=UseColor_Specified, specifiedColor=<rgba 0.2232 0.6584 1 1>}
    font {
     bForceMonospaced=false
     bMaterialIsStencil=false
     fontMaterial=null
     fontObject=&HeadingNow
     letterSpacing=0
     monospacedWidth=1
     size=18
     skewAmount=0
     typefaceFontName=75 Medium
     outlineSettings {
      bApplyOutlineToDropShadows=false
      bMiteredCorners=false
      bSeparateFillAlpha=false
      outlineMaterial=null
      outlineSize=0
      outlineColor=<rgba 0 0 0 1>
     }
    }
    shadowOffset=<0 0>
   }
   slot {class=VerticalBoxSlot, properties {verticalAlignment=VAlign_Center, padding=<ltrb 24 18 0 6>}}
  }
  VerticalBox_1 {
   widgetClassPath=&VerticalBox
   parent=VerticalBox_4
   widget=VerticalBox_1
   properties {navigation=&self:WidgetTree.VerticalBox_1.WidgetNavigation_0}
   slot {class=VerticalBoxSlot, properties {padding=<ltrb 0 0 0 10>, size {sizeRule=Fill, value=1}}}
  }
  Switcher_ActiveList {
   widgetClassPath=&WidgetSwitcher
   bIsVariable=true
   parent=VerticalBox_1
   widget=Switcher_ActiveList
   slot {class=VerticalBoxSlot, properties {size {sizeRule=Fill, value=1}}}
  }
  ActiveListView {
   widgetClassPath=&VerticalBox
   bIsVariable=true
   parent=Switcher_ActiveList
   widget=ActiveListView
   slot {class=WidgetSwitcherSlot}
  }
  GridPanel_0 {
   widgetClassPath=&GridPanel
   parent=ActiveListView
   widget=GridPanel_0
   properties {columnFill [0.5,0.3,0.2]}
   slot {class=VerticalBoxSlot, properties {size {sizeRule=Automatic, value=0.5}}}
  }
  CommonTextBlock {
   widgetClassPath=&CommonTextBlock
   parent=GridPanel_0
   widget=CommonTextBlock
   properties {
    text=Select the destination for this rift
    textTransformPolicy=ToUpper
    colorAndOpacity {colorUseRule=UseColor_Specified, specifiedColor=<rgba 0.2232 0.6584 1 1>}
    font {
     bForceMonospaced=false
     bMaterialIsStencil=false
     fontMaterial=null
     fontObject=&HeadingNow
     letterSpacing=0
     monospacedWidth=1
     size=15
     skewAmount=0
     typefaceFontName=76 Bold
     outlineSettings {
      bApplyOutlineToDropShadows=false
      bMiteredCorners=false
      bSeparateFillAlpha=false
      outlineMaterial=null
      outlineSize=0
      outlineColor=<rgba 0 0 0 1>
     }
    }
    shadowOffset=<4.9266 5.2555>
   }
   slot {class=GridSlot, properties {padding=<ltrb 24 0 0 0>}}
  }
  IslandCodeHeaderColumn {
   widgetClassPath=&CommonTextBlock
   bIsVariable=true
   parent=GridPanel_0
   widget=IslandCodeHeaderColumn
   properties {
    text=Island Code
    textTransformPolicy=ToUpper
    visibility=Collapsed
    colorAndOpacity {colorUseRule=UseColor_Specified, specifiedColor=<rgba 0.2232 0.6584 1 1>}
    font {
     bForceMonospaced=false
     bMaterialIsStencil=false
     fontMaterial=null
     fontObject=&HeadingNow
     letterSpacing=0
     monospacedWidth=1
     size=15
     skewAmount=0
     typefaceFontName=76 Bold
     outlineSettings {
      bApplyOutlineToDropShadows=false
      bMiteredCorners=false
      bSeparateFillAlpha=false
      outlineMaterial=null
      outlineSize=0
      outlineColor=<rgba 0 0 0 1>
     }
    }
    shadowOffset=<4.9266 5.2555>
   }
   slot {class=GridSlot, properties {column=1, padding=<ltrb 56 0 0 0>}}
  }
  CommonTextBlock_4 {
   widgetClassPath=&CommonTextBlock
   parent=GridPanel_0
   widget=CommonTextBlock_4
   properties {
    text=Last Used
    textTransformPolicy=ToUpper
    colorAndOpacity {colorUseRule=UseColor_Specified, specifiedColor=<rgba 0.2232 0.6584 1 1>}
    font {
     bForceMonospaced=false
     bMaterialIsStencil=false
     fontMaterial=null
     fontObject=&HeadingNow
     letterSpacing=0
     monospacedWidth=1
     size=15
     skewAmount=0
     typefaceFontName=76 Bold
     outlineSettings {
      bApplyOutlineToDropShadows=false
      bMiteredCorners=false
      bSeparateFillAlpha=false
      outlineMaterial=null
      outlineSize=0
      outlineColor=<rgba 0 0 0 1>
     }
    }
    shadowOffset=<4.9266 5.2555>
   }
   slot {class=GridSlot, properties {column=2}}
  }
  Border_105 {
   widgetClassPath=&Border
   parent=ActiveListView
   widget=Border_105
   properties {visibility=Collapsed, brushColor=<rgba 0.15 0.314 0.8714 1>, padding=<ltrb 0 0 0 0>}
   slot {class=VerticalBoxSlot}
  }
  SizeBox_119 {
   widgetClassPath=&SizeBox
   parent=Border_105
   widget=SizeBox_119
   properties {bOverride_HeightOverride=true, heightOverride=2, widthOverride=147.6986}
   slot {class=BorderSlot, properties {padding=<ltrb 0 0 0 0>}}
  }
  ListView_IslandList {
   widgetClassPath=&CommonListView
   bInherited=true
   bIsVariable=true
   parent=ActiveListView
   widget=ListView_IslandList
   properties {
    entryWidgetClass=&/IslandOptionEntry_V2.IslandOptionEntry_V2_C
    horizontalEntrySpacing=4
    numDesignerPreviewEntries=12
   }
   slot {class=VerticalBoxSlot, properties {padding=<ltrb -12 4 -12 0>, size {sizeRule=Fill, value=1}}}
  }
  ActiveListViewErrorMessage {
   widgetClassPath=&SizeBox
   bIsVariable=true
   parent=Switcher_ActiveList
   widget=ActiveListViewErrorMessage
   properties {bOverride_HeightOverride=true, bOverride_MaxDesiredHeight=true, heightOverride=300, maxDesiredHeight=300}
   slot {class=WidgetSwitcherSlot}
  }
  TextBlock_0 {
   widgetClassPath=&TextBlock
   bIsVariable=true
   parent=ActiveListViewErrorMessage
   widget=TextBlock_0
   properties {
    renderOpacity=0
    text=You have no available Islands
    colorAndOpacity {colorUseRule=UseColor_Specified, specifiedColor=<rgba 0.9301 0.0742 0.107 1>}
    font {
     bForceMonospaced=false
     bMaterialIsStencil=false
     fontMaterial=null
     fontObject=&BurbankSmall
     letterSpacing=0
     monospacedWidth=1
     size=24
     skewAmount=0
     typefaceFontName=Black
     outlineSettings {
      bApplyOutlineToDropShadows=false
      bMiteredCorners=false
      bSeparateFillAlpha=false
      outlineMaterial=null
      outlineSize=0
      outlineColor=<rgba 0 0 0 1>
     }
    }
   }
   slot {class=SizeBoxSlot, properties {verticalAlignment=VAlign_Center}}
  }
  VB_EnterCodeSection {
   widgetClassPath=&VerticalBox
   parent=VerticalBox_4
   widget=VB_EnterCodeSection
   properties {
    navigation=&self:WidgetTree.VB_EnterCodeSection.WidgetNavigation_0
    visibility=Collapsed
   }
   slot {class=VerticalBoxSlot}
  }
  CommonTextBlock_0 {
   widgetClassPath=&CommonTextBlock
   parent=VB_EnterCodeSection
   widget=CommonTextBlock_0
   properties {
    style=/FortUILibrary/Styles/Typography/Foundation/TextStyle-Base-S.TextStyle-Base-S_C
    text=Enter code
    font {
     bForceMonospaced=false
     bMaterialIsStencil=false
     fontMaterial=null
     fontObject=&NotoSans
     letterSpacing=0
     monospacedWidth=1
     size=18
     skewAmount=0
     typefaceFontName=NotoSans-Regular
     outlineSettings {
      bApplyOutlineToDropShadows=false
      bMiteredCorners=false
      bSeparateFillAlpha=false
      outlineMaterial=null
      outlineSize=0
      outlineColor=<rgba 0 0 0 1>
     }
    }
    shadowOffset=<0 0>
   }
   slot {class=VerticalBoxSlot, properties {verticalAlignment=VAlign_Center}}
  }
  CommonBorder_1 {
   widgetClassPath=&CommonBorder
   parent=VB_EnterCodeSection
   widget=CommonBorder_1
   properties {style=/FortUILibrary/Styles/Border/Border-SolidBG-DkBlue.Border-SolidBG-DkBlue_C, padding=<ltrb 0 0 0 0>}
   slot {class=VerticalBoxSlot, properties {verticalAlignment=VAlign_Center, padding=<ltrb 0 16 0 16>}}
  }
  FriendCodeEntry {
   widgetClassPath=/Script/UMG.EditableText
   bIsVariable=true
   parent=CommonBorder_1
   widget=FriendCodeEntry
   properties {
    hintText=Code
    widgetStyle {
     backgroundImageComposing {
      bIsDynamicallyLoaded=false
      drawAs=Border
      imageType=FullColor
      mirroring=NoMirror
      resourceName=../../../Engine/Content/Slate/Old/HyperlinkDotted.png
      resourceObject=null
      tiling=Both
      imageSize=<0 0>
      margin=<ltrb 0 0 0 0.1875>
      outlineSettings {
       bUseBrushTransparency=false
       roundingType=HalfHeightRadius
       width=0
       color {colorUseRule=UseColor_Specified, specifiedColor=<rgba 0 0 0 0>}
       cornerRadii=<0 0 0 0>
      }
      tintColor {colorUseRule=UseColor_Specified, specifiedColor=<rgba 1 1 1 1>}
      uVRegion {bIsValid=false, max=<0 0>, min=<0 0>}
     }
     backgroundImageSelected {
      bIsDynamicallyLoaded=false
      drawAs=Box
      imageType=FullColor
      mirroring=NoMirror
      resourceName=../../../Engine/Content/Slate/Common/EditableTextSelectionBackground.png
      resourceObject=null
      tiling=NoTile
      imageSize=<0 0>
      margin=<ltrb 0.25 0.25 0.25 0.25>
      outlineSettings {
       bUseBrushTransparency=false
       roundingType=HalfHeightRadius
       width=0
       color {colorUseRule=UseColor_Specified, specifiedColor=<rgba 0 0 0 0>}
       cornerRadii=<0 0 0 0>
      }
      tintColor {colorUseRule=UseColor_Specified, specifiedColor=<rgba 1 1 1 1>}
      uVRegion {bIsValid=false, max=<0 0>, min=<0 0>}
     }
     caretImage {
      bIsDynamicallyLoaded=false
      drawAs=Image
      imageType=FullColor
      mirroring=NoMirror
      resourceName=../../../Engine/Content/Slate/Old/White.png
      resourceObject=null
      tiling=NoTile
      imageSize=<16 16>
      margin=<ltrb 0 0 0 0>
      outlineSettings {
       bUseBrushTransparency=false
       roundingType=HalfHeightRadius
       width=0
       color {colorUseRule=UseColor_Specified, specifiedColor=<rgba 0 0 0 0>}
       cornerRadii=<0 0 0 0>
      }
      tintColor {colorUseRule=UseColor_Specified, specifiedColor=<rgba 1 1 1 1>}
      uVRegion {bIsValid=false, max=<0 0>, min=<0 0>}
     }
     colorAndOpacity {colorUseRule=UseColor_Specified, specifiedColor=<rgba 1 1 1 1>}
     font {
      bForceMonospaced=false
      bMaterialIsStencil=false
      fontMaterial=null
      fontObject=&NotoSans
      letterSpacing=0
      monospacedWidth=1
      size=16
      skewAmount=0
      typefaceFontName=NotoSans-Regular
      outlineSettings {
       bApplyOutlineToDropShadows=false
       bMiteredCorners=false
       bSeparateFillAlpha=false
       outlineMaterial=null
       outlineSize=0
       outlineColor=<rgba 0 0 0 1>
      }
     }
     hintColorAndOpacity {colorUseRule=UseColor_Foreground, specifiedColor=<rgba 1 0 1 1>}
     selectedBackgroundColor {colorUseRule=UseColor_Foreground, specifiedColor=<rgba 1 0 1 1>}
    }
   }
   slot {class=BorderSlot, properties {padding=<ltrb 0 0 0 0>}}
  }
  ErrorText {
   widgetClassPath=&CommonTextBlock
   bIsVariable=true
   parent=VB_EnterCodeSection
   widget=ErrorText
   properties {
    style=/FortUILibrary/Styles/Typography/Foundation/TextStyle-Base-XS-B-S-Red.TextStyle-Base-XS-B-S-Red_C
    text=Text Block
    visibility=Hidden
    colorAndOpacity {colorUseRule=UseColor_Specified, specifiedColor=<rgba 1 0.0561 0.0648 1>}
    font {
     bForceMonospaced=false
     bMaterialIsStencil=false
     fontMaterial=null
     fontObject=&NotoSans
     letterSpacing=0
     monospacedWidth=1
     size=16
     skewAmount=0
     typefaceFontName=NotoSans-Bold
     outlineSettings {
      bApplyOutlineToDropShadows=true
      bMiteredCorners=false
      bSeparateFillAlpha=true
      outlineMaterial=&M-FontShadow
      outlineSize=2
      outlineColor=<rgba 0.085 0.0146 0.0177 0.5>
     }
    }
    shadowOffset=<0 0>
   }
   slot {class=VerticalBoxSlot}
  }
  HorizontalBox_3 {
   widgetClassPath=&HorizontalBox
   parent=VB_EnterCodeSection
   widget=HorizontalBox_3
   slot {class=VerticalBoxSlot, properties {horizontalAlignment=HAlign_Left, verticalAlignment=VAlign_Center, padding=<ltrb 10 10 20 10>}}
  }
  IconTextButton {
   widgetClassPath=&IconTextButton
   bIsVariable=true
   parent=HorizontalBox_3
   widget=IconTextButton
   properties {
    buttonText=Use Code
    onButtonBaseClicked=(&self:WidgetTree.IconTextButton.HandleOnButtonClicked)
    onButtonBaseDoubleClicked=(&self:WidgetTree.IconTextButton.HandleOnButtonDoubleClicked)
    onButtonBaseHovered=(&self:WidgetTree.IconTextButton.HandleOnButtonHovered)
    onButtonBaseUnhovered=(&self:WidgetTree.IconTextButton.HandleOnButtonUnhovered)
    onSelectedChangedBase=(&self:WidgetTree.IconTextButton.HandleOnSelectedChanged)
   }
   slot {class=HorizontalBoxSlot, properties {horizontalAlignment=HAlign_Left}}
  }
  DeletedIslandsSection {
   widgetClassPath=&HorizontalBox
   bIsVariable=true
   parent=ListWidgetSwitcher
   widget=DeletedIslandsSection
   slot {class=WidgetSwitcherSlot}
  }
  VerticalBox {
   widgetClassPath=&VerticalBox
   parent=DeletedIslandsSection
   widget=VerticalBox
   slot {class=HorizontalBoxSlot, properties {padding=<ltrb 20 10 20 32>, size {sizeRule=Fill, value=1}}}
  }
  VerticalBox_3 {
   widgetClassPath=&VerticalBox
   parent=VerticalBox
   widget=VerticalBox_3
   properties {navigation=&self:WidgetTree.VerticalBox_3.WidgetNavigation_0}
   slot {class=VerticalBoxSlot}
  }
  Text_Islands_1 {
   widgetClassPath=&CommonTextBlock
   parent=VerticalBox_3
   widget=Text_Islands_1
   properties {
    justification=Center
    text=Deleted islands listed here may be recovered
    visibility=Collapsed
    font {
     bForceMonospaced=false
     bMaterialIsStencil=false
     fontMaterial=null
     fontObject=&HeadingNow
     letterSpacing=0
     monospacedWidth=1
     size=18
     skewAmount=0
     typefaceFontName=75 Medium
     outlineSettings {
      bApplyOutlineToDropShadows=false
      bMiteredCorners=false
      bSeparateFillAlpha=false
      outlineMaterial=null
      outlineSize=0
      outlineColor=<rgba 0 0 0 1>
     }
    }
    shadowOffset=<0 0>
   }
   slot {class=VerticalBoxSlot, properties {horizontalAlignment=HAlign_Left, verticalAlignment=VAlign_Center, padding=<ltrb 24 18 0 6>}}
  }
  Switcher_DeletedList {
   widgetClassPath=&WidgetSwitcher
   bIsVariable=true
   parent=VerticalBox_3
   widget=Switcher_DeletedList
   slot {class=VerticalBoxSlot}
  }
  DeletedListView {
   widgetClassPath=&VerticalBox
   bIsVariable=true
   parent=Switcher_DeletedList
   widget=DeletedListView
   slot {class=WidgetSwitcherSlot}
  }
  HorizontalBox_2 {widgetClassPath=&HorizontalBox, parent=DeletedListView, widget=HorizontalBox_2, slot {class=VerticalBoxSlot}}
  CommonTextBlock_6 {
   widgetClassPath=&CommonTextBlock
   parent=HorizontalBox_2
   widget=CommonTextBlock_6
   properties {
    text=Deleted islands
    textTransformPolicy=ToUpper
    colorAndOpacity {colorUseRule=UseColor_Specified, specifiedColor=<rgba 0.2232 0.6584 1 1>}
    font {
     bForceMonospaced=false
     bMaterialIsStencil=false
     fontMaterial=null
     fontObject=&HeadingNow
     letterSpacing=0
     monospacedWidth=1
     size=15
     skewAmount=0
     typefaceFontName=76 Bold
     outlineSettings {
      bApplyOutlineToDropShadows=false
      bMiteredCorners=false
      bSeparateFillAlpha=false
      outlineMaterial=null
      outlineSize=0
      outlineColor=<rgba 0 0 0 1>
     }
    }
    shadowOffset=<4.9266 5.2555>
   }
   slot {class=HorizontalBoxSlot, properties {padding=<ltrb 24 0 0 0>, size {sizeRule=Fill, value=0.5}}}
  }
  IslandCodeHeaderColumn_1 {
   widgetClassPath=&CommonTextBlock
   bIsVariable=true
   parent=HorizontalBox_2
   widget=IslandCodeHeaderColumn_1
   properties {
    justification=Center
    text=Island Code
    textTransformPolicy=ToUpper
    visibility=Hidden
    colorAndOpacity {colorUseRule=UseColor_Specified, specifiedColor=<rgba 0.2232 0.6584 1 1>}
    font {
     bForceMonospaced=false
     bMaterialIsStencil=false
     fontMaterial=null
     fontObject=&HeadingNow
     letterSpacing=0
     monospacedWidth=1
     size=15
     skewAmount=0
     typefaceFontName=76 Bold
     outlineSettings {
      bApplyOutlineToDropShadows=false
      bMiteredCorners=false
      bSeparateFillAlpha=false
      outlineMaterial=null
      outlineSize=0
      outlineColor=<rgba 0 0 0 1>
     }
    }
    shadowOffset=<4.9266 5.2555>
   }
   slot {class=HorizontalBoxSlot, properties {size {sizeRule=Fill, value=0.3}}}
  }
  Spacer_1 {
   widgetClassPath=&Spacer
   parent=HorizontalBox_2
   widget=Spacer_1
   slot {class=HorizontalBoxSlot, properties {size {sizeRule=Fill, value=0.01}}}
  }
  CommonTextBlock_7 {
   widgetClassPath=&CommonTextBlock
   parent=HorizontalBox_2
   widget=CommonTextBlock_7
   properties {
    justification=Center
    text=Deleted At
    textTransformPolicy=ToUpper
    colorAndOpacity {colorUseRule=UseColor_Specified, specifiedColor=<rgba 0.2232 0.6584 1 1>}
    font {
     bForceMonospaced=false
     bMaterialIsStencil=false
     fontMaterial=null
     fontObject=&HeadingNow
     letterSpacing=0
     monospacedWidth=1
     size=15
     skewAmount=0
     typefaceFontName=76 Bold
     outlineSettings {
      bApplyOutlineToDropShadows=false
      bMiteredCorners=false
      bSeparateFillAlpha=false
      outlineMaterial=null
      outlineSize=0
      outlineColor=<rgba 0 0 0 1>
     }
    }
    shadowOffset=<4.9266 5.2555>
   }
   slot {class=HorizontalBoxSlot, properties {size {sizeRule=Fill, value=0.3}}}
  }
  Border {
   widgetClassPath=&Border
   parent=DeletedListView
   widget=Border
   properties {visibility=Collapsed, brushColor=<rgba 0.15 0.314 0.8714 1>, padding=<ltrb 0 0 0 0>}
   slot {class=VerticalBoxSlot}
  }
  SizeBox {
   widgetClassPath=&SizeBox
   parent=Border
   widget=SizeBox
   properties {bOverride_HeightOverride=true, heightOverride=2, widthOverride=147.6986}
   slot {class=BorderSlot, properties {padding=<ltrb 0 0 0 0>}}
  }
  SizeBox_3 {
   widgetClassPath=&SizeBox
   parent=DeletedListView
   widget=SizeBox_3
   properties {bOverride_MaxDesiredHeight=true, bOverride_WidthOverride=true, heightOverride=400, maxDesiredHeight=350, widthOverride=700}
   slot {class=VerticalBoxSlot, properties {padding=<ltrb 0 4 0 0>}}
  }
  ListView_DeletedIslandList {
   widgetClassPath=&CommonListView
   bIsVariable=true
   parent=SizeBox_3
   widget=ListView_DeletedIslandList
   properties {
    entryWidgetClass=&IslandOptionEntry_V2_C
    horizontalEntrySpacing=4
    numDesignerPreviewEntries=12
    verticalEntrySpacing=4
   }
   slot {class=SizeBoxSlot, properties {padding=<ltrb -12 0 -12 0>}}
  }
  DeletedListViewErrorMessage {
   widgetClassPath=&SizeBox
   bIsVariable=true
   parent=Switcher_DeletedList
   widget=DeletedListViewErrorMessage
   properties {bOverride_HeightOverride=true, bOverride_MaxDesiredHeight=true, heightOverride=300, maxDesiredHeight=300}
   slot {class=WidgetSwitcherSlot}
  }
  TextBlock {
   widgetClassPath=&TextBlock
   bIsVariable=true
   parent=DeletedListViewErrorMessage
   widget=TextBlock
   properties {
    renderOpacity=0
    text=You have no deleted islands.
    colorAndOpacity {colorUseRule=UseColor_Specified, specifiedColor=<rgba 0.9301 0.0742 0.107 1>}
    font {
     bForceMonospaced=false
     bMaterialIsStencil=false
     fontMaterial=null
     fontObject=&BurbankSmall
     letterSpacing=0
     monospacedWidth=1
     size=24
     skewAmount=0
     typefaceFontName=Black
     outlineSettings {
      bApplyOutlineToDropShadows=false
      bMiteredCorners=false
      bSeparateFillAlpha=false
      outlineMaterial=null
      outlineSize=0
      outlineColor=<rgba 0 0 0 1>
     }
    }
   }
   slot {class=SizeBoxSlot, properties {verticalAlignment=VAlign_Center}}
  }
  VB_EnterCodeSection_1 {
   widgetClassPath=&VerticalBox
   parent=VerticalBox
   widget=VB_EnterCodeSection_1
   properties {
    navigation=&self:WidgetTree.VB_EnterCodeSection_1.WidgetNavigation_0
    visibility=Collapsed
   }
   slot {class=VerticalBoxSlot}
  }
  CommonTextBlock_8 {
   widgetClassPath=&CommonTextBlock
   parent=VB_EnterCodeSection_1
   widget=CommonTextBlock_8
   properties {
    style=&TextStyle-Base-S_C
    text=Enter code
    font {
     bForceMonospaced=false
     bMaterialIsStencil=false
     fontMaterial=null
     fontObject=&NotoSans
     letterSpacing=0
     monospacedWidth=1
     size=18
     skewAmount=0
     typefaceFontName=NotoSans-Regular
     outlineSettings {
      bApplyOutlineToDropShadows=false
      bMiteredCorners=false
      bSeparateFillAlpha=false
      outlineMaterial=null
      outlineSize=0
      outlineColor=<rgba 0 0 0 1>
     }
    }
    shadowOffset=<0 0>
   }
   slot {class=VerticalBoxSlot, properties {verticalAlignment=VAlign_Center}}
  }
  CommonBorder {
   widgetClassPath=&CommonBorder
   parent=VB_EnterCodeSection_1
   widget=CommonBorder
   properties {style=&Border-SolidBG-DkBlue_C, padding=<ltrb 0 0 0 0>}
   slot {class=VerticalBoxSlot, properties {verticalAlignment=VAlign_Center, padding=<ltrb 0 16 0 16>}}
  }
  FriendCodeEntry_1 {
   widgetClassPath=&EditableText
   bIsVariable=true
   parent=CommonBorder
   widget=FriendCodeEntry_1
   properties {
    hintText=Code
    widgetStyle {
     backgroundImageComposing {
      bIsDynamicallyLoaded=false
      drawAs=Border
      imageType=FullColor
      mirroring=NoMirror
      resourceName=../../../Engine/Content/Slate/Old/HyperlinkDotted.png
      resourceObject=null
      tiling=Both
      imageSize=<0 0>
      margin=<ltrb 0 0 0 0.1875>
      outlineSettings {
       bUseBrushTransparency=false
       roundingType=HalfHeightRadius
       width=0
       color {colorUseRule=UseColor_Specified, specifiedColor=<rgba 0 0 0 0>}
       cornerRadii=<0 0 0 0>
      }
      tintColor {colorUseRule=UseColor_Specified, specifiedColor=<rgba 1 1 1 1>}
      uVRegion {bIsValid=false, max=<0 0>, min=<0 0>}
     }
     backgroundImageSelected {
      bIsDynamicallyLoaded=false
      drawAs=Box
      imageType=FullColor
      mirroring=NoMirror
      resourceName=../../../Engine/Content/Slate/Common/EditableTextSelectionBackground.png
      resourceObject=null
      tiling=NoTile
      imageSize=<0 0>
      margin=<ltrb 0.25 0.25 0.25 0.25>
      outlineSettings {
       bUseBrushTransparency=false
       roundingType=HalfHeightRadius
       width=0
       color {colorUseRule=UseColor_Specified, specifiedColor=<rgba 0 0 0 0>}
       cornerRadii=<0 0 0 0>
      }
      tintColor {colorUseRule=UseColor_Specified, specifiedColor=<rgba 1 1 1 1>}
      uVRegion {bIsValid=false, max=<0 0>, min=<0 0>}
     }
     caretImage {
      bIsDynamicallyLoaded=false
      drawAs=Image
      imageType=FullColor
      mirroring=NoMirror
      resourceName=../../../Engine/Content/Slate/Old/White.png
      resourceObject=null
      tiling=NoTile
      imageSize=<16 16>
      margin=<ltrb 0 0 0 0>
      outlineSettings {
       bUseBrushTransparency=false
       roundingType=HalfHeightRadius
       width=0
       color {colorUseRule=UseColor_Specified, specifiedColor=<rgba 0 0 0 0>}
       cornerRadii=<0 0 0 0>
      }
      tintColor {colorUseRule=UseColor_Specified, specifiedColor=<rgba 1 1 1 1>}
      uVRegion {bIsValid=false, max=<0 0>, min=<0 0>}
     }
     colorAndOpacity {colorUseRule=UseColor_Specified, specifiedColor=<rgba 1 1 1 1>}
     font {
      bForceMonospaced=false
      bMaterialIsStencil=false
      fontMaterial=null
      fontObject=&NotoSans
      letterSpacing=0
      monospacedWidth=1
      size=16
      skewAmount=0
      typefaceFontName=NotoSans-Regular
      outlineSettings {
       bApplyOutlineToDropShadows=false
       bMiteredCorners=false
       bSeparateFillAlpha=false
       outlineMaterial=null
       outlineSize=0
       outlineColor=<rgba 0 0 0 1>
      }
     }
     hintColorAndOpacity {colorUseRule=UseColor_Foreground, specifiedColor=<rgba 1 0 1 1>}
     selectedBackgroundColor {colorUseRule=UseColor_Foreground, specifiedColor=<rgba 1 0 1 1>}
    }
   }
   slot {class=BorderSlot, properties {padding=<ltrb 0 0 0 0>}}
  }
  ErrorText_1 {
   widgetClassPath=&CommonTextBlock
   bIsVariable=true
   parent=VB_EnterCodeSection_1
   widget=ErrorText_1
   properties {
    style=&TextStyle-Base-XS-B-S-Red_C
    text=Text Block
    visibility=Hidden
    colorAndOpacity {colorUseRule=UseColor_Specified, specifiedColor=<rgba 1 0.0561 0.0648 1>}
    font {
     bForceMonospaced=false
     bMaterialIsStencil=false
     fontMaterial=null
     fontObject=&NotoSans
     letterSpacing=0
     monospacedWidth=1
     size=16
     skewAmount=0
     typefaceFontName=NotoSans-Bold
     outlineSettings {
      bApplyOutlineToDropShadows=true
      bMiteredCorners=false
      bSeparateFillAlpha=true
      outlineMaterial=&M-FontShadow
      outlineSize=2
      outlineColor=<rgba 0.085 0.0146 0.0177 0.5>
     }
    }
    shadowOffset=<0 0>
   }
   slot {class=VerticalBoxSlot}
  }
  HorizontalBox_4 {
   widgetClassPath=&HorizontalBox
   parent=VB_EnterCodeSection_1
   widget=HorizontalBox_4
   slot {class=VerticalBoxSlot, properties {horizontalAlignment=HAlign_Left, verticalAlignment=VAlign_Center, padding=<ltrb 10 10 20 10>}}
  }
  IconTextButton_1 {
   widgetClassPath=&IconTextButton
   bIsVariable=true
   parent=HorizontalBox_4
   widget=IconTextButton_1
   properties {
    buttonText=Use Code
    onButtonBaseClicked=(&self:WidgetTree.IconTextButton_1.HandleOnButtonClicked)
    onButtonBaseDoubleClicked=(&self:WidgetTree.IconTextButton_1.HandleOnButtonDoubleClicked)
    onButtonBaseHovered=(&self:WidgetTree.IconTextButton_1.HandleOnButtonHovered)
    onButtonBaseUnhovered=(&self:WidgetTree.IconTextButton_1.HandleOnButtonUnhovered)
    onSelectedChangedBase=(&self:WidgetTree.IconTextButton_1.HandleOnSelectedChanged)
   }
   slot {class=HorizontalBoxSlot, properties {horizontalAlignment=HAlign_Left}}
  }
  Overlay_3 {
   widgetClassPath=&Overlay
   parent=VerticalBox_0
   widget=Overlay_3
   slot {class=VerticalBoxSlot, properties {verticalAlignment=VAlign_Bottom}}
  }
  WBP_CaptureForPostBufferUpdate_109 {
   widgetClassPath=&WBP_CaptureForPostBufferUpdate
   bIsVariable=true
   parent=Overlay_3
   widget=WBP_CaptureForPostBufferUpdate_109
   slot {class=OverlaySlot, properties {horizontalAlignment=HAlign_Fill, verticalAlignment=VAlign_Fill}}
  }
  Image_515 {
   widgetClassPath=&Image
   bIsVariable=true
   parent=Overlay_3
   widget=Image_515
   properties {
    visibility=Collapsed
    brush {
     bIsDynamicallyLoaded=false
     drawAs=Image
     imageType=NoImage
     mirroring=NoMirror
     resourceName=None
     resourceObject=/Game/Athena/UI/InGame/Creative/Art/M_UI_BoltGradientBorder_Creative_IslandSelectFooter
     tiling=NoTile
     imageSize=<32 64>
     margin=<ltrb 0 0 0 0>
     outlineSettings {
      bUseBrushTransparency=false
      roundingType=HalfHeightRadius
      width=0
      color {colorUseRule=UseColor_Specified, specifiedColor=<rgba 0 0 0 0>}
      cornerRadii=<0 0 0 0>
     }
     tintColor {colorUseRule=UseColor_Specified, specifiedColor=<rgba 1 1 1 1>}
     uVRegion {bIsValid=false, max=<0 0>, min=<0 0>}
    }
    colorAndOpacity=<rgba 0.0052 0.0578 0.3916 0.5>
   }
   slot {class=OverlaySlot, properties {horizontalAlignment=HAlign_Fill, verticalAlignment=VAlign_Bottom}}
  }
  HorizontalBox_1 {
   widgetClassPath=&HorizontalBox
   parent=Overlay_3
   widget=HorizontalBox_1
   slot {class=OverlaySlot, properties {horizontalAlignment=HAlign_Fill, verticalAlignment=VAlign_Bottom, padding=<ltrb 40 10 40 32>}}
  }
  HorizontalBox_0 {
   widgetClassPath=&HorizontalBox
   parent=HorizontalBox_1
   widget=HorizontalBox_0
   properties {visibility=Collapsed}
   slot {class=HorizontalBoxSlot, properties {verticalAlignment=VAlign_Bottom}}
  }
  RemoveIslandButton {
   widgetClassPath=&IconTextButton
   bIsVariable=true
   parent=HorizontalBox_0
   widget=RemoveIslandButton
   properties {
    bIsFocusable=false
    buttonText=Remove
    minWidth=256
    onButtonBaseClicked=(&self:WidgetTree.RemoveIslandButton.HandleOnButtonClicked)
    onButtonBaseDoubleClicked=(&self:WidgetTree.RemoveIslandButton.HandleOnButtonDoubleClicked)
    onButtonBaseHovered=(&self:WidgetTree.RemoveIslandButton.HandleOnButtonHovered)
    onButtonBaseUnhovered=(&self:WidgetTree.RemoveIslandButton.HandleOnButtonUnhovered)
    onSelectedChangedBase=(&self:WidgetTree.RemoveIslandButton.HandleOnSelectedChanged)
    style=/FortUILibrary/Foundation/Buttons/ButtonStyles/Athena/Athena_ButtonStyle_AngledBlueMenuCreativeButton.Athena_ButtonStyle_AngledBlueMenuCreativeButton_C
    visibility=Collapsed
    font {
     bForceMonospaced=false
     bMaterialIsStencil=false
     fontMaterial=null
     fontObject=&HeadingNow
     letterSpacing=0
     monospacedWidth=1
     size=16.5
     skewAmount=0
     typefaceFontName=75 Medium
     outlineSettings {
      bApplyOutlineToDropShadows=false
      bMiteredCorners=false
      bSeparateFillAlpha=false
      outlineMaterial=null
      outlineSize=0
      outlineColor=<rgba 0 0 0 1>
     }
    }
    triggeringInputAction {dataTable=&InputActionDataTable, rowName=GenericRightStickClick}
   }
   slot {class=HorizontalBoxSlot}
  }
  Spacer {widgetClassPath=&Spacer, parent=HorizontalBox_0, widget=Spacer, properties {size=<10 1>}, slot {class=HorizontalBoxSlot}}
  DuplicateIslandButton {
   widgetClassPath=&IconTextButton
   bIsVariable=true
   parent=HorizontalBox_0
   widget=DuplicateIslandButton
   properties {
    bIsFocusable=false
    buttonText=Duplicate
    minWidth=256
    onButtonBaseClicked=(&self:WidgetTree.DuplicateIslandButton.HandleOnButtonClicked)
    onButtonBaseDoubleClicked=(&self:WidgetTree.DuplicateIslandButton.HandleOnButtonDoubleClicked)
    onButtonBaseHovered=(&self:WidgetTree.DuplicateIslandButton.HandleOnButtonHovered)
    onButtonBaseUnhovered=(&self:WidgetTree.DuplicateIslandButton.HandleOnButtonUnhovered)
    onSelectedChangedBase=(&self:WidgetTree.DuplicateIslandButton.HandleOnSelectedChanged)
    style=&Athena_ButtonStyle_AngledBlueMenuCreativeButton_C
    visibility=Collapsed
    font {
     bForceMonospaced=false
     bMaterialIsStencil=false
     fontMaterial=null
     fontObject=&HeadingNow
     letterSpacing=0
     monospacedWidth=1
     size=16.5
     skewAmount=0
     typefaceFontName=75 Medium
     outlineSettings {
      bApplyOutlineToDropShadows=false
      bMiteredCorners=false
      bSeparateFillAlpha=false
      outlineMaterial=null
      outlineSize=0
      outlineColor=<rgba 0 0 0 1>
     }
    }
    triggeringInputAction {dataTable=&InputActionDataTable, rowName=GenericFaceButtonTop}
   }
   slot {class=HorizontalBoxSlot}
  }
  Spacer_2 {widgetClassPath=&Spacer, parent=HorizontalBox_0, widget=Spacer_2, properties {size=<10 1>}, slot {class=HorizontalBoxSlot}}
  RenameIslandButton {
   widgetClassPath=&IconTextButton
   bIsVariable=true
   parent=HorizontalBox_0
   widget=RenameIslandButton
   properties {
    bIsFocusable=false
    buttonText=Rename
    minWidth=256
    onButtonBaseClicked=(&self:WidgetTree.RenameIslandButton.HandleOnButtonClicked)
    onButtonBaseDoubleClicked=(&self:WidgetTree.RenameIslandButton.HandleOnButtonDoubleClicked)
    onButtonBaseHovered=(&self:WidgetTree.RenameIslandButton.HandleOnButtonHovered)
    onButtonBaseUnhovered=(&self:WidgetTree.RenameIslandButton.HandleOnButtonUnhovered)
    onSelectedChangedBase=(&self:WidgetTree.RenameIslandButton.HandleOnSelectedChanged)
    style=&Athena_ButtonStyle_AngledBlueMenuCreativeButton_C
    visibility=Collapsed
    font {
     bForceMonospaced=false
     bMaterialIsStencil=false
     fontMaterial=null
     fontObject=&HeadingNow
     letterSpacing=0
     monospacedWidth=1
     size=16.5
     skewAmount=0
     typefaceFontName=75 Medium
     outlineSettings {
      bApplyOutlineToDropShadows=false
      bMiteredCorners=false
      bSeparateFillAlpha=false
      outlineMaterial=null
      outlineSize=0
      outlineColor=<rgba 0 0 0 1>
     }
    }
    triggeringInputAction {dataTable=&InputActionDataTable, rowName=GenericFaceButtonLeft}
   }
   slot {class=HorizontalBoxSlot}
  }
  Remove_Button {
   widgetClassPath=&WBP_UIKit_Button_Regular
   bIsVariable=true
   parent=HorizontalBox_1
   widget=Remove_Button
   properties {
    buttonSize=Small
    minHeight=56
    text=Remove
    visibility=Hidden
    triggeringInputAction {dataTable=&InputActionDataTable, rowName=GenericRightStickClick}
   }
   slot {class=HorizontalBoxSlot, properties {verticalAlignment=VAlign_Bottom}}
  }
  Duplicate_Button {
   widgetClassPath=&WBP_UIKit_Button_Regular
   bIsVariable=true
   parent=HorizontalBox_1
   widget=Duplicate_Button
   properties {
    buttonSize=Small
    text=Duplicate
    visibility=Hidden
    triggeringInputAction {dataTable=&InputActionDataTable, rowName=GenericFaceButtonTop}
   }
   slot {class=HorizontalBoxSlot, properties {verticalAlignment=VAlign_Bottom, padding=<ltrb 24 0 0 0>}}
  }
  Rename_Button {
   widgetClassPath=&WBP_UIKit_Button_Regular
   bIsVariable=true
   parent=HorizontalBox_1
   widget=Rename_Button
   properties {
    buttonSize=Small
    text=Rename
    visibility=Hidden
    triggeringInputAction {dataTable=&InputActionDataTable, rowName=GenericFaceButtonLeft}
   }
   slot {class=HorizontalBoxSlot, properties {verticalAlignment=VAlign_Bottom, padding=<ltrb 24 0 24 0>}}
  }
  RemoveSpinner {
   widgetClassPath=&Image
   bIsVariable=true
   parent=HorizontalBox_1
   widget=RemoveSpinner
   properties {
    visibility=Collapsed
    brush {
     bIsDynamicallyLoaded=false
     drawAs=Image
     imageType=NoImage
     mirroring=NoMirror
     resourceName=None
     resourceObject=&MI_UIKit_Monochrome_Throbber
     tiling=NoTile
     imageSize=<48 48>
     margin=<ltrb 0 0 0 0>
     outlineSettings {
      bUseBrushTransparency=false
      roundingType=HalfHeightRadius
      width=0
      color {colorUseRule=UseColor_Specified, specifiedColor=<rgba 0 0 0 0>}
      cornerRadii=<0 0 0 0>
     }
     tintColor {colorUseRule=UseColor_Specified, specifiedColor=<rgba 1 1 1 1>}
     uVRegion {bIsValid=false, max=<0 0>, min=<0 0>}
    }
   }
   slot {class=HorizontalBoxSlot, properties {horizontalAlignment=HAlign_Center, verticalAlignment=VAlign_Bottom}}
  }
  DuplicateIslandButtonSpinner {
   widgetClassPath=&Image
   bIsVariable=true
   parent=HorizontalBox_1
   widget=DuplicateIslandButtonSpinner
   properties {
    visibility=Collapsed
    brush {
     bIsDynamicallyLoaded=false
     drawAs=Image
     imageType=NoImage
     mirroring=NoMirror
     resourceName=None
     resourceObject=&MI_UIKit_Monochrome_Throbber
     tiling=NoTile
     imageSize=<48 48>
     margin=<ltrb 0 0 0 0>
     outlineSettings {
      bUseBrushTransparency=false
      roundingType=HalfHeightRadius
      width=0
      color {colorUseRule=UseColor_Specified, specifiedColor=<rgba 0 0 0 0>}
      cornerRadii=<0 0 0 0>
     }
     tintColor {colorUseRule=UseColor_Specified, specifiedColor=<rgba 1 1 1 1>}
     uVRegion {bIsValid=false, max=<0 0>, min=<0 0>}
    }
   }
   slot {class=HorizontalBoxSlot, properties {horizontalAlignment=HAlign_Center, verticalAlignment=VAlign_Bottom}}
  }
  RenameIslandButtonSpinner {
   widgetClassPath=&Image
   bIsVariable=true
   parent=HorizontalBox_1
   widget=RenameIslandButtonSpinner
   properties {
    visibility=Collapsed
    brush {
     bIsDynamicallyLoaded=false
     drawAs=Image
     imageType=NoImage
     mirroring=NoMirror
     resourceName=None
     resourceObject=&MI_UIKit_Monochrome_Throbber
     tiling=NoTile
     imageSize=<48 48>
     margin=<ltrb 0 0 0 0>
     outlineSettings {
      bUseBrushTransparency=false
      roundingType=HalfHeightRadius
      width=0
      color {colorUseRule=UseColor_Specified, specifiedColor=<rgba 0 0 0 0>}
      cornerRadii=<0 0 0 0>
     }
     tintColor {colorUseRule=UseColor_Specified, specifiedColor=<rgba 1 1 1 1>}
     uVRegion {bIsValid=false, max=<0 0>, min=<0 0>}
    }
   }
   slot {class=HorizontalBoxSlot, properties {horizontalAlignment=HAlign_Center, verticalAlignment=VAlign_Bottom}}
  }
  VisibilityBorder_NonTouch {
   widgetClassPath=&CommonHardwareVisibilityBorder
   parent=HorizontalBox_1
   widget=VisibilityBorder_NonTouch
   properties {visibilityQuery {
     autoDescription=" NONE( Input.Touch )"
     tokenStreamVersion=0
     userDescription=""
     queryTokenStream [0,1,3,1,0]
     tagDictionary [{tagName=Input.Touch}]
    }}
   slot {class=HorizontalBoxSlot, properties {horizontalAlignment=HAlign_Right, size {sizeRule=Fill, value=1}}}
  }
  CloseMainPanel {
   widgetClassPath=&IconTextButton
   bIsVariable=true
   parent=VisibilityBorder_NonTouch
   widget=CloseMainPanel
   properties {
    buttonText=CLOSE
    onButtonBaseClicked=(&self:WidgetTree.CloseMainPanel.HandleOnButtonClicked)
    onButtonBaseDoubleClicked=(&self:WidgetTree.CloseMainPanel.HandleOnButtonDoubleClicked)
    onButtonBaseHovered=(&self:WidgetTree.CloseMainPanel.HandleOnButtonHovered)
    onButtonBaseUnhovered=(&self:WidgetTree.CloseMainPanel.HandleOnButtonUnhovered)
    onSelectedChangedBase=(&self:WidgetTree.CloseMainPanel.HandleOnSelectedChanged)
    style=/FortUILibrary/Foundation/Buttons/ButtonStyles/Skew/ButtonStyle-Skew_LessTransparent.ButtonStyle-Skew_LessTransparent_C
    visibility=Collapsed
    triggeringInputAction {dataTable=/FortUILibrary/Input/CreativeInputActionDataTable, rowName=Close}
   }
   slot {class=BorderSlot, properties {padding=<ltrb 0 0 0 0>}}
  }
  SelectIslandSpinner {
   widgetClassPath=&Image
   bIsVariable=true
   parent=HorizontalBox_1
   widget=SelectIslandSpinner
   properties {
    visibility=Collapsed
    brush {
     bIsDynamicallyLoaded=false
     drawAs=Image
     imageType=NoImage
     mirroring=NoMirror
     resourceName=None
     resourceObject=&MI_UIKit_Monochrome_Throbber
     tiling=NoTile
     imageSize=<48 48>
     margin=<ltrb 0 0 0 0>
     outlineSettings {
      bUseBrushTransparency=false
      roundingType=HalfHeightRadius
      width=0
      color {colorUseRule=UseColor_Specified, specifiedColor=<rgba 0 0 0 0>}
      cornerRadii=<0 0 0 0>
     }
     tintColor {colorUseRule=UseColor_Specified, specifiedColor=<rgba 1 1 1 1>}
     uVRegion {bIsValid=false, max=<0 0>, min=<0 0>}
    }
   }
   slot {class=HorizontalBoxSlot, properties {horizontalAlignment=HAlign_Center, verticalAlignment=VAlign_Center, padding=<ltrb 0 0 12 0>}}
  }
  Overlay_5 {
   widgetClassPath=&Overlay
   parent=HorizontalBox_1
   widget=Overlay_5
   properties {visibility=Collapsed}
   slot {class=HorizontalBoxSlot, properties {horizontalAlignment=HAlign_Right, verticalAlignment=VAlign_Bottom, size {sizeRule=Fill, value=1}}}
  }
  SelectIsland {
   widgetClassPath=&IconTextButton
   bIsVariable=true
   parent=Overlay_5
   widget=SelectIsland
   properties {
    bIsFocusable=false
    buttonText=SET PORTAL
    minHeight=72
    minWidth=300
    onButtonBaseClicked=(&self:WidgetTree.SelectIsland.HandleOnButtonClicked)
    onButtonBaseDoubleClicked=(&self:WidgetTree.SelectIsland.HandleOnButtonDoubleClicked)
    onButtonBaseHovered=(&self:WidgetTree.SelectIsland.HandleOnButtonHovered)
    onButtonBaseUnhovered=(&self:WidgetTree.SelectIsland.HandleOnButtonUnhovered)
    onSelectedChangedBase=(&self:WidgetTree.SelectIsland.HandleOnSelectedChanged)
    style=/FortUILibrary/Foundation/Buttons/ButtonStyles/ButtonStyle_Feature_M_Yellow.ButtonStyle_Feature_M_Yellow_C
    visibility=Hidden
    font {
     bForceMonospaced=false
     bMaterialIsStencil=false
     fontMaterial=null
     fontObject=&HeadingNow
     letterSpacing=0
     monospacedWidth=1
     size=21
     skewAmount=0
     typefaceFontName=76 Bold
     outlineSettings {
      bApplyOutlineToDropShadows=false
      bMiteredCorners=false
      bSeparateFillAlpha=false
      outlineMaterial=null
      outlineSize=0
      outlineColor=<rgba 0 0 0 1>
     }
    }
    triggeringInputAction {dataTable=&InputActionDataTable, rowName=GenericForward}
   }
   slot {class=OverlaySlot}
  }
  SetPortal_Button {
   widgetClassPath=/FortUILibrary/UIKit/Buttons/WBP_UIKit_Button_Loud
   bIsVariable=true
   parent=HorizontalBox_1
   widget=SetPortal_Button
   properties {
    buttonSize=Large
    minWidth=312
    text=Set Portal
    visibility=Hidden
    triggeringInputAction {dataTable=&InputActionDataTable, rowName=GenericForward}
   }
   slot {class=HorizontalBoxSlot, properties {verticalAlignment=VAlign_Center}}
  }
  VisibilityBorder_Touch {
   widgetClassPath=&CommonHardwareVisibilityBorder
   parent=Overlay_1
   widget=VisibilityBorder_Touch
   properties {visibilityQuery {
     autoDescription=" ANY( Input.Touch )"
     tokenStreamVersion=0
     userDescription=""
     queryTokenStream [0,1,1,1,0]
     tagDictionary [{tagName=Input.Touch}]
    }}
   slot {class=OverlaySlot, properties {horizontalAlignment=HAlign_Right}}
  }
  Button_CloseTouch {
   widgetClassPath=/FortUILibrary/Foundation/Buttons/CloseButton
   bIsVariable=true
   parent=VisibilityBorder_Touch
   widget=Button_CloseTouch
   properties {
    onButtonBaseClicked=(&self:WidgetTree.Button_CloseTouch.HandleOnButtonClicked)
    onButtonBaseDoubleClicked=(&self:WidgetTree.Button_CloseTouch.HandleOnButtonDoubleClicked)
    onButtonBaseHovered=(&self:WidgetTree.Button_CloseTouch.HandleOnButtonHovered)
    onButtonBaseUnhovered=(&self:WidgetTree.Button_CloseTouch.HandleOnButtonUnhovered)
    onSelectedChangedBase=(&self:WidgetTree.Button_CloseTouch.HandleOnSelectedChanged)
    renderTransform {angle=0, scale=<0.7 0.7>, shear=<0 0>, translation=<0 0>}
   }
   slot {class=BorderSlot, properties {padding=<ltrb 0 0 0 0>}}
  }
 ]
}
