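# PortalInfoPlate is the nameplate widget for a Creative island portal, showing owner, island name, population, link code, description and download status.
# It binds to BGA_IslandPortal events on init, switching between player, featured and published-code layouts.
# Distance thresholds and HUD-interact availability gate visibility and the description reveal, with animations adding an outline-material sheen to names.
# Interaction plays a shrink-and-fade animation, then a timer collapses the plate; ending interaction reverses it.
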WidgetBlueprint /Game/Playgrounds/Items/Portals/UI/PortalInfoPlate
animations [
 Intro {
  binding_count=3
  length_seconds=0.75
  object=&self:Intro
  sequence {bindings [
    Text_PlayerName {
     id=#1
     possessed_class=CommonTextBlock
     tracks [
      {
       type=MovieScene2DTransformTrack
       sections [{
         type=MovieScene2DTransformSection
         end_frame=5
         start_frame=0
         channels [
          Translation.X {type=MovieSceneScriptingFloatChannel}
          Translation.Y {type=MovieSceneScriptingFloatChannel}
          Angle {type=MovieSceneScriptingFloatChannel}
          Scale.X {type=MovieSceneScriptingFloatChannel, keys [[0,0.7056],[2,1.1],[3,0.95],[5,1]]}
          Scale.Y {type=MovieSceneScriptingFloatChannel, keys [[0,0.7056],[2,1.1],[3,0.95],[5,1]]}
          Shear.X {type=MovieSceneScriptingFloatChannel}
          Shear.Y {type=MovieSceneScriptingFloatChannel}
         ]
        }]
      }
      {
       type=MovieSceneFloatTrack
       display_name=Render Opacity
       sections [{type=MovieSceneFloatSection, end_frame=5, start_frame=0, channels [None {type=MovieSceneScriptingFloatChannel, keys [[0,0],[2,1],[5,1]]}]}]
      }
     ]
    }
    PopulationBox {
     id=#2
     possessed_class=HorizontalBox
     tracks [
      {
       type=MovieScene2DTransformTrack
       sections [{
         type=MovieScene2DTransformSection
         end_frame=15
         start_frame=8
         channels [
          Translation.X {type=MovieSceneScriptingFloatChannel, const [0,4]}
          Translation.Y {type=MovieSceneScriptingFloatChannel, keys [[8,-9],[11,3.6452],[13,-1.5518],[15,0]]}
          Angle {type=MovieSceneScriptingFloatChannel}
          Scale.X {type=MovieSceneScriptingFloatChannel}
          Scale.Y {type=MovieSceneScriptingFloatChannel}
          Shear.X {type=MovieSceneScriptingFloatChannel}
          Shear.Y {type=MovieSceneScriptingFloatChannel}
         ]
        }]
      }
      {
       type=MovieSceneFloatTrack
       display_name=Render Opacity
       sections [{type=MovieSceneFloatSection, end_frame=10, start_frame=0, channels [None {type=MovieSceneScriptingFloatChannel, keys [[0,0],[8,0],[10,1]]}]}]
      }
     ]
    }
    Text_IslandName {
     id=#3
     possessed_class=CommonTextBlock
     tracks [
      {
       type=MovieScene2DTransformTrack
       sections [{
         type=MovieScene2DTransformSection
         end_frame=15
         start_frame=0
         channels [
          Translation.X {type=MovieSceneScriptingFloatChannel, const [0,6]}
          Translation.Y {type=MovieSceneScriptingFloatChannel, keys [[0,0],[5,-9],[8,1.4508],[10,-0.9672],[12,0],[15,0]]}
          Angle {type=MovieSceneScriptingFloatChannel}
          Scale.X {type=MovieSceneScriptingFloatChannel}
          Scale.Y {type=MovieSceneScriptingFloatChannel}
          Shear.X {type=MovieSceneScriptingFloatChannel}
          Shear.Y {type=MovieSceneScriptingFloatChannel}
         ]
        }]
      }
      {
       type=MovieSceneFloatTrack
       display_name=Render Opacity
       sections [{
         type=MovieSceneFloatSection
         end_frame=15
         start_frame=0
         channels [None {type=MovieSceneScriptingFloatChannel, keys [[0,0],[5,0],[7,1],[15,1]]}]
        }]
      }
     ]
    }
   ]}
 }
 intro2 {
  binding_count=3
  length_seconds=0.15
  object=&self:intro2
  sequence {bindings [
    Text_PlayerName {
     id=#4
     possessed_class=CommonTextBlock
     tracks [
      {
       type=MovieScene2DTransformTrack
       sections [{
         type=MovieScene2DTransformSection
         end_frame=3
         start_frame=0
         channels [
          Translation.X {type=MovieSceneScriptingFloatChannel, keys [[0,-48],[3,0]]}
          Translation.Y {type=MovieSceneScriptingFloatChannel, const [0,2]}
          Angle {type=MovieSceneScriptingFloatChannel}
          Scale.X {type=MovieSceneScriptingFloatChannel}
          Scale.Y {type=MovieSceneScriptingFloatChannel}
          Shear.X {type=MovieSceneScriptingFloatChannel}
          Shear.Y {type=MovieSceneScriptingFloatChannel}
         ]
        }]
      }
      {
       type=MovieSceneFloatTrack
       display_name=Render Opacity
       sections [{type=MovieSceneFloatSection, end_frame=3, start_frame=0, channels [None {type=MovieSceneScriptingFloatChannel, keys [[0,0],[3,1]]}]}]
      }
     ]
    }
    PopulationBox {
     id=#5
     possessed_class=HorizontalBox
     tracks [
      {
       type=MovieScene2DTransformTrack
       sections [{
         type=MovieScene2DTransformSection
         end_frame=3
         start_frame=0
         channels [
          Translation.X {type=MovieSceneScriptingFloatChannel, keys [[0,-50],[3,0]]}
          Translation.Y {type=MovieSceneScriptingFloatChannel, const [0,2]}
          Angle {type=MovieSceneScriptingFloatChannel}
          Scale.X {type=MovieSceneScriptingFloatChannel}
          Scale.Y {type=MovieSceneScriptingFloatChannel}
          Shear.X {type=MovieSceneScriptingFloatChannel}
          Shear.Y {type=MovieSceneScriptingFloatChannel}
         ]
        }]
      }
      {
       type=MovieSceneFloatTrack
       display_name=Render Opacity
       sections [{type=MovieSceneFloatSection, end_frame=3, start_frame=0, channels [None {type=MovieSceneScriptingFloatChannel, keys [[0,0],[3,1]]}]}]
      }
     ]
    }
    Text_IslandName {
     id=#6
     possessed_class=CommonTextBlock
     tracks [
      {
       type=MovieScene2DTransformTrack
       sections [{
         type=MovieScene2DTransformSection
         end_frame=3
         start_frame=0
         channels [
          Translation.X {type=MovieSceneScriptingFloatChannel, keys [[0,50],[3,0]]}
          Translation.Y {type=MovieSceneScriptingFloatChannel, const [0,2]}
          Angle {type=MovieSceneScriptingFloatChannel}
          Scale.X {type=MovieSceneScriptingFloatChannel}
          Scale.Y {type=MovieSceneScriptingFloatChannel}
          Shear.X {type=MovieSceneScriptingFloatChannel}
          Shear.Y {type=MovieSceneScriptingFloatChannel}
         ]
        }]
      }
      {
       type=MovieSceneFloatTrack
       display_name=Render Opacity
       sections [{type=MovieSceneFloatSection, end_frame=3, start_frame=0, channels [None {type=MovieSceneScriptingFloatChannel, keys [[0,0],[3,1]]}]}]
      }
     ]
    }
   ]}
 }
 DescriptionReveal {
  binding_count=2
  length_seconds=0.25
  object=&self:DescriptionReveal
  sequence {bindings [
    CommonTextBlock_1 {
     id=#7
     possessed_class=CommonTextBlock
     tracks [
      {
       type=MovieScene2DTransformTrack
       sections [{
         type=MovieScene2DTransformSection
         end_frame=5
         start_frame=0
         channels [
          Translation.X {type=MovieSceneScriptingFloatChannel}
          Translation.Y {type=MovieSceneScriptingFloatChannel}
          Angle {type=MovieSceneScriptingFloatChannel}
          Scale.X {type=MovieSceneScriptingFloatChannel, keys [[0,1.15],[4,1],[5,1]]}
          Scale.Y {type=MovieSceneScriptingFloatChannel, keys [[0,1.15],[4,1],[5,1]]}
          Shear.X {type=MovieSceneScriptingFloatChannel}
          Shear.Y {type=MovieSceneScriptingFloatChannel}
         ]
        }]
      }
      {
       type=MovieSceneFloatTrack
       display_name=Render Opacity
       sections [{type=MovieSceneFloatSection, end_frame=2, start_frame=0, channels [None {type=MovieSceneScriptingFloatChannel, keys [[0,0],[2,1]]}]}]
      }
     ]
    }
    Text_Description {
     id=#8
     possessed_class=CommonTextBlock
     tracks [
      {
       type=MovieScene2DTransformTrack
       sections [{
         type=MovieScene2DTransformSection
         end_frame=5
         start_frame=0
         channels [
          Translation.X {type=MovieSceneScriptingFloatChannel}
          Translation.Y {type=MovieSceneScriptingFloatChannel}
          Angle {type=MovieSceneScriptingFloatChannel}
          Scale.X {type=MovieSceneScriptingFloatChannel, keys [[0,1.2424],[5,1]]}
          Scale.Y {type=MovieSceneScriptingFloatChannel, keys [[0,1.2424],[5,1]]}
          Shear.X {type=MovieSceneScriptingFloatChannel}
          Shear.Y {type=MovieSceneScriptingFloatChannel}
         ]
        }]
      }
      {
       type=MovieSceneFloatTrack
       display_name=Render Opacity
       sections [{type=MovieSceneFloatSection, end_frame=5, start_frame=0, channels [None {type=MovieSceneScriptingFloatChannel, keys [[0,0],[3,1],[5,1]]}]}]
      }
     ]
    }
   ]}
 }
 VisibilityReveal {
  binding_count=2
  length_seconds=0.45
  object=&self:VisibilityReveal
  sequence {bindings [
    Text_PlayerName {
     id=#9
     possessed_class=CommonTextBlock
     tracks [
      {
       type=MovieSceneWidgetMaterialTrack
       display_name=Font.Outline Settings.OutlineMaterial
       sections [{type=MovieSceneParameterSection, end_frame=9, start_frame=0, channels [Position {type=MovieSceneScriptingFloatChannel, keys [[0,0],[9,1]]}]}]
      }
      {
       type=MovieSceneFloatTrack
       display_name=Render Opacity
       sections [{
         type=MovieSceneFloatSection
         end_frame=9
         start_frame=0
         channels [None {type=MovieSceneScriptingFloatChannel, keys [[0,0],[5,1],[9.0003,1]]}]
        }]
      }
      {
       type=MovieScene2DTransformTrack
       sections [{
         type=MovieScene2DTransformSection
         end_frame=9
         start_frame=0
         channels [
          Translation.X {type=MovieSceneScriptingFloatChannel}
          Translation.Y {type=MovieSceneScriptingFloatChannel}
          Angle {type=MovieSceneScriptingFloatChannel}
          Scale.X {type=MovieSceneScriptingFloatChannel, keys [[0,0.6841],[5,1.1],[7,0.9646],[9,1]]}
          Scale.Y {type=MovieSceneScriptingFloatChannel, const [1,4]}
          Shear.X {type=MovieSceneScriptingFloatChannel}
          Shear.Y {type=MovieSceneScriptingFloatChannel}
         ]
        }]
      }
     ]
    }
    Text_FeaturedPlayerName {
     id=#10
     possessed_class=CommonTextBlock
     tracks [
      {
       type=MovieSceneWidgetMaterialTrack
       display_name=Font.Outline Settings.OutlineMaterial
       sections [{
         type=MovieSceneParameterSection
         end_frame=9
         start_frame=0
         channels [Position {type=MovieSceneScriptingFloatChannel, keys [[0,0],[9.0003,1]]}]
        }]
      }
      {
       type=MovieScene2DTransformTrack
       sections [{
         type=MovieScene2DTransformSection
         end_frame=9
         start_frame=0
         channels [
          Translation.X {type=MovieSceneScriptingFloatChannel}
          Translation.Y {type=MovieSceneScriptingFloatChannel}
          Angle {type=MovieSceneScriptingFloatChannel}
          Scale.X {type=MovieSceneScriptingFloatChannel, keys [[0,0.7],[5,1.1],[7,0.95],[9,1]]}
          Scale.Y {type=MovieSceneScriptingFloatChannel, const [1,4]}
          Shear.X {type=MovieSceneScriptingFloatChannel}
          Shear.Y {type=MovieSceneScriptingFloatChannel}
         ]
        }]
      }
      {
       type=MovieSceneFloatTrack
       display_name=Render Opacity
       sections [{type=MovieSceneFloatSection, end_frame=9, start_frame=0, channels [None {type=MovieSceneScriptingFloatChannel, keys [[0,0],[5,1],[9,1]]}]}]
      }
     ]
    }
   ]}
 }
 SheenNames {
  binding_count=2
  length_seconds=0.45
  object=&self:SheenNames
  sequence {bindings [
    Text_FeaturedPlayerName {
     id=#11
     possessed_class=CommonTextBlock
     tracks [{
       type=MovieSceneWidgetMaterialTrack
       display_name=Font.Outline Settings.OutlineMaterial
       sections [{type=MovieSceneParameterSection, end_frame=9, start_frame=0, channels [Position {type=MovieSceneScriptingFloatChannel, keys [[0,0],[9,1]]}]}]
      }]
    }
    Text_PlayerName {
     id=#12
     possessed_class=CommonTextBlock
     tracks [{
       type=MovieSceneWidgetMaterialTrack
       display_name=Font.Outline Settings.OutlineMaterial
       sections [{type=MovieSceneParameterSection, end_frame=9, start_frame=0, channels [Position {type=MovieSceneScriptingFloatChannel, keys [[0,0],[9,1]]}]}]
      }]
    }
   ]}
 }
 OnInteract {
  binding_count=1
  length_seconds=0.25
  object=&self:OnInteract
  sequence {bindings [DistanceSwitcher {
     id=#13
     possessed_class=WidgetSwitcher
     tracks [
      {
       type=MovieScene2DTransformTrack
       sections [{
         type=MovieScene2DTransformSection
         end_frame=5
         start_frame=0
         channels [
          Translation.X {type=MovieSceneScriptingFloatChannel}
          Translation.Y {type=MovieSceneScriptingFloatChannel}
          Angle {type=MovieSceneScriptingFloatChannel}
          Scale.X {type=MovieSceneScriptingFloatChannel, keys [[0,1],[5,0]]}
          Scale.Y {type=MovieSceneScriptingFloatChannel, keys [[0,1],[5,0]]}
          Shear.X {type=MovieSceneScriptingFloatChannel}
          Shear.Y {type=MovieSceneScriptingFloatChannel}
         ]
        }]
      }
      {
       type=MovieSceneFloatTrack
       display_name=Render Opacity
       sections [{type=MovieSceneFloatSection, end_frame=5, start_frame=0, channels [None {type=MovieSceneScriptingFloatChannel, keys [[0,1],[5,0]]}]}]
      }
     ]
    }]}
 }
]
graphs [
 ClearDownloadStatus {nodes [
   ClearDownloadStatus() {is=title K2Node_FunctionEntry}
   > n1 = For Each Loop(Array:Array of Common Text Object References=Get DownloadInfoArray() -> (DownloadInfoArray:Array of Common Text Object References)) -> (Array Element:Common Text Object Reference)
   Set StatusLine(execute=@n1.Completed, StatusLine:int=0) -> (Output_Get:int=0)
   "SetText (Text)"(execute=@n1.LoopBody, self=@n1)
  ]}
 EventGraph {nodes [
   PlayIntro() {is=title K2Node_CustomEvent}
   > PlayAnimationForward(InAnimation=Get intro2() -> (intro2:Widget Animation Object Reference))
   n1 = OnUpdateNameplateVis!() -> (OutputDelegate, then, Show:bool=~, WasHidden:bool=~, Distance:double=~) {is=title K2Node_CustomEvent}
   > UpdatePortalInfoPlateVisibility!(execute, self, ShouldShow:bool=@n1.Show, Distance:double=@n1.Distance) -> (then)
   n2 = OnNewVisibilityThresholdChanged!() -> (OutputDelegate, then, bNewVis:bool=~) {is=title K2Node_CustomEvent}
   > Branch(Condition=@n2.bNewVis)
   > PlayAnimation(InAnimation=Get DescriptionReveal() -> (DescriptionReveal:Widget Animation Object Reference))
   > PlayAnimation(InAnimation=Get SheenNames() -> (SheenNames:Widget Animation Object Reference))
   n3 = OnInteractedWith!() -> (OutputDelegate, then, IsInteracting:bool=~) {is=title K2Node_CustomEvent}
   > Branch(Condition="Not Equal (Boolean)"!(A:bool=@n3.IsInteracting, B:bool=Get isInteractWith() -> (isInteractWith:bool)) -> (ReturnValue:bool) {is=title K2Node_CallFunction})
   > Set isInteractWith(isInteractWith:bool=@n3.IsInteracting) -> (Output_Get:bool)
   > Clear and Invalidate Timer by Handle(Handle=Get InteractionAnimTimerHandle() -> (InteractionAnimTimerHandle:Timer Handle Structure))
   > n4 = Branch(Condition=@n3.IsInteracting)
   n5 = Get OnInteract() -> (OnInteract:Widget Animation Object Reference)
   > n6 = PlayAnimation(InAnimation=@n5)
   n7 = OnInteractedAnimFinished() {is=title K2Node_CustomEvent}
   n8 = Set Timer by Event(execute=@n6.then, Delegate=@n7, Time=GetEndTime(self=@n5), bLooping=false)
   > Set InteractionAnimTimerHandle(InteractionAnimTimerHandle:Timer Handle Structure=@n8) -> (Output_Get:Timer Handle Structure)
   PlayAnimation(execute=@n4.else, InAnimation=Get OnInteract(), PlayMode=Reverse)
   > UpdatePortalInfoPlateVisibility(ShouldShow=true, Distance=Get CachedDistance() -> (CachedDistance:double))
   n9 = OnUpateStatusMessage!() -> (OutputDelegate, then, NewMessage:String) {is=title K2Node_CustomEvent}
   > UpdateDownloadStatus!(execute, self, Message:String=@n9.NewMessage) -> (then)
   n10 = Event InitFromObject!() -> (OutputDelegate, then, InitObject:Object Reference) {is=title K2Node_Event}
   > n11 = Cast To BGA_IslandPortal!(execute, Object=@n10.InitObject) -> (then, CastFailed, AsBGA Island Portal:BGA Island Portal Object Reference) {is=title K2Node_DynamicCast}
   > n12 = Set MyBGAIslandPortal(MyBGAIslandPortal:BGA Island Portal Object Reference=@n11) -> (Output_Get:BGA Island Portal Object Reference)
   > SetupPortalEventBinding!(execute, self, InPortal:BGA Island Portal Object Reference=@n12, Bind:bool=true) -> (then)
   > n13 = Branch(Condition=Get bEnableCreativePortalContentStatusUpdateMessages!(self:Fort Runtime Options Object Reference=GetRuntimeOptions()) -> (bEnableCreativePortalContentStatusUpdateMessages:bool))
   > n14 = Cast To FortGameStateAthena!(execute, Object=GetGameState!() -> (ReturnValue:Game State Base Object Reference)) -> (then, CastFailed, AsFort Game State Athena:Fort Game State Athena Object Reference) {is=title K2Node_DynamicCast}
   > n15 = Bind Event to On New Content Update Message Recieved(self:Fort Game State Athena Object Reference=@n14, Delegate=@n9) {is=title K2Node_AddDelegate}
   SetupDownloadStatus(execute=[@n15.then, @n13.else])
   n16 = Get MyBGAIslandPortal() -> (MyBGAIslandPortal:BGA Island Portal Object Reference)
   n17 = TextIsEmpty(InText=Get OverrideName!(self:BGA Island Portal Object Reference=@n16) -> (OverrideName:Text))
   > n18 = Branch(Condition=OR Boolean(A=Get bReturnToCreativeHub!(self:Fort Athena Creative Portal Object Reference=@n16) -> (bReturnToCreativeHub:bool), B=NOT Boolean(A=@n17)))
   > SetVisibility(self=Get PopulationBox() -> (PopulationBox:Horizontal Box Object Reference), InVisibility=Collapsed)
   n19 = AND Boolean(
    A=Get IsMyPortal!(self:BGA Island Portal Object Reference=Get MyBGAIslandPortal()) -> (IsMyPortal:bool)
    B=IsCreativeModeAccessLimited!(self:Fort Global UIContext Object Reference=Get FortGlobalUIContext() -> (ReturnValue:Fort Global UIContext Object Reference) {is=title K2Node_GetSubsystem}) -> (ReturnValue:bool)
   )
   SetVisibility(
    execute=@n18.else
    self=[Get IslandNameBox() -> (IslandNameBox:Vertical Box Object Reference), Get PopulationBox()]
    InVisibility=Select(Option 0=SelfHitTestInvisible, Option 1=Collapsed, Index=@n19)
   )
   n20 = Get LimitedAccessText() -> (LimitedAccessText:Common Text Object Reference)
   > SetVisibility(self=@n20, InVisibility=Select(Option 0=Collapsed, Option 1=SelfHitTestInvisible, Index=@n19))
   > "SetText (Text)"(self=@n20, InText=Format Text!(Format="LOCTEXT(\"[{x}]\")", x:Text=MakeLiteralText(Value="LOCTEXT(\"LIMITED ACCESS\")")) -> (Result))
   UpdatePortalInfoPlateVisibility(execute=@n7.then, ShouldShow=false, Distance=Get CachedDistance())
   Get Intro()
  ]}
 SetCodeVisibilityByPrivacy {nodes [
   n1 = SetCodeVisibilityByPrivacy!() -> (then, NewPrivacy:"EMMSPrivacy Enum (by ref)"=Public) {is=title K2Node_FunctionEntry}
   > n2 = Switch on EMMSPrivacy!(execute, Selection:EMMSPrivacy Enum=@n1) -> (Public:Exec, Private:Exec) {is=title K2Node_SwitchEnum}
   n3 = Get Text_LinkCode() -> (Text_LinkCode:Common Text Object Reference)
   SetVisibility(execute=@n2.Private, self=@n3, InVisibility=Collapsed)
   SetVisibility(execute=@n2.Public, self=@n3, InVisibility=HitTestInvisible)
  ]}
 SetCuratedSwitcher {nodes [
   n1 = SetCuratedSwitcher!() -> (then, bIsFeatured:bool=~, bIsLinkCode:bool=~, bIsUserInitiatedLoad:bool=~) {is=title K2Node_FunctionEntry}
   > n2 = Set bIsCurated(bIsCurated:bool=@n1) -> (Output_Get:bool)
   > n3 = Set IsLinkCode(IsLinkCode:bool=@n1.bIsLinkCode) -> (Output_Get:bool)
   > n4 = Branch(Condition=AND Boolean(A=@n2, B=NOT Boolean(A=@n1.bIsUserInitiatedLoad)))
   n5 = Get Switcher_IslandOwner() -> (Switcher_IslandOwner:Widget Switcher Object Reference)
   > SetActiveWidgetIndex(self=@n5, Index=1)
   n6 = Branch(execute=@n4.else, Condition=NOT Boolean(A=@n3))
   > SetActiveWidgetIndex(self=@n5, Index=0)
   SetActiveWidgetIndex(execute=@n6.else, self=@n5, Index=2)
  ]}
 SetDescription {nodes [
   n1 = SetDescription!() -> (then, Description:Text) {is=title K2Node_FunctionEntry}
   n2 = TextIsEmpty(InText=@n1)
   > Set HasDescription(HasDescription:bool=NOT Boolean(A=@n2)) -> (Output_Get:bool)
   n3 = Get Text_Description() -> (Text_Description:Common Text Object Reference)
   > "SetText (Text)"(self=@n3, InText=@n1)
   > SetVisibility(self=@n3, InVisibility=Select(Option 0=SelfHitTestInvisible, Option 1=Collapsed, Index=@n2))
  ]}
 SetIslandName {nodes [
   n1 = SetIslandName!() -> (then, NewIslandName:Text) {is=title K2Node_FunctionEntry}
   > n2 = Set IslandName(IslandName:Text=@n1) -> (Output_Get:Text)
   n3 = Get Text_IslandName() -> (Text_IslandName:Common Text Object Reference)
   > "SetText (Text)"(self=@n3, InText=TextToUpper!(InText:"Text (by ref)"=@n2) -> (ReturnValue:Text))
   > SetVisibility(self=@n3, InVisibility=Select(Option 0=SelfHitTestInvisible, Option 1=Collapsed, Index=TextIsEmpty(InText=@n1)))
   > ClearDownloadStatus() {is=title K2Node_CallFunction}
  ]}
 SetIslandPop {nodes [
   n1 = SetIslandPop!() -> (then, NewPop:int, MaxPop:int) {is=title K2Node_FunctionEntry}
   > n2 = Set CurrentIslandPop(CurrentIslandPop:int=@n1) -> (Output_Get:int=0)
   > n3 = Compare Int!(Exec:Exec, Input:int=@n2, CompareWith:int) -> (">":Exec, "==":Exec, "<":Exec) {is=title K2Node_MacroInstance}
   n4 = Get Text_IslandPop() -> (Text_IslandPop:Common Text Object Reference)
   "SetText (Text)"(execute=@n3."<", self=@n4)
   n5 = Get SocialImage_Shadow() -> (SocialImage_Shadow:Image Object Reference)
   n6 = Get SocialImage() -> (SocialImage:Image Object Reference)
   > SetVisibility(self=[@n5, @n6], InVisibility=Collapsed)
   "SetText (Text)"(execute=[@n3."==", @n3.">"], self=@n4, InText=Format Text!(Format="LOCTEXT(\"{x}/{y}\")", x:int=@n2, y:int=@n1.MaxPop) -> (Result))
   > SetVisibility(self=[@n5, @n6], InVisibility=HitTestInvisible)
  ]}
 SetLinkCode {nodes [
   n1 = SetLinkCode!() -> (then, LinkCode:String) {is=title K2Node_FunctionEntry}
   n2 = Get Text_LinkCode()
   > "SetText (Text)"(self=@n2, InText="To Text (String)"(InString=@n1))
   n3 = "Equal (Integer)"(A=Len(S=@n1), B=0) {is=title K2Node_CallFunction}
   > SetVisibility(self=@n2, InVisibility=Select(Option 0=SelfHitTestInvisible, Option 1=Collapsed, Index=@n3))
   > n4 = Branch(Condition=@n3)
   SetCodeVisibilityByPrivacy!(
    execute=@n4.else
    self
    NewPrivacy:"EMMSPrivacy Enum (by ref)"=GetPrivacy!(self:Fort Athena Creative Portal Object Reference=Get MyBGAIslandPortal()) -> (ReturnValue:EMMSPrivacy Enum=Public)
   ) -> (then) {is=title K2Node_CallFunction}
  ]}
 SetPlayerName {nodes [
   n1 = SetPlayerName!() -> (then, PlayerName:Text) {is=title K2Node_FunctionEntry}
   > Set PlayerName(PlayerName:Text=@n1) -> (Output_Get:Text)
   > "SetText (Text)"(
    self=[
     Get Text_PlayerName() -> (Text_PlayerName:Common Text Object Reference)
     Get Text_FeaturedPlayerName() -> (Text_FeaturedPlayerName:Common Text Object Reference)
     Get Text_PublishedPlayer() -> (Text_PublishedPlayer:Common Text Object Reference)
    ]
    InText=@n1
   )
   n2 = TextIsEmpty(InText=@n1)
   > SetVisibility(self=Get Text_PlayerName(), InVisibility=Select(Option 0=SelfHitTestInvisible, Option 1=Hidden, Index=@n2))
   > SetVisibility(
    self=[Get Text_PublishedPlayer(), Get Text_FeaturedPlayerName()]
    InVisibility=Select(Option 0=SelfHitTestInvisible, Option 1=Collapsed, Index=@n2)
   )
  ]}
 SetupDownloadStatus {nodes [
   SetupDownloadStatus() {is=title K2Node_FunctionEntry}
   n1 = Get DownloadInfoArray()
   > Add Unique(
    TargetArray:Array of Common Text Object References=@n1
    NewItem:"Common Text Object Reference (by ref)"=Get DownloadInfo_0() -> (DownloadInfo_0:Common Text Object Reference)
   )
   > Add Unique(TargetArray=@n1, NewItem=Get DownloadInfo_1() -> (DownloadInfo_1:Common Text Object Reference))
   > Add Unique(TargetArray=@n1, NewItem=Get DownloadInfo_2() -> (DownloadInfo_2:Common Text Object Reference))
   > Add Unique(TargetArray=@n1, NewItem=Get DownloadInfo_3() -> (DownloadInfo_3:Common Text Object Reference))
  ]}
 SetupPortalEventBinding {nodes [
   n1 = SetupPortalEventBinding!() -> (then, InPortal:BGA Island Portal Object Reference, Bind:bool=~) {is=title K2Node_FunctionEntry}
   > n2 = Branch(Condition=@n1.Bind)
   > Bind Event to Update Island Name(self:BGA Island Portal Object Reference=@n1, Delegate=Create Event()) {is=title K2Node_AddDelegate}
   > Bind Event to Update Player Name(self:BGA Island Portal Object Reference=@n1, Delegate=Create Event()) {is=title K2Node_AddDelegate}
   > Bind Event to Update Player Pop(self:BGA Island Portal Object Reference=@n1, Delegate=Create Event()) {is=title K2Node_AddDelegate}
   > Bind Event to Update Is Curated Portal(self:BGA Island Portal Object Reference=@n1, Delegate=Create Event()) {is=title K2Node_AddDelegate}
   > Bind Event to Update Description(self:BGA Island Portal Object Reference=@n1, Delegate=Create Event()) {is=title K2Node_AddDelegate}
   > Bind Event to Update Link Code(self:BGA Island Portal Object Reference=@n1, Delegate=Create Event()) {is=title K2Node_AddDelegate}
   > Bind Event to Update Portal UIVis(self:BGA Island Portal Object Reference=@n1, Delegate=Create Event()) {is=title K2Node_AddDelegate}
   n3 = Create Event()
   > Bind Event to Is Being Interacted With(self:BGA Island Portal Object Reference=@n1, Delegate=@n3) {is=title K2Node_AddDelegate}
   Unbind Event from Update Island Name(execute=@n2.else, self:BGA Island Portal Object Reference=@n1, Delegate=Create Event()) {is=title K2Node_RemoveDelegate}
   > Unbind Event from Update Player Name(self:BGA Island Portal Object Reference=@n1, Delegate=Create Event()) {is=title K2Node_RemoveDelegate}
   > Unbind Event from Update Player Pop(self:BGA Island Portal Object Reference=@n1, Delegate=Create Event()) {is=title K2Node_RemoveDelegate}
   > Unbind Event from Update Is Curated Portal(self:BGA Island Portal Object Reference=@n1, Delegate=Create Event()) {is=title K2Node_RemoveDelegate}
   > Unbind Event from Update Description(self:BGA Island Portal Object Reference=@n1, Delegate=Create Event()) {is=title K2Node_RemoveDelegate}
   > Unbind Event from Update Link Code(self:BGA Island Portal Object Reference=@n1, Delegate=Create Event()) {is=title K2Node_RemoveDelegate}
   > Unbind Event from Update Portal UIVis(self:BGA Island Portal Object Reference=@n1, Delegate=Create Event()) {is=title K2Node_RemoveDelegate}
   > Unbind Event from Is Being Interacted With(self:BGA Island Portal Object Reference=@n1, Delegate=@n3) {is=title K2Node_RemoveDelegate}
  ]}
 UpdateDownloadStatus {nodes [
   n1 = UpdateDownloadStatus!() -> (then, Message:String) {is=title K2Node_FunctionEntry}
   > n2 = Branch(Condition="Equal, Case Insensitive (String)"(A=@n1, B=Clear))
   > ClearDownloadStatus()
   n3 = Branch(execute=@n2.else, Condition="integer < integer"(A=Get StatusLine() -> (StatusLine:int=0), B=4))
   "SetText (Text)"(
    execute=[@n4.then, @n3.then]
    self="Get (a copy)"(Array:Array of Common Text Object References=Get DownloadInfoArray(), Dimension 1=Get StatusLine() -> (StatusLine=0)) -> (Output:Common Text Object Reference)
    InText="To Text (String)"(InString=@n1)
   )
   > Increment Int(Value=Get StatusLine() -> (StatusLine=0))
   n4 = ClearDownloadStatus(execute=@n3.else)
  ]}
 UpdatePortalInfoPlateVisibility {nodes [
   n1 = UpdatePortalInfoPlateVisibility!() -> (then, ShouldShow:bool=~, Distance:double=~) {is=title K2Node_FunctionEntry}
   > Set CachedDistance(CachedDistance:double=@n1.Distance) -> (Output_Get:double)
   > n2 = Branch(Condition=@n1)
   > Set bShouldShowDescription(bShouldShowDescription:bool="float < float"!(A:double=@n1.Distance, B:double=Get DescriptionDistanceLimit() -> (DescriptionDistanceLimit:double)) -> (ReturnValue:bool) {is=title K2Node_CallFunction}) -> (Output_Get:bool)
   > UpdateVisBasedOnDistance()
   n3 = "float > float"(A=@n1.Distance, B=Get IconThresholdDistance() -> (IconThresholdDistance:double)) {is=title K2Node_CallFunction}
   n4 = "Equal (Integer)"(A=Get CurrentIslandPop() -> (CurrentIslandPop:int=0), B=0)
   > SetVisibility(InVisibility=Select(Option 0=SelfHitTestInvisible, Option 1=Collapsed, Index=AND Boolean(A=@n3, B=@n4)))
   > n5 = Branch(Condition="Equal (Boolean)"!(A:bool=@n3, B:bool=Get FarDistance() -> (FarDistance:bool)) -> (ReturnValue:bool))
   n6 = Set FarDistance(execute=@n5.else, FarDistance:bool=@n3) -> (Output_Get:bool)
   n7 = Get CurrentIslandPop() -> (CurrentIslandPop=0)
   > n8 = Branch(Condition=OR Boolean(A=@n6, B="integer > integer"(A=@n7, B=0) {is=title K2Node_CallFunction}))
   PlayAnimation(execute=@n8.else, InAnimation=Get VisibilityReveal() -> (VisibilityReveal:Widget Animation Object Reference))
   SetVisibility(execute=@n2.else, InVisibility=Collapsed)
  ]}
 UpdateVisBasedOnDistance {nodes [
   UpdateVisBasedOnDistance() {is=title K2Node_FunctionEntry}
   > n1 = Is Valid(InputObject=Get MyBGAIslandPortal())
   n2 = Set bNewCurVis(
    execute=@n1.Is Not Valid
    bNewCurVis:bool=AND Boolean(A=Get HasDescription() -> (HasDescription:bool), B=Get bShouldShowDescription() -> (bShouldShowDescription:bool))
   ) -> (Output_Get:bool)
   n3 = Get bNewCurVis() -> (bNewCurVis:bool)
   SetVisibility(execute=[@n7.then, @n2.then], self=Get Text_Description(), InVisibility=Select(Option 0=Collapsed, Option 1=SelfHitTestInvisible, Index=@n3))
   > n4 = Branch(Condition="Equal (Boolean)"(A=Get bNewCurVis(), B=Get IsCurVis() -> (IsCurVis:bool)))
   n5 = Set IsCurVis(execute=@n4.else, IsCurVis:bool=@n3) -> (Output_Get:bool)
   > OnNewVisibilityThresholdChanged!(execute, self, bNewVis:bool=@n5) -> (then) {is=title K2Node_CallFunction}
   n6 = CanUseHUDInteract!(self:BGA Island Portal Object Reference=Get MyBGAIslandPortal()) -> (CanInteract:bool)
   n7 = Set bNewCurVis(
    execute=@n1.Is Valid
    bNewCurVis=AND Boolean!(A=Get HasDescription(), B=Get bShouldShowDescription(), C:bool=NOT Boolean(A=@n6)) -> (ReturnValue)
   )
  ]}
]
variables [
 bIsCurated {type=boolean}
 bNewCurVis {type=boolean}
 bShouldShowDescription {type=boolean}
 CachedDistance {type=number}
 CurrentIslandPop {type=integer}
 DescriptionDistanceLimit {type=number, default=800}
 DownloadInfoArray {type="array<CommonTextBlock>", default []}
 FarDistance {type=boolean}
 HasDescription {type=boolean}
 IconThresholdDistance {type=number, default=1000}
 InteractionAnimTimerHandle {type=TimerHandle}
 IsCurVis {type=boolean}
 isInteractWith {type=boolean}
 IslandName {type=string, default=""}
 IsLinkCode {type=boolean}
 MyBGAIslandPortal {type=BGA_IslandPortal_C, default=null}
 PlayerName {type=string, default=""}
 StatusLine {type=integer}
]
widgets {
 info {inheritedWidgetCount=0, parentClass=&UserWidget, rootWidgetClass=&Overlay, widgetCount=26}
 widgets [
  Overlay_0 {widgetClassPath=&Overlay, widget=Overlay_0, properties {clipping=ClipToBounds}}
  VerticalBox_0 {
   widgetClassPath=&VerticalBox
   parent=Overlay_0
   widget=VerticalBox_0
   slot {class=OverlaySlot, properties {horizontalAlignment=HAlign_Fill, verticalAlignment=VAlign_Fill}}
  }
  HorizontalBox_0 {
   widgetClassPath=&HorizontalBox
   parent=VerticalBox_0
   widget=HorizontalBox_0
   slot {class=VerticalBoxSlot, properties {horizontalAlignment=HAlign_Center}}
  }
  Switcher_IslandOwner {
   widgetClassPath=&WidgetSwitcher
   bIsVariable=true
   parent=HorizontalBox_0
   widget=Switcher_IslandOwner
   slot {class=HorizontalBoxSlot}
  }
  VerticalBox_1 {widgetClassPath=&VerticalBox, parent=Switcher_IslandOwner, widget=VerticalBox_1, slot {class=WidgetSwitcherSlot}}
  Text_PlayerName {
   widgetClassPath=&CommonTextBlock
   bIsVariable=true
   parent=VerticalBox_1
   widget=Text_PlayerName
   properties {
    justification=Center
    text=PLAYERNAME
    visibility=Hidden
    colorAndOpacity {colorUseRule=UseColor_Specified, specifiedColor=<rgba 0 1 1 1>}
    font {
     bForceMonospaced=false
     bMaterialIsStencil=false
     fontMaterial=null
     fontObject=&Burbank
     letterSpacing=0
     monospacedWidth=1
     size=24
     skewAmount=0
     typefaceFontName=Burbank-Black
     outlineSettings {
      bApplyOutlineToDropShadows=false
      bMiteredCorners=false
      bSeparateFillAlpha=false
      outlineMaterial=/Game/Athena/HUD/Art/Materials/M_HUD_Sheen_PortalInfo
      outlineSize=1
      outlineColor=<rgba 1 1 1 1>
     }
    }
    shadowColorAndOpacity=<rgba 0.0052 0.0052 0.0056 0.822>
    shadowOffset=<2 2>
   }
   slot {class=VerticalBoxSlot}
  }
  LimitedAccessText {
   widgetClassPath=&CommonTextBlock
   bIsVariable=true
   parent=VerticalBox_1
   widget=LimitedAccessText
   properties {
    justification=Center
    text="[LIMITED ACCESS]"
    visibility=Collapsed
    colorAndOpacity {colorUseRule=UseColor_Specified, specifiedColor=<rgba 1 0.7696 0.6787 1>}
    font {
     bForceMonospaced=false
     bMaterialIsStencil=false
     fontMaterial=null
     fontObject=&BurbankSmall
     letterSpacing=0
     monospacedWidth=1
     size=14
     skewAmount=0
     typefaceFontName=Bold
     outlineSettings {
      bApplyOutlineToDropShadows=false
      bMiteredCorners=false
      bSeparateFillAlpha=false
      outlineMaterial=null
      outlineSize=1
      outlineColor=<rgba 0.0052 0.0052 0.0056 0.149>
     }
    }
    shadowColorAndOpacity=<rgba 0.0052 0.0052 0.0056 0.822>
    shadowOffset=<2 2>
   }
   slot {class=VerticalBoxSlot, properties {horizontalAlignment=HAlign_Center}}
  }
  VerticalBox_2 {widgetClassPath=&VerticalBox, parent=Switcher_IslandOwner, widget=VerticalBox_2, slot {class=WidgetSwitcherSlot}}
  CommonTextBlock_0 {
   widgetClassPath=&CommonTextBlock
   bIsVariable=true
   parent=VerticalBox_2
   widget=CommonTextBlock_0
   properties {
    justification=Center
    text=FEATURED
    colorAndOpacity {colorUseRule=UseColor_Specified, specifiedColor=<rgba 0.9301 1 0.0091 1>}
    font {
     bForceMonospaced=false
     bMaterialIsStencil=false
     fontMaterial=null
     fontObject=&BurbankSmall
     letterSpacing=0
     monospacedWidth=1
     size=12
     skewAmount=0
     typefaceFontName=Black
     outlineSettings {
      bApplyOutlineToDropShadows=false
      bMiteredCorners=false
      bSeparateFillAlpha=false
      outlineMaterial=null
      outlineSize=1
      outlineColor=<rgba 0.0052 0.0052 0.0056 0.149>
     }
    }
    renderTransform {angle=0, scale=<1 1>, shear=<-12 0>, translation=<0 3>}
    shadowColorAndOpacity=<rgba 0.0052 0.0052 0.0056 0.822>
    shadowOffset=<2 2>
   }
   slot {class=VerticalBoxSlot}
  }
  Text_FeaturedPlayerName {
   widgetClassPath=&CommonTextBlock
   bIsVariable=true
   parent=VerticalBox_2
   widget=Text_FeaturedPlayerName
   properties {
    justification=Center
    text="{PlayerName}"
    visibility=Collapsed
    colorAndOpacity {colorUseRule=UseColor_Specified, specifiedColor=<rgba 0 1 1 1>}
    font {
     bForceMonospaced=false
     bMaterialIsStencil=false
     fontMaterial=null
     fontObject=&Burbank
     letterSpacing=0
     monospacedWidth=1
     size=24
     skewAmount=0
     typefaceFontName=Burbank-Black
     outlineSettings {
      bApplyOutlineToDropShadows=false
      bMiteredCorners=false
      bSeparateFillAlpha=false
      outlineMaterial=&M_HUD_Sheen_PortalInfo
      outlineSize=1
      outlineColor=<rgba 1 1 1 1>
     }
    }
    shadowColorAndOpacity=<rgba 0.0052 0.0052 0.0056 0.822>
    shadowOffset=<2 2>
   }
   slot {class=VerticalBoxSlot}
  }
  VerticalBox_3 {widgetClassPath=&VerticalBox, parent=Switcher_IslandOwner, widget=VerticalBox_3, slot {class=WidgetSwitcherSlot}}
  Text_PublishedPlayer {
   widgetClassPath=&CommonTextBlock
   bIsVariable=true
   parent=VerticalBox_3
   widget=Text_PublishedPlayer
   properties {
    justification=Center
    text=PlayerName
    visibility=Collapsed
    colorAndOpacity {colorUseRule=UseColor_Specified, specifiedColor=<rgba 0 1 1 1>}
    font {
     bForceMonospaced=false
     bMaterialIsStencil=false
     fontMaterial=null
     fontObject=&Burbank
     letterSpacing=0
     monospacedWidth=1
     size=24
     skewAmount=0
     typefaceFontName=Burbank-Black
     outlineSettings {
      bApplyOutlineToDropShadows=false
      bMiteredCorners=false
      bSeparateFillAlpha=false
      outlineMaterial=null
      outlineSize=1
      outlineColor=<rgba 0.0052 0.0052 0.0056 0.149>
     }
    }
    shadowColorAndOpacity=<rgba 0.0052 0.0052 0.0056 0.822>
    shadowOffset=<2 2>
   }
   slot {class=VerticalBoxSlot}
  }
  Text_LinkCode {
   widgetClassPath=&CommonTextBlock
   bIsVariable=true
   parent=VerticalBox_3
   widget=Text_LinkCode
   properties {
    justification=Center
    text=1111-1111-1111
    font {
     bForceMonospaced=false
     bMaterialIsStencil=false
     fontMaterial=null
     fontObject=&BurbankSmall
     letterSpacing=0
     monospacedWidth=1
     size=18
     skewAmount=0
     typefaceFontName=Black
     outlineSettings {
      bApplyOutlineToDropShadows=false
      bMiteredCorners=false
      bSeparateFillAlpha=false
      outlineMaterial=null
      outlineSize=1
      outlineColor=<rgba 0.0052 0.0052 0.0056 0.2>
     }
    }
    shadowColorAndOpacity=<rgba 0.0052 0.0052 0.0056 0.822>
    shadowOffset=<2 2>
   }
   slot {class=VerticalBoxSlot}
  }
  IslandNameBox {
   widgetClassPath=&VerticalBox
   bIsVariable=true
   parent=VerticalBox_0
   widget=IslandNameBox
   slot {class=VerticalBoxSlot, properties {horizontalAlignment=HAlign_Center}}
  }
  Border_PlateDivider {
   widgetClassPath=&Border
   parent=IslandNameBox
   widget=Border_PlateDivider
   properties {visibility=Collapsed, padding=<ltrb 0 0 0 0>, renderTransform {angle=0, scale=<1.3 1>, shear=<0 0>, translation=<0 0>}}
   slot {class=VerticalBoxSlot}
  }
  Text_IslandName {
   widgetClassPath=&CommonTextBlock
   bIsVariable=true
   parent=IslandNameBox
   widget=Text_IslandName
   properties {
    justification=Center
    text=MEDIUM ISLAND 1
    font {
     bForceMonospaced=false
     bMaterialIsStencil=false
     fontMaterial=null
     fontObject=&BurbankSmall
     letterSpacing=0
     monospacedWidth=1
     size=20
     skewAmount=0
     typefaceFontName=Black
     outlineSettings {
      bApplyOutlineToDropShadows=false
      bMiteredCorners=false
      bSeparateFillAlpha=false
      outlineMaterial=null
      outlineSize=1
      outlineColor=<rgba 0.0052 0.0052 0.0056 0.149>
     }
    }
    shadowColorAndOpacity=<rgba 0.0052 0.0052 0.0056 0.822>
    shadowOffset=<2 2>
   }
   slot {class=VerticalBoxSlot}
  }
  PopulationBox {
   widgetClassPath=&HorizontalBox
   bIsVariable=true
   parent=IslandNameBox
   widget=PopulationBox
   slot {class=VerticalBoxSlot, properties {horizontalAlignment=HAlign_Center, verticalAlignment=VAlign_Top}}
  }
  Overlay_1 {
   widgetClassPath=&Overlay
   parent=PopulationBox
   widget=Overlay_1
   slot {class=HorizontalBoxSlot, properties {verticalAlignment=VAlign_Center}}
  }
  SocialImage_Shadow {
   widgetClassPath=&Image
   bIsVariable=true
   parent=Overlay_1
   widget=SocialImage_Shadow
   properties {
    brush {
     bIsDynamicallyLoaded=false
     drawAs=Image
     imageType=NoImage
     mirroring=NoMirror
     resourceName=None
     resourceObject=/Game/UI/Frontend/Parties/PartiesButtonStyles/T_UI_Icon_QuadMeeple
     tiling=NoTile
     imageSize=<22 22>
     margin=<ltrb 0 0 0 0>
     outlineSettings {
      bUseBrushTransparency=false
      roundingType=HalfHeightRadius
      width=0
      color {colorUseRule=UseColor_Specified, specifiedColor=<rgba 0 0 0 0>}
      cornerRadii=<0 0 0 0>
     }
     tintColor {colorUseRule=UseColor_Specified, specifiedColor=<rgba 1 1 1 1>}
     uVRegion {bIsValid=false, max=<0 0>, min=<0 0>}
    }
    colorAndOpacity=<rgba 0 0 0 1>
    renderTransform {angle=0, scale=<1.1 1.1>, shear=<0 0>, translation=<1 -2>}
   }
   slot {class=OverlaySlot}
  }
  SocialImage {
   widgetClassPath=&Image
   bIsVariable=true
   parent=Overlay_1
   widget=SocialImage
   properties {
    brush {
     bIsDynamicallyLoaded=false
     drawAs=Image
     imageType=NoImage
     mirroring=NoMirror
     resourceName=None
     resourceObject=&T_UI_Icon_QuadMeeple
     tiling=NoTile
     imageSize=<22 22>
     margin=<ltrb 0 0 0 0>
     outlineSettings {
      bUseBrushTransparency=false
      roundingType=HalfHeightRadius
      width=0
      color {colorUseRule=UseColor_Specified, specifiedColor=<rgba 0 0 0 0>}
      cornerRadii=<0 0 0 0>
     }
     tintColor {colorUseRule=UseColor_Specified, specifiedColor=<rgba 1 1 1 1>}
     uVRegion {bIsValid=false, max=<0 0>, min=<0 0>}
    }
    renderTransform {angle=0, scale=<1 1>, shear=<0 0>, translation=<0 -3>}
   }
   slot {class=OverlaySlot}
  }
  Text_IslandPop {
   widgetClassPath=&CommonTextBlock
   bIsVariable=true
   parent=PopulationBox
   widget=Text_IslandPop
   properties {
    text=00/00
    font {
     bForceMonospaced=false
     bMaterialIsStencil=false
     fontMaterial=null
     fontObject=&BurbankSmall
     letterSpacing=0
     monospacedWidth=1
     size=16
     skewAmount=0
     typefaceFontName=Black
     outlineSettings {
      bApplyOutlineToDropShadows=false
      bMiteredCorners=false
      bSeparateFillAlpha=false
      outlineMaterial=null
      outlineSize=1
      outlineColor=<rgba 0.0052 0.0052 0.0056 0.149>
     }
    }
    renderTransform {angle=0, scale=<1 1>, shear=<0 0>, translation=<0 3>}
    shadowColorAndOpacity=<rgba 0.0052 0.0052 0.0056 0.822>
    shadowOffset=<2 2>
   }
   slot {class=HorizontalBoxSlot, properties {verticalAlignment=VAlign_Center, padding=<ltrb 4 0 0 0>}}
  }
  Text_Description {
   widgetClassPath=&CommonTextBlock
   bIsVariable=true
   parent=IslandNameBox
   widget=Text_Description
   properties {
    justification=Center
    text=Suspiciously long vague description text about this that or the other WHO KNOWS
    wrapTextAt=465
    colorAndOpacity {colorUseRule=UseColor_Specified, specifiedColor=<rgba 0 1 1 1>}
    font {
     bForceMonospaced=false
     bMaterialIsStencil=false
     fontMaterial=null
     fontObject=&BurbankSmall
     letterSpacing=0
     monospacedWidth=1
     size=12
     skewAmount=0
     typefaceFontName=Bold
     outlineSettings {
      bApplyOutlineToDropShadows=false
      bMiteredCorners=false
      bSeparateFillAlpha=false
      outlineMaterial=null
      outlineSize=1
      outlineColor=<rgba 0.0052 0.0052 0.0056 0.149>
     }
    }
    shadowColorAndOpacity=<rgba 0.0052 0.0052 0.0056 0.822>
    shadowOffset=<0 0>
   }
   slot {class=VerticalBoxSlot}
  }
  DownloadInfo_0 {
   widgetClassPath=&CommonTextBlock
   bIsVariable=true
   parent=IslandNameBox
   widget=DownloadInfo_0
   properties {
    justification=Center
    wrapTextAt=465
    colorAndOpacity {colorUseRule=UseColor_Specified, specifiedColor=<rgba 0 1 1 1>}
    font {
     bForceMonospaced=false
     bMaterialIsStencil=false
     fontMaterial=null
     fontObject=&BurbankSmall
     letterSpacing=0
     monospacedWidth=1
     size=12
     skewAmount=0
     typefaceFontName=Bold
     outlineSettings {
      bApplyOutlineToDropShadows=false
      bMiteredCorners=false
      bSeparateFillAlpha=false
      outlineMaterial=null
      outlineSize=1
      outlineColor=<rgba 0.0052 0.0052 0.0056 0.149>
     }
    }
    shadowColorAndOpacity=<rgba 0.0052 0.0052 0.0056 0.822>
    shadowOffset=<0 0>
   }
   slot {class=VerticalBoxSlot}
  }
  DownloadInfo_1 {
   widgetClassPath=&CommonTextBlock
   bIsVariable=true
   parent=IslandNameBox
   widget=DownloadInfo_1
   properties {
    justification=Center
    wrapTextAt=465
    colorAndOpacity {colorUseRule=UseColor_Specified, specifiedColor=<rgba 0 1 1 1>}
    font {
     bForceMonospaced=false
     bMaterialIsStencil=false
     fontMaterial=null
     fontObject=&BurbankSmall
     letterSpacing=0
     monospacedWidth=1
     size=12
     skewAmount=0
     typefaceFontName=Bold
     outlineSettings {
      bApplyOutlineToDropShadows=false
      bMiteredCorners=false
      bSeparateFillAlpha=false
      outlineMaterial=null
      outlineSize=1
      outlineColor=<rgba 0.0052 0.0052 0.0056 0.149>
     }
    }
    shadowColorAndOpacity=<rgba 0.0052 0.0052 0.0056 0.822>
    shadowOffset=<0 0>
   }
   slot {class=VerticalBoxSlot}
  }
  DownloadInfo_2 {
   widgetClassPath=&CommonTextBlock
   bIsVariable=true
   parent=IslandNameBox
   widget=DownloadInfo_2
   properties {
    justification=Center
    wrapTextAt=465
    colorAndOpacity {colorUseRule=UseColor_Specified, specifiedColor=<rgba 0 1 1 1>}
    font {
     bForceMonospaced=false
     bMaterialIsStencil=false
     fontMaterial=null
     fontObject=&BurbankSmall
     letterSpacing=0
     monospacedWidth=1
     size=12
     skewAmount=0
     typefaceFontName=Bold
     outlineSettings {
      bApplyOutlineToDropShadows=false
      bMiteredCorners=false
      bSeparateFillAlpha=false
      outlineMaterial=null
      outlineSize=1
      outlineColor=<rgba 0.0052 0.0052 0.0056 0.149>
     }
    }
    shadowColorAndOpacity=<rgba 0.0052 0.0052 0.0056 0.822>
    shadowOffset=<0 0>
   }
   slot {class=VerticalBoxSlot}
  }
  DownloadInfo_3 {
   widgetClassPath=&CommonTextBlock
   bIsVariable=true
   parent=IslandNameBox
   widget=DownloadInfo_3
   properties {
    justification=Center
    wrapTextAt=465
    colorAndOpacity {colorUseRule=UseColor_Specified, specifiedColor=<rgba 0 1 1 1>}
    font {
     bForceMonospaced=false
     bMaterialIsStencil=false
     fontMaterial=null
     fontObject=&BurbankSmall
     letterSpacing=0
     monospacedWidth=1
     size=12
     skewAmount=0
     typefaceFontName=Bold
     outlineSettings {
      bApplyOutlineToDropShadows=false
      bMiteredCorners=false
      bSeparateFillAlpha=false
      outlineMaterial=null
      outlineSize=1
      outlineColor=<rgba 0.0052 0.0052 0.0056 0.149>
     }
    }
    shadowColorAndOpacity=<rgba 0.0052 0.0052 0.0056 0.822>
    shadowOffset=<0 0>
   }
   slot {class=VerticalBoxSlot}
  }
 ]
}
